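; Utility functions for band-play character animation in Sparks, translating song MIDI events and gameplay tags into animation inputs.
; MIDI text maps to an intensity tag plus idle and realtime-idle flags, while genre and intensity tags map to indices.
; SetSongDefaults resets an instrument anim instance to medium intensity and default genre.
; DrumUpdateArmHit eases a hit value toward a boolean target, rising faster than falling; an inline note doubts its effectiveness.
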(fn DrumUpdateArmHit
  (params (Input :type Float%20%28double-precision%29%20%28by%20ref%29) (Target :type Boolean) (DeltaTime :type Float%20%28double-precision%29))
  ; Note that this whole system is probably wrong.  In at least one song, WhatIsLove, Target only spend .2 seconds being false, so this never gets below .8 or so.  So it probably may as well not exist -jwf
  (Variables|SetBy-RefVar Input (Math|Interpolation|FInterptoConstant Input (Math|Conversions|ToFloat(Boolean) Target) DeltaTime (Math|Float|SelectFloat 0.833333 2.5 (> Input Target)))))

(fn IntensityMidiTextEvents
  (return (Utilities|Array|MakeArray (Class|SparksCoreBPFL|GetVocalsMidiTextStringfromEnum "PlayingMellow") (Class|SparksCoreBPFL|GetVocalsMidiTextStringfromEnum "PlayingStandard") (Class|SparksCoreBPFL|GetVocalsMidiTextStringfromEnum "PlayingIntense") (Class|SparksCoreBPFL|GetVocalsMidiTextStringfromEnum "IdleRhythmMellow") (Class|SparksCoreBPFL|GetVocalsMidiTextStringfromEnum "IdleRhythmStandard") (Class|SparksCoreBPFL|GetVocalsMidiTextStringfromEnum "IdleRhythmIntense") (Class|SparksCoreBPFL|GetGuitarMidiTextStringfromEnum "IdleRealtime"))))

(fn SetSongDefaults
  (params (AnimInstance :type Sparks%20Instrument%20Anim%20Instance%20Object%20Reference))
  (Variables|/Script/SparksCharacterCommonRuntime.SparksInstrumentAnimInstance|SetDesiredSongIntensity "(TagName=\"Sparks.Anim.Intensity.Medium\")" AnimInstance)
  (Variables|/Script/SparksCharacterCommonRuntime.SparksInstrumentAnimInstance|SetSongIntensity "(TagName=\"Sparks.Anim.Intensity.Medium\")" AnimInstance)
  (Variables|/Script/SparksCharacterCommonRuntime.SparksInstrumentAnimInstance|SetSongGenre :self AnimInstance))

(fn SongGenreToIndex
  (params (Genre :type Gameplay%20Tag%20Structure))
  (if (== Genre "(TagName=\"Sparks.Anim.Genre.RapHipHop\")")
    (return 0)
    (elif (== Genre "(TagName=\"Sparks.Anim.Genre.Pop\")")
      (return 1)
      (elif (== Genre "(TagName=\"Sparks.Anim.Genre.Rock\")")
        (return 2)
        (else
          (return 0))))))

(fn SongIntensityToIndex
  (params (Intensity :type Gameplay%20Tag%20Structure))
  (if (== Intensity "(TagName=\"Sparks.Anim.Intensity.Light\")")
    (return 0)
    (elif (== Intensity "(TagName=\"Sparks.Anim.Intensity.Heavy\")")
      (return 2)
      (else
        (return 1)))))

(fn TranslateMidiEventTextToSongIntensity
  (params (MidiEventText :type Name))
  (if (== MidiEventText "idle_realtime")
    (return true true "(TagName=\"Sparks.Anim.Intensity.Light\")")
    (elif (== MidiEventText "mellow")
      (return false false "(TagName=\"Sparks.Anim.Intensity.Light\")")
      (elif (== MidiEventText "idle_mellow")
        (return false true "(TagName=\"Sparks.Anim.Intensity.Light\")")
        (elif (== MidiEventText "play")
          (return false false "(TagName=\"Sparks.Anim.Intensity.Medium\")")
          (elif (== MidiEventText "idle")
            (return false true "(TagName=\"Sparks.Anim.Intensity.Medium\")")
            (elif (== MidiEventText "intense")
              (return false false "(TagName=\"Sparks.Anim.Intensity.Heavy\")")
              (elif (== MidiEventText "idle_intense")
                (return false true "(TagName=\"Sparks.Anim.Intensity.Heavy\")")))))))))
