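# Editor utility panel for testing Jam V2 band-play animation in PIE by driving jam session songs, loop types and tempo.
# CheckJamJession caches the pawn's session and audio state; PlayLoop starts loops through the music manager, first forcing a jam emote if no session exists.
# UpdateTempo applies BPM changes, and Tick can oscillate tempo.
# Details views expose debug overrides and the linked instrument-slot anim instance.
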EditorUtilityWidgetBlueprint /SparksCharacterCommon/Animation/MainPlayer/BandPlay/EUW_JamV2Control
graphs [
 CheckJamJession {nodes [
   CheckJamJession() {is=title K2Node_FunctionEntry}
   > n1 = Sequence!(execute) -> (then_0, then_1, then_2:Exec)
   > n2 = Branch(Condition=IsValidInterface!(Interface:"Interface Interface (by ref)"=Get Jam Session() -> (Jam Session:Jam Session Interface Interface)) -> (ReturnValue:bool))
   n3 = GetPIEWorlds!(execute=@n2.else, bIncludeDedicatedServer:bool=true) -> (then, ReturnValue:Array of World Object References)
   > Branch(Condition=Is Not Empty!(TargetArray:Array of World Object References=@n3) -> (ReturnValue:bool))
   n4 = "Get (a copy)"(Array:Array of World Object References=@n3, Dimension 1="Get  WorldIndex"() -> (" WorldIndex":int=0)) -> (Output:World Object Reference)
   > n5 = Cast To FortPlayerController(Object=GetPlayerController!(WorldContextObject:Object Reference=@n4, PlayerIndex:int=Get PlayerIndex() -> (PlayerIndex:int=0)) -> (ReturnValue:Player Controller Object Reference))
   > Set As Fort Player Controller(As Fort Player Controller:Fort Player Controller Object Reference=@n5) -> (Output_Get:Fort Player Controller Object Reference)
   > n6 = Set PlayerPawn(PlayerPawn:Fort Player Pawn Object Reference=FindPlayerPawn!(
     Pawn:Object Reference=GetPlayerPawn!(self:Fort Player Controller Object Reference=@n5) -> (ReturnValue:Fort Player Pawn Object Reference)
     WorldContextObject:Object Reference=@n4
    ) -> (AthenaPawn:Fort Player Pawn Athena Object Reference)) -> (Output_Get:Fort Player Pawn Object Reference)
   > n7 = GetJamSessionForActor!(execute, Actor:Actor Object Reference=@n6) -> (then, ReturnValue:Jam Session Interface Interface)
   > Set Jam Session(Jam Session:Jam Session Interface Interface=@n7) -> (Output_Get:Jam Session Interface Interface)
   n8 = Get Jam Session()
   Branch(execute=@n1.then_1, Condition=IsValidInterface(Interface=@n8))
   n9 = Get Component On JamSession!(
    ComponentClass:Actor Component Class Reference=/Script/FMJamSessionRuntime.JamSessionComponent_AudioState
    JamSession:Jam Session Interface Interface=@n8
   ) -> (OutFoundComponent:bool, ReturnValue:Jam Session Component Audio State Object Reference) {is=title K2Node_CallFunction}
   > Is Valid(InputObject=@n9.ReturnValue)
   > Set JamAudioState(JamAudioState:Jam Session Component Audio State Object Reference=@n9.ReturnValue) -> (Output_Get:Jam Session Component Audio State Object Reference)
   Return Node!(execute=@n1.then_2, JamSession:Jam Session Interface Interface=Get Jam Session())
  ]}
 EventGraph {nodes [
   Add!(execute, TargetSet:Set of Strings=Get CustomButtonList() -> (CustomButtonList:Set of Strings), NewItem:"String (by ref)") -> (then) {is=title K2Node_CallFunction}
   > Generate button()
   "On Clicked (EditorUtilityButton_0)"() {is=title K2Node_ComponentBoundEvent}
   > n1 = GetPIEWorlds(bIncludeDedicatedServer=false)
   > Branch(Condition=Is Not Empty(TargetArray=@n1))
   n2 = "Get (a copy)"(Array=@n1)
   > n3 = Cast To PlayerPawn_Athena(Object=GetPlayerPawn!(WorldContextObject:Object Reference=@n2, PlayerIndex:int=0) -> (ReturnValue:Pawn Object Reference))
   n4 = GetAnimInstance(self=Get Mesh(self=@n3))
   > n5 = Set Anim Instance Ref(Anim Instance Ref:Anim Instance Object Reference=GetLinkedAnimGraphInstanceByTag!(
     self:Anim Instance Object Reference=GetLinkedAnimLayerInstanceByClass!(
      self:Anim Instance Object Reference=@n4
      InClass:Anim Instance Class Reference=/Script/SparksCharacterCommonRuntime.SparksBaseLayerAnimInstance
      bCheckForChildClass:bool=true
     ) -> (ReturnValue:Anim Instance Object Reference)
     InTag:Name=SparksJam_Instrument_Slot
    ) -> (ReturnValue:Anim Instance Object Reference)) -> (Output_Get:Anim Instance Object Reference)
   SetObject(
    execute=[@n5.then, @n12.then]
    self=Get DetailsView_0() -> (DetailsView_0:Details View Object Reference)
    NewObject=Get Anim Instance Ref() -> (Anim Instance Ref:Anim Instance Object Reference)
   )
   n6 = Get Anim Instance Ref()
   n7 = GetDisplayName(Object=@n6)
   > "SetText (Text)"(
    self=Get Text_CurrentAnimInstance() -> (Text_CurrentAnimInstance:Text Object Reference)
    InText=SelectText(A="To Text (String)"(InString=@n7), B="LOCTEXT(\"Not Valid\")", bPickA=IsValid(Object=@n6))
   )
   n8 = GetAllActorsOfClass!(execute=@n3.CastFailed, WorldContextObject:Object Reference=@n2, ActorClass:Actor Class Reference=/Script/FortniteGame.FortPlayerPawnAthena) -> (then, OutActors:Array of Fort Player Pawn Athena Object References)
   n9 = "Get (a copy)"(Array:Array of Fort Player Pawn Athena Object References=@n8, Dimension 1=Get PlayerIndex() -> (PlayerIndex=0)) -> (Output:Fort Player Pawn Athena Object Reference)
   > n10 = Is Valid(InputObject=@n9)
   "SetText (Text)"(execute=@n10.Is Not Valid, self=Get Text_CurrentAnimInstance(), InText="LOCTEXT(\"Not Valid\")")
   n11 = GetAnimInstance(self=Get Mesh(self=@n9))
   n12 = Set Anim Instance Ref(
    execute=@n10.Is Valid
    Anim Instance Ref=GetLinkedAnimGraphInstanceByTag(
     self=GetLinkedAnimLayerInstanceByClass(self=@n11, InClass=&SparksBaseLayerAnimInstance, bCheckForChildClass=true)
     InTag=SparksJam_Instrument_Slot
    )
   )
   n13 = "On Text Committed (EditorUtilityEditableTextBox_0)"!() -> (OutputDelegate, then, Text:"Text (by ref)", CommitMethod:ETextCommit Enum=Default) {is=title K2Node_ComponentBoundEvent}
   n14 = "To String (Text)"(InText=@n13.Text)
   > Branch(Condition=IsNumeric!(SourceString:String=@n14) -> (ReturnValue:bool))
   n15 = "Clamp (Integer)"(Value=String To Integer(InString=@n14), Min=80, Max=180) -> (ReturnValue=0)
   > "SetText (Text Box)"!(
    execute
    self:Text Box Object Reference=Get EditorUtilityEditableTextBox_0() -> (EditorUtilityEditableTextBox_0:Editor Utility Editable Text Box Object Reference)
    InText:Text="To Text (Integer)"(Value=@n15)
   ) -> (then)
   > n16 = Set CustomTempo(CustomTempo:int=@n15) -> (Output_Get:int=0)
   > UpdateTempo!(execute, self, NewTempo:int=@n16) -> (then)
   "On Clicked (Button_Nightmare)"() {is=title K2Node_ComponentBoundEvent}
   > Set CurrentSongShortName(CurrentSongShortName:String=nightmare) -> (Output_Get:String)
   > PlayLoop()
   "On Clicked (Button_Vocal)"() {is=title K2Node_ComponentBoundEvent}
   > Set CurrentLoopType(CurrentLoopType:EFMJamLoopType Enum=Lead) -> (Output_Get:EFMJamLoopType Enum=Lead)
   > PlayLoop()
   "On Clicked (Button_Overcompensate)"() {is=title K2Node_ComponentBoundEvent}
   > Set CurrentSongShortName(CurrentSongShortName=overcompensate)
   > PlayLoop()
   "On Clicked (Button_Bass)"() {is=title K2Node_ComponentBoundEvent}
   > Set CurrentLoopType(CurrentLoopType=Bass) -> (Output_Get=Lead)
   > PlayLoop()
   "On Clicked (Button_SideToSide)"() {is=title K2Node_ComponentBoundEvent}
   > Set CurrentSongShortName(CurrentSongShortName=sidetoside)
   > PlayLoop()
   "On Clicked (Button_MIA)"() {is=title K2Node_ComponentBoundEvent}
   > Set CurrentSongShortName(CurrentSongShortName=mia)
   > PlayLoop()
   "On Clicked (Button_Superbass)"() {is=title K2Node_ComponentBoundEvent}
   > Set CurrentSongShortName(CurrentSongShortName=superbass)
   > PlayLoop()
   "On Clicked (Button_Closer)"() {is=title K2Node_ComponentBoundEvent}
   > Set CurrentSongShortName(CurrentSongShortName=closer)
   > PlayLoop()
   "On Clicked (Button_Melt)"() {is=title K2Node_ComponentBoundEvent}
   > Set CurrentSongShortName(CurrentSongShortName=melt)
   > PlayLoop()
   "On Clicked (Button_YourLove9PM)"() {is=title K2Node_ComponentBoundEvent}
   > Set CurrentSongShortName(CurrentSongShortName=yourlove9pm)
   > PlayLoop()
   "On Clicked (EditorUtilityButton_123)"() {is=title K2Node_ComponentBoundEvent}
   > UpdateTempo(NewTempo=Get CustomTempo() -> (CustomTempo:int=0))
   "On Clicked (Button_Vocal_1)"() {is=title K2Node_ComponentBoundEvent}
   > n17 = GetPIEWorlds(bIncludeDedicatedServer=true)
   > Branch(Condition=Is Not Empty(TargetArray=@n17))
   > GetPawnAndController!(
    execute
    self
    WorldContextObject:Object Reference="Get (a copy)"(Array:Array of World Object References=@n17) -> (Output:World Object Reference)
   ) -> (then, Pawn:Fort Player Pawn Object Reference, Controller:Fort Player Controller Object Reference)
   > n18 = CheckJamJession!(execute, self) -> (then, JamSession:Jam Session Interface Interface) {is=title K2Node_CallFunction}
   > StopAllLoops!(
    execute
    self:Jam Session Component Music Manager Object Reference=GetMusicManager!(self:Jam Session Interface Interface=@n18) -> (ReturnValue:Jam Session Component Music Manager Object Reference)
    bStopAutoJammers:bool=true
    bTryToBackfill:bool=true
   ) -> (then)
   "On Clicked (Button_CustomSongTag)"() {is=title K2Node_ComponentBoundEvent}
   n19 = "To String (Text)"(InText="GetText (Text Box)"!(self:Text Box Object Reference=Get TextBox_CustomSongName() -> (TextBox_CustomSongName:Editor Utility Editable Text Box Object Reference)) -> (ReturnValue:Text))
   > n20 = Branch(Condition="Not Equal, Case Insenstive (String)"!(A:String=Get CurrentSongShortName() -> (CurrentSongShortName:String), B:String=@n19) -> (ReturnValue:bool))
   > Set CurrentSongShortName(CurrentSongShortName=@n19)
   > PlayLoop()
   n21 = CheckJamJession(execute=@n20.else)
   > n22 = Branch(Condition=IsValidInterface(Interface=@n21))
   PlayLoop(execute=@n22.else)
   "On Clicked (EditorUtilityButton_Bpm180)"() {is=title K2Node_ComponentBoundEvent}
   > UpdateTempo(NewTempo=180)
   n23 = "On Clicked (EditorUtilityButton)"() {is=title K2Node_ComponentBoundEvent}
   CheckJamJession(execute=[@n23.then, @n37.then])
   > n24 = Get!(execute:Exec) -> (then:Exec, else:Exec, JamAudioState:Jam Session Component Audio State Object Reference)
   > UpdateTempo(NewTempo=int + int(A=Get CurrentTempo!(self:Jam Session Component Audio State Object Reference=@n24) -> (CurrentTempo:int=0)))
   "On Clicked (EditorUtilityButton_Bpm95)"() {is=title K2Node_ComponentBoundEvent}
   > UpdateTempo(NewTempo=95)
   n25 = "On Clicked (EditorUtilityButton_1)"() {is=title K2Node_ComponentBoundEvent}
   CheckJamJession(execute=[@n25.then, @n38.then])
   > n26 = Get()
   > UpdateTempo(NewTempo=int + int(A=Get CurrentTempo(self=@n26) -> (CurrentTempo=0), B=-1))
   "On Clicked (EditorUtilityButton_Bpm165)"() {is=title K2Node_ComponentBoundEvent}
   > UpdateTempo(NewTempo=165)
   "On Clicked (EditorUtilityButton_Bpm125)"() {is=title K2Node_ComponentBoundEvent}
   > UpdateTempo(NewTempo=125)
   "On Clicked (EditorUtilityButton_Bpm80)"() {is=title K2Node_ComponentBoundEvent}
   > UpdateTempo(NewTempo=80)
   "On Clicked (Button_Drum)"() {is=title K2Node_ComponentBoundEvent}
   > Set CurrentLoopType(CurrentLoopType=Beat) -> (Output_Get=Lead)
   > PlayLoop()
   "On Clicked (Button_BaseBodyVisbility)"() {is=title K2Node_ComponentBoundEvent}
   n27 = Get BaseBody_Container() -> (BaseBody_Container:Vertical Box Object Reference)
   n28 = "Not Equal (Enum)"!(
    A:ESlateVisibility Enum=Get Visibility!(self:Widget Object Reference=@n27) -> (Visibility:ESlateVisibility Enum=Visible)
    B:ESlateVisibility Enum=SelfHitTestInvisible
   ) -> (ReturnValue:bool)
   > SetVisibility(self=@n27, InVisibility=Select(Option 0=Collapsed, Option 1=SelfHitTestInvisible, Index=@n28))
   n29 = Get HiddenVisibleIcons() -> (HiddenVisibleIcons:Array of Texture 2D Object References)
   > SetBrushFromTexture!(
    execute
    self:Image Object Reference=Get Image_BaseBodyVisbility() -> (Image_BaseBodyVisbility:Image Object Reference)
    Texture:Texture 2D Object Reference=Select(
     Option 0:Texture 2D Object Reference="Get (a copy)"(Array:Array of Texture 2D Object References=@n29) -> (Output:Texture 2D Object Reference)
     Option 1:Texture 2D Object Reference="Get (a copy)"(Array=@n29, Dimension 1=1)
     Index=@n28
    ) -> (ReturnValue:Texture 2D Object Reference)
    bMatchSize:bool=false
   ) -> (then)
   "On Clicked (Button_CountingStars)"() {is=title K2Node_ComponentBoundEvent}
   > Set CurrentSongShortName(CurrentSongShortName=countingstars)
   > PlayLoop()
   "On Clicked (Button_Lead)"() {is=title K2Node_ComponentBoundEvent}
   > Set CurrentLoopType(CurrentLoopType=Misc) -> (Output_Get=Lead)
   > PlayLoop()
   PlayLoop() {is=title K2Node_CustomEvent}
   > n30 = CheckJamJession()
   > n31 = Branch(Condition=IsValidInterface(Interface=@n30))
   n32 = Set PlayerNetId(
    execute=[@n33.then, @n31.then]
    PlayerNetId:Unique Net Id Structure=GetNetIdForPawn!(
     JamSession:Jam Session Interface Interface=Get Jam Session()
     Pawn:Pawn Object Reference=Get PlayerPawn() -> (PlayerPawn:Fort Player Pawn Object Reference)
    ) -> (ReturnValue:Unique Net Id Structure)
   ) -> (Output_Get:Unique Net Id Structure)
   > PlayLoop!(
    execute
    self:Jam Session Component Music Manager Object Reference=Get Component On JamSession(ComponentClass=/Script/FMJamSessionRuntime.JamSessionComponent_MusicManager, JamSession=Get Jam Session()) -> (ReturnValue:Jam Session Component Music Manager Object Reference).ReturnValue
    PlayerNetId:"Unique Net Id Structure (by ref)"=@n32
    SongShortName:Name=String To Name(InString=Get CurrentSongShortName())
    LoopType:EFMJamLoopType Enum=Get CurrentLoopType() -> (CurrentLoopType:EFMJamLoopType Enum=Lead)
    StartMethod:EJamSlotStartMethod Enum=Manual
    bForceAudioState:bool=false
    bIsAutoJammer:bool=false
   ) -> (then, ReturnValue:Jam Music Slot Object Reference) {is=title K2Node_CallFunction}
   PlayEmoteItem!(
    execute=@n31.else
    self:Fort Player Controller Object Reference=Get As Fort Player Controller() -> (As Fort Player Controller:Fort Player Controller Object Reference)
    EmoteAsset:Fort Montage Item Definition Base Object Reference=/SparksSongTemplates/Items/JamEmotes/EID_Placeholder_Global_Bass
    PlayMode:EFortEmotePlayMode Enum=ForcePlay
   ) -> (then)
   > Delay!(execute, WorldContextObject:Object Reference, Duration=0.3) -> (then)
   > n33 = CheckJamJession()
   Generate button() {is=title K2Node_CustomEvent}
   n34 = Get WrapBox_2() -> (WrapBox_2:Wrap Box Object Reference)
   > n35 = Set WidgetBuffer(WidgetBuffer:Widget Object Reference=GetChildAt!(self:Panel Widget Object Reference=@n34, Index:int=0) -> (ReturnValue:Widget Object Reference)) -> (Output_Get:Widget Object Reference)
   > ClearChildren(self=@n34)
   > AddChild(self=@n34, Content=@n35)
   > n36 = To Array!(execute, A:Set of Strings=Get CustomButtonList()) -> (then, Result:Array of Strings)
   > For Each Loop(Array=@n36)
   n37 = Tempo+1() {is=title K2Node_CustomEvent}
   n38 = Tempo-1() {is=title K2Node_CustomEvent}
   n39 = Event Tick()
   > Branch(Condition=IsChecked(self=Get EditorUtilityCheckBox_0() -> (EditorUtilityCheckBox_0:Editor Utility Check Box Object Reference)))
   > n40 = Branch(Condition="float < float"(A=Get BpmOscillatorSpeed() -> (BpmOscillatorSpeed:double)))
   > Set OscillateTempoSense(OscillateTempoSense:bool=NOT Boolean(A=Get OscillateTempoSense() -> (OscillateTempoSense:bool))) -> (Output_Get:bool)
   > n41 = Set BpmOscillatorSpeed(BpmOscillatorSpeed:double="Absolute (Float)"(A=Get BpmOscillatorSpeed())) -> (Output_Get:double)
   n42 = Get BpmOscillatorSpeed()
   n43 = Set TempoBuffer(
    execute=[@n41.then, @n40.else]
    TempoBuffer:double=float + float(A=float * float(A=@n39.InDeltaTime, B=@n42) {is=title K2Node_PromotableOperator}, B=Get TempoBuffer() -> (TempoBuffer:double))
   ) -> (Output_Get:double)
   > Branch(Condition="float >= float"(A=@n43, B=1))
   > Set TempoBuffer(TempoBuffer=0)
   > CheckJamJession()
   > n44 = Get()
   > n45 = Branch(Condition=Get OscillateTempoSense())
   n46 = Get CurrentTempo(self=@n44) -> (CurrentTempo=0)
   > Set OscillateTempoSense(OscillateTempoSense="integer < integer"(A=@n46, B=Get Max() -> (Max:int=0)))
   > Tempo+1() {is=title K2Node_CallFunction}
   Set OscillateTempoSense(execute=@n45.else, OscillateTempoSense="integer <= integer"(A=@n46, B=Get Min() -> (Min:int=0)))
   > Tempo-1() {is=title K2Node_CallFunction}
   Event PreConstruct()
   > SetObject(
    self=[
     Get DetailsView_1() -> (DetailsView_1:Details View Object Reference)
     Get DetailsView_58() -> (DetailsView_58:Details View Object Reference)
     Get SinglePropertyView_BaseBody_3() -> (SinglePropertyView_BaseBody_3:Single Property View Object Reference)
     Get SinglePropertyView_BaseBody_2() -> (SinglePropertyView_BaseBody_2:Single Property View Object Reference)
     Get SinglePropertyView_BaseBody_1() -> (SinglePropertyView_BaseBody_1:Single Property View Object Reference)
     Get SinglePropertyView_BaseBody() -> (SinglePropertyView_BaseBody:Single Property View Object Reference)
     Get SinglePropertyView_1() -> (SinglePropertyView_1:Single Property View Object Reference)
     Get SinglePropertyView_0() -> (SinglePropertyView_0:Single Property View Object Reference)
     Get SinglePropertyView() -> (SinglePropertyView:Single Property View Object Reference)
    ]
    NewObject=Self-Reference()
   )
   Event Construct()
   > CheckJamJession()
   > n47 = Get()
   > UpdateTempo(NewTempo=Get CurrentTempo(self=@n47) -> (CurrentTempo=0))
   String To Name(InString=Get CurrentSongShortName())
   AddChild(self=Get WrapBox_2())
  ]}
 FindPlayerPawn {nodes [
   n1 = FindPlayerPawn!() -> (then, Pawn:Object Reference, WorldContextObject:Object Reference) {is=title K2Node_FunctionEntry}
   > n2 = Cast To PlayerPawn_Athena(Object=@n1)
   > Return Node!(execute, AthenaPawn:Fort Player Pawn Athena Object Reference=@n2)
   n3 = GetAllActorsOfClass(execute=@n2.CastFailed, WorldContextObject=@n1.WorldContextObject, ActorClass=&FortPlayerPawnAthena)
   n4 = "Get (a copy)"(Array:Array of Fort Player Pawn Athena Object References=@n3, Dimension 1=Get PlayerIndex() -> (PlayerIndex=0)) -> (Output:Fort Player Pawn Athena Object Reference)
   > Is Valid(InputObject=@n4)
   > Return Node(AthenaPawn=@n4)
  ]}
 GetColorAndOpacity_Bass {nodes [
   GetColorAndOpacity_Bass() {is=title K2Node_FunctionEntry}
   n1 = Get CurrentLoopType() -> (CurrentLoopType=Lead)
   > Return Node!(
    execute
    ReturnValue:Linear Color Structure=SelectColor(
     A=Get Higlight() -> (Higlight:Linear Color Structure)
     B=(R=1,G=1,B=1,A=1)
     bPickA="Equal (Enum)"(A:EFMJamLoopType Enum=@n1, B:EFMJamLoopType Enum=Bass)
    )
   )
  ]}
 GetColorAndOpacity_Drum {nodes [
   GetColorAndOpacity_Drum() {is=title K2Node_FunctionEntry}
   n1 = Get CurrentLoopType() -> (CurrentLoopType=Lead)
   > Return Node(ReturnValue=SelectColor(A=Get Higlight(), B=(R=1,G=1,B=1,A=1), bPickA="Equal (Enum)"(A=@n1, B=Beat)))
  ]}
 GetColorAndOpacity_Lead {nodes [
   GetColorAndOpacity_Lead() {is=title K2Node_FunctionEntry}
   n1 = Get CurrentLoopType() -> (CurrentLoopType=Lead)
   > Return Node(ReturnValue=SelectColor(A=Get Higlight(), B=(R=1,G=1,B=1,A=1), bPickA="Equal (Enum)"(A=@n1, B=Misc)))
  ]}
 GetColorAndOpacity_Vocal {nodes [
   GetColorAndOpacity_Vocal() {is=title K2Node_FunctionEntry}
   n1 = Get CurrentLoopType() -> (CurrentLoopType=Lead)
   > Return Node(ReturnValue=SelectColor(A=Get Higlight(), B=(R=1,G=1,B=1,A=1), bPickA="Equal (Enum)"(A=@n1, B=Lead)))
  ]}
 GetPawnAndController {nodes [
   n1 = GetPawnAndController!() -> (then, WorldContextObject:Object Reference) {is=title K2Node_FunctionEntry}
   n2 = GetPlayerController(WorldContextObject=@n1, PlayerIndex="Get  WorldIndex"() -> (" WorldIndex"=0))
   > n3 = Cast To FortPlayerController(Object=@n2)
   > n4 = Set PlayerPawn(PlayerPawn=FindPlayerPawn(
     Pawn=GetPlayerPawn!(self:Fort Player Controller Object Reference=@n3) -> (ReturnValue:Fort Player Pawn Object Reference)
     WorldContextObject=@n2
    ) {is=title K2Node_CallFunction})
   > Return Node!(execute, Pawn:Fort Player Pawn Object Reference=@n4, Controller:Fort Player Controller Object Reference=@n3)
  ]}
 GetWorldContext {nodes [
   n1 = GetWorldContext!() -> (then, bIncludeDedicatedServer:bool=~) {is=title K2Node_FunctionEntry}
   > n2 = GetPIEWorlds(bIncludeDedicatedServer=@n1)
   n3 = "Get  WorldIndex"() -> (" WorldIndex"=0)
   > n4 = Branch(Condition="integer > integer"(A=Length(TargetArray:Array of World Object References=@n2), B=@n3))
   > Return Node!(
    execute
    World:World Object Reference="Get (a copy)"(Array:Array of World Object References=@n2, Dimension 1=@n3) -> (Output:World Object Reference)
   )
   Return Node(execute=@n4.else)
  ]}
 Get_Button_CustomSongTag_bIsEnabled {nodes [
   Get_Button_CustomSongTag_bIsEnabled() {is=title K2Node_FunctionEntry}
   n1 = GetWorldContext!(bIncludeDedicatedServer:bool=true) -> (World:World Object Reference) {is=title K2Node_CallFunction}
   > n2 = Branch(Condition=IsValid(Object=@n1))
   n3 = "To String (Text)"(InText=Get Text(self=Get TextBox_CustomSongName()))
   n4 = GetSongCatalogEntryByShortName!(
    self:FMJam Song Catalog Object Reference=TryGetFMJamSongCatalog!(WorldContextObject:Object Reference=@n1) -> (ReturnValue:FMJam Song Catalog Object Reference)
    ShortName:Name=String To Name(InString=@n3)
   ) -> (ReturnValue:Catalog Data Object Reference)
   n5 = Contains Item!(TargetSet:Set of Strings=Get CustomButtonList(), ItemToFind:"String (by ref)"=@n3) -> (ReturnValue:bool) {is=title K2Node_CallFunction}
   n6 = IsEmpty(InString=@n3)
   > Return Node!(execute, ReturnValue:bool=AND Boolean!(A=IsValid(Object=@n4), B=NOT Boolean(A=@n5), C:bool=NOT Boolean(A=@n6)) -> (ReturnValue))
   Return Node(execute=@n2.else, ReturnValue=false)
  ]}
 UpdateTempo {nodes [
   n1 = UpdateTempo!() -> (then, NewTempo:int) {is=title K2Node_FunctionEntry}
   > CheckJamJession()
   > n2 = Get()
   n3 = Get CurrentTempo(self=@n2) -> (CurrentTempo=0)
   > Set Current Tempo(Current Tempo:int=@n3) -> (Output_Get:int=0)
   > Branch(Condition="Not Equal (Integer)"(A=@n3, B=@n1))
   > TriggerTempoChange!(execute, self:Jam Session Component Audio State Object Reference=@n2, Bpm:int=@n1) -> (then)
   > Server_SetTempo!(
    execute
    self:Jam Controller Component Object Reference=Get JamControllerComponent() -> (JamControllerComponent:Jam Controller Component Object Reference)
    NewBPM:int=@n1
   ) -> (then)
  ]}
]
variables [
 " WorldIndex" {type=integer}
 Anim Instance Ref {type=AnimInstance, default=null}
 As Fort Player Controller {type=FortPlayerController, default=null}
 BpmOscillatorSpeed {type=number, default=6}
 Current Tempo {type=integer}
 CurrentLoopType {type=EFMJamLoopType, default=Beat}
 CurrentSongShortName {type=string, default=nightmare}
 CustomButtonList {type="array<string>", default []}
 CustomTempo {type=integer, default=112}
 DEBUG_JamV2_BaseBody_OverrideSequence {type=AnimSequenceBase, default=/SparksCharacterCommon/Animation/MainPlayer/BandPlay/Drum/Hits/TPose}
 DEBUG_JamV2_BaseBody_TraversalHomebaseBPMs {
  type=SparksAnimBPMSet
  default {
   allowDoublingHalving=true
   hysteresis=0.07
   bPMBuckets [
    /SparksCharacterCommon/Animation/MainPlayer/BandPlay/Guitar/Loops/JamV2Guitar_IntensitySet_095
    /SparksCharacterCommon/Animation/MainPlayer/BandPlay/Guitar/Loops/JamV2Guitar_IntensitySet_125
    /SparksCharacterCommon/Animation/MainPlayer/BandPlay/Guitar/Loops/JamV2Guitar_IntensitySet_165
   ]
  }
 }
 DEBUG_JamV2_FaceExpression_OverrideSongGenre {type=Enum_MusicGenre, default=Rap/HipHop}
 DEBUG_JamV2_FaceExpression_OverrideSongIntensity {type=AnimIntensity, default=Light}
 HiddenVisibleIcons {
  type="array<Texture2D>"
  default [/Game/Athena/HUD/Bacchus/HUDLayoutTool/HLT_Icons/T-HLT-Visible, /Game/Athena/HUD/Bacchus/HUDLayoutTool/HLT_Icons/T-HLT-Hidden]
 }
 Higlight {type=LinearColor, default=<rgba 0 0.7585 1 1>}
 Jam Session {default=None}
 JamAudioState {type=JamSessionComponent_AudioState, default=null}
 JamControllerComponent {type=JamControllerComponent, default=null}
 JamMusicManagerComponent {type=JamSessionComponent_MusicManager, default=null}
 Max {type=integer, default=165}
 Min {type=integer, default=75}
 OscillateTempoSense {type=boolean}
 PlayerIndex {type=integer}
 PlayerNetId {type=UniqueNetIdRepl, default=INVALID}
 PlayerPawn {type=FortPlayerPawn}
 TempoBuffer {type=number}
 WidgetBuffer {type=Widget, default=null}
]
widgets {
 info {rootWidgetClass=&CanvasPanel, widgetCount=155}
 widgets [
  CanvasPanel_27 {widgetClassPath=&CanvasPanel, widget=CanvasPanel_27}
  AthenaScrollBox_0 {
   widgetClassPath=/Script/FortniteUI.AthenaScrollBox
   parent=CanvasPanel_27
   widget=AthenaScrollBox_0
   properties {
    widgetBarStyle {
     thickness=8
     draggedThumbImage {
      bIsDynamicallyLoaded=false
      drawAs=RoundedBox
      imageType=NoImage
      mirroring=NoMirror
      resourceName=None
      resourceObject=null
      tiling=NoTile
      imageSize=<0 0>
      margin=<ltrb 0 0 0 0>
      outlineSettings {
       bUseBrushTransparency=false
       roundingType=FixedRadius
       width=0
       color {colorUseRule=UseColor_Specified, specifiedColor=<rgba 0 0 0 0>}
       cornerRadii=<4 4 4 4>
      }
      tintColor {colorUseRule=UseColor_Specified, specifiedColor=<rgba 0.2159 0.2159 0.2159 1>}
      uVRegion {bIsValid=false, max=<0 0>, min=<0 0>}
     }
     horizontalBackgroundImage {
      bIsDynamicallyLoaded=false
      drawAs=NoDrawType
      imageType=NoImage
      mirroring=NoMirror
      resourceName=None
      resourceObject=null
      tiling=NoTile
      imageSize=<0 0>
      margin=<ltrb 0 0 0 0>
      outlineSettings {
       bUseBrushTransparency=false
       roundingType=HalfHeightRadius
       width=0
       color {colorUseRule=UseColor_Specified, specifiedColor=<rgba 0 0 0 0>}
       cornerRadii=<0 0 0 0>
      }
      tintColor {colorUseRule=UseColor_Specified, specifiedColor=<rgba 1 1 1 1>}
      uVRegion {bIsValid=false, max=<0 0>, min=<0 0>}
     }
     horizontalBottomSlotImage {
      bIsDynamicallyLoaded=false
      drawAs=NoDrawType
      imageType=NoImage
      mirroring=NoMirror
      resourceName=None
      resourceObject=null
      tiling=NoTile
      imageSize=<0 0>
      margin=<ltrb 0 0 0 0>
      outlineSettings {
       bUseBrushTransparency=false
       roundingType=HalfHeightRadius
       width=0
       color {colorUseRule=UseColor_Specified, specifiedColor=<rgba 0 0 0 0>}
       cornerRadii=<0 0 0 0>
      }
      tintColor {colorUseRule=UseColor_Specified, specifiedColor=<rgba 1 1 1 1>}
      uVRegion {bIsValid=false, max=<0 0>, min=<0 0>}
     }
     horizontalTopSlotImage {
      bIsDynamicallyLoaded=false
      drawAs=NoDrawType
      imageType=NoImage
      mirroring=NoMirror
      resourceName=None
      resourceObject=null
      tiling=NoTile
      imageSize=<0 0>
      margin=<ltrb 0 0 0 0>
      outlineSettings {
       bUseBrushTransparency=false
       roundingType=HalfHeightRadius
       width=0
       color {colorUseRule=UseColor_Specified, specifiedColor=<rgba 0 0 0 0>}
       cornerRadii=<0 0 0 0>
      }
      tintColor {colorUseRule=UseColor_Specified, specifiedColor=<rgba 1 1 1 1>}
      uVRegion {bIsValid=false, max=<0 0>, min=<0 0>}
     }
     hoveredThumbImage {
      bIsDynamicallyLoaded=false
      drawAs=RoundedBox
      imageType=NoImage
      mirroring=NoMirror
      resourceName=None
      resourceObject=null
      tiling=NoTile
      imageSize=<0 0>
      margin=<ltrb 0 0 0 0>
      outlineSettings {
       bUseBrushTransparency=false
       roundingType=FixedRadius
       width=0
       color {colorUseRule=UseColor_Specified, specifiedColor=<rgba 0 0 0 0>}
       cornerRadii=<4 4 4 4>
      }
      tintColor {colorUseRule=UseColor_Specified, specifiedColor=<rgba 0.2159 0.2159 0.2159 1>}
      uVRegion {bIsValid=false, max=<0 0>, min=<0 0>}
     }
     normalThumbImage {
      bIsDynamicallyLoaded=false
      drawAs=RoundedBox
      imageType=NoImage
      mirroring=NoMirror
      resourceName=None
      resourceObject=null
      tiling=NoTile
      imageSize=<0 0>
      margin=<ltrb 0 0 0 0>
      outlineSettings {
       bUseBrushTransparency=false
       roundingType=FixedRadius
       width=0
       color {colorUseRule=UseColor_Specified, specifiedColor=<rgba 0 0 0 0>}
       cornerRadii=<4 4 4 4>
      }
      tintColor {colorUseRule=UseColor_Specified, specifiedColor=<rgba 0.0953 0.0953 0.0953 1>}
      uVRegion {bIsValid=false, max=<0 0>, min=<0 0>}
     }
     verticalBackgroundImage {
      bIsDynamicallyLoaded=false
      drawAs=NoDrawType
      imageType=NoImage
      mirroring=NoMirror
      resourceName=None
      resourceObject=null
      tiling=NoTile
      imageSize=<0 0>
      margin=<ltrb 0 0 0 0>
      outlineSettings {
       bUseBrushTransparency=false
       roundingType=HalfHeightRadius
       width=0
       color {colorUseRule=UseColor_Specified, specifiedColor=<rgba 0 0 0 0>}
       cornerRadii=<0 0 0 0>
      }
      tintColor {colorUseRule=UseColor_Specified, specifiedColor=<rgba 1 1 1 1>}
      uVRegion {bIsValid=false, max=<0 0>, min=<0 0>}
     }
     verticalBottomSlotImage {
      bIsDynamicallyLoaded=false
      drawAs=NoDrawType
      imageType=NoImage
      mirroring=NoMirror
      resourceName=None
      resourceObject=null
      tiling=NoTile
      imageSize=<0 0>
      margin=<ltrb 0 0 0 0>
      outlineSettings {
       bUseBrushTransparency=false
       roundingType=HalfHeightRadius
       width=0
       color {colorUseRule=UseColor_Specified, specifiedColor=<rgba 0 0 0 0>}
       cornerRadii=<0 0 0 0>
      }
      tintColor {colorUseRule=UseColor_Specified, specifiedColor=<rgba 1 1 1 1>}
      uVRegion {bIsValid=false, max=<0 0>, min=<0 0>}
     }
     verticalTopSlotImage {
      bIsDynamicallyLoaded=false
      drawAs=NoDrawType
      imageType=NoImage
      mirroring=NoMirror
      resourceName=None
      resourceObject=null
      tiling=NoTile
      imageSize=<0 0>
      margin=<ltrb 0 0 0 0>
      outlineSettings {
       bUseBrushTransparency=false
       roundingType=HalfHeightRadius
       width=0
       color {colorUseRule=UseColor_Specified, specifiedColor=<rgba 0 0 0 0>}
       cornerRadii=<0 0 0 0>
      }
      tintColor {colorUseRule=UseColor_Specified, specifiedColor=<rgba 1 1 1 1>}
      uVRegion {bIsValid=false, max=<0 0>, min=<0 0>}
     }
    }
    widgetStyle {
     barThickness=8
     bottomShadowBrush {
      bIsDynamicallyLoaded=false
      drawAs=Box
      imageType=FullColor
      mirroring=NoMirror
      resourceName=../../../Engine/Content/Slate/Common/ScrollBoxShadowBottom.png
      resourceObject=null
      tiling=NoTile
      imageSize=<16 8>
      margin=<ltrb 0.5 0 0.5 1>
      outlineSettings {
       bUseBrushTransparency=false
       roundingType=HalfHeightRadius
       width=0
       color {colorUseRule=UseColor_Specified, specifiedColor=<rgba 0 0 0 0>}
       cornerRadii=<0 0 0 0>
      }
      tintColor {colorUseRule=UseColor_Specified, specifiedColor=<rgba 1 1 1 1>}
      uVRegion {bIsValid=false, max=<0 0>, min=<0 0>}
     }
     horizontalScrolledContentPadding=<ltrb 0 0 1 0>
     leftShadowBrush {
      bIsDynamicallyLoaded=false
      drawAs=Box
      imageType=FullColor
      mirroring=NoMirror
      resourceName=../../../Engine/Content/Slate/Common/ScrollBoxShadowLeft.png
      resourceObject=null
      tiling=NoTile
      imageSize=<8 16>
      margin=<ltrb 1 0.5 0 0.5>
      outlineSettings {
       bUseBrushTransparency=false
       roundingType=HalfHeightRadius
       width=0
       color {colorUseRule=UseColor_Specified, specifiedColor=<rgba 0 0 0 0>}
       cornerRadii=<0 0 0 0>
      }
      tintColor {colorUseRule=UseColor_Specified, specifiedColor=<rgba 1 1 1 1>}
      uVRegion {bIsValid=false, max=<0 0>, min=<0 0>}
     }
     rightShadowBrush {
      bIsDynamicallyLoaded=false
      drawAs=Box
      imageType=FullColor
      mirroring=NoMirror
      resourceName=../../../Engine/Content/Slate/Common/ScrollBoxShadowRight.png
      resourceObject=null
      tiling=NoTile
      imageSize=<8 16>
      margin=<ltrb 0 0.5 1 0.5>
      outlineSettings {
       bUseBrushTransparency=false
       roundingType=HalfHeightRadius
       width=0
       color {colorUseRule=UseColor_Specified, specifiedColor=<rgba 0 0 0 0>}
       cornerRadii=<0 0 0 0>
      }
      tintColor {colorUseRule=UseColor_Specified, specifiedColor=<rgba 1 1 1 1>}
      uVRegion {bIsValid=false, max=<0 0>, min=<0 0>}
     }
     topShadowBrush {
      bIsDynamicallyLoaded=false
      drawAs=Box
      imageType=FullColor
      mirroring=NoMirror
      resourceName=../../../Engine/Content/Slate/Common/ScrollBoxShadowTop.png
      resourceObject=null
      tiling=NoTile
      imageSize=<16 8>
      margin=<ltrb 0.5 1 0.5 0>
      outlineSettings {
       bUseBrushTransparency=false
       roundingType=HalfHeightRadius
       width=0
       color {colorUseRule=UseColor_Specified, specifiedColor=<rgba 0 0 0 0>}
       cornerRadii=<0 0 0 0>
      }
      tintColor {colorUseRule=UseColor_Specified, specifiedColor=<rgba 1 1 1 1>}
      uVRegion {bIsValid=false, max=<0 0>, min=<0 0>}
     }
     verticalScrolledContentPadding=<ltrb 0 0 0 1>
    }
   }
   slot {class=CanvasPanelSlot, properties {layoutData {alignment=<0 0>, anchors {maximum=<1 1>, minimum=<0 0>}, offsets=<ltrb 0 0 0 0>}}}
  }
  VerticalBox_0 {widgetClassPath=&VerticalBox, parent=AthenaScrollBox_0, widget=VerticalBox_0, slot {class=ScrollBoxSlot}}
  HorizontalBox_77 {widgetClassPath=&HorizontalBox, parent=VerticalBox_0, widget=HorizontalBox_77, slot {class=VerticalBoxSlot}}
  TextBlock_99 {
   widgetClassPath=&TextBlock
   parent=HorizontalBox_77
   widget=TextBlock_99
   properties {
    text="BPM : "
    font {
     bForceMonospaced=false
     bMaterialIsStencil=false
     fontMaterial=null
     fontObject=&Roboto
     letterSpacing=0
     monospacedWidth=1
     size=16.5
     skewAmount=0
     typefaceFontName=Bold
     outlineSettings {
      bApplyOutlineToDropShadows=false
      bMiteredCorners=false
      bSeparateFillAlpha=false
      outlineMaterial=null
      outlineSize=0
      outlineColor=<rgba 0 0 0 1>
     }
    }
   }
   slot {class=HorizontalBoxSlot, properties {verticalAlignment=VAlign_Center}}
  }
  TextBlock_40 {
   widgetClassPath=&TextBlock
   parent=HorizontalBox_77
   widget=TextBlock_40
   properties {
    text="Current - "
    font {
     bForceMonospaced=false
     bMaterialIsStencil=false
     fontMaterial=null
     fontObject=&Roboto
     letterSpacing=0
     monospacedWidth=1
     size=11.25
     skewAmount=0
     typefaceFontName=Bold
     outlineSettings {
      bApplyOutlineToDropShadows=false
      bMiteredCorners=false
      bSeparateFillAlpha=false
      outlineMaterial=null
      outlineSize=0
      outlineColor=<rgba 0 0 0 1>
     }
    }
   }
   slot {class=HorizontalBoxSlot, properties {verticalAlignment=VAlign_Center}}
  }
  SizeBox_2 {
   widgetClassPath=&SizeBox
   parent=HorizontalBox_77
   widget=SizeBox_2
   properties {bOverride_WidthOverride=true, widthOverride=25.269}
   slot {class=HorizontalBoxSlot, properties {verticalAlignment=VAlign_Center}}
  }
  TextBlock_41 {
   widgetClassPath=&TextBlock
   parent=SizeBox_2
   widget=TextBlock_41
   properties {
    text=0
    colorAndOpacity {colorUseRule=UseColor_Specified, specifiedColor=<rgba 0.029 0.6786 1 1>}
    font {
     bForceMonospaced=false
     bMaterialIsStencil=false
     fontMaterial=null
     fontObject=&Roboto
     letterSpacing=0
     monospacedWidth=1
     size=11.25
     skewAmount=0
     typefaceFontName=Bold
     outlineSettings {
      bApplyOutlineToDropShadows=false
      bMiteredCorners=false
      bSeparateFillAlpha=false
      outlineMaterial=null
      outlineSize=0
      outlineColor=<rgba 0 0 0 1>
     }
    }
   }
   slot {class=SizeBoxSlot}
  }
  EditorUtilityButton_1 {
   widgetClassPath=&EditorUtilityButton
   parent=HorizontalBox_77
   widget=EditorUtilityButton_1
   slot {class=HorizontalBoxSlot, properties {padding=<ltrb 10 0 0 0>}}
  }
  TextBlock_39 {
   widgetClassPath=&TextBlock
   parent=EditorUtilityButton_1
   widget=TextBlock_39
   properties {
    text=-
    font {
     bForceMonospaced=false
     bMaterialIsStencil=false
     fontMaterial=null
     fontObject=&Roboto
     letterSpacing=0
     monospacedWidth=1
     size=15
     skewAmount=0
     typefaceFontName=Bold
     outlineSettings {
      bApplyOutlineToDropShadows=false
      bMiteredCorners=false
      bSeparateFillAlpha=false
      outlineMaterial=null
      outlineSize=0
      outlineColor=<rgba 0 0 0 1>
     }
    }
   }
   slot {class=ButtonSlot, properties {padding=<ltrb 0 0 0 0>}}
  }
  EditorUtilityButton {
   widgetClassPath=&EditorUtilityButton
   parent=HorizontalBox_77
   widget=EditorUtilityButton
   slot {class=HorizontalBoxSlot, properties {padding=<ltrb 0 0 10 0>}}
  }
  TextBlock_38 {
   widgetClassPath=&TextBlock
   parent=EditorUtilityButton
   widget=TextBlock_38
   properties {
    text=+
    font {
     bForceMonospaced=false
     bMaterialIsStencil=false
     fontMaterial=null
     fontObject=&Roboto
     letterSpacing=0
     monospacedWidth=1
     size=15
     skewAmount=0
     typefaceFontName=Bold
     outlineSettings {
      bApplyOutlineToDropShadows=false
      bMiteredCorners=false
      bSeparateFillAlpha=false
      outlineMaterial=null
      outlineSize=0
      outlineColor=<rgba 0 0 0 1>
     }
    }
   }
   slot {class=ButtonSlot, properties {padding=<ltrb 0 0 0 0>}}
  }
  HorizontalBox_18 {widgetClassPath=&HorizontalBox, parent=HorizontalBox_77, widget=HorizontalBox_18, slot {class=HorizontalBoxSlot}}
  EditorUtilityButton_Bpm80 {
   widgetClassPath=&EditorUtilityButton
   parent=HorizontalBox_18
   widget=EditorUtilityButton_Bpm80
   slot {class=HorizontalBoxSlot}
  }
  TextBlock_67 {
   widgetClassPath=&TextBlock
   parent=EditorUtilityButton_Bpm80
   widget=TextBlock_67
   properties {
    text=80
    font {
     bForceMonospaced=false
     bMaterialIsStencil=false
     fontMaterial=null
     fontObject=&Roboto
     letterSpacing=0
     monospacedWidth=1
     size=11.25
     skewAmount=0
     typefaceFontName=Bold
     outlineSettings {
      bApplyOutlineToDropShadows=false
      bMiteredCorners=false
      bSeparateFillAlpha=false
      outlineMaterial=null
      outlineSize=0
      outlineColor=<rgba 0 0 0 1>
     }
    }
   }
   slot {class=ButtonSlot}
  }
  EditorUtilityButton_Bpm95 {
   widgetClassPath=&EditorUtilityButton
   parent=HorizontalBox_18
   widget=EditorUtilityButton_Bpm95
   slot {class=HorizontalBoxSlot}
  }
  TextBlock_47 {
   widgetClassPath=&TextBlock
   parent=EditorUtilityButton_Bpm95
   widget=TextBlock_47
   properties {
    text=95
    font {
     bForceMonospaced=false
     bMaterialIsStencil=false
     fontMaterial=null
     fontObject=&Roboto
     letterSpacing=0
     monospacedWidth=1
     size=11.25
     skewAmount=0
     typefaceFontName=Bold
     outlineSettings {
      bApplyOutlineToDropShadows=false
      bMiteredCorners=false
      bSeparateFillAlpha=false
      outlineMaterial=null
      outlineSize=0
      outlineColor=<rgba 0 0 0 1>
     }
    }
   }
   slot {class=ButtonSlot}
  }
  EditorUtilityButton_Bpm125 {
   widgetClassPath=&EditorUtilityButton
   parent=HorizontalBox_18
   widget=EditorUtilityButton_Bpm125
   slot {class=HorizontalBoxSlot}
  }
  TextBlock_113 {
   widgetClassPath=&TextBlock
   parent=EditorUtilityButton_Bpm125
   widget=TextBlock_113
   properties {
    text=125
    font {
     bForceMonospaced=false
     bMaterialIsStencil=false
     fontMaterial=null
     fontObject=&Roboto
     letterSpacing=0
     monospacedWidth=1
     size=11.25
     skewAmount=0
     typefaceFontName=Bold
     outlineSettings {
      bApplyOutlineToDropShadows=false
      bMiteredCorners=false
      bSeparateFillAlpha=false
      outlineMaterial=null
      outlineSize=0
      outlineColor=<rgba 0 0 0 1>
     }
    }
   }
   slot {class=ButtonSlot}
  }
  EditorUtilityButton_Bpm165 {
   widgetClassPath=&EditorUtilityButton
   parent=HorizontalBox_18
   widget=EditorUtilityButton_Bpm165
   slot {class=HorizontalBoxSlot}
  }
  TextBlock_156 {
   widgetClassPath=&TextBlock
   parent=EditorUtilityButton_Bpm165
   widget=TextBlock_156
   properties {
    text=165
    font {
     bForceMonospaced=false
     bMaterialIsStencil=false
     fontMaterial=null
     fontObject=&Roboto
     letterSpacing=0
     monospacedWidth=1
     size=11.25
     skewAmount=0
     typefaceFontName=Bold
     outlineSettings {
      bApplyOutlineToDropShadows=false
      bMiteredCorners=false
      bSeparateFillAlpha=false
      outlineMaterial=null
      outlineSize=0
      outlineColor=<rgba 0 0 0 1>
     }
    }
   }
   slot {class=ButtonSlot}
  }
  EditorUtilityButton_Bpm180 {
   widgetClassPath=&EditorUtilityButton
   parent=HorizontalBox_18
   widget=EditorUtilityButton_Bpm180
   slot {class=HorizontalBoxSlot}
  }
  TextBlock_46 {
   widgetClassPath=&TextBlock
   parent=EditorUtilityButton_Bpm180
   widget=TextBlock_46
   properties {
    text=180
    font {
     bForceMonospaced=false
     bMaterialIsStencil=false
     fontMaterial=null
     fontObject=&Roboto
     letterSpacing=0
     monospacedWidth=1
     size=11.25
     skewAmount=0
     typefaceFontName=Bold
     outlineSettings {
      bApplyOutlineToDropShadows=false
      bMiteredCorners=false
      bSeparateFillAlpha=false
      outlineMaterial=null
      outlineSize=0
      outlineColor=<rgba 0 0 0 1>
     }
    }
   }
   slot {class=ButtonSlot}
  }
  EditorUtilityButton_123 {
   widgetClassPath=&EditorUtilityButton
   parent=HorizontalBox_77
   widget=EditorUtilityButton_123
   slot {class=HorizontalBoxSlot}
  }
  EditorUtilityEditableTextBox_0 {
   widgetClassPath=&EditorUtilityEditableTextBox
   bIsVariable=true
   parent=EditorUtilityButton_123
   widget=EditorUtilityEditableTextBox_0
   properties {text=112}
   slot {class=ButtonSlot, properties {padding=<ltrb 4 2 12 2>}}
  }
  HorizontalBox_19 {widgetClassPath=&HorizontalBox, parent=VerticalBox_0, widget=HorizontalBox_19, slot {class=VerticalBoxSlot}}
  EditorUtilityCheckBox_0 {
   widgetClassPath=&EditorUtilityCheckBox
   bIsVariable=true
   parent=HorizontalBox_19
   widget=EditorUtilityCheckBox_0
   slot {class=HorizontalBoxSlot, properties {padding=<ltrb 10 0 0 0>}}
  }
  TextBlock_43 {
   widgetClassPath=&TextBlock
   parent=HorizontalBox_19
   widget=TextBlock_43
   properties {
    text="Osillate BPM "
    font {
     bForceMonospaced=false
     bMaterialIsStencil=false
     fontMaterial=null
     fontObject=&Roboto
     letterSpacing=0
     monospacedWidth=1
     size=9
     skewAmount=0
     typefaceFontName=Bold
     outlineSettings {
      bApplyOutlineToDropShadows=false
      bMiteredCorners=false
      bSeparateFillAlpha=false
      outlineMaterial=null
      outlineSize=0
      outlineColor=<rgba 0 0 0 1>
     }
    }
   }
   slot {class=HorizontalBoxSlot, properties {verticalAlignment=VAlign_Center}}
  }
  SinglePropertyView_0 {
   widgetClassPath=&SinglePropertyView
   parent=HorizontalBox_19
   widget=SinglePropertyView_0
   properties {propertyName=Min}
   slot {class=HorizontalBoxSlot, properties {horizontalAlignment=HAlign_Left, padding=<ltrb 0 0 -25 0>}}
  }
  TextBlock_44 {
   widgetClassPath=&TextBlock
   parent=HorizontalBox_19
   widget=TextBlock_44
   properties {
    text=" to "
    font {
     bForceMonospaced=false
     bMaterialIsStencil=false
     fontMaterial=null
     fontObject=&Roboto
     letterSpacing=0
     monospacedWidth=1
     size=9
     skewAmount=0
     typefaceFontName=Bold
     outlineSettings {
      bApplyOutlineToDropShadows=false
      bMiteredCorners=false
      bSeparateFillAlpha=false
      outlineMaterial=null
      outlineSize=0
      outlineColor=<rgba 0 0 0 1>
     }
    }
   }
   slot {class=HorizontalBoxSlot, properties {verticalAlignment=VAlign_Center}}
  }
  SinglePropertyView_1 {
   widgetClassPath=&SinglePropertyView
   parent=HorizontalBox_19
   widget=SinglePropertyView_1
   properties {propertyName=Max}
   slot {class=HorizontalBoxSlot, properties {padding=<ltrb 0 0 -25 0>}}
  }
  TextBlock_45 {
   widgetClassPath=&TextBlock
   parent=HorizontalBox_19
   widget=TextBlock_45
   properties {
    text=at
    font {
     bForceMonospaced=false
     bMaterialIsStencil=false
     fontMaterial=null
     fontObject=&Roboto
     letterSpacing=0
     monospacedWidth=1
     size=9
     skewAmount=0
     typefaceFontName=Bold
     outlineSettings {
      bApplyOutlineToDropShadows=false
      bMiteredCorners=false
      bSeparateFillAlpha=false
      outlineMaterial=null
      outlineSize=0
      outlineColor=<rgba 0 0 0 1>
     }
    }
   }
   slot {class=HorizontalBoxSlot, properties {verticalAlignment=VAlign_Center}}
  }
  SinglePropertyView {
   widgetClassPath=&SinglePropertyView
   parent=HorizontalBox_19
   widget=SinglePropertyView
   properties {
    nameOverride=Speed
    propertyName=BpmOscillatorSpeed
    toolTipText=If you put a negative value it will flip oscillation and remove the negative
   }
   slot {class=HorizontalBoxSlot}
  }
  HorizontalBox {widgetClassPath=&HorizontalBox, parent=VerticalBox_0, widget=HorizontalBox, slot {class=VerticalBoxSlot}}
  TextBlock {
   widgetClassPath=&TextBlock
   parent=HorizontalBox
   widget=TextBlock
   properties {
    text="Instrument : "
    font {
     bForceMonospaced=false
     bMaterialIsStencil=false
     fontMaterial=null
     fontObject=&Roboto
     letterSpacing=0
     monospacedWidth=1
     size=16.5
     skewAmount=0
     typefaceFontName=Bold
     outlineSettings {
      bApplyOutlineToDropShadows=false
      bMiteredCorners=false
      bSeparateFillAlpha=false
      outlineMaterial=null
      outlineSize=0
      outlineColor=<rgba 0 0 0 1>
     }
    }
   }
   slot {class=HorizontalBoxSlot, properties {verticalAlignment=VAlign_Center}}
  }
  Button_Drum {widgetClassPath=&EditorUtilityButton, parent=HorizontalBox, widget=Button_Drum, slot {class=HorizontalBoxSlot}}
  Image_1 {
   widgetClassPath=&Image
   bIsVariable=true
   parent=Button_Drum
   widget=Image_1
   properties {
    brush {
     bIsDynamicallyLoaded=false
     drawAs=Image
     imageType=FullColor
     mirroring=NoMirror
     resourceName=None
     resourceObject=/PilgrimQuickplay/Widgets/Materials/MI_UI_Icon_Instrument_Type_Drums
     tiling=NoTile
     imageSize=<32 32>
     margin=<ltrb 0 0 0 0>
     outlineSettings {
      bUseBrushTransparency=false
      roundingType=HalfHeightRadius
      width=0
      color {colorUseRule=UseColor_Specified, specifiedColor=<rgba 0 0 0 0>}
      cornerRadii=<0 0 0 0>
     }
     tintColor {colorUseRule=UseColor_Specified, specifiedColor=<rgba 1 1 1 1>}
     uVRegion {bIsValid=false, max=<0 0>, min=<0 0>}
    }
    renderTransform {angle=0, scale=<1.5 1.5>, shear=<0 0>, translation=<0 0>}
   }
   slot {class=ButtonSlot}
  }
  Button_Lead {widgetClassPath=&EditorUtilityButton, parent=HorizontalBox, widget=Button_Lead, slot {class=HorizontalBoxSlot}}
  Image_2 {
   widgetClassPath=&Image
   bIsVariable=true
   parent=Button_Lead
   widget=Image_2
   properties {
    brush {
     bIsDynamicallyLoaded=false
     drawAs=Image
     imageType=FullColor
     mirroring=NoMirror
     resourceName=None
     resourceObject=/PilgrimQuickplay/Widgets/Materials/MI_UI_Icon_Instrument_Type_Guitar
     tiling=NoTile
     imageSize=<32 32>
     margin=<ltrb 0 0 0 0>
     outlineSettings {
      bUseBrushTransparency=false
      roundingType=HalfHeightRadius
      width=0
      color {colorUseRule=UseColor_Specified, specifiedColor=<rgba 0 0 0 0>}
      cornerRadii=<0 0 0 0>
     }
     tintColor {colorUseRule=UseColor_Specified, specifiedColor=<rgba 1 1 1 1>}
     uVRegion {bIsValid=false, max=<0 0>, min=<0 0>}
    }
    renderTransform {angle=0, scale=<1.5 1.5>, shear=<0 0>, translation=<0 0>}
   }
   slot {class=ButtonSlot}
  }
  Button_Bass {widgetClassPath=&EditorUtilityButton, parent=HorizontalBox, widget=Button_Bass, slot {class=HorizontalBoxSlot}}
  Image_3 {
   widgetClassPath=&Image
   bIsVariable=true
   parent=Button_Bass
   widget=Image_3
   properties {
    brush {
     bIsDynamicallyLoaded=false
     drawAs=Image
     imageType=FullColor
     mirroring=NoMirror
     resourceName=None
     resourceObject=/PilgrimQuickplay/Widgets/Materials/MI_UI_Icon_Instrument_Type_Bass
     tiling=NoTile
     imageSize=<32 32>
     margin=<ltrb 0 0 0 0>
     outlineSettings {
      bUseBrushTransparency=false
      roundingType=HalfHeightRadius
      width=0
      color {colorUseRule=UseColor_Specified, specifiedColor=<rgba 0 0 0 0>}
      cornerRadii=<0 0 0 0>
     }
     tintColor {colorUseRule=UseColor_Specified, specifiedColor=<rgba 1 1 1 1>}
     uVRegion {bIsValid=false, max=<0 0>, min=<0 0>}
    }
    renderTransform {angle=0, scale=<1.5 1.5>, shear=<0 0>, translation=<0 0>}
   }
   slot {class=ButtonSlot}
  }
  Button_Vocal {widgetClassPath=&EditorUtilityButton, parent=HorizontalBox, widget=Button_Vocal, slot {class=HorizontalBoxSlot}}
  Image_4 {
   widgetClassPath=&Image
   bIsVariable=true
   parent=Button_Vocal
   widget=Image_4
   properties {
    brush {
     bIsDynamicallyLoaded=false
     drawAs=Image
     imageType=FullColor
     mirroring=NoMirror
     resourceName=None
     resourceObject=/PilgrimQuickplay/Widgets/Materials/MI_UI_Icon_Instrument_Type_Vocals
     tiling=NoTile
     imageSize=<32 32>
     margin=<ltrb 0 0 0 0>
     outlineSettings {
      bUseBrushTransparency=false
      roundingType=HalfHeightRadius
      width=0
      color {colorUseRule=UseColor_Specified, specifiedColor=<rgba 0 0 0 0>}
      cornerRadii=<0 0 0 0>
     }
     tintColor {colorUseRule=UseColor_Specified, specifiedColor=<rgba 1 1 1 1>}
     uVRegion {bIsValid=false, max=<0 0>, min=<0 0>}
    }
    renderTransform {angle=0, scale=<1.5 1.5>, shear=<0 0>, translation=<0 0>}
   }
   slot {class=ButtonSlot}
  }
  Button_Vocal_1 {widgetClassPath=&EditorUtilityButton, parent=HorizontalBox, widget=Button_Vocal_1, slot {class=HorizontalBoxSlot}}
  TextBlock_33 {
   widgetClassPath=&TextBlock
   parent=Button_Vocal_1
   widget=TextBlock_33
   properties {
    text=Stop
    font {
     bForceMonospaced=false
     bMaterialIsStencil=false
     fontMaterial=null
     fontObject=&Roboto
     letterSpacing=0
     monospacedWidth=1
     size=11.25
     skewAmount=0
     typefaceFontName=Bold
     outlineSettings {
      bApplyOutlineToDropShadows=false
      bMiteredCorners=false
      bSeparateFillAlpha=false
      outlineMaterial=null
      outlineSize=0
      outlineColor=<rgba 0 0 0 1>
     }
    }
   }
   slot {class=ButtonSlot}
  }
  VerticalBox_2 {widgetClassPath=&VerticalBox, parent=VerticalBox_0, widget=VerticalBox_2, slot {class=VerticalBoxSlot}}
  HorizontalBox_1 {widgetClassPath=&HorizontalBox, parent=VerticalBox_2, widget=HorizontalBox_1, slot {class=VerticalBoxSlot}}
  TextBlock_5 {
   widgetClassPath=&TextBlock
   parent=HorizontalBox_1
   widget=TextBlock_5
   properties {
    text="Song : "
    font {
     bForceMonospaced=false
     bMaterialIsStencil=false
     fontMaterial=null
     fontObject=&Roboto
     letterSpacing=0
     monospacedWidth=1
     size=16.5
     skewAmount=0
     typefaceFontName=Bold
     outlineSettings {
      bApplyOutlineToDropShadows=false
      bMiteredCorners=false
      bSeparateFillAlpha=false
      outlineMaterial=null
      outlineSize=0
      outlineColor=<rgba 0 0 0 1>
     }
    }
   }
   slot {class=HorizontalBoxSlot, properties {verticalAlignment=VAlign_Center}}
  }
  TextBlock_34 {
   widgetClassPath=&TextBlock
   parent=HorizontalBox_1
   widget=TextBlock_34
   properties {
    text="Current song - "
    font {
     bForceMonospaced=false
     bMaterialIsStencil=false
     fontMaterial=null
     fontObject=&Roboto
     letterSpacing=0
     monospacedWidth=1
     size=11.25
     skewAmount=0
     typefaceFontName=Bold
     outlineSettings {
      bApplyOutlineToDropShadows=false
      bMiteredCorners=false
      bSeparateFillAlpha=false
      outlineMaterial=null
      outlineSize=0
      outlineColor=<rgba 0 0 0 1>
     }
    }
   }
   slot {class=HorizontalBoxSlot, properties {verticalAlignment=VAlign_Center}}
  }
  TextBlock_35 {
   widgetClassPath=&TextBlock
   parent=HorizontalBox_1
   widget=TextBlock_35
   properties {
    text=none
    colorAndOpacity {colorUseRule=UseColor_Specified, specifiedColor=<rgba 0.029 0.6786 1 1>}
    font {
     bForceMonospaced=false
     bMaterialIsStencil=false
     fontMaterial=null
     fontObject=&Roboto
     letterSpacing=0
     monospacedWidth=1
     size=11.25
     skewAmount=0
     typefaceFontName=Bold
     outlineSettings {
      bApplyOutlineToDropShadows=false
      bMiteredCorners=false
      bSeparateFillAlpha=false
      outlineMaterial=null
      outlineSize=0
      outlineColor=<rgba 0 0 0 1>
     }
    }
   }
   slot {class=HorizontalBoxSlot, properties {verticalAlignment=VAlign_Center}}
  }
  HorizontalBox_13 {widgetClassPath=&HorizontalBox, parent=VerticalBox_2, widget=HorizontalBox_13, slot {class=VerticalBoxSlot}}
  TextBlock_42 {
   widgetClassPath=&TextBlock
   parent=HorizontalBox_13
   widget=TextBlock_42
   properties {
    text="Song Short Name : "
    font {
     bForceMonospaced=false
     bMaterialIsStencil=false
     fontMaterial=null
     fontObject=&Roboto
     letterSpacing=0
     monospacedWidth=1
     size=11.25
     skewAmount=0
     typefaceFontName=Bold
     outlineSettings {
      bApplyOutlineToDropShadows=false
      bMiteredCorners=false
      bSeparateFillAlpha=false
      outlineMaterial=null
      outlineSize=0
      outlineColor=<rgba 0 0 0 1>
     }
    }
   }
   slot {class=HorizontalBoxSlot, properties {verticalAlignment=VAlign_Center}}
  }
  SizeBox_1 {
   widgetClassPath=&SizeBox
   parent=HorizontalBox_13
   widget=SizeBox_1
   properties {bOverride_WidthOverride=true, widthOverride=201.8042}
   slot {class=HorizontalBoxSlot, properties {horizontalAlignment=HAlign_Center, verticalAlignment=VAlign_Center, padding=<ltrb 2 2 2 2>}}
  }
  TextBox_CustomSongName {
   widgetClassPath=&EditorUtilityEditableTextBox
   bIsVariable=true
   parent=SizeBox_1
   widget=TextBox_CustomSongName
   properties {text=Nightmare}
   slot {class=SizeBoxSlot}
  }
  Button_CustomSongTag {widgetClassPath=&EditorUtilityButton, parent=HorizontalBox_13, widget=Button_CustomSongTag, slot {class=HorizontalBoxSlot}}
  TextBlock_37 {
   widgetClassPath=&TextBlock
   parent=Button_CustomSongTag
   widget=TextBlock_37
   properties {
    text=Play
    font {
     bForceMonospaced=false
     bMaterialIsStencil=false
     fontMaterial=null
     fontObject=&Roboto
     letterSpacing=0
     monospacedWidth=1
     size=11.25
     skewAmount=0
     typefaceFontName=Bold
     outlineSettings {
      bApplyOutlineToDropShadows=false
      bMiteredCorners=false
      bSeparateFillAlpha=false
      outlineMaterial=null
      outlineSize=0
      outlineColor=<rgba 0 0 0 1>
     }
    }
   }
   slot {class=ButtonSlot, properties {horizontalAlignment=HAlign_Fill, padding=<ltrb 0 0 0 0>}}
  }
  WrapBox_2 {
   widgetClassPath=&WrapBox
   bIsVariable=true
   parent=VerticalBox_2
   widget=WrapBox_2
   properties {visibility=Collapsed}
   slot {class=VerticalBoxSlot}
  }
  TextBlock_28 {
   widgetClassPath=&TextBlock
   parent=WrapBox_2
   widget=TextBlock_28
   properties {
    text="Bookmarked : "
    font {
     bForceMonospaced=false
     bMaterialIsStencil=false
     fontMaterial=null
     fontObject=&Roboto
     letterSpacing=0
     monospacedWidth=1
     size=11.25
     skewAmount=0
     typefaceFontName=Bold
     outlineSettings {
      bApplyOutlineToDropShadows=false
      bMiteredCorners=false
      bSeparateFillAlpha=false
      outlineMaterial=null
      outlineSize=0
      outlineColor=<rgba 0 0 0 1>
     }
    }
   }
   slot {class=WrapBoxSlot, properties {verticalAlignment=VAlign_Center}}
  }
  HorizontalBox_4 {
   widgetClassPath=&HorizontalBox
   parent=VerticalBox_2
   widget=HorizontalBox_4
   slot {class=VerticalBoxSlot, properties {padding=<ltrb 0 3 0 0>}}
  }
  TextBlock_13 {
   widgetClassPath=&TextBlock
   parent=HorizontalBox_4
   widget=TextBlock_13
   properties {
    text="Rock : "
    font {
     bForceMonospaced=false
     bMaterialIsStencil=false
     fontMaterial=null
     fontObject=&Roboto
     letterSpacing=0
     monospacedWidth=1
     size=11.25
     skewAmount=0
     typefaceFontName=Bold
     outlineSettings {
      bApplyOutlineToDropShadows=false
      bMiteredCorners=false
      bSeparateFillAlpha=false
      outlineMaterial=null
      outlineSize=0
      outlineColor=<rgba 0 0 0 1>
     }
    }
   }
   slot {class=HorizontalBoxSlot, properties {verticalAlignment=VAlign_Center}}
  }
  Button_CountingStars {widgetClassPath=&EditorUtilityButton, parent=HorizontalBox_4, widget=Button_CountingStars, slot {class=HorizontalBoxSlot}}
  TextBlock_14 {
   widgetClassPath=&TextBlock
   parent=Button_CountingStars
   widget=TextBlock_14
   properties {
    text=Counting Stars
    font {
     bForceMonospaced=false
     bMaterialIsStencil=false
     fontMaterial=null
     fontObject=&Roboto
     letterSpacing=0
     monospacedWidth=1
     size=11.25
     skewAmount=0
     typefaceFontName=Bold
     outlineSettings {
      bApplyOutlineToDropShadows=false
      bMiteredCorners=false
      bSeparateFillAlpha=false
      outlineMaterial=null
      outlineSize=0
      outlineColor=<rgba 0 0 0 1>
     }
    }
   }
   slot {class=ButtonSlot}
  }
  Button_Nightmare {widgetClassPath=&EditorUtilityButton, parent=HorizontalBox_4, widget=Button_Nightmare, slot {class=HorizontalBoxSlot}}
  TextBlock_16 {
   widgetClassPath=&TextBlock
   parent=Button_Nightmare
   widget=TextBlock_16
   properties {
    text=Nightmare
    font {
     bForceMonospaced=false
     bMaterialIsStencil=false
     fontMaterial=null
     fontObject=&Roboto
     letterSpacing=0
     monospacedWidth=1
     size=11.25
     skewAmount=0
     typefaceFontName=Bold
     outlineSettings {
      bApplyOutlineToDropShadows=false
      bMiteredCorners=false
      bSeparateFillAlpha=false
      outlineMaterial=null
      outlineSize=0
      outlineColor=<rgba 0 0 0 1>
     }
    }
   }
   slot {class=ButtonSlot}
  }
  Button_Overcompensate {widgetClassPath=&EditorUtilityButton, parent=HorizontalBox_4, widget=Button_Overcompensate, slot {class=HorizontalBoxSlot}}
  TextBlock_17 {
   widgetClassPath=&TextBlock
   parent=Button_Overcompensate
   widget=TextBlock_17
   properties {
    text=Overcompensate
    font {
     bForceMonospaced=false
     bMaterialIsStencil=false
     fontMaterial=null
     fontObject=&Roboto
     letterSpacing=0
     monospacedWidth=1
     size=11.25
     skewAmount=0
     typefaceFontName=Bold
     outlineSettings {
      bApplyOutlineToDropShadows=false
      bMiteredCorners=false
      bSeparateFillAlpha=false
      outlineMaterial=null
      outlineSize=0
      outlineColor=<rgba 0 0 0 1>
     }
    }
   }
   slot {class=ButtonSlot}
  }
  HorizontalBox_3 {widgetClassPath=&HorizontalBox, parent=VerticalBox_2, widget=HorizontalBox_3, slot {class=VerticalBoxSlot}}
  TextBlock_9 {
   widgetClassPath=&TextBlock
   parent=HorizontalBox_3
   widget=TextBlock_9
   properties {
    text="Hip-Hop : "
    font {
     bForceMonospaced=false
     bMaterialIsStencil=false
     fontMaterial=null
     fontObject=&Roboto
     letterSpacing=0
     monospacedWidth=1
     size=11.25
     skewAmount=0
     typefaceFontName=Bold
     outlineSettings {
      bApplyOutlineToDropShadows=false
      bMiteredCorners=false
      bSeparateFillAlpha=false
      outlineMaterial=null
      outlineSize=0
      outlineColor=<rgba 0 0 0 1>
     }
    }
   }
   slot {class=HorizontalBoxSlot, properties {verticalAlignment=VAlign_Center}}
  }
  Button_SideToSide {widgetClassPath=&EditorUtilityButton, parent=HorizontalBox_3, widget=Button_SideToSide, slot {class=HorizontalBoxSlot}}
  TextBlock_10 {
   widgetClassPath=&TextBlock
   parent=Button_SideToSide
   widget=TextBlock_10
   properties {
    text=Side to Side
    font {
     bForceMonospaced=false
     bMaterialIsStencil=false
     fontMaterial=null
     fontObject=&Roboto
     letterSpacing=0
     monospacedWidth=1
     size=11.25
     skewAmount=0
     typefaceFontName=Bold
     outlineSettings {
      bApplyOutlineToDropShadows=false
      bMiteredCorners=false
      bSeparateFillAlpha=false
      outlineMaterial=null
      outlineSize=0
      outlineColor=<rgba 0 0 0 1>
     }
    }
   }
   slot {class=ButtonSlot}
  }
  Button_MIA {widgetClassPath=&EditorUtilityButton, parent=HorizontalBox_3, widget=Button_MIA, slot {class=HorizontalBoxSlot}}
  TextBlock_11 {
   widgetClassPath=&TextBlock
   parent=Button_MIA
   widget=TextBlock_11
   properties {
    text=MIA
    font {
     bForceMonospaced=false
     bMaterialIsStencil=false
     fontMaterial=null
     fontObject=&Roboto
     letterSpacing=0
     monospacedWidth=1
     size=11.25
     skewAmount=0
     typefaceFontName=Bold
     outlineSettings {
      bApplyOutlineToDropShadows=false
      bMiteredCorners=false
      bSeparateFillAlpha=false
      outlineMaterial=null
      outlineSize=0
      outlineColor=<rgba 0 0 0 1>
     }
    }
   }
   slot {class=ButtonSlot}
  }
  Button_Superbass {widgetClassPath=&EditorUtilityButton, parent=HorizontalBox_3, widget=Button_Superbass, slot {class=HorizontalBoxSlot}}
  TextBlock_12 {
   widgetClassPath=&TextBlock
   parent=Button_Superbass
   widget=TextBlock_12
   properties {
    text=Superbass
    font {
     bForceMonospaced=false
     bMaterialIsStencil=false
     fontMaterial=null
     fontObject=&Roboto
     letterSpacing=0
     monospacedWidth=1
     size=11.25
     skewAmount=0
     typefaceFontName=Bold
     outlineSettings {
      bApplyOutlineToDropShadows=false
      bMiteredCorners=false
      bSeparateFillAlpha=false
      outlineMaterial=null
      outlineSize=0
      outlineColor=<rgba 0 0 0 1>
     }
    }
   }
   slot {class=ButtonSlot}
  }
  HorizontalBox_2 {widgetClassPath=&HorizontalBox, parent=VerticalBox_2, widget=HorizontalBox_2, slot {class=VerticalBoxSlot}}
  TextBlock_15 {
   widgetClassPath=&TextBlock
   parent=HorizontalBox_2
   widget=TextBlock_15
   properties {
    text="Pop : "
    font {
     bForceMonospaced=false
     bMaterialIsStencil=false
     fontMaterial=null
     fontObject=&Roboto
     letterSpacing=0
     monospacedWidth=1
     size=11.25
     skewAmount=0
     typefaceFontName=Bold
     outlineSettings {
      bApplyOutlineToDropShadows=false
      bMiteredCorners=false
      bSeparateFillAlpha=false
      outlineMaterial=null
      outlineSize=0
      outlineColor=<rgba 0 0 0 1>
     }
    }
   }
   slot {class=HorizontalBoxSlot, properties {verticalAlignment=VAlign_Center}}
  }
  Button_Closer {widgetClassPath=&EditorUtilityButton, parent=HorizontalBox_2, widget=Button_Closer, slot {class=HorizontalBoxSlot}}
  TextBlock_6 {
   widgetClassPath=&TextBlock
   parent=Button_Closer
   widget=TextBlock_6
   properties {
    text=Closer
    font {
     bForceMonospaced=false
     bMaterialIsStencil=false
     fontMaterial=null
     fontObject=&Roboto
     letterSpacing=0
     monospacedWidth=1
     size=11.25
     skewAmount=0
     typefaceFontName=Bold
     outlineSettings {
      bApplyOutlineToDropShadows=false
      bMiteredCorners=false
      bSeparateFillAlpha=false
      outlineMaterial=null
      outlineSize=0
      outlineColor=<rgba 0 0 0 1>
     }
    }
   }
   slot {class=ButtonSlot}
  }
  Button_Melt {widgetClassPath=&EditorUtilityButton, parent=HorizontalBox_2, widget=Button_Melt, slot {class=HorizontalBoxSlot}}
  TextBlock_7 {
   widgetClassPath=&TextBlock
   parent=Button_Melt
   widget=TextBlock_7
   properties {
    text=Melt
    font {
     bForceMonospaced=false
     bMaterialIsStencil=false
     fontMaterial=null
     fontObject=&Roboto
     letterSpacing=0
     monospacedWidth=1
     size=11.25
     skewAmount=0
     typefaceFontName=Bold
     outlineSettings {
      bApplyOutlineToDropShadows=false
      bMiteredCorners=false
      bSeparateFillAlpha=false
      outlineMaterial=null
      outlineSize=0
      outlineColor=<rgba 0 0 0 1>
     }
    }
   }
   slot {class=ButtonSlot}
  }
  Button_YourLove9PM {widgetClassPath=&EditorUtilityButton, parent=HorizontalBox_2, widget=Button_YourLove9PM, slot {class=HorizontalBoxSlot}}
  TextBlock_8 {
   widgetClassPath=&TextBlock
   parent=Button_YourLove9PM
   widget=TextBlock_8
   properties {
    text="Your Love (9PM)"
    font {
     bForceMonospaced=false
     bMaterialIsStencil=false
     fontMaterial=null
     fontObject=&Roboto
     letterSpacing=0
     monospacedWidth=1
     size=11.25
     skewAmount=0
     typefaceFontName=Bold
     outlineSettings {
      bApplyOutlineToDropShadows=false
      bMiteredCorners=false
      bSeparateFillAlpha=false
      outlineMaterial=null
      outlineSize=0
      outlineColor=<rgba 0 0 0 1>
     }
    }
   }
   slot {class=ButtonSlot}
  }
  HorizontalBox_0 {
   widgetClassPath=&HorizontalBox
   parent=VerticalBox_0
   widget=HorizontalBox_0
   properties {bOverride_Cursor=true}
   slot {class=VerticalBoxSlot, properties {horizontalAlignment=HAlign_Left, padding=<ltrb 0 10 0 0>}}
  }
  EditorUtilityButton_0 {widgetClassPath=&EditorUtilityButton, parent=HorizontalBox_0, widget=EditorUtilityButton_0, slot {class=HorizontalBoxSlot}}
  TextBlock_0 {
   widgetClassPath=&TextBlock
   parent=EditorUtilityButton_0
   widget=TextBlock_0
   properties {
    text=Load Anim Instance
    font {
     bForceMonospaced=false
     bMaterialIsStencil=false
     fontMaterial=null
     fontObject=&Roboto
     letterSpacing=0
     monospacedWidth=1
     size=15
     skewAmount=0
     typefaceFontName=Bold
     outlineSettings {
      bApplyOutlineToDropShadows=false
      bMiteredCorners=false
      bSeparateFillAlpha=false
      outlineMaterial=null
      outlineSize=0
      outlineColor=<rgba 0 0 0 1>
     }
    }
   }
   slot {class=ButtonSlot}
  }
  Text_CurrentAnimInstance {
   widgetClassPath=&TextBlock
   bIsVariable=true
   parent=HorizontalBox_0
   widget=Text_CurrentAnimInstance
   properties {
    text=Not Valid
    colorAndOpacity {colorUseRule=UseColor_Specified, specifiedColor=<rgba 0.4167 0.4167 0.4167 1>}
    font {
     bForceMonospaced=false
     bMaterialIsStencil=false
     fontMaterial=null
     fontObject=&Roboto
     letterSpacing=0
     monospacedWidth=1
     size=9
     skewAmount=0
     typefaceFontName=Bold
     outlineSettings {
      bApplyOutlineToDropShadows=false
      bMiteredCorners=false
      bSeparateFillAlpha=false
      outlineMaterial=null
      outlineSize=0
      outlineColor=<rgba 0 0 0 1>
     }
    }
   }
   slot {class=HorizontalBoxSlot, properties {verticalAlignment=VAlign_Center}}
  }
  ScaleBox_0 {
   widgetClassPath=&ScaleBox
   parent=VerticalBox_0
   widget=ScaleBox_0
   properties {stretch=UserSpecified, userSpecifiedScale=0.7}
   slot {class=VerticalBoxSlot, properties {horizontalAlignment=HAlign_Left}}
  }
  DetailsView_58 {
   widgetClassPath=&DetailsView
   parent=ScaleBox_0
   widget=DetailsView_58
   properties {categoriesToShow [Data]}
   slot {class=ScaleBoxSlot}
  }
  EditorUtilityScrollBox_1 {
   widgetClassPath=&EditorUtilityScrollBox
   parent=VerticalBox_0
   widget=EditorUtilityScrollBox_1
   properties {visibility=Collapsed}
   slot {class=VerticalBoxSlot, properties {size {sizeRule=Fill, value=1}}}
  }
  VerticalBox_174 {
   widgetClassPath=&VerticalBox
   parent=EditorUtilityScrollBox_1
   widget=VerticalBox_174
   properties {visibility=Visible}
   slot {class=ScrollBoxSlot}
  }
  TextBlock_316 {
   widgetClassPath=&TextBlock
   parent=VerticalBox_174
   widget=TextBlock_316
   properties {
    text=Debug :
    font {
     bForceMonospaced=false
     bMaterialIsStencil=false
     fontMaterial=null
     fontObject=&Roboto
     letterSpacing=0
     monospacedWidth=1
     size=16.5
     skewAmount=0
     typefaceFontName=Bold
     outlineSettings {
      bApplyOutlineToDropShadows=false
      bMiteredCorners=false
      bSeparateFillAlpha=false
      outlineMaterial=null
      outlineSize=0
      outlineColor=<rgba 0 0 0 1>
     }
    }
   }
   slot {class=VerticalBoxSlot}
  }
  HorizontalBox_10 {widgetClassPath=&HorizontalBox, parent=VerticalBox_174, widget=HorizontalBox_10, slot {class=VerticalBoxSlot}}
  TextBlock_18 {
   widgetClassPath=&TextBlock
   parent=HorizontalBox_10
   widget=TextBlock_18
   properties {
    text=Base Body
    font {
     bForceMonospaced=false
     bMaterialIsStencil=false
     fontMaterial=null
     fontObject=&Roboto
     letterSpacing=0
     monospacedWidth=1
     size=13.5
     skewAmount=0
     typefaceFontName=Bold
     outlineSettings {
      bApplyOutlineToDropShadows=false
      bMiteredCorners=false
      bSeparateFillAlpha=false
      outlineMaterial=null
      outlineSize=0
      outlineColor=<rgba 0 0 0 1>
     }
    }
   }
   slot {class=HorizontalBoxSlot}
  }
  Button_BaseBodyVisbility {
   widgetClassPath=&EditorUtilityButton
   parent=HorizontalBox_10
   widget=Button_BaseBodyVisbility
   properties {widgetStyle {
     clickedSlateSound {resourceObject=null}
     disabled {
      bIsDynamicallyLoaded=false
      drawAs=RoundedBox
      imageType=NoImage
      mirroring=NoMirror
      resourceName=None
      resourceObject=null
      tiling=NoTile
      imageSize=<0 0>
      margin=<ltrb 0 0 0 0>
      outlineSettings {
       bUseBrushTransparency=false
       roundingType=FixedRadius
       width=1
       color {colorUseRule=UseColor_ColorTable, specifiedColor=<rgba 0 0 0 0>}
       cornerRadii=<4 4 4 4>
      }
      tintColor {colorUseRule=UseColor_Specified, specifiedColor=<rgba 0.0395 0.0395 0.0395 1>}
      uVRegion {bIsValid=false, max=<0 0>, min=<0 0>}
     }
     disabledForeground {colorUseRule=UseColor_ColorTable, specifiedColor=<rgba 0 0 0 0>}
     hovered {
      bIsDynamicallyLoaded=false
      drawAs=RoundedBox
      imageType=NoImage
      mirroring=NoMirror
      resourceName=None
      resourceObject=null
      tiling=NoTile
      imageSize=<32 32>
      margin=<ltrb 0 0 0 0>
      outlineSettings {
       bUseBrushTransparency=false
       roundingType=FixedRadius
       width=1
       color {colorUseRule=UseColor_ColorTable, specifiedColor=<rgba 0 0 0 0>}
       cornerRadii=<4 4 4 4>
      }
      tintColor {colorUseRule=UseColor_Specified, specifiedColor=<rgba 0.0953 0.0953 0.0953 1>}
      uVRegion {bIsValid=false, max=<0 0>, min=<0 0>}
     }
     hoveredForeground {colorUseRule=UseColor_ColorTable, specifiedColor=<rgba 0 0 0 0>}
     hoveredSlateSound {resourceObject=null}
     normal {
      bIsDynamicallyLoaded=false
      drawAs=RoundedBox
      imageType=NoImage
      mirroring=NoMirror
      resourceName=None
      resourceObject=null
      tiling=NoTile
      imageSize=<32 32>
      margin=<ltrb 0 0 0 0>
      outlineSettings {
       bUseBrushTransparency=false
       roundingType=FixedRadius
       width=1
       color {colorUseRule=UseColor_ColorTable, specifiedColor=<rgba 0 0 0 0>}
       cornerRadii=<4 4 4 4>
      }
      tintColor {colorUseRule=UseColor_Specified, specifiedColor=<rgba 0.0395 0.0395 0.0395 1>}
      uVRegion {bIsValid=false, max=<0 0>, min=<0 0>}
     }
     normalForeground {colorUseRule=UseColor_ColorTable, specifiedColor=<rgba 0 0 0 0>}
     normalPadding=<ltrb 0 0 0 0>
     pressed {
      bIsDynamicallyLoaded=false
      drawAs=RoundedBox
      imageType=NoImage
      mirroring=NoMirror
      resourceName=None
      resourceObject=null
      tiling=NoTile
      imageSize=<32 32>
      margin=<ltrb 0 0 0 0>
      outlineSettings {
       bUseBrushTransparency=false
       roundingType=FixedRadius
       width=1
       color {colorUseRule=UseColor_ColorTable, specifiedColor=<rgba 0 0 0 0>}
       cornerRadii=<4 4 4 4>
      }
      tintColor {colorUseRule=UseColor_Specified, specifiedColor=<rgba 0.0284 0.0284 0.0284 1>}
      uVRegion {bIsValid=false, max=<0 0>, min=<0 0>}
     }
     pressedForeground {colorUseRule=UseColor_ColorTable, specifiedColor=<rgba 0 0 0 0>}
     pressedPadding=<ltrb 0 0 0 0>
     pressedSlateSound {resourceObject=null}
    }}
   slot {class=HorizontalBoxSlot, properties {verticalAlignment=VAlign_Center}}
  }
  ScaleBox_1 {
   widgetClassPath=&ScaleBox
   parent=Button_BaseBodyVisbility
   widget=ScaleBox_1
   properties {stretch=UserSpecified, userSpecifiedScale=0.3}
   slot {class=ButtonSlot, properties {padding=<ltrb 2 4 2 4>}}
  }
  Image_BaseBodyVisbility {
   widgetClassPath=&Image
   bIsVariable=true
   parent=ScaleBox_1
   widget=Image_BaseBodyVisbility
   properties {brush {
     bIsDynamicallyLoaded=false
     drawAs=Image
     imageType=FullColor
     mirroring=NoMirror
     resourceName=None
     resourceObject=&T-HLT-Hidden
     tiling=NoTile
     imageSize=<55 43>
     margin=<ltrb 0 0 0 0>
     outlineSettings {
      bUseBrushTransparency=false
      roundingType=HalfHeightRadius
      width=0
      color {colorUseRule=UseColor_Specified, specifiedColor=<rgba 0 0 0 0>}
      cornerRadii=<0 0 0 0>
     }
     tintColor {colorUseRule=UseColor_Specified, specifiedColor=<rgba 1 1 1 1>}
     uVRegion {bIsValid=false, max=<0 0>, min=<0 0>}
    }}
   slot {class=ScaleBoxSlot}
  }
  BaseBody_Container {
   widgetClassPath=&VerticalBox
   bIsVariable=true
   parent=VerticalBox_174
   widget=BaseBody_Container
   properties {visibility=Collapsed}
   slot {class=VerticalBoxSlot, properties {padding=<ltrb 20 0 0 0>}}
  }
  HorizontalBox_5 {widgetClassPath=&HorizontalBox, parent=BaseBody_Container, widget=HorizontalBox_5, slot {class=VerticalBoxSlot}}
  TextBlock_19 {
   widgetClassPath=&TextBlock
   parent=HorizontalBox_5
   widget=TextBlock_19
   properties {
    text="Override Sequence : "
    font {
     bForceMonospaced=false
     bMaterialIsStencil=false
     fontMaterial=null
     fontObject=&Roboto
     letterSpacing=0
     monospacedWidth=1
     size=9.75
     skewAmount=0
     typefaceFontName=Regular
     outlineSettings {
      bApplyOutlineToDropShadows=false
      bMiteredCorners=false
      bSeparateFillAlpha=false
      outlineMaterial=null
      outlineSize=0
      outlineColor=<rgba 0 0 0 1>
     }
    }
   }
   slot {class=HorizontalBoxSlot, properties {verticalAlignment=VAlign_Center}}
  }
  EditorUtilityCheckBox {
   widgetClassPath=&EditorUtilityCheckBox
   bIsVariable=true
   parent=HorizontalBox_5
   widget=EditorUtilityCheckBox
   slot {class=HorizontalBoxSlot}
  }
  SinglePropertyView_BaseBody {
   widgetClassPath=&SinglePropertyView
   parent=HorizontalBox_5
   widget=SinglePropertyView_BaseBody
   properties {nameOverride=Sequence, propertyName=DEBUG_JamV2_BaseBody_OverrideSequence}
   slot {class=HorizontalBoxSlot}
  }
  WrapBox_0 {widgetClassPath=&WrapBox, parent=BaseBody_Container, widget=WrapBox_0, slot {class=VerticalBoxSlot}}
  HorizontalBox_6 {widgetClassPath=&HorizontalBox, parent=WrapBox_0, widget=HorizontalBox_6, slot {class=WrapBoxSlot}}
  TextBlock_20 {
   widgetClassPath=&TextBlock
   parent=HorizontalBox_6
   widget=TextBlock_20
   properties {
    text="Force Squat : "
    font {
     bForceMonospaced=false
     bMaterialIsStencil=false
     fontMaterial=null
     fontObject=&Roboto
     letterSpacing=0
     monospacedWidth=1
     size=9.75
     skewAmount=0
     typefaceFontName=Regular
     outlineSettings {
      bApplyOutlineToDropShadows=false
      bMiteredCorners=false
      bSeparateFillAlpha=false
      outlineMaterial=null
      outlineSize=0
      outlineColor=<rgba 0 0 0 1>
     }
    }
   }
   slot {class=HorizontalBoxSlot}
  }
  EditorUtilityCheckBox_1 {
   widgetClassPath=&EditorUtilityCheckBox
   bIsVariable=true
   parent=HorizontalBox_6
   widget=EditorUtilityCheckBox_1
   slot {class=HorizontalBoxSlot}
  }
  HorizontalBox_7 {widgetClassPath=&HorizontalBox, parent=WrapBox_0, widget=HorizontalBox_7, slot {class=WrapBoxSlot}}
  TextBlock_24 {
   widgetClassPath=&TextBlock
   parent=HorizontalBox_7
   widget=TextBlock_24
   properties {
    text="ForcePlayStateTransition : "
    font {
     bForceMonospaced=false
     bMaterialIsStencil=false
     fontMaterial=null
     fontObject=&Roboto
     letterSpacing=0
     monospacedWidth=1
     size=9.75
     skewAmount=0
     typefaceFontName=Regular
     outlineSettings {
      bApplyOutlineToDropShadows=false
      bMiteredCorners=false
      bSeparateFillAlpha=false
      outlineMaterial=null
      outlineSize=0
      outlineColor=<rgba 0 0 0 1>
     }
    }
   }
   slot {class=HorizontalBoxSlot}
  }
  EditorUtilityCheckBox_4 {
   widgetClassPath=&EditorUtilityCheckBox
   bIsVariable=true
   parent=HorizontalBox_7
   widget=EditorUtilityCheckBox_4
   slot {class=HorizontalBoxSlot}
  }
  HorizontalBox_8 {widgetClassPath=&HorizontalBox, parent=WrapBox_0, widget=HorizontalBox_8, slot {class=WrapBoxSlot}}
  TextBlock_21 {
   widgetClassPath=&TextBlock
   parent=HorizontalBox_8
   widget=TextBlock_21
   properties {
    text="UsePlayLoopOnly : "
    font {
     bForceMonospaced=false
     bMaterialIsStencil=false
     fontMaterial=null
     fontObject=&Roboto
     letterSpacing=0
     monospacedWidth=1
     size=9.75
     skewAmount=0
     typefaceFontName=Regular
     outlineSettings {
      bApplyOutlineToDropShadows=false
      bMiteredCorners=false
      bSeparateFillAlpha=false
      outlineMaterial=null
      outlineSize=0
      outlineColor=<rgba 0 0 0 1>
     }
    }
   }
   slot {class=HorizontalBoxSlot}
  }
  EditorUtilityCheckBox_2 {
   widgetClassPath=&EditorUtilityCheckBox
   bIsVariable=true
   parent=HorizontalBox_8
   widget=EditorUtilityCheckBox_2
   slot {class=HorizontalBoxSlot}
  }
  VerticalBox_1 {
   widgetClassPath=&VerticalBox
   parent=BaseBody_Container
   widget=VerticalBox_1
   properties {bOverride_Cursor=true}
   slot {class=VerticalBoxSlot, properties {padding=<ltrb 0 20 0 0>}}
  }
  TextBlock_22 {
   widgetClassPath=&TextBlock
   parent=VerticalBox_1
   widget=TextBlock_22
   properties {
    text="TraversalHomebaseBPMs : "
    font {
     bForceMonospaced=false
     bMaterialIsStencil=false
     fontMaterial=null
     fontObject=&Roboto
     letterSpacing=0
     monospacedWidth=1
     size=9.75
     skewAmount=0
     typefaceFontName=Regular
     outlineSettings {
      bApplyOutlineToDropShadows=false
      bMiteredCorners=false
      bSeparateFillAlpha=false
      outlineMaterial=null
      outlineSize=0
      outlineColor=<rgba 0 0 0 1>
     }
    }
   }
   slot {class=VerticalBoxSlot}
  }
  DetailsView_1 {
   widgetClassPath=&DetailsView
   parent=VerticalBox_1
   widget=DetailsView_1
   properties {bOverride_Cursor=true, propertiesToShow [DEBUG_JamV2_BaseBody_TraversalHomebaseBPMs]}
   slot {class=VerticalBoxSlot}
  }
  Image_0 {
   widgetClassPath=&Image
   bIsVariable=true
   parent=VerticalBox_174
   widget=Image_0
   properties {
    brush {
     bIsDynamicallyLoaded=false
     drawAs=Image
     imageType=NoImage
     mirroring=NoMirror
     resourceName=None
     resourceObject=null
     tiling=NoTile
     imageSize=<32 5>
     margin=<ltrb 0 0 0 0>
     outlineSettings {
      bUseBrushTransparency=false
      roundingType=HalfHeightRadius
      width=0
      color {colorUseRule=UseColor_Specified, specifiedColor=<rgba 0 0 0 0>}
      cornerRadii=<0 0 0 0>
     }
     tintColor {colorUseRule=UseColor_Specified, specifiedColor=<rgba 1 1 1 1>}
     uVRegion {bIsValid=false, max=<0 0>, min=<0 0>}
    }
    colorAndOpacity=<rgba 0.0469 0.0469 0.0469 1>
   }
   slot {class=VerticalBoxSlot, properties {padding=<ltrb 5 5 5 5>}}
  }
  TextBlock_4 {
   widgetClassPath=&TextBlock
   parent=VerticalBox_174
   widget=TextBlock_4
   properties {
    text=Head
    font {
     bForceMonospaced=false
     bMaterialIsStencil=false
     fontMaterial=null
     fontObject=&Roboto
     letterSpacing=0
     monospacedWidth=1
     size=13.5
     skewAmount=0
     typefaceFontName=Bold
     outlineSettings {
      bApplyOutlineToDropShadows=false
      bMiteredCorners=false
      bSeparateFillAlpha=false
      outlineMaterial=null
      outlineSize=0
      outlineColor=<rgba 0 0 0 1>
     }
    }
   }
   slot {class=VerticalBoxSlot}
  }
  VerticalBox_267 {
   widgetClassPath=&VerticalBox
   parent=VerticalBox_174
   widget=VerticalBox_267
   slot {class=VerticalBoxSlot, properties {padding=<ltrb 20 0 0 0>}}
  }
  TextBlock_103 {
   widgetClassPath=&TextBlock
   parent=VerticalBox_267
   widget=TextBlock_103
   properties {
    text=Head Twist
    font {
     bForceMonospaced=false
     bMaterialIsStencil=false
     fontMaterial=null
     fontObject=&Roboto
     letterSpacing=0
     monospacedWidth=1
     size=9
     skewAmount=0
     typefaceFontName=Bold
     outlineSettings {
      bApplyOutlineToDropShadows=false
      bMiteredCorners=false
      bSeparateFillAlpha=false
      outlineMaterial=null
      outlineSize=0
      outlineColor=<rgba 0 0 0 1>
     }
    }
   }
   slot {class=VerticalBoxSlot}
  }
  HorizontalBox_106 {widgetClassPath=&HorizontalBox, parent=VerticalBox_267, widget=HorizontalBox_106, slot {class=VerticalBoxSlot}}
  EditorUtilityCheckBox_59 {
   widgetClassPath=&EditorUtilityCheckBox
   bIsVariable=true
   parent=HorizontalBox_106
   widget=EditorUtilityCheckBox_59
   slot {class=HorizontalBoxSlot}
  }
  SinglePropertyView_76 {
   widgetClassPath=&SinglePropertyView
   parent=HorizontalBox_106
   widget=SinglePropertyView_76
   slot {class=HorizontalBoxSlot}
  }
  TextBlock_1 {
   widgetClassPath=&TextBlock
   parent=VerticalBox_174
   widget=TextBlock_1
   properties {
    text=Spine
    font {
     bForceMonospaced=false
     bMaterialIsStencil=false
     fontMaterial=null
     fontObject=&Roboto
     letterSpacing=0
     monospacedWidth=1
     size=13.5
     skewAmount=0
     typefaceFontName=Bold
     outlineSettings {
      bApplyOutlineToDropShadows=false
      bMiteredCorners=false
      bSeparateFillAlpha=false
      outlineMaterial=null
      outlineSize=0
      outlineColor=<rgba 0 0 0 1>
     }
    }
   }
   slot {class=VerticalBoxSlot}
  }
  TextBlock_2 {
   widgetClassPath=&TextBlock
   parent=VerticalBox_174
   widget=TextBlock_2
   properties {
    text=Foot
    font {
     bForceMonospaced=false
     bMaterialIsStencil=false
     fontMaterial=null
     fontObject=&Roboto
     letterSpacing=0
     monospacedWidth=1
     size=13.5
     skewAmount=0
     typefaceFontName=Bold
     outlineSettings {
      bApplyOutlineToDropShadows=false
      bMiteredCorners=false
      bSeparateFillAlpha=false
      outlineMaterial=null
      outlineSize=0
      outlineColor=<rgba 0 0 0 1>
     }
    }
   }
   slot {class=VerticalBoxSlot}
  }
  TextBlock_3 {
   widgetClassPath=&TextBlock
   parent=VerticalBox_174
   widget=TextBlock_3
   properties {
    text=Face Expression
    font {
     bForceMonospaced=false
     bMaterialIsStencil=false
     fontMaterial=null
     fontObject=&Roboto
     letterSpacing=0
     monospacedWidth=1
     size=13.5
     skewAmount=0
     typefaceFontName=Bold
     outlineSettings {
      bApplyOutlineToDropShadows=false
      bMiteredCorners=false
      bSeparateFillAlpha=false
      outlineMaterial=null
      outlineSize=0
      outlineColor=<rgba 0 0 0 1>
     }
    }
   }
   slot {class=VerticalBoxSlot}
  }
  VerticalBox_3 {
   widgetClassPath=&VerticalBox
   parent=VerticalBox_174
   widget=VerticalBox_3
   slot {class=VerticalBoxSlot, properties {padding=<ltrb 20 0 0 0>}}
  }
  WrapBox_1 {widgetClassPath=&WrapBox, parent=VerticalBox_3, widget=WrapBox_1, slot {class=VerticalBoxSlot}}
  HorizontalBox_14 {widgetClassPath=&HorizontalBox, parent=WrapBox_1, widget=HorizontalBox_14, slot {class=WrapBoxSlot}}
  TextBlock_29 {
   widgetClassPath=&TextBlock
   parent=HorizontalBox_14
   widget=TextBlock_29
   properties {
    text="Face Expression Enabled : "
    font {
     bForceMonospaced=false
     bMaterialIsStencil=false
     fontMaterial=null
     fontObject=&Roboto
     letterSpacing=0
     monospacedWidth=1
     size=9.75
     skewAmount=0
     typefaceFontName=Regular
     outlineSettings {
      bApplyOutlineToDropShadows=false
      bMiteredCorners=false
      bSeparateFillAlpha=false
      outlineMaterial=null
      outlineSize=0
      outlineColor=<rgba 0 0 0 1>
     }
    }
   }
   slot {class=HorizontalBoxSlot}
  }
  EditorUtilityCheckBox_8 {
   widgetClassPath=&EditorUtilityCheckBox
   bIsVariable=true
   parent=HorizontalBox_14
   widget=EditorUtilityCheckBox_8
   slot {class=HorizontalBoxSlot}
  }
  HorizontalBox_15 {widgetClassPath=&HorizontalBox, parent=WrapBox_1, widget=HorizontalBox_15, slot {class=WrapBoxSlot}}
  TextBlock_30 {
   widgetClassPath=&TextBlock
   parent=HorizontalBox_15
   widget=TextBlock_30
   properties {
    text="Disable Lip Sync : "
    font {
     bForceMonospaced=false
     bMaterialIsStencil=false
     fontMaterial=null
     fontObject=&Roboto
     letterSpacing=0
     monospacedWidth=1
     size=9.75
     skewAmount=0
     typefaceFontName=Regular
     outlineSettings {
      bApplyOutlineToDropShadows=false
      bMiteredCorners=false
      bSeparateFillAlpha=false
      outlineMaterial=null
      outlineSize=0
      outlineColor=<rgba 0 0 0 1>
     }
    }
   }
   slot {class=HorizontalBoxSlot}
  }
  EditorUtilityCheckBox_9 {
   widgetClassPath=&EditorUtilityCheckBox
   bIsVariable=true
   parent=HorizontalBox_15
   widget=EditorUtilityCheckBox_9
   slot {class=HorizontalBoxSlot}
  }
  HorizontalBox_9 {widgetClassPath=&HorizontalBox, parent=VerticalBox_3, widget=HorizontalBox_9, slot {class=VerticalBoxSlot}}
  TextBlock_23 {
   widgetClassPath=&TextBlock
   parent=HorizontalBox_9
   widget=TextBlock_23
   properties {
    text="Override Sequence : "
    font {
     bForceMonospaced=false
     bMaterialIsStencil=false
     fontMaterial=null
     fontObject=&Roboto
     letterSpacing=0
     monospacedWidth=1
     size=9.75
     skewAmount=0
     typefaceFontName=Regular
     outlineSettings {
      bApplyOutlineToDropShadows=false
      bMiteredCorners=false
      bSeparateFillAlpha=false
      outlineMaterial=null
      outlineSize=0
      outlineColor=<rgba 0 0 0 1>
     }
    }
   }
   slot {class=HorizontalBoxSlot, properties {verticalAlignment=VAlign_Center}}
  }
  EditorUtilityCheckBox_3 {
   widgetClassPath=&EditorUtilityCheckBox
   bIsVariable=true
   parent=HorizontalBox_9
   widget=EditorUtilityCheckBox_3
   slot {class=HorizontalBoxSlot}
  }
  SinglePropertyView_BaseBody_1 {
   widgetClassPath=&SinglePropertyView
   parent=HorizontalBox_9
   widget=SinglePropertyView_BaseBody_1
   properties {nameOverride=Sequence, propertyName=DEBUG_JamV2_BaseBody_OverrideSequence}
   slot {class=HorizontalBoxSlot}
  }
  HorizontalBox_16 {widgetClassPath=&HorizontalBox, parent=VerticalBox_3, widget=HorizontalBox_16, slot {class=VerticalBoxSlot}}
  TextBlock_31 {
   widgetClassPath=&TextBlock
   parent=HorizontalBox_16
   widget=TextBlock_31
   properties {
    text="Override Genre : "
    font {
     bForceMonospaced=false
     bMaterialIsStencil=false
     fontMaterial=null
     fontObject=&Roboto
     letterSpacing=0
     monospacedWidth=1
     size=9.75
     skewAmount=0
     typefaceFontName=Regular
     outlineSettings {
      bApplyOutlineToDropShadows=false
      bMiteredCorners=false
      bSeparateFillAlpha=false
      outlineMaterial=null
      outlineSize=0
      outlineColor=<rgba 0 0 0 1>
     }
    }
   }
   slot {class=HorizontalBoxSlot, properties {verticalAlignment=VAlign_Center}}
  }
  EditorUtilityCheckBox_10 {
   widgetClassPath=&EditorUtilityCheckBox
   bIsVariable=true
   parent=HorizontalBox_16
   widget=EditorUtilityCheckBox_10
   slot {class=HorizontalBoxSlot}
  }
  SinglePropertyView_BaseBody_2 {
   widgetClassPath=&SinglePropertyView
   parent=HorizontalBox_16
   widget=SinglePropertyView_BaseBody_2
   properties {nameOverride=Genre, propertyName=DEBUG_JamV2_FaceExpression_OverrideSongGenre}
   slot {class=HorizontalBoxSlot}
  }
  HorizontalBox_17 {widgetClassPath=&HorizontalBox, parent=VerticalBox_3, widget=HorizontalBox_17, slot {class=VerticalBoxSlot}}
  TextBlock_32 {
   widgetClassPath=&TextBlock
   parent=HorizontalBox_17
   widget=TextBlock_32
   properties {
    text="Override Intensity : "
    font {
     bForceMonospaced=false
     bMaterialIsStencil=false
     fontMaterial=null
     fontObject=&Roboto
     letterSpacing=0
     monospacedWidth=1
     size=9.75
     skewAmount=0
     typefaceFontName=Regular
     outlineSettings {
      bApplyOutlineToDropShadows=false
      bMiteredCorners=false
      bSeparateFillAlpha=false
      outlineMaterial=null
      outlineSize=0
      outlineColor=<rgba 0 0 0 1>
     }
    }
   }
   slot {class=HorizontalBoxSlot, properties {verticalAlignment=VAlign_Center}}
  }
  EditorUtilityCheckBox_11 {
   widgetClassPath=&EditorUtilityCheckBox
   bIsVariable=true
   parent=HorizontalBox_17
   widget=EditorUtilityCheckBox_11
   slot {class=HorizontalBoxSlot}
  }
  SinglePropertyView_BaseBody_3 {
   widgetClassPath=&SinglePropertyView
   parent=HorizontalBox_17
   widget=SinglePropertyView_BaseBody_3
   properties {nameOverride=Intensity, propertyName=DEBUG_JamV2_FaceExpression_OverrideSongIntensity}
   slot {class=HorizontalBoxSlot}
  }
  HorizontalBox_11 {widgetClassPath=&HorizontalBox, parent=VerticalBox_3, widget=HorizontalBox_11, slot {class=VerticalBoxSlot}}
  TextBlock_25 {
   widgetClassPath=&TextBlock
   parent=HorizontalBox_11
   widget=TextBlock_25
   properties {
    text=Expression Weight : 0
    font {
     bForceMonospaced=false
     bMaterialIsStencil=false
     fontMaterial=null
     fontObject=&Roboto
     letterSpacing=0
     monospacedWidth=1
     size=9.75
     skewAmount=0
     typefaceFontName=Regular
     outlineSettings {
      bApplyOutlineToDropShadows=false
      bMiteredCorners=false
      bSeparateFillAlpha=false
      outlineMaterial=null
      outlineSize=0
      outlineColor=<rgba 0 0 0 1>
     }
    }
   }
   slot {class=HorizontalBoxSlot, properties {verticalAlignment=VAlign_Center}}
  }
  SizeBox_0 {
   widgetClassPath=&SizeBox
   parent=HorizontalBox_11
   widget=SizeBox_0
   properties {bOverride_WidthOverride=true, widthOverride=144.4182}
   slot {class=HorizontalBoxSlot}
  }
  EditorUtilitySlider_61 {
   widgetClassPath=/Script/Blutility.EditorUtilitySlider
   bIsVariable=true
   parent=SizeBox_0
   widget=EditorUtilitySlider_61
   properties {value=1}
   slot {class=SizeBoxSlot, properties {padding=<ltrb 2 2 2 2>}}
  }
  TextBlock_26 {
   widgetClassPath=&TextBlock
   parent=HorizontalBox_11
   widget=TextBlock_26
   properties {
    text=1
    font {
     bForceMonospaced=false
     bMaterialIsStencil=false
     fontMaterial=null
     fontObject=&Roboto
     letterSpacing=0
     monospacedWidth=1
     size=9.75
     skewAmount=0
     typefaceFontName=Regular
     outlineSettings {
      bApplyOutlineToDropShadows=false
      bMiteredCorners=false
      bSeparateFillAlpha=false
      outlineMaterial=null
      outlineSize=0
      outlineColor=<rgba 0 0 0 1>
     }
    }
   }
   slot {class=HorizontalBoxSlot, properties {verticalAlignment=VAlign_Center}}
  }
  HorizontalBox_12 {widgetClassPath=&HorizontalBox, parent=VerticalBox_3, widget=HorizontalBox_12, slot {class=VerticalBoxSlot}}
  TextBlock_27 {
   widgetClassPath=&TextBlock
   parent=HorizontalBox_12
   widget=TextBlock_27
   properties {
    text="Close Eyes : "
    font {
     bForceMonospaced=false
     bMaterialIsStencil=false
     fontMaterial=null
     fontObject=&Roboto
     letterSpacing=0
     monospacedWidth=1
     size=9.75
     skewAmount=0
     typefaceFontName=Regular
     outlineSettings {
      bApplyOutlineToDropShadows=false
      bMiteredCorners=false
      bSeparateFillAlpha=false
      outlineMaterial=null
      outlineSize=0
      outlineColor=<rgba 0 0 0 1>
     }
    }
   }
   slot {class=HorizontalBoxSlot, properties {verticalAlignment=VAlign_Center}}
  }
  EditorUtilityCheckBox_5 {
   widgetClassPath=&EditorUtilityCheckBox
   bIsVariable=true
   parent=HorizontalBox_12
   widget=EditorUtilityCheckBox_5
   slot {class=HorizontalBoxSlot}
  }
  DetailsView_0 {
   widgetClassPath=&DetailsView
   parent=VerticalBox_0
   widget=DetailsView_0
   properties {
    bAllowFiltering=true
    categoriesToShow [
     DEBUG / TEST
     Jam 2.0
     Jam 2.0 - Base Body
     Jam 2.0 - Face Expression
     Jam 2.0 - Foot Shift
     Jam 2.0 - Head Motion
     Jam 2.0 - Left Arm
     Jam 2.0 - Right Arm
     Jam 2.0 - Spine Motion
     Jam 2.0 - Lower Body
     DEBUG
    ]
   }
   slot {class=VerticalBoxSlot, properties {size {sizeRule=Fill, value=1}}}
  }
  EditorUtilityScrollBox {
   widgetClassPath=&EditorUtilityScrollBox
   parent=VerticalBox_0
   widget=EditorUtilityScrollBox
   properties {visibility=Collapsed}
   slot {class=VerticalBoxSlot, properties {size {sizeRule=Fill, value=1}}}
  }
  TextBlock_36 {
   widgetClassPath=&TextBlock
   bIsVariable=true
   parent=EditorUtilityScrollBox
   widget=TextBlock_36
   properties {
    text=Text Block
    font {
     bForceMonospaced=false
     bMaterialIsStencil=false
     fontMaterial=null
     fontObject=&Roboto
     letterSpacing=0
     monospacedWidth=1
     size=7.5
     skewAmount=0
     typefaceFontName=Bold
     outlineSettings {
      bApplyOutlineToDropShadows=false
      bMiteredCorners=false
      bSeparateFillAlpha=false
      outlineMaterial=null
      outlineSize=0
      outlineColor=<rgba 0 0 0 1>
     }
    }
   }
   slot {class=ScrollBoxSlot}
  }
 ]
}
