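# Player animation layer AnimBP for Sparks Jam band play that overrides the full body with an instrument animation graph.
# FullBodyOverride uses the linked graph tagged SparksJam_Instrument_Slot, or passes the cached input pose through when not animatable or without an instrument.
# When the instrument pawn component changes, it rebinds the cosmetic instrument type delegate that updates bInstrumentIsNone.
# The UpperBody, PreLegIK and post-leg-IK layers contain no logic.
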AnimBlueprint /SparksCharacterCommon/Animation/MainPlayer/BandPlay/Sparks_BandPlay_Jam_LayerAnimBP
target_skeleton=&Fortnite_M_Avg_Player_Skeleton
graphs [
 AnimGraph {nodes [Output Pose() {layerGroup=None}]}
 CharacterSkeletalControlPostLegIK {nodes [
   Input Pose() {name=InPosePostLegIK}
   Output Pose() {layerGroup=PlayerOverride}
  ]}
 EventGraph {nodes [
   n1 = OnInstrumentTypeChanged!() -> (OutputDelegate, then, CosmeticInstrumentType:ESparksInstrumentType Enum=Guitar) {is=title K2Node_CustomEvent}
   > Set bInstrumentIsNone(bInstrumentIsNone:bool="Equal (Enum)"(A:ESparksInstrumentType Enum=@n1.CosmeticInstrumentType, B:ESparksInstrumentType Enum=None)) -> (Output_Get:bool)
   n2 = Event BP_OnSetInstrumentPSC!() -> (
    OutputDelegate
    then
    OldInstrumentPSC:Sparks Instrument Anim Pawn Component Object Reference
    NewInstrumentPSC:Sparks Instrument Anim Pawn Component Object Reference
   ) {is=title K2Node_Event}
   > Exchange Instrument Type Delegate Binds!(
    execute
    self
    Old PSC:Sparks Instrument Anim Pawn Component Object Reference=@n2.OldInstrumentPSC
    New PSC:Sparks Instrument Anim Pawn Component Object Reference=@n2.NewInstrumentPSC
    Event:"Delegate (by ref)"=@n1
   ) -> (then)
  ]}
 Exchange Instrument Type Delegate Binds {nodes [
   Exchange Instrument Type Delegate Binds!() -> (
    then
    Old PSC:Sparks Instrument Anim Pawn Component Object Reference
    New PSC:Sparks Instrument Anim Pawn Component Object Reference
    Event:"Delegate (by ref)"
   ) {is=title K2Node_FunctionEntry}
   > n1 = Sequence()
   > n2 = Get!(execute:Exec) -> (then:Exec, else:Exec, Old PSC:Sparks Instrument Anim Pawn Component Object Reference)
   > Unbind Event from On Player Cosmetic Instrument Type Changed(self:Sparks Instrument Anim Pawn Component Object Reference=@n2, Delegate=Get Event() -> (Event:"Delegate (by ref)")) {is=title K2Node_RemoveDelegate}
   n3 = Get!(execute=@n1.then_1) -> (then, else, New PSC:Sparks Instrument Anim Pawn Component Object Reference)
   > Bind Event to On Player Cosmetic Instrument Type Changed(self:Sparks Instrument Anim Pawn Component Object Reference=@n3, Delegate=Get Event()) {is=title K2Node_AddDelegate}
  ]}
 FullBodyOverride {nodes [
   n1 = Linked Anim Graph!() -> (Pose:Pose Link Structure) {is=title AnimGraphNode_LinkedAnimGraph, tag=SparksJam_Instrument_Slot, bReceiveNotifiesFromLinkedInstances=true}
   n2 = Use cached pose 'PassThroughFullBody'()
   n3 = Get bInstrumentIsNone() -> (bInstrumentIsNone:bool)
   Output Pose(Result=Blend Poses by bool(
     BlendPose_0=Use cached pose 'PassThroughFullBody'()
     BlendPose_1=Blend Poses by bool(BlendPose_0=@n2, BlendPose_1=@n1, bActiveValue=@n3) {blendType=Linear}
     bActiveValue=Get bIsNotAnimatable() -> (bIsNotAnimatable:bool)
    ) {blendType=Linear}) {layerGroup=PlayerOverride}
   Save cached pose 'PassThroughFullBody'(Pose=Input Pose() {name=PassThroughFullBodyPose}) {cacheName=PassThroughFullBody}
  ]}
 PreLegIK {nodes [
   Input Pose()
   Output Pose() {layerGroup=PlayerOverride}
  ]}
 UpperBody {nodes [
   Input Pose() {name=InPoseUpperBody}
   Output Pose() {layerGroup=PlayerOverride}
  ]}
]
variables [bInstrumentIsNone {type=boolean}]
