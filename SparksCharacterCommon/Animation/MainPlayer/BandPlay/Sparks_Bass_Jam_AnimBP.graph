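# Animation Blueprint for the Fortnite player skeleton playing bass in Sparks jam performances, driven by song MIDI events.
# A Body state machine alternates two beat-synced loops chosen by song intensity, with per-bone arm layers and an additive montage slot on top.
# MIDI listeners update intensity and fret number and trigger gated plucks alternating index and middle fingers.
# A bool switches between this MIDI-driven pose and a fixed loop, under an additive instrument-spawn pose.
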AnimBlueprint /SparksCharacterCommon/Animation/MainPlayer/BandPlay/Sparks_Bass_Jam_AnimBP
target_skeleton=&Fortnite_M_Avg_Player_Skeleton
graphs [
 AnimGraph {nodes [
   AnimGraphNode_StateMachine_1 = Right Arm() {is=title AnimGraphNode_StateMachine}
   AnimGraphNode_StateMachine_2 = Left Arm() {is=title AnimGraphNode_StateMachine, bAllowConduitEntryStates=true}
   n1 = Layered blend per bone!(
    BasePose="Use cached pose 'Body (Cache Pose)'"()
    BlendPoses_0=@AnimGraphNode_StateMachine_2
    BlendPoses_1:Pose Link Structure=@AnimGraphNode_StateMachine_1
    BlendWeights_0
    BlendWeights_1:float=1
   ) -> (Pose) {
    blendPoses [{linkId=-1, sourceLinkId=-1}, {linkId=-1, sourceLinkId=-1}]
    blendWeights [1,1]
    layerSetup [
     {branchFilters [{boneName=clavicle_l}, {boneName=VB VB spine_05_hand_attach_r_ik_hand_l}]}
     {branchFilters [{boneName=clavicle_r}, {boneName=VB VB spine_05_hand_attach_r_ik_hand_gun}]}
    ]
   }
   n2 = Apply Additive(Base=@n1, Additive=Slot 'AdditiveLayer'(Source=Additive Identity Pose()) {slotName=AdditiveLayer}) {alphaBoolBlend {blendOption=Linear}}
   n3 = Sequence Player 'JamBass_PlayMedium_125_A'() {sequence=/SparksCharacterCommon/Animation/MainPlayer/BandPlay/Bass/Loops/JamBass_PlayMedium_125_A}
   n4 = Get bCanPlayMidiAnimations() -> (bCanPlayMidiAnimations:bool)
   Output Pose(Result=Apply Additive(
     Base=Blend Poses by bool(BlendPose_0=@n2, BlendPose_1=@n3, BlendTime_0=0.5, BlendTime_1=0.5, bActiveValue=@n4) {blendType=Linear}
     Additive="Sequence Player (additive) 'Bass_SpawnInstrument_Pose'"() {sequence=/SparksCharacterCommon/Animation/MainPlayer/BandPlay/Bass/Bass_SpawnInstrument_Pose}
    ) {alphaBoolBlend {blendOption=Linear}}) {layerGroup=None}
   AnimGraphNode_StateMachine_5 = Body() {is=title AnimGraphNode_StateMachine, bAllowConduitEntryStates=true}
   "Save cached pose 'Body (Cache Pose)'"(Pose=@AnimGraphNode_StateMachine_5) {cacheName="Body (Cache Pose)"}
   AnimGraphNode_StateMachine_0 = Head() {is=title AnimGraphNode_StateMachine, bAllowConduitEntryStates=true}
  ]}
 Head {entry=LipSync, path=AnimGraph.AnimGraphNode_StateMachine_0.Head, ids=id, nodes [AnimStateNode_0 = LipSync() {is=title AnimStateNode}]}
 LipSync {
  path=AnimGraph.AnimGraphNode_StateMachine_0.Head.AnimStateNode_0.LipSync
  nodes [Output Animation Pose(Result=Play Beat Synced Animation!(
     ExplicitTime:float=0
     bShouldLoop:bool=false
     InSequence:Anim Sequence Base Object Reference=Get LipSyncAnimSequence() -> (LipSyncAnimSequence:Anim Sequence Object Reference)
     MusicClock:Music Clock Object Reference
     TempoMultiplier:float=1
     bEnableTempoVariants:bool=False
     TempoVariantHysteresis:float=0.05
     bShowDebug:bool=False
     BeatAndTime:Beat And Time Structure=(Beat=0,Time=0,BeatsPerSecond=0,IsValid=False)
    ) -> (Pose:Pose Link Structure) {is=title AnimGraphNode_PlayBeatSyncedAnim})]
 }
 Right Arm {
  entry=Right Arm
  path=AnimGraph.AnimGraphNode_StateMachine_1.Right Arm
  ids=id
  nodes [AnimStateNode_0 = Right Arm() {is=title AnimStateNode}]
 }
 Right Arm {
  path=AnimGraph.AnimGraphNode_StateMachine_1.Right Arm.AnimStateNode_0.Right Arm
  nodes [
   n1 = Sequence Player!(Sequence:Anim Sequence Base Object Reference, PlayRate:float=0, bLoopAnimation:bool=False) -> (Pose) {bLoopAnimation=false, playRate=0}
   n2 = "Use cached pose 'Body (Cache Pose)'"()
   Output Animation Pose(Result=Apply Additive(
     Base=Apply Additive(Base=@n2, Additive=@n1) {alphaBoolBlend {blendOption=Linear}}
     Additive=Sequence Player(PlayRate=0, bLoopAnimation=false) {playRate=0}
    ) {alphaBoolBlend {blendOption=Linear}})
  ]
 }
 Left Arm {
  entry=Conduit
  path=AnimGraph.AnimGraphNode_StateMachine_2.Left Arm
  ids=id
  nodes [
   AnimStateConduitNode_1 = Conduit()
   AnimStateNode_4 = "Frets && Chords"() {is=title AnimStateNode}
   AnimStateTransitionNode_12 = AnimStateTransitionNode()
   AnimStateTransitionNode_13 = AnimStateTransitionNode()
  ]
  links [
   {
    dst="Frets && Chords"
    rule=AnimGraph.AnimGraphNode_StateMachine_2.Left Arm.AnimStateTransitionNode_12.AnimationTransitionGraph_0
    src=Conduit
    via=AnimStateTransitionNode_12
   }
   {
    dst=Conduit
    rule=AnimGraph.AnimGraphNode_StateMachine_2.Left Arm.AnimStateTransitionNode_13.AnimationTransitionGraph_0
    src="Frets && Chords"
    via=AnimStateTransitionNode_13
   }
  ]
 }
 Conduit {path=AnimGraph.AnimGraphNode_StateMachine_2.Left Arm.AnimStateConduitNode_1.Conduit, nodes [Result(bCanEnterTransition=true)]}
 "Frets && Chords" {
  path="AnimGraph.AnimGraphNode_StateMachine_2.Left Arm.AnimStateNode_4.Frets && Chords"
  nodes [
   n1 = Blendspace Player 'Bass_FretHand_BS'(X=0, Y=0) {blendSpace=/SparksCharacterCommon/Animation/MainPlayer/BandPlay/Bass/FretHand/Bass_FretHand_BS, startPosition=1}
   n2 = Additive Identity Pose()
   n3 = "Not Equal (Enum)"!(
    A:ESparksAnimatableState Enum=Get AnimatableState() -> (AnimatableState:ESparksAnimatableState Enum=None)
    B:ESparksAnimatableState Enum=Animatable
   ) -> (ReturnValue:bool)
   Output Animation Pose(Result=Apply Additive(
     Base=Sequence Player 'Bass_FretHand_BasePose'() {sequence=/SparksCharacterCommon/Animation/MainPlayer/BandPlay/Bass/FretHand/Bass_FretHand_BasePose}
     Additive=Blend Poses by bool(BlendPose_0=@n1, BlendPose_1=@n2, bActiveValue=@n3) {blendType=Linear}
    ) {alphaBoolBlend {blendOption=Linear}})
  ]
 }
 AnimationTransitionGraph_0 {
  path=AnimGraph.AnimGraphNode_StateMachine_2.Left Arm.AnimStateTransitionNode_12.AnimationTransitionGraph_0
  nodes [Result(bCanEnterTransition=true)]
 }
 AnimationTransitionGraph_0 {
  path=AnimGraph.AnimGraphNode_StateMachine_2.Left Arm.AnimStateTransitionNode_13.AnimationTransitionGraph_0
  nodes [Result(bCanEnterTransition=NOT Boolean(A=false))]
 }
 Body {
  entry=Play 0
  path=AnimGraph.AnimGraphNode_StateMachine_5.Body
  ids=id
  nodes [
   AnimStateNode_0 = Play 1() {is=title AnimStateNode, stateEntered {guid=#1}, stateLeft {guid=#2}}
   AnimStateNode_8 = Play 0() {is=title AnimStateNode, stateEntered {guid=#3}, stateLeft {guid=#4}}
   AnimStateTransitionNode_2 = AnimStateTransitionNode()
   AnimStateTransitionNode_8 = AnimStateTransitionNode()
  ]
  links [
   {dst=Play 1, rule=AnimGraph.AnimGraphNode_StateMachine_5.Body.AnimStateTransitionNode_2.Transition, src=Play 0, via=AnimStateTransitionNode_2}
   {dst=Play 0, rule=AnimGraph.AnimGraphNode_StateMachine_5.Body.AnimStateTransitionNode_8.Transition, src=Play 1, via=AnimStateTransitionNode_8}
  ]
 }
 Play 1 {
  path=AnimGraph.AnimGraphNode_StateMachine_5.Body.AnimStateNode_0.Play 1
  nodes [
   n1 = Get PlaySequences() -> (PlaySequences:Array of Anim Sequence Object References)
   n2 = Get TempoMultipliers() -> (TempoMultipliers:"Array of Float (double-precision)s")
   Output Animation Pose(Result=Play Beat Synced Animation!(
     ExplicitTime=0
     bShouldLoop=True
     InSequence="Get (a copy)"(Array:Array of Anim Sequence Object References=@n1, Dimension 1=1) -> (Output:Anim Sequence Object Reference)
     MusicClock
     TempoMultiplier="Get (a copy)"(Array:"Array of Float (double-precision)s"=@n2, Dimension 1=1) -> (Output:double)
     bEnableTempoVariants=False
     TempoVariantHysteresis=0.05
     bShowDebug=False
    ) -> (Pose)) {
    node {stateEntryFunction {functionName=OnPlay1Entry}}
    stateEntryFunction {bSelfContext=true, memberGuid=#5, memberName=OnPlay1Entry}
   }
  ]
 }
 Play 0 {
  path=AnimGraph.AnimGraphNode_StateMachine_5.Body.AnimStateNode_8.Play 0
  nodes [
   n1 = Get PlaySequences()
   n2 = Get TempoMultipliers()
   Output Animation Pose(Result=Play Beat Synced Animation(
     ExplicitTime=0
     bShouldLoop=True
     InSequence="Get (a copy)"(Array:Array of Anim Sequence Object References=@n1) -> (Output:Anim Sequence Object Reference)
     TempoMultiplier="Get (a copy)"(Array:"Array of Float (double-precision)s"=@n2) -> (Output:double)
     bEnableTempoVariants=False
     TempoVariantHysteresis=0.05
     bShowDebug=False
    )) {
    node {stateEntryFunction {functionName=OnPlay0Entry}}
    stateEntryFunction {bSelfContext=true, memberGuid=#6, memberName=OnPlay0Entry}
   }
  ]
 }
 Transition {
  path=AnimGraph.AnimGraphNode_StateMachine_5.Body.AnimStateTransitionNode_2.Transition
  nodes [Result(bCanEnterTransition=JamShouldTransition!(self:Sparks Instrument Anim Instance Object Reference) -> (ReturnValue:bool))]
 }
 Transition {
  path=AnimGraph.AnimGraphNode_StateMachine_5.Body.AnimStateTransitionNode_8.Transition
  nodes [Result(bCanEnterTransition=JamShouldTransition())]
 }
 AnimNotify_PrintStringAssembler {nodes [
   n1 = AnimNotify_PrintStringAssembler!() -> (then, AnimNotifyEventName:String) {is=title K2Node_FunctionEntry}
   > Branch(Condition=Get DEBUG_PrintAnimLogic() -> (DEBUG_PrintAnimLogic:bool))
   n2 = Get DEBUG_PrintAnimLogic()
   n3 = Enum to String!(Enumerator:ESparksInstrumentType Enum=Get InstrumentType() -> (InstrumentType:ESparksInstrumentType Enum=Guitar)) -> (ReturnValue:String)
   > PrintString!(
    execute
    InString=JoinStringArray(SourceArray=Make Array!("[0]":String="[AnimNotify]", "[1]":String=@n3, "[2]":String=@n1) -> (Array:Array of Strings), Separator=" ")
    bPrintToScreen=@n2
    bPrintToLog=@n2
    TextColor=Get AnimNotify_TextColor() -> (AnimNotify_TextColor:Linear Color Structure)
    Duration
    Key
   ) -> (then)
   > Return Node()
  ]}
 EventGraph {nodes [
   n1 = Montage_Play!(
    execute
    self:Anim Instance Object Reference
    MontageToPlay:Anim Montage Object Reference=/SparksCharacterCommon/Animation/MainPlayer/BandPlay/Bass/Strums/Bass_Strum_PluckTwoFinger_M
    InPlayRate:float=1
    ReturnValueType:EMontagePlayReturnType Enum=MontageLength
    InTimeToStartMontageAt:float=0
    bStopAllMontages:bool=true
   ) -> (then, ReturnValue:float)
   n2 = Sequence(execute=[@n4.then, @n5.then, @n1.then])
   Gate!(Enter:Exec=@n8.else, Open:Exec=@n6.then, Close:Exec=@n2.then_0, Toggle:Exec, bStartClosed:bool=false) -> (Exit:Exec) {is=title K2Node_MacroInstance}
   > n3 = Flip Flop!(None:Exec) -> (A:Exec, B:Exec, IsA:bool=~) {is=title K2Node_MacroInstance}
   > n4 = Montage_Play(
    MontageToPlay=/SparksCharacterCommon/Animation/MainPlayer/BandPlay/Bass/Strums/Bass_Strum_PluckMiddle_M
    InPlayRate=1
    ReturnValueType=MontageLength
    InTimeToStartMontageAt=0
    bStopAllMontages=true
   )
   n5 = Montage_Play(
    execute=@n3.B
    MontageToPlay=/SparksCharacterCommon/Animation/MainPlayer/BandPlay/Bass/Strums/Bass_Strum_PluckIndex_M
    InPlayRate=1
    ReturnValueType=MontageLength
    InTimeToStartMontageAt=0
    bStopAllMontages=true
   )
   n6 = Delay(execute=@n2.then_1, Duration=0.1)
   Bass Strum() {is=title K2Node_CustomEvent}
   n7 = "Equal (Enum)"(A:ESparksAnimatableState Enum=Get AnimatableState() -> (AnimatableState=None), B:ESparksAnimatableState Enum=Animatable)
   > n8 = Branch(Condition=@n7)
   Bass Change Chord() {is=title K2Node_CustomEvent}
   > n9 = Branch(Condition=@n7)
   Set ChordHandPoseIndex(execute=@n9.else, ChordHandPoseIndex:int=RandomIntegerInRange!(Min:int=0, Max:int=4) -> (ReturnValue:int=0)) -> (Output_Get:int=0)
   n10 = Event On LipSync Asset Changed!() -> (OutputDelegate, then, NewLipSyncAsset:Anim Sequence Object Reference) {is=title K2Node_Event}
   > n11 = Is Valid(InputObject=@n10.NewLipSyncAsset)
   Set LipSyncAnimSequence(execute=@n11.Is Not Valid, LipSyncAnimSequence:Anim Sequence Object Reference) -> (Output_Get:Anim Sequence Object Reference)
   Set LipSyncAnimSequence(execute=@n11.Is Valid, LipSyncAnimSequence=@n10.NewLipSyncAsset)
   n12 = Event On MIDI Event Driver Changed!() -> (OutputDelegate, then, NewMidiEventDriver:"Midi Event Driver Interface Interface (by ref)") {is=title K2Node_Event}
   > n13 = Branch(Condition=IsValidInterface!(Interface:"Interface Interface (by ref)"=@n12.NewMidiEventDriver) -> (ReturnValue:bool))
   > n14 = Set MidiTrackName(MidiTrackName:Name=GetMidiTrackNameFromSparksInstrument!(self:Midi Event Driver Interface Interface=@n12.NewMidiEventDriver, InstrumentType:ESparksInstrumentType Enum=Bass) -> (ReturnValue:Name=None)) -> (Output_Get:Name=None)
   K2Node_Composite_3 = Song Intensity!(In:Exec=[@n14.then, @n15.then]) -> (Out:Exec) {is=title K2Node_Composite}
   > K2Node_Composite_1 = Fret Position!(In:Exec) -> (Out:Exec) {is=title K2Node_Composite}
   > K2Node_Composite_2 = Strum!(In:Exec) -> (Out:Exec) {is=title K2Node_Composite}
   n15 = Set MidiTrackName(execute=@n13.else, MidiTrackName=None) -> (Output_Get=None)
   Event Sparks Debug Update() {is=title K2Node_Event}
   > Branch(Condition=Get DEBUG_StrumTestFire() -> (DEBUG_StrumTestFire:bool))
   > Bass Strum() {is=title K2Node_CallFunction}
   > Set DEBUG_StrumTestFire(DEBUG_StrumTestFire:bool=false) -> (Output_Get:bool)
   Event On New Song() {is=title K2Node_Event}
   > SetSongDefaults!(
    execute
    AnimInstance:Sparks Instrument Anim Instance Object Reference=GetSelfAsSparksInstrumentAnimInstance!() -> (ReturnValue:Sparks Instrument Anim Instance Object Reference)
   ) -> (then)
   Event BlueprintPostEvaluateAnimation() {is=title K2Node_Event}
   Event BlueprintInitializeAnimation()
   AnimNotify_BeatExit() {is=title K2Node_Event}
  ]}
 Fret Position {
  path=EventGraph.K2Node_Composite_1.Fret Position
  nodes [
   Inputs!() -> (In:Exec)
   > n1 = Sequence!(execute) -> (then_0, then_1, then_2:Exec)
   n2 = Get FretMidiListener() -> (FretMidiListener:Async Action Listen For Midi Note Event Object Reference)
   > Is Valid(InputObject=@n2)
   > Cancel(self:Cancellable Async Action Object Reference=@n2)
   > Set FretMidiListener(FretMidiListener:Async Action Listen For Midi Note Event Object Reference) -> (Output_Get:Async Action Listen For Midi Note Event Object Reference)
   Branch(execute=@n1.then_1, Condition=IsValidInterface(Interface=Get MidiEventDriver() -> (MidiEventDriver:Midi Event Driver Interface Interface)))
   > n3 = ListenForMidiNoteEvents!(
    execute:Exec
    MidiEventDriver:Midi Event Driver Interface Interface=Get MidiEventDriver()
    MidiTrackName:Name=Get MidiTrackName() -> (MidiTrackName:Name=None)
    EventNotes:Array of Bytes=GetBassMidiNoteIntegersFromEnums!(MidiNoteEvents:Array of EBassMidiNoteEvent Enums=Get MidiFretEvents() -> (MidiFretEvents:Array of EBassMidiNoteEvent Enums=None)) -> (ReturnValue:Array of Bytes)
    SecondsBeforeEventStart:float=0
    SecondsBeforeEventEnd:float=0
    InitBroadcast:EMidiEventListenerInitBroadcast Enum=MostRecentEvent
   ) -> (
    then:Exec
    AsyncTaskProxy:Async Action Listen For Midi Note Event Object Reference
    OnMidiEventStart:Exec
    OnMidiEventEnd:Exec
    OnFinished:Exec
    MidiEventNote:Byte
    MidiEventData:"Parsed Midi Note Event Structure (by ref)"
   ) {is=title K2Node_AsyncAction}
   > Set FretMidiListener(FretMidiListener=@n3)
   Set FretNumber(execute=@n3.OnMidiEventStart, FretNumber:int=GetFretNumberFromMidiNote!(MidiNote:Byte=@n3.MidiEventNote) -> (ReturnValue:int=0)) -> (Output_Get:int=0)
   Outputs!(Out:Exec=@n1.then_2)
  ]
 }
 Strum {
  path=EventGraph.K2Node_Composite_2.Strum
  nodes [
   Inputs()
   > n1 = Sequence()
   n2 = Get StrumMidiListener() -> (StrumMidiListener:Async Action Listen For Midi Note Event Object Reference)
   > Is Valid(InputObject=@n2)
   > Cancel(self=@n2)
   > Set StrumMidiListener(StrumMidiListener:Async Action Listen For Midi Note Event Object Reference) -> (Output_Get:Async Action Listen For Midi Note Event Object Reference)
   Branch(execute=@n1.then_1, Condition=IsValidInterface(Interface=Get MidiEventDriver()))
   > n3 = ListenForMidiNoteEvents(
    MidiEventDriver=Get MidiEventDriver()
    MidiTrackName=Get MidiTrackName() -> (MidiTrackName=None)
    EventNotes=Make Array!(
     "[0]":Byte=GetBassMidiNoteIntegerFromEnum!(MidiNoteEvent:EBassMidiNoteEvent Enum=StrumDown) -> (ReturnValue:Byte=0)
     "[1]":Byte=GetBassMidiNoteIntegerFromEnum(MidiNoteEvent=StrumUp) -> (ReturnValue=0)
    ) -> (Array:Array of Bytes)
    SecondsBeforeEventStart=0.06
    SecondsBeforeEventEnd=0
    InitBroadcast=NoPreviousEvents
   )
   > Set StrumMidiListener(StrumMidiListener=@n3)
   Bass Strum(execute=@n3.OnMidiEventStart)
   Outputs(Out=@n1.then_2)
  ]
 }
 Song Intensity {
  path=EventGraph.K2Node_Composite_3.Song Intensity
  nodes [
   Inputs()
   > n1 = Sequence()
   n2 = Get IntensityMidiListener() -> (IntensityMidiListener:Async Action Listen For Midi Text Event Object Reference)
   > Is Valid(InputObject=@n2)
   > Cancel(self=@n2)
   > Set IntensityMidiListener(IntensityMidiListener:Async Action Listen For Midi Text Event Object Reference) -> (Output_Get:Async Action Listen For Midi Text Event Object Reference)
   Branch(execute=@n1.then_1, Condition=IsValidInterface(Interface=Get MidiEventDriver()))
   > n3 = ListenForMidiTextEvents!(
    execute:Exec
    MidiEventDriver:Midi Event Driver Interface Interface=Get MidiEventDriver()
    MidiTrackName:Name=Get MidiTrackName() -> (MidiTrackName=None)
    EventStrings:Array of Strings=Make Array!(
     "[0]":String=GetBassMidiTextStringFromEnum!(MidiTextEvent:EBassMidiTextEvent Enum=PlayingMellow) -> (ReturnValue:String)
     "[1]":String=GetBassMidiTextStringFromEnum(MidiTextEvent=PlayingStandard)
     "[2]":String=GetBassMidiTextStringFromEnum(MidiTextEvent=PlayingIntense)
     "[3]":String=GetBassMidiTextStringFromEnum(MidiTextEvent=IdleRhythmMellow)
     "[4]":String=GetBassMidiTextStringFromEnum(MidiTextEvent=IdleRhythmStandard)
     "[5]":String=GetBassMidiTextStringFromEnum(MidiTextEvent=IdleRhythmIntense)
     "[6]":String=GetBassMidiTextStringFromEnum(MidiTextEvent=IdleRealtime)
    ) -> (Array:Array of Strings)
    SecondsBeforeEventStart:float=Get SecBeforeIntensity() -> (SecBeforeIntensity:float)
    InitBroadcast:EMidiEventListenerInitBroadcast Enum=MostRecentEvent
   ) -> (
    then:Exec
    AsyncTaskProxy:Async Action Listen For Midi Text Event Object Reference
    OnMidiEventStart:Exec
    OnFinished:Exec
    MidiEventText:String
    MidiEventData:"Parsed Midi Text Event Structure (by ref)"
   ) {is=title K2Node_AsyncAction}
   > Set IntensityMidiListener(IntensityMidiListener=@n3)
   MidiEvent_Intensity_OnEvent!(execute=@n3.OnMidiEventStart, self, MidiEventText:String=@n3.MidiEventText) -> (then)
   Outputs(Out=@n1.then_2)
  ]
 }
 MidiEvent_Intensity_OnEvent {nodes [
   n1 = MidiEvent_Intensity_OnEvent!() -> (then, MidiEventText:String) {is=title K2Node_FunctionEntry}
   > n2 = Sequence!(execute) -> (then_0, then_1)
   > n3 = TranslateMidiEventTextToSongIntensity!(execute, MidiEventText:Name=String To Name(InString=@n1)) -> (then, IdleRealtime:bool, MidiSongIntensityIsIdle:bool, SongIntensity:Gameplay Tag Structure)
   > n4 = Branch(Condition=@n3)
   Set DesiredSongIntensity(execute=@n4.else, DesiredSongIntensity:Gameplay Tag Structure=@n3.SongIntensity) -> (Output_Get:Gameplay Tag Structure)
   MidiEvent_PrintStringAssembler!(
    execute=@n2.then_1
    self
    InstrumentType:ESparksInstrumentType Enum=Get InstrumentType() -> (InstrumentType=Guitar)
    MidiEventText:String=@n1
    SecondsBeforeEvent:double=Get SecBeforeIntensity()
   ) -> (then)
  ]}
 MidiEvent_PrintStringAssembler {nodes [
   n1 = MidiEvent_PrintStringAssembler!() -> (then, InstrumentType:ESparksInstrumentType Enum=Guitar, MidiEventText:String, SecondsBeforeEvent:double=~) {is=title K2Node_FunctionEntry}
   > Branch(Condition=Get DEBUG_PrintAnimLogic())
   > n2 = Branch(Condition="Equal (Float)"(A=@n1.SecondsBeforeEvent))
   > n3 = Set Duration(Duration:float=2) -> (Output_Get:float)
   n4 = Get DEBUG_PrintAnimLogic()
   n5 = "To String (Float)"(InDouble=@n1.SecondsBeforeEvent)
   n6 = Enum to String(Enumerator=@n1)
   PrintString(
    execute=[@n3.then, @n7.then]
    InString=JoinStringArray(
     SourceArray=Make Array!("[0]"="[Anim Midi]", "[1]"=@n6, "[2]"=@n1.MidiEventText, "[3]"=in, "[4]"=@n5, "[5]"=seconds...) -> (Array)
     Separator=" "
    )
    bPrintToScreen=@n4
    bPrintToLog=@n4
    TextColor=Get AnimNotify_TextColor()
    Duration=Get Duration() -> (Duration:float)
   )
   > Return Node()
   n7 = Set Duration(execute=@n2.else, Duration=@n1.SecondsBeforeEvent)
  ]}
 OnPlay0Entry {nodes [
   OnPlay0Entry!() -> (then, Context:"Anim Update Context Structure (by ref)", Node:"Anim Node Reference Structure (by ref)") {is=title K2Node_FunctionEntry}
   > SetPlayAnimAndTempoMultiplier!(execute, self, Index:int=0) -> (then)
   > Return Node()
  ]}
 OnPlay1Entry {nodes [
   OnPlay1Entry!() -> (then, Context:"Anim Update Context Structure (by ref)", Node:"Anim Node Reference Structure (by ref)") {is=title K2Node_FunctionEntry}
   > SetPlayAnimAndTempoMultiplier(Index=1)
   > Return Node()
  ]}
 SetPlayAnimAndTempoMultiplier {nodes [
   n1 = SetPlayAnimAndTempoMultiplier!() -> (then, Index:int) {is=title K2Node_FunctionEntry}
   > Set SongIntensity(SongIntensity:Gameplay Tag Structure=Get DesiredSongIntensity() -> (DesiredSongIntensity:Gameplay Tag Structure)) -> (Output_Get:Gameplay Tag Structure)
   > n2 = PickIntensityAnimLoopCycle!(
    execute
    self:Sparks Instrument Anim Instance Object Reference
    BPMSet:"Sparks Anim BPMSet Structure (by ref)"=Get HomebaseBPMs() -> (HomebaseBPMs:Sparks Anim BPMSet Structure)
    TransitionTag:Gameplay Tag Structure
    ForceVariant:bool=false
    VariantIndex:int=0
   ) -> (then, ReturnValue:"Sparks Anim Loop Cycle Structure (by ref)")
   > Set Array Elem!(
    execute
    TargetArray:Array of Anim Sequence Object References=Get PlaySequences()
    Index:int=@n1
    Item:"Anim Sequence Object Reference (by ref)"=Break Sparks Anim Loop Cycle!(SparksAnimLoopCycle:"Sparks Anim Loop Cycle Structure (by ref)"=@n2) -> (Loop:Anim Sequence Object Reference, Entry:Anim Sequence Object Reference, Exit:Anim Sequence Object Reference) {is=title K2Node_BreakStruct}
    bSizeToFit:bool=true
   ) -> (then)
   > Set Array Elem(
    TargetArray:"Array of Float (double-precision)s"=Get TempoMultipliers()
    Index=@n1
    Item:"Float (double-precision) (by ref)"=Break Sparks Anim BPMSet!(SparksAnimBPMSet:"Sparks Anim BPMSet Structure (by ref)"=Get HomebaseBPMs()) -> (BPMBuckets:Array of Sparks Anim Intensity Set Object References, CurrentTempoMultiplier:float=~) {is=title K2Node_BreakStruct}.CurrentTempoMultiplier
    bSizeToFit=true
   )
  ]}
]
variables [
 AnimatableState {type=ESparksAnimatableState, default=None}
 AnimNotify_TextColor {type=LinearColor, default=<rgba 1 0 1 1>}
 BPM Preview {type=number, default=125}
 ChordHandPoseIndex {type=integer}
 DEBUG_PrintAnimLogic {type=boolean}
 DEBUG_StrumTestFire {type=boolean}
 FretMidiListener {type=AsyncAction_ListenForMidiNoteEvent, default=null}
 InstrumentType {type=ESparksInstrumentType, default=Bass}
 IntensityMidiListener {type=AsyncAction_ListenForMidiTextEvent, default=null}
 LipSyncAnimSequence {type=AnimSequence, default=null}
 MidiFretEvents {
  type="array<EBassMidiNoteEvent>"
  default [
   FretPosition1
   FretPosition2
   FretPosition3
   FretPosition4
   FretPosition5
   FretPosition6
   FretPosition7
   FretPosition8
   FretPosition9
   FretPosition10
   FretPosition11
   FretPosition12
   FretPosition13
   FretPosition14
   FretPosition15
   FretPosition16
   FretPosition17
   FretPosition18
   FretPosition19
   FretPosition20
  ]
 }
 MidiTrackName {type=string, default=None}
 PlaySequences {type="array<AnimSequence>", default [null,null]}
 SecBeforeIntensity {type=number}
 StrumCorrectivePose {type=AnimSequenceBase, default=null}
 StrumMidiListener {type=AsyncAction_ListenForMidiNoteEvent, default=null}
 StrumStartPose {type=AnimSequenceBase, default=null}
 TempoMultipliers {type="array<number>", default [0,0]}
]
