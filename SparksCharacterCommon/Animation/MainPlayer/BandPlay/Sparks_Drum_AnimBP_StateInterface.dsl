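; Animation state interface for the band-play drummer character, declaring getters for drum animation state.
; The getters cover the bass hit index, a shortening amount for the right leg's maximum delay, and hi-hat open and close triggers.
; Each function returns only a zero or false default and contains no other logic.
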(fn Get_HitBassIndex
  (return 0))

(fn Get_RightLegMaxDelayShortenedBy
  (return 0.0))

(fn Get_TriggerHiHatClose
  (return false))

(fn Get_TriggerHiHatOpen
  (return false))
