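# Animation Blueprint that poses a Fortnite player character playing guitar in Sparks Jam band play.
# A BaseBody state machine with strum and fret arm overrides takes face, spine, arm, head and foot-lift layers, with additives scaled by dampen/amplify alphas and beat-synced clips.
# A traversal pose blends in lower-body motion.
# Procedural foot-shift state machines move feet and pelvis through timed crossfades, followed by leg IK, debug-toggled hand IK and curve-driven attach-bone copies.
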AnimBlueprint /SparksCharacterCommon/Animation/MainPlayer/BandPlay/Sparks_Guitar_JamV2_AnimBP
target_skeleton=&Fortnite_M_Avg_Player_Skeleton
graphs [
 AnimGraph {nodes [
   n1 = "Transform (Modify) Bone - Bone: pelvis"!(
    ComponentPose=Local To Component()
    Alpha=Get BaseBody_Squat_BlendWeight() -> (BaseBody_Squat_BlendWeight:double)
    Translation_X:double=0
    Translation_Y:double=0
    Translation_Z:double=Get BaseBody_SquatDistance() -> (BaseBody_SquatDistance:double)
   ) -> (Pose) {
    translationMode=BMM_Additive
    translationSpace=BCS_WorldSpace
    alphaBoolBlend {blendOption=Linear}
    boneToModify {boneName=pelvis}
    translation {z=-30}
   }
   Component To Local(ComponentPose=@n1)
   n2 = Sequence Player 'TPose_Guitar'!(Sequence:Anim Sequence Base Object Reference=/SparksCharacterCommon/Animation/MainPlayer/BandPlay/Guitar/TPose_Guitar) -> (Pose)
   n3 = Use cached pose 'BaseBody Pre-Additives Pose'()
   n4 = Get LeftArm_CurrentlyActive?() -> (LeftArm_CurrentlyActive?:bool)
   Layered blend per bone(
    BasePose=Layered blend per bone(
     BlendPoses_0=Blend Poses by bool(BlendPose_0=@n2, BlendPose_1=@n3, BlendTime_0=1, BlendTime_1=1, bActiveValue=@n4)
     BlendWeights_0=Get LeftArm_BlendWeight() -> (LeftArm_BlendWeight:float)
    ) {layerSetup [{branchFilters [{boneName=clavicle_l}]}]}
    BlendPoses_0=Blend Poses by bool(
     BlendPose_0=Sequence Player 'TPose_Guitar'(Sequence=&TPose_Guitar)
     BlendPose_1=Use cached pose 'BaseBody Pre-Additives Pose'()
     BlendTime_0=1
     BlendTime_1=1
     bActiveValue=Get RightArm_CurrentlyActive?() -> (RightArm_CurrentlyActive?:bool)
    )
    BlendWeights_0=Get RightArm_BlendWeight() -> (RightArm_BlendWeight:float)
   ) {layerSetup [{branchFilters [{boneName=clavicle_r}]}]}
   n5 = Apply Additive(
    Base=Sequence Player 'TPose_Guitar'(Sequence=&TPose_Guitar)
    Additive=Use cached pose 'SpineMotion Pose'()
    Alpha=Get SpineMotion_DampenAlpha() -> (SpineMotion_DampenAlpha:float)
   ) {alphaBoolBlend {blendOption=Linear}}
   n6 = Use cached pose 'SpineMotion Pose'()
   n7 = Get SpineMotion_AmplifyAlpha() -> (SpineMotion_AmplifyAlpha:float)
   Layered blend per bone(
    BlendPoses_0=Blend Poses by bool(
     BlendPose_0=Apply Additive(Base=@n5, Additive=@n6, Alpha=@n7) {alphaBoolBlend {blendOption=Linear}}
     BlendPose_1=Use cached pose 'BaseBody Pre-Additives Pose'()
     BlendTime_0=1
     BlendTime_1=1
     bActiveValue=Get SpineMotion_CurrentlyActive?() -> (SpineMotion_CurrentlyActive?:bool)
    )
    BlendWeights_0=Get SpineMotion_BlendWeight() -> (SpineMotion_BlendWeight:float)
   ) {layerSetup [{branchFilters [{boneName=neck_01}]}]}
   n8 = Leg IK!(ComponentPose:Component Space Pose Link Structure, AlphaCurveName:Name=LeftHandIK) -> (Pose:Component Space Pose Link Structure) {
    is=title AnimGraphNode_LegIK
    alphaInputType=Curve
    alphaBoolBlend {blendOption=Linear}
    legsDefinition [{
      bEnableKneeTwistCorrection=true
      bEnableRotationLimit=false
      footBoneForwardAxis=Y
      hingeRotationAxis=None
      minRotationAngle=15
      numBonesInLimb=2
      twistOffsetCurveName=None
      fKFootBone {boneName=hand_l}
      iKFootBone {boneName=ik_hand_l}
     }]
   }
   Leg IK(ComponentPose=@n8, AlphaCurveName=RightHandIK) {
    alphaInputType=Curve
    alphaBoolBlend {blendOption=Linear}
    legsDefinition [{
      bEnableKneeTwistCorrection=true
      bEnableRotationLimit=false
      footBoneForwardAxis=Y
      hingeRotationAxis=None
      minRotationAngle=15
      numBonesInLimb=2
      twistOffsetCurveName=None
      fKFootBone {boneName=hand_r}
      iKFootBone {boneName=ik_hand_r}
     }]
   }
   Blend Multi!(Poses_0:Pose Link Structure, Poses_1:Pose Link Structure, DesiredAlphas_0:float=0, DesiredAlphas_1:float=0) -> (Pose:Pose Link Structure) {is=title AnimGraphNode_MultiWayBlend, desiredAlphas [0,0], poses [{linkId=-1, sourceLinkId=-1}, {linkId=-1, sourceLinkId=-1}]}
   AnimGraphNode_StateMachine_15 = FootShift 1stFoot Translate() {is=title AnimGraphNode_StateMachine}
   n9 = Get FootShift_BlendWeight() -> (FootShift_BlendWeight:float)
   AnimGraphNode_StateMachine_6 = Instrument Mesh Chooser() {is=title AnimGraphNode_StateMachine}
   n10 = Blend Poses by bool(
    BlendPose_0=Sequence Player 'JamBass_PlayMedium_125_A'() {sequence=/SparksCharacterCommon/Animation/MainPlayer/BandPlay/Bass/Loops/JamBass_PlayMedium_125_A}
    BlendPose_1=Sequence Player 'JamGuitar_PlayMedium_125_A'() {sequence=/SparksCharacterCommon/Animation/MainPlayer/BandPlay/Guitar/Loops/JamGuitar_PlayMedium_125_A}
    BlendTime_0=0
    BlendTime_1=0
    bActiveValue="Equal (Enum)"(A:ESparksInstrumentType Enum=Get Instrument() -> (Instrument:ESparksInstrumentType Enum=Guitar), B:ESparksInstrumentType Enum=Bass)
   ) {blendType=Linear}
   n11 = Get bCanPlayMidiAnimations() -> (bCanPlayMidiAnimations:bool)
   n12 = Use cached pose 'BaseBody Pre-Additives Pose'()
   n13 = Get DEBUG__BypassBodySectionAdditives?() -> (DEBUG__BypassBodySectionAdditives?:bool)
   n14 = Apply Additive(
    Base=Sequence Player 'TPose_Guitar'(Sequence=&TPose_Guitar)
    Additive=Use cached pose 'HeadMotion Pose'()
    Alpha=Get HeadMotion_DampenAlpha() -> (HeadMotion_DampenAlpha:float)
   ) {alphaBoolBlend {blendOption=Linear}}
   n15 = Use cached pose 'HeadMotion Pose'()
   n16 = Get HeadMotion_AmplifyAlpha() -> (HeadMotion_AmplifyAlpha:float)
   n17 = Sequence Player 'TPose_Guitar'(Sequence=&TPose_Guitar)
   n18 = Use cached pose 'BaseBody Pre-Additives Pose'()
   n19 = Get HeadMotion_CurrentlyActive?() -> (HeadMotion_CurrentlyActive?:bool)
   n20 = Layered blend per bone(
    BasePose=Layered blend per bone(
     BasePose=Use cached pose 'BaseBody Post-Additives Pose'()
     BlendPoses_0=Blend Poses by bool(BlendPose_0=@n17, BlendPose_1=@n18, BlendTime_0=1, BlendTime_1=1, bActiveValue=@n19)
     BlendWeights_0=Get HeadMotion_BlendWeight() -> (HeadMotion_BlendWeight:float)
    ) {layerSetup [{branchFilters [{boneName=neck_01}]}]}
    BlendPoses_0=Blend Poses by bool(
     BlendPose_0=Apply Additive(Base=@n14, Additive=@n15, Alpha=@n16) {alphaBoolBlend {blendOption=Linear}}
     BlendPose_1=Use cached pose 'BaseBody Pre-Additives Pose'()
     BlendTime_0=1
     BlendTime_1=1
     bActiveValue=Get HeadMotion_CurrentlyActive?()
    )
    BlendWeights_0=Get HeadMotion_BlendWeight()
   ) {layerSetup [{branchFilters [{boneName=neck_01}]}]}
   n21 = Apply Additive(
    Base=Blend Poses by bool(
     BlendPose_0=Use cached pose 'Traversal Pose'()
     BlendPose_1=Blend Poses by bool(BlendPose_0=@n12, BlendPose_1=@n20, BlendTime_0=0, BlendTime_1=0, bActiveValue=@n13)
     BlendTime_0=0.6
     BlendTime_1=0.4
     bActiveValue=Get Traversal_IsCharacterMoving?() -> (Traversal_IsCharacterMoving?:bool)
    ) {blendType=Linear}
    Additive=Slot 'AdditiveLayer'(Source=Additive Identity Pose()) {slotName=AdditiveLayer}
   ) {alphaBoolBlend {blendOption=Linear}}
   n22 = Modify Curve!(
    SourcePose=Apply Additive(
     Base=Blend Poses by bool(BlendPose_0=@n21, BlendPose_1=@n10, BlendTime_0=0.5, BlendTime_1=0.5, bActiveValue=@n11) {blendType=Linear}
     Additive=@AnimGraphNode_StateMachine_6
    ) {alphaBoolBlend {blendOption=Linear}}
    CurveValues_0=1
    CurveValues_1:float=1
    Alpha=Get FaceExpression_ClosedEyes() -> (FaceExpression_ClosedEyes:float)
   ) -> (Pose) {curveNames [l_blink_pose,r_blink_pose], curveValues [0,0]}
   AnimGraphNode_StateMachine_16 = FootShift 2ndFoot Translate() {is=title AnimGraphNode_StateMachine}
   AnimGraphNode_StateMachine_7 = FootShift 1stPelvis Translate() {is=title AnimGraphNode_StateMachine}
   n23 = Blend Multi(
    Poses_0=Blend Multi(
     Poses_0=Blend Multi(Poses_0=@n22, Poses_1=@AnimGraphNode_StateMachine_15, DesiredAlphas_0=1, DesiredAlphas_1=@n9) {desiredAlphas [0,0], poses [{linkId=-1, sourceLinkId=-1}, {linkId=-1, sourceLinkId=-1}]}
     Poses_1=@AnimGraphNode_StateMachine_16
     DesiredAlphas_0=1
     DesiredAlphas_1=Get FootShift_BlendWeight()
    ) {desiredAlphas [0,0], poses [{linkId=-1, sourceLinkId=-1}, {linkId=-1, sourceLinkId=-1}]}
    Poses_1=@AnimGraphNode_StateMachine_7
    DesiredAlphas_0=1
    DesiredAlphas_1=Get FootShift_BlendWeight()
   ) {desiredAlphas [0,0], poses [{linkId=-1, sourceLinkId=-1}, {linkId=-1, sourceLinkId=-1}]}
   AnimGraphNode_StateMachine_0 = FootShift 2ndPelvis Translate() {is=title AnimGraphNode_StateMachine}
   n24 = Get FootShift_BlendWeight()
   n25 = Leg IK!(
    ComponentPose=Local To Component(LocalPose=Blend Multi(Poses_0=@n23, Poses_1=@AnimGraphNode_StateMachine_0, DesiredAlphas_0=1, DesiredAlphas_1=@n24) {desiredAlphas [0,0], poses [{linkId=-1, sourceLinkId=-1}, {linkId=-1, sourceLinkId=-1}]})
    bAlphaBoolEnabled:bool=True
   ) -> (Pose) {
    alphaInputType=Bool
    alphaBoolBlend {blendOption=Linear}
    legsDefinition [
     {
      bEnableKneeTwistCorrection=true
      bEnableRotationLimit=false
      footBoneForwardAxis=Y
      hingeRotationAxis=Z
      minRotationAngle=15
      numBonesInLimb=2
      twistOffsetCurveName=None
      fKFootBone {boneName=foot_l}
      iKFootBone {boneName=ik_foot_l}
     }
     {
      bEnableKneeTwistCorrection=true
      bEnableRotationLimit=false
      footBoneForwardAxis=Y
      hingeRotationAxis=Z
      minRotationAngle=15
      numBonesInLimb=2
      twistOffsetCurveName=None
      fKFootBone {boneName=foot_r}
      iKFootBone {boneName=ik_foot_r}
     }
    ]
   }
   n26 = Leg IK(ComponentPose=@n25, bAlphaBoolEnabled=Get DEBUG_EnableLeftHandIK?() -> (DEBUG_EnableLeftHandIK?:bool)) {
    alphaInputType=Bool
    alphaBoolBlend {blendOption=Linear}
    legsDefinition [{
      bEnableKneeTwistCorrection=true
      bEnableRotationLimit=false
      footBoneForwardAxis=Y
      hingeRotationAxis=Z
      minRotationAngle=15
      numBonesInLimb=2
      twistOffsetCurveName=None
      fKFootBone {boneName=hand_l}
      iKFootBone {boneName=ik_hand_l}
     }]
   }
   n27 = Leg IK(ComponentPose=@n26, bAlphaBoolEnabled=Get DEBUG_EnableRightHandIK?() -> (DEBUG_EnableRightHandIK?:bool)) {
    alphaInputType=Bool
    alphaBoolBlend {blendOption=Linear}
    legsDefinition [{
      bEnableKneeTwistCorrection=true
      bEnableRotationLimit=false
      footBoneForwardAxis=Y
      hingeRotationAxis=Z
      minRotationAngle=15
      numBonesInLimb=2
      twistOffsetCurveName=None
      fKFootBone {boneName=hand_r}
      iKFootBone {boneName=ik_hand_r}
     }]
   }
   n28 = Copy Bone - Source Bone: VB spine_05_hand_attach_r - Target Bone: hand_attach_r!(ComponentPose:Component Space Pose Link Structure=@n27, AlphaCurveName:Name=HandAttachRinChestSpace) -> (Pose:Component Space Pose Link Structure) {
    is=title AnimGraphNode_CopyBone
    alphaInputType=Curve
    bCopyRotation=true
    bCopyScale=true
    bCopyTranslation=true
    alphaBoolBlend {blendOption=Linear}
    sourceBone {boneName=VB spine_05_hand_attach_r}
    targetBone {boneName=hand_attach_r}
   }
   n29 = Copy Bone - Source Bone: VB spine_05_weapon_r - Target Bone: weapon_r!(ComponentPose:Component Space Pose Link Structure=@n28, AlphaCurveName:Name=WeaponRinChestSpace) -> (Pose:Component Space Pose Link Structure) {
    is=title AnimGraphNode_CopyBone
    alphaInputType=Curve
    bCopyRotation=true
    bCopyScale=true
    bCopyTranslation=true
    alphaBoolBlend {blendOption=Linear}
    sourceBone {boneName=VB spine_05_weapon_r}
    targetBone {boneName=weapon_r}
   }
   Output Pose(Result=Blend Poses by bool(
     BlendPose_0=Sequence Player 'JamV2Guitar_PlayMedium_Pop_125_A'!(Sequence:Anim Sequence Base Object Reference=/SparksCharacterCommon/Animation/MainPlayer/BandPlay/Guitar/Loops/JamV2Guitar_PlayMedium_Pop_125_A) -> (Pose)
     BlendPose_1=Component To Local(ComponentPose=@n29)
     BlendTime_0=0
     BlendTime_1=0
     bActiveValue=Get DEBUG_BaseBody_UsePlayLoopOnly?() -> (DEBUG_BaseBody_UsePlayLoopOnly?:bool)
    ) {blendType=Linear}) {layerGroup=None}
   AnimGraphNode_StateMachine_11 = FootShift 2ndFoot Lift() {is=title AnimGraphNode_StateMachine}
   Save cached pose 'FootShift 2ndFoot Lift Pose'(Pose=@AnimGraphNode_StateMachine_11) {cacheName=FootShift 2ndFoot Lift Pose}
   Save cached pose 'PassThroughFullBodyPose'(Pose=Input Pose() {name=PassThroughFullBodyPose}) {cacheName=PassThroughFullBodyPose}
   n30 = Layered blend per bone(
    BasePose=Use cached pose 'BaseBody Pre-Additives Pose'()
    BlendPoses_0=Blend Poses by bool(
     BlendPose_0=Use cached pose 'PassThroughFullBodyPose'()
     BlendPose_1=Use cached pose 'LowerBody Pose'()
     BlendTime_0=0
     BlendTime_1=0
     bActiveValue=Get DEBUG_LowerBody_UseLocomotionSystem?() -> (DEBUG_LowerBody_UseLocomotionSystem?:bool)
    ) {blendType=Linear}
   ) {blendMode=BlendMask, blendMasks [&Fortnite_M_Avg_Player_Skeleton:LowerBody], layerSetup []}
   Save cached pose 'Traversal Pose'(Pose=@n30) {cacheName=Traversal Pose}
   n31 = Get SpineMotion_BlendWeight()
   n32 = Layered blend per bone(BasePose=Additive Identity Pose(), BlendPoses_0=Use cached pose 'SpineMotion Pose'(), BlendWeights_0=@n31) {layerSetup [{branchFilters [{boneName=pelvis}, {boneName=ik_hand_l}, {boneName=ik_hand_r}]}]}
   n33 = Layered blend per bone(
    BasePose=Use cached pose 'BaseBody Pre-Additives Pose'()
    BlendPoses_0=Use cached pose 'FaceLipSync Pose'()
    BlendWeights_0=Get LipSyncBlendWeight() -> (LipSyncBlendWeight:double)
   ) {
    blendMode=BlendMask
    blendMasks [&Fortnite_M_Avg_Player_Skeleton:FacialCurvesNoWeights]
    layerSetup []
   }
   n34 = Layered blend per bone(
    BasePose=Use cached pose 'BaseBody Pre-Additives Pose'()
    BlendPoses_0=Blend Poses by bool(
     BlendPose_0=Use cached pose 'FaceExpression Pose'()
     BlendPose_1=@n33
     BlendTime_0=0
     BlendTime_1=0
     bActiveValue=Get DEBUG_FaceExpression_DisableLipSync?() -> (DEBUG_FaceExpression_DisableLipSync?:bool)
    ) {blendType=Linear}
    BlendWeights_0=Get DEBUG_FaceExpression_BlendWeight() -> (DEBUG_FaceExpression_BlendWeight:float)
   ) {
    blendMode=BlendMask
    curveBlendOption=BlendByWeight
    blendMasks [&Fortnite_M_Avg_Player_Skeleton:FacialCurvesNoWeights]
    layerSetup []
   }
   n35 = Apply Additive(Base=@n34, Additive=@n32, Alpha=Get SpineMotion_DampenAlpha()) {alphaBoolBlend {blendOption=Linear}}
   n36 = Layered blend per bone(BasePose=Additive Identity Pose(), BlendPoses_0=Use cached pose 'SpineMotion Pose'(), BlendWeights_0=@n31) {layerSetup [{branchFilters [{boneName=pelvis}, {boneName=ik_hand_l}, {boneName=ik_hand_r}]}]}
   n37 = Get SpineMotion_AmplifyAlpha()
   n38 = Get FootShift_BlendWeight()
   n39 = Layered blend per bone(BasePose=Additive Identity Pose(), BlendPoses_0=Use cached pose 'FootShift 1stFoot Lift Pose'(), BlendWeights_0=@n38) {
    blendMode=BlendMask
    blendMasks [&Fortnite_M_Avg_Player_Skeleton:LowerBodyNoFeathering]
    layerSetup []
   }
   n40 = Layered blend per bone(BasePose=Additive Identity Pose(), BlendPoses_0=Use cached pose 'FootShift 1stFoot Lift Pose'(), BlendWeights_0=@n38) {
    blendMode=BlendMask
    blendMasks [&Fortnite_M_Avg_Player_Skeleton:LowerBodyNoFeathering]
    layerSetup []
   }
   n41 = Apply Additive(
    Base=Apply Additive(
     Base=Apply Additive(Base=@n35, Additive=@n36, Alpha=@n37) {alphaBoolBlend {blendOption=Linear}}
     Additive=@n39
     Alpha=Get FootShift_DampenAlpha() -> (FootShift_DampenAlpha:float)
    ) {alphaBoolBlend {blendOption=Linear}}
    Additive=@n40
    Alpha=Get FootShift_AmplifyAlpha() -> (FootShift_AmplifyAlpha:float)
   ) {alphaBoolBlend {blendOption=Linear}}
   n42 = Get FootShift_BlendWeight()
   n43 = Layered blend per bone(BasePose=Additive Identity Pose(), BlendPoses_0=Use cached pose 'FootShift 2ndFoot Lift Pose'(), BlendWeights_0=@n42) {
    blendMode=BlendMask
    blendMasks [&Fortnite_M_Avg_Player_Skeleton:LowerBodyNoFeathering]
    layerSetup []
   }
   n44 = Get FootShift_DampenAlpha()
   n45 = Layered blend per bone(BasePose=Additive Identity Pose(), BlendPoses_0=Use cached pose 'FootShift 2ndFoot Lift Pose'(), BlendWeights_0=@n42) {
    blendMode=BlendMask
    blendMasks [&Fortnite_M_Avg_Player_Skeleton:LowerBodyNoFeathering]
    layerSetup []
   }
   n46 = Get LeftArm_BlendWeight()
   n47 = Apply Additive(
    Base=Apply Additive(
     Base=Apply Additive(Base=@n41, Additive=@n43, Alpha=@n44) {alphaBoolBlend {blendOption=Linear}}
     Additive=@n45
     Alpha=Get FootShift_AmplifyAlpha()
    ) {alphaBoolBlend {blendOption=Linear}}
    Additive=Layered blend per bone(BasePose=Additive Identity Pose(), BlendPoses_0=Use cached pose 'LeftArm Pose'(), BlendWeights_0=@n46) {layerSetup [{branchFilters [{boneName=spine_01}]}]}
    Alpha=Get LeftArm_DampenAlpha() -> (LeftArm_DampenAlpha:float)
   ) {alphaBoolBlend {blendOption=Linear}}
   n48 = Layered blend per bone(BasePose=Additive Identity Pose(), BlendPoses_0=Use cached pose 'LeftArm Pose'(), BlendWeights_0=@n46) {layerSetup [{branchFilters [{boneName=spine_01}]}]}
   n49 = Get LeftArm_AmplifyAlpha() -> (LeftArm_AmplifyAlpha:float)
   n50 = Additive Identity Pose()
   n51 = Use cached pose 'RightArm Pose'()
   n52 = Get RightArm_BlendWeight()
   n53 = Apply Additive(
    Base=Apply Additive(
     Base=Apply Additive(Base=@n47, Additive=@n48, Alpha=@n49) {alphaBoolBlend {blendOption=Linear}}
     Additive=Layered blend per bone(BasePose=@n50, BlendPoses_0=@n51, BlendWeights_0=@n52) {layerSetup [{branchFilters [{boneName=spine_01}]}]}
     Alpha=Get RightArm_DampenAlpha() -> (RightArm_DampenAlpha:float)
    ) {alphaBoolBlend {blendOption=Linear}}
    Additive=Layered blend per bone(BasePose=Additive Identity Pose(), BlendPoses_0=Use cached pose 'RightArm Pose'(), BlendWeights_0=@n52) {layerSetup [{branchFilters [{boneName=spine_01}]}]}
    Alpha=Get RightArm_AmplifyAlpha() -> (RightArm_AmplifyAlpha:float)
   ) {alphaBoolBlend {blendOption=Linear}}
   n54 = Get HeadMotion_BlendWeight()
   n55 = Layered blend per bone(BasePose=Additive Identity Pose(), BlendPoses_0=Use cached pose 'HeadMotion Pose'(), BlendWeights_0=@n54) {layerSetup [{branchFilters [{boneName=spine_01}]}]}
   n56 = Get HeadMotion_DampenAlpha()
   n57 = Additive Identity Pose()
   n58 = Use cached pose 'HeadMotion Pose'()
   Save cached pose 'BaseBody Post-Additives Pose'(Pose=Apply Additive(
     Base=Apply Additive(Base=@n53, Additive=@n55, Alpha=@n56) {alphaBoolBlend {blendOption=Linear}}
     Additive=Layered blend per bone(BasePose=@n57, BlendPoses_0=@n58, BlendWeights_0=@n54) {layerSetup [{branchFilters [{boneName=spine_01}]}]}
     Alpha=Get HeadMotion_AmplifyAlpha()
    ) {alphaBoolBlend {blendOption=Linear}}) {cacheName=BaseBody Post-Additives Pose}
   AnimGraphNode_StateMachine_5 = HeadMotion() {is=title AnimGraphNode_StateMachine}
   Save cached pose 'HeadMotion Pose'(Pose=@AnimGraphNode_StateMachine_5) {cacheName=HeadMotion Pose}
   AnimGraphNode_StateMachine_13 = BaseBody StrumArm() {is=title AnimGraphNode_StateMachine, bAllowConduitEntryStates=true}
   AnimGraphNode_StateMachine_14 = BaseBody FretArm() {is=title AnimGraphNode_StateMachine, bAllowConduitEntryStates=true}
   n59 = Layered blend per bone!(
    BasePose=Use cached pose 'BaseBody Pose'()
    BlendPoses_0=@AnimGraphNode_StateMachine_13
    BlendPoses_1:Pose Link Structure=@AnimGraphNode_StateMachine_14
    BlendWeights_0=Get BaseBody_StrumArm_BlendWeight() -> (BaseBody_StrumArm_BlendWeight:float)
    BlendWeights_1:float=Get BaseBody_FretArm_BlendWeight() -> (BaseBody_FretArm_BlendWeight:float)
   ) -> (Pose) {
    blendPoses [{linkId=-1, sourceLinkId=-1}, {linkId=-1, sourceLinkId=-1}]
    blendWeights [1,1]
    layerSetup [
     {branchFilters [{blendDepth=-1, boneName=clavicle_r}, {boneName=hand_r}, {boneName=VB VB spine_05_hand_attach_r_ik_hand_gun}]}
     {branchFilters [{boneName=clavicle_l}, {boneName=VB VB spine_05_hand_attach_r_ik_hand_l}]}
    ]
   }
   Save cached pose 'BaseBody Pre-Additives Pose'(Pose=@n59) {cacheName=BaseBody Pre-Additives Pose}
   n60 = Get DEBUG_BaseBody_OverrideSequence() -> (DEBUG_BaseBody_OverrideSequence:Anim Sequence Base Object Reference)
   AnimGraphNode_StateMachine_8 = BaseBody() {is=title AnimGraphNode_StateMachine, bAllowConduitEntryStates=true}
   Save cached pose 'BaseBody Pose'(Pose=Blend Poses by bool(
     BlendPose_0=Sequence Player!(Sequence:Anim Sequence Base Object Reference=@n60) -> (Pose)
     BlendPose_1=@AnimGraphNode_StateMachine_8
     BlendTime_0=0
     BlendTime_1=0
     bActiveValue=Get DEBUG_BaseBody_UseOverrideSequence?() -> (DEBUG_BaseBody_UseOverrideSequence?:bool)
    ) {blendType=Linear}) {cacheName=BaseBody Pose}
   AnimGraphNode_StateMachine_9 = FaceExpression() {is=title AnimGraphNode_StateMachine, bAllowConduitEntryStates=true}
   Save cached pose 'FaceExpression Pose'(Pose=@AnimGraphNode_StateMachine_9) {cacheName=FaceExpression Pose}
   AnimGraphNode_StateMachine_2 = FootShift 1stFoot Lift() {is=title AnimGraphNode_StateMachine}
   Save cached pose 'FootShift 1stFoot Lift Pose'(Pose=@AnimGraphNode_StateMachine_2) {cacheName=FootShift 1stFoot Lift Pose}
   AnimGraphNode_StateMachine_10 = FaceLipSync() {is=title AnimGraphNode_StateMachine, bAllowConduitEntryStates=true}
   Save cached pose 'FaceLipSync Pose'(Pose=@AnimGraphNode_StateMachine_10) {cacheName=FaceLipSync Pose}
   AnimGraphNode_StateMachine_4 = SpineMotion() {is=title AnimGraphNode_StateMachine}
   Save cached pose 'SpineMotion Pose'(Pose=@AnimGraphNode_StateMachine_4) {cacheName=SpineMotion Pose}
   AnimGraphNode_StateMachine_3 = RightArm() {is=title AnimGraphNode_StateMachine}
   Save cached pose 'RightArm Pose'(Pose=@AnimGraphNode_StateMachine_3) {cacheName=RightArm Pose}
   AnimGraphNode_StateMachine_1 = LeftArm() {is=title AnimGraphNode_StateMachine}
   Save cached pose 'LeftArm Pose'(Pose=@AnimGraphNode_StateMachine_1) {cacheName=LeftArm Pose}
   AnimGraphNode_StateMachine_12 = LowerBody() {is=title AnimGraphNode_StateMachine}
   Save cached pose 'LowerBody Pose'(Pose=@AnimGraphNode_StateMachine_12) {cacheName=LowerBody Pose}
   Sequence Player 'JamV2Guitar_PlayMedium_HipHop_095_A'!(Sequence:Anim Sequence Base Object Reference=/SparksCharacterCommon/Animation/MainPlayer/BandPlay/Guitar/Loops/JamV2Guitar_PlayMedium_HipHop_095_A) -> (Pose)
   Sequence Player 'JamV2Guitar_PlayMedium_Pop_095_A'!(Sequence:Anim Sequence Base Object Reference=/SparksCharacterCommon/Animation/MainPlayer/BandPlay/Guitar/Loops/JamV2Guitar_PlayMedium_Pop_095_A) -> (Pose)
   Sequence Player 'JamV2Mic_PlayMedium_HipHop_165_A'!(Sequence:Anim Sequence Base Object Reference=/SparksCharacterCommon/Animation/MainPlayer/BandPlay/Mic/Loops/JamV2Mic_PlayMedium_HipHop_165_A) -> (Pose)
   Sequence Player 'JamV2Mic_PlayLight_HipHop_095_A'!(Sequence:Anim Sequence Base Object Reference=/SparksCharacterCommon/Animation/MainPlayer/BandPlay/Mic/Loops/JamV2Mic_PlayLight_HipHop_095_A) -> (Pose)
   Sequence Player 'TPose'() {sequence=/SparksCharacterCommon/Animation/MainPlayer/BandPlay/Drum/Hits/TPose}
   Sequence Player 'JamV2Guitar_PlayLight_HipHop_095_A'!(Sequence:Anim Sequence Base Object Reference=/SparksCharacterCommon/Animation/MainPlayer/BandPlay/Guitar/Loops/JamV2Guitar_PlayLight_HipHop_095_A) -> (Pose)
   Sequence Player 'JamV2Mic_PlayHeavy_Rock_095_A'!(Sequence:Anim Sequence Base Object Reference=/SparksCharacterCommon/Animation/MainPlayer/BandPlay/Mic/Loops/JamV2Mic_PlayHeavy_Rock_095_A) -> (Pose)
   Use cached pose 'BaseBody Post-Additives Pose'()
   Use cached pose 'BaseBody Post-Additives Pose'()
   Use cached pose 'BaseBody Post-Additives Pose'()
   Get InstrumentType() -> (InstrumentType:ESparksInstrumentType Enum=Guitar)
  ]}
 FootShift 2ndPelvis Translate {
  entry=FootShift 2ndPelvis Translate RESET
  path=AnimGraph.AnimGraphNode_StateMachine_0.FootShift 2ndPelvis Translate
  ids=id
  nodes [
   AnimStateNode_0 = FootShift 2ndPelvis Translate RESET() {
    is=title AnimStateNode
    stateEntered {guid=#1}
    stateFullyBlended {guid=#2}
    stateLeft {guid=#3}
   }
   AnimStateNode_1 = FootShift 2ndPelvis Translate ON() {
    is=title AnimStateNode
    stateEntered {guid=#4}
    stateFullyBlended {guid=#5}
    stateLeft {guid=#6}
   }
   AnimStateNode_3 = FootShift 2ndPelvis Translate OFF() {
    is=title AnimStateNode
    stateEntered {guid=#1}
    stateFullyBlended {guid=#2}
    stateLeft {guid=#3}
   }
   AnimStateTransitionNode_0 = AnimStateTransitionNode() {crossfadeDuration=0.53, customBlendCurve=/SparksCharacterCommon/Animation/MainPlayer/Additives/FootShift_Pelvis_ProceduralTranslate}
   AnimStateTransitionNode_1 = AnimStateTransitionNode() {bAllowInertializationForSelfTransitions=true, crossfadeDuration=0}
   AnimStateTransitionNode_10 = AnimStateTransitionNode() {bAllowInertializationForSelfTransitions=true, crossfadeDuration=0}
   AnimStateTransitionNode_16 = AnimStateTransitionNode() {
    bAllowInertializationForSelfTransitions=true
    crossfadeDuration=0.795
    customBlendCurve=&FootShift_Pelvis_ProceduralTranslate
    priorityOrder=2
   }
   AnimStateTransitionNode_17 = AnimStateTransitionNode() {
    bAllowInertializationForSelfTransitions=true
    crossfadeDuration=1.06
    customBlendCurve=&FootShift_Pelvis_ProceduralTranslate
    priorityOrder=3
   }
   AnimStateTransitionNode_18 = AnimStateTransitionNode() {
    bAllowInertializationForSelfTransitions=true
    crossfadeDuration=0.795
    customBlendCurve=&FootShift_Pelvis_ProceduralTranslate
    priorityOrder=2
   }
   AnimStateTransitionNode_19 = AnimStateTransitionNode() {
    bAllowInertializationForSelfTransitions=true
    crossfadeDuration=1.06
    customBlendCurve=&FootShift_Pelvis_ProceduralTranslate
    priorityOrder=3
   }
   AnimStateTransitionNode_6 = AnimStateTransitionNode() {bAllowInertializationForSelfTransitions=true, crossfadeDuration=0}
   AnimStateTransitionNode_8 = AnimStateTransitionNode() {
    crossfadeDuration=0.53
    customBlendCurve=&FootShift_Pelvis_ProceduralTranslate
    transitionEnd {guid=#7}
    transitionInterrupt {guid=#8}
    transitionStart {guid=#9}
   }
  ]
  links [
   {
    dst=FootShift 2ndPelvis Translate ON
    rule=AnimGraph.AnimGraphNode_StateMachine_0.FootShift 2ndPelvis Translate.AnimStateTransitionNode_0.Transition
    src=FootShift 2ndPelvis Translate OFF
    via=AnimStateTransitionNode_0
   }
   {
    dst=FootShift 2ndPelvis Translate ON
    rule=AnimGraph.AnimGraphNode_StateMachine_0.FootShift 2ndPelvis Translate.AnimStateTransitionNode_18.Transition
    src=FootShift 2ndPelvis Translate OFF
    via=AnimStateTransitionNode_18
   }
   {
    dst=FootShift 2ndPelvis Translate ON
    rule=AnimGraph.AnimGraphNode_StateMachine_0.FootShift 2ndPelvis Translate.AnimStateTransitionNode_19.Transition
    src=FootShift 2ndPelvis Translate OFF
    via=AnimStateTransitionNode_19
   }
   {
    dst=FootShift 2ndPelvis Translate RESET
    rule=AnimGraph.AnimGraphNode_StateMachine_0.FootShift 2ndPelvis Translate.AnimStateTransitionNode_10.Transition
    src=FootShift 2ndPelvis Translate OFF
    via=AnimStateTransitionNode_10
   }
   {
    dst=FootShift 2ndPelvis Translate OFF
    rule=AnimGraph.AnimGraphNode_StateMachine_0.FootShift 2ndPelvis Translate.AnimStateTransitionNode_16.Transition
    src=FootShift 2ndPelvis Translate ON
    via=AnimStateTransitionNode_16
   }
   {
    dst=FootShift 2ndPelvis Translate OFF
    rule=AnimGraph.AnimGraphNode_StateMachine_0.FootShift 2ndPelvis Translate.AnimStateTransitionNode_17.Transition
    src=FootShift 2ndPelvis Translate ON
    via=AnimStateTransitionNode_17
   }
   {
    dst=FootShift 2ndPelvis Translate OFF
    rule=AnimGraph.AnimGraphNode_StateMachine_0.FootShift 2ndPelvis Translate.AnimStateTransitionNode_8.AnimationTransitionGraph_0
    src=FootShift 2ndPelvis Translate ON
    via=AnimStateTransitionNode_8
   }
   {
    dst=FootShift 2ndPelvis Translate RESET
    rule=AnimGraph.AnimGraphNode_StateMachine_0.FootShift 2ndPelvis Translate.AnimStateTransitionNode_6.Transition
    src=FootShift 2ndPelvis Translate ON
    via=AnimStateTransitionNode_6
   }
   {
    dst=FootShift 2ndPelvis Translate OFF
    rule=AnimGraph.AnimGraphNode_StateMachine_0.FootShift 2ndPelvis Translate.AnimStateTransitionNode_1.Transition
    src=FootShift 2ndPelvis Translate RESET
    via=AnimStateTransitionNode_1
   }
  ]
 }
 FootShift 2ndPelvis Translate RESET {
  path=AnimGraph.AnimGraphNode_StateMachine_0.FootShift 2ndPelvis Translate.AnimStateNode_0.FootShift 2ndPelvis Translate RESET
  nodes [Output Animation Pose(Result=Use cached pose 'BaseBody Post-Additives Pose'())]
 }
 FootShift 2ndPelvis Translate ON {
  path=AnimGraph.AnimGraphNode_StateMachine_0.FootShift 2ndPelvis Translate.AnimStateNode_1.FootShift 2ndPelvis Translate ON
  nodes [
   n1 = "Transform (Modify) Bone - Bone: pelvis"!(
    ComponentPose=Local To Component(LocalPose=Use cached pose 'BaseBody Pre-Additives Pose'())
    Alpha=1
    Translation_X=Get FootShift_2ndPelvis_ProcTranslateLR() -> (FootShift_2ndPelvis_ProcTranslateLR:double)
    Translation_Y=Get FootShift_2ndPelvis_ProcTranslateFB() -> (FootShift_2ndPelvis_ProcTranslateFB:double)
    Translation_Z=Get FootShift_2ndPelvis_ProcTranslateVertical() -> (FootShift_2ndPelvis_ProcTranslateVertical:double)
    Rotation_Roll:float=0
    Rotation_Pitch:float=0
    Rotation_Yaw:float=Get FootShift_2ndPelvis_ProcRotateTwist() -> (FootShift_2ndPelvis_ProcRotateTwist:float)
   ) -> (Pose) {
    rotationMode=BMM_Additive
    translationMode=BMM_Additive
    alphaBoolBlend {blendOption=Linear}
    boneToModify {boneName=pelvis}
    translation=<30 30>
   }
   Output Animation Pose(Result=Component To Local(ComponentPose=@n1))
  ]
 }
 FootShift 2ndPelvis Translate OFF {
  path=AnimGraph.AnimGraphNode_StateMachine_0.FootShift 2ndPelvis Translate.AnimStateNode_3.FootShift 2ndPelvis Translate OFF
  nodes [
   n1 = "Transform (Modify) Bone - Bone: pelvis"(
    ComponentPose=Local To Component(LocalPose=Use cached pose 'BaseBody Pre-Additives Pose'())
    Alpha=1
    Translation_X=0
    Translation_Y=0
    Translation_Z=0
    Rotation_Roll=0
    Rotation_Pitch=0
    Rotation_Yaw=0
   ) {
    rotationMode=BMM_Additive
    translationMode=BMM_Additive
    alphaBoolBlend {blendOption=Linear}
    boneToModify {boneName=pelvis}
    translation=<30 30>
   }
   Output Animation Pose(Result=Component To Local(ComponentPose=@n1))
  ]
 }
 Transition {
  path=AnimGraph.AnimGraphNode_StateMachine_0.FootShift 2ndPelvis Translate.AnimStateTransitionNode_0.Transition
  nodes [
   n1 = Get FootShift_2ndFoot_ProcTransitionSpeed() -> (FootShift_2ndFoot_ProcTransitionSpeed:Enum_FootShift_ProcTransitionSpeed Enum=NewEnumerator0)
   Result(bCanEnterTransition=AND Boolean(
     A=Get FootShift_2ndFoot_ProcLiftCurrentlyActive?() -> (FootShift_2ndFoot_ProcLiftCurrentlyActive?:bool)
     B="Equal (Enum)"(A:Enum_FootShift_ProcTransitionSpeed Enum=@n1, B:Enum_FootShift_ProcTransitionSpeed Enum=NewEnumerator2)
    ))
  ]
 }
 Transition {
  path=AnimGraph.AnimGraphNode_StateMachine_0.FootShift 2ndPelvis Translate.AnimStateTransitionNode_1.Transition
  nodes [Result(bCanEnterTransition=true)]
 }
 Transition {
  path=AnimGraph.AnimGraphNode_StateMachine_0.FootShift 2ndPelvis Translate.AnimStateTransitionNode_10.Transition
  nodes [Result(bCanEnterTransition=Get DEBUG_FootShift_ResetSystem?() -> (DEBUG_FootShift_ResetSystem?:bool))]
 }
 Transition {
  path=AnimGraph.AnimGraphNode_StateMachine_0.FootShift 2ndPelvis Translate.AnimStateTransitionNode_16.Transition
  nodes [
   n1 = Get FootShift_2ndFoot_ProcLiftCurrentlyActive?()
   n2 = Get FootShift_2ndFoot_ProcTransitionSpeed() -> (FootShift_2ndFoot_ProcTransitionSpeed=NewEnumerator0)
   Result(bCanEnterTransition=AND Boolean(A=NOT Boolean(A=@n1), B="Equal (Enum)"(A=@n2, B=NewEnumerator1)))
  ]
 }
 Transition {
  path=AnimGraph.AnimGraphNode_StateMachine_0.FootShift 2ndPelvis Translate.AnimStateTransitionNode_17.Transition
  nodes [
   n1 = Get FootShift_2ndFoot_ProcLiftCurrentlyActive?()
   n2 = Get FootShift_2ndFoot_ProcTransitionSpeed() -> (FootShift_2ndFoot_ProcTransitionSpeed=NewEnumerator0)
   Result(bCanEnterTransition=AND Boolean(A=NOT Boolean(A=@n1), B="Equal (Enum)"(A=@n2, B=NewEnumerator0)))
  ]
 }
 Transition {
  path=AnimGraph.AnimGraphNode_StateMachine_0.FootShift 2ndPelvis Translate.AnimStateTransitionNode_18.Transition
  nodes [
   n1 = Get FootShift_2ndFoot_ProcTransitionSpeed() -> (FootShift_2ndFoot_ProcTransitionSpeed=NewEnumerator0)
   Result(bCanEnterTransition=AND Boolean(A=Get FootShift_2ndFoot_ProcLiftCurrentlyActive?(), B="Equal (Enum)"(A=@n1, B=NewEnumerator1)))
  ]
 }
 Transition {
  path=AnimGraph.AnimGraphNode_StateMachine_0.FootShift 2ndPelvis Translate.AnimStateTransitionNode_19.Transition
  nodes [
   n1 = Get FootShift_2ndFoot_ProcTransitionSpeed() -> (FootShift_2ndFoot_ProcTransitionSpeed=NewEnumerator0)
   Result(bCanEnterTransition=AND Boolean(A=Get FootShift_2ndFoot_ProcLiftCurrentlyActive?(), B="Equal (Enum)"(A=@n1, B=NewEnumerator0)))
  ]
 }
 Transition {
  path=AnimGraph.AnimGraphNode_StateMachine_0.FootShift 2ndPelvis Translate.AnimStateTransitionNode_6.Transition
  nodes [Result(bCanEnterTransition=Get DEBUG_FootShift_ResetSystem?())]
 }
 AnimationTransitionGraph_0 {
  path=AnimGraph.AnimGraphNode_StateMachine_0.FootShift 2ndPelvis Translate.AnimStateTransitionNode_8.AnimationTransitionGraph_0
  nodes [
   n1 = Get FootShift_2ndFoot_ProcLiftCurrentlyActive?()
   n2 = Get FootShift_2ndFoot_ProcTransitionSpeed() -> (FootShift_2ndFoot_ProcTransitionSpeed=NewEnumerator0)
   Result(bCanEnterTransition=AND Boolean(A=NOT Boolean(A=@n1), B="Equal (Enum)"(A=@n2, B=NewEnumerator2)))
  ]
 }
 LeftArm {
  entry=LeftArm RESET
  path=AnimGraph.AnimGraphNode_StateMachine_1.LeftArm
  ids=id
  nodes [
   AnimStateNode_0 = LeftArm RESET() {is=title AnimStateNode}
   AnimStateNode_1 = LeftArm OFF() {
    is=title AnimStateNode
    stateEntered {guid=#1}
    stateFullyBlended {guid=#2}
    stateLeft {guid=#3}
   }
   AnimStateNode_5 = LeftArm ON() {
    is=title AnimStateNode
    stateEntered {guid=#4}
    stateFullyBlended {guid=#5}
    stateLeft {guid=#6}
   }
   AnimStateTransitionNode_0 = AnimStateTransitionNode() {crossfadeDuration=0.5}
   AnimStateTransitionNode_1 = AnimStateTransitionNode() {crossfadeDuration=0}
   AnimStateTransitionNode_7 = AnimStateTransitionNode() {crossfadeDuration=0}
   AnimStateTransitionNode_8 = AnimStateTransitionNode() {crossfadeDuration=0.5}
   AnimStateTransitionNode_9 = AnimStateTransitionNode() {crossfadeDuration=0.5}
  ]
  links [
   {
    dst=LeftArm ON
    rule=AnimGraph.AnimGraphNode_StateMachine_1.LeftArm.AnimStateTransitionNode_9.Transition
    src=LeftArm OFF
    via=AnimStateTransitionNode_9
   }
   {
    dst=LeftArm RESET
    rule=AnimGraph.AnimGraphNode_StateMachine_1.LeftArm.AnimStateTransitionNode_7.Transition
    src=LeftArm OFF
    via=AnimStateTransitionNode_7
   }
   {
    dst=LeftArm OFF
    rule=AnimGraph.AnimGraphNode_StateMachine_1.LeftArm.AnimStateTransitionNode_8.Transition
    src=LeftArm ON
    via=AnimStateTransitionNode_8
   }
   {
    dst=LeftArm RESET
    rule=AnimGraph.AnimGraphNode_StateMachine_1.LeftArm.AnimStateTransitionNode_1.Transition
    src=LeftArm ON
    via=AnimStateTransitionNode_1
   }
   {
    dst=LeftArm OFF
    rule=AnimGraph.AnimGraphNode_StateMachine_1.LeftArm.AnimStateTransitionNode_0.Transition
    src=LeftArm RESET
    via=AnimStateTransitionNode_0
   }
  ]
 }
 LeftArm RESET {
  path=AnimGraph.AnimGraphNode_StateMachine_1.LeftArm.AnimStateNode_0.LeftArm RESET
  nodes [Output Animation Pose(Result=Additive Identity Pose())]
 }
 LeftArm OFF {
  path=AnimGraph.AnimGraphNode_StateMachine_1.LeftArm.AnimStateNode_1.LeftArm OFF
  nodes [Output Animation Pose(Result=Additive Identity Pose())]
 }
 LeftArm ON {
  path=AnimGraph.AnimGraphNode_StateMachine_1.LeftArm.AnimStateNode_5.LeftArm ON
  nodes [Output Animation Pose(Result=Play Beat Synced Animation!(
     ExplicitTime:float=0
     bShouldLoop:bool=True
     InSequence:Anim Sequence Base Object Reference=Get LeftArm_CurrentSequence() -> (LeftArm_CurrentSequence:Anim Sequence Base Object Reference)
     MusicClock:Music Clock Object Reference
     TempoMultiplier:float=1
     bEnableTempoVariants:bool=False
     TempoVariantHysteresis:float=0.05
     bShowDebug:bool=False
    ) -> (Pose:Pose Link Structure) {is=title AnimGraphNode_PlayBeatSyncedAnim})]
 }
 Transition {
  path=AnimGraph.AnimGraphNode_StateMachine_1.LeftArm.AnimStateTransitionNode_0.Transition
  nodes [Result(bCanEnterTransition=Get DEBUG_LeftArm_ResetSystem?() -> (DEBUG_LeftArm_ResetSystem?:bool))]
 }
 Transition {
  path=AnimGraph.AnimGraphNode_StateMachine_1.LeftArm.AnimStateTransitionNode_1.Transition
  nodes [Result(bCanEnterTransition=NOT Boolean(A=Get DEBUG_LeftArm_ResetSystem?()))]
 }
 Transition {
  path=AnimGraph.AnimGraphNode_StateMachine_1.LeftArm.AnimStateTransitionNode_7.Transition
  nodes [Result(bCanEnterTransition=NOT Boolean(A=Get DEBUG_LeftArm_ResetSystem?()))]
 }
 Transition {
  path=AnimGraph.AnimGraphNode_StateMachine_1.LeftArm.AnimStateTransitionNode_8.Transition
  nodes [Result(bCanEnterTransition=NOT Boolean(A=Get LeftArm_CurrentlyActive?()))]
 }
 Transition {
  path=AnimGraph.AnimGraphNode_StateMachine_1.LeftArm.AnimStateTransitionNode_9.Transition
  nodes [Result(bCanEnterTransition=Get LeftArm_CurrentlyActive?())]
 }
 FaceLipSync {
  entry=LipSync
  path=AnimGraph.AnimGraphNode_StateMachine_10.FaceLipSync
  ids=id
  nodes [AnimStateNode_0 = LipSync() {is=title AnimStateNode}]
 }
 LipSync {
  path=AnimGraph.AnimGraphNode_StateMachine_10.FaceLipSync.AnimStateNode_0.LipSync
  nodes [Output Animation Pose(Result=Play Beat Synced Animation(ExplicitTime=0, bShouldLoop=false, TempoMultiplier=1, bEnableTempoVariants=False, TempoVariantHysteresis=0.05, bShowDebug=False))]
 }
 FootShift 2ndFoot Lift {
  entry=FootShift 2ndFoot Lift RESET
  path=AnimGraph.AnimGraphNode_StateMachine_11.FootShift 2ndFoot Lift
  ids=id
  nodes [
   AnimStateNode_0 = FootShift 2ndFoot Lift RESET() {
    is=title AnimStateNode
    stateEntered {guid=#1}
    stateFullyBlended {guid=#2}
    stateLeft {guid=#3}
   }
   AnimStateNode_1 = FootShift 2ndFoot Lift OFF() {
    is=title AnimStateNode
    stateEntered {guid=#1}
    stateFullyBlended {guid=#2}
    stateLeft {guid=#3}
   }
   AnimStateNode_2 = FootShift 2ndFoot Lift ON() {
    is=title AnimStateNode
    stateEntered {guid=#4}
    stateFullyBlended {guid=#5}
    stateLeft {guid=#6}
   }
   AnimStateTransitionNode_0 = AnimStateTransitionNode() {crossfadeDuration=0}
   AnimStateTransitionNode_1 = AnimStateTransitionNode() {
    crossfadeDuration=0
    transitionEnd {guid=#7}
    transitionInterrupt {guid=#8}
    transitionStart {guid=#9}
   }
   AnimStateTransitionNode_10 = AnimStateTransitionNode() {bAllowInertializationForSelfTransitions=true, crossfadeDuration=0}
   AnimStateTransitionNode_6 = AnimStateTransitionNode() {bAllowInertializationForSelfTransitions=true, crossfadeDuration=0}
   AnimStateTransitionNode_9 = AnimStateTransitionNode() {bAllowInertializationForSelfTransitions=true, crossfadeDuration=0}
  ]
  links [
   {
    dst=FootShift 2ndFoot Lift ON
    rule=AnimGraph.AnimGraphNode_StateMachine_11.FootShift 2ndFoot Lift.AnimStateTransitionNode_0.Transition
    src=FootShift 2ndFoot Lift OFF
    via=AnimStateTransitionNode_0
   }
   {
    dst=FootShift 2ndFoot Lift RESET
    rule=AnimGraph.AnimGraphNode_StateMachine_11.FootShift 2ndFoot Lift.AnimStateTransitionNode_10.Transition
    src=FootShift 2ndFoot Lift OFF
    via=AnimStateTransitionNode_10
   }
   {
    dst=FootShift 2ndFoot Lift OFF
    rule=AnimGraph.AnimGraphNode_StateMachine_11.FootShift 2ndFoot Lift.AnimStateTransitionNode_1.AnimationTransitionGraph_1
    src=FootShift 2ndFoot Lift ON
    via=AnimStateTransitionNode_1
   }
   {
    dst=FootShift 2ndFoot Lift RESET
    rule=AnimGraph.AnimGraphNode_StateMachine_11.FootShift 2ndFoot Lift.AnimStateTransitionNode_6.Transition
    src=FootShift 2ndFoot Lift ON
    via=AnimStateTransitionNode_6
   }
   {
    dst=FootShift 2ndFoot Lift OFF
    rule=AnimGraph.AnimGraphNode_StateMachine_11.FootShift 2ndFoot Lift.AnimStateTransitionNode_9.Transition
    src=FootShift 2ndFoot Lift RESET
    via=AnimStateTransitionNode_9
   }
  ]
 }
 FootShift 2ndFoot Lift RESET {
  path=AnimGraph.AnimGraphNode_StateMachine_11.FootShift 2ndFoot Lift.AnimStateNode_0.FootShift 2ndFoot Lift RESET
  nodes [Output Animation Pose()]
 }
 FootShift 2ndFoot Lift OFF {
  path=AnimGraph.AnimGraphNode_StateMachine_11.FootShift 2ndFoot Lift.AnimStateNode_1.FootShift 2ndFoot Lift OFF
  nodes [
   n1 = Get FootShift_2ndFoot_ProcCurrentSequence() -> (FootShift_2ndFoot_ProcCurrentSequence:Anim Sequence Base Object Reference)
   n2 = Get FootShift_2ndFoot_ProcAnimPlayRate() -> (FootShift_2ndFoot_ProcAnimPlayRate:float)
   n3 = Get FootShift_Reset_CurrentSequence() -> (FootShift_Reset_CurrentSequence:Anim Sequence Base Object Reference)
   Output Animation Pose(Result=Blend Poses by bool(
     BlendPose_0=Sequence Player!(Sequence=@n1, PlayRate:float=@n2) -> (Pose) {bLoopAnimation=false}
     BlendPose_1=Sequence Player(Sequence=@n3, PlayRate=1) {bLoopAnimation=false}
     BlendTime_0=0
     BlendTime_1=0
     bActiveValue=Get DEBUG_FootShift_UseProceduralSystem?() -> (DEBUG_FootShift_UseProceduralSystem?:bool)
    ) {blendType=Linear})
  ]
 }
 FootShift 2ndFoot Lift ON {
  path=AnimGraph.AnimGraphNode_StateMachine_11.FootShift 2ndFoot Lift.AnimStateNode_2.FootShift 2ndFoot Lift ON
  nodes [
   n1 = Get FootShift_2ndFoot_ProcCurrentSequence()
   n2 = Get FootShift_2ndFoot_ProcAnimPlayRate()
   n3 = Get FootShift_CurrentSequence() -> (FootShift_CurrentSequence:Anim Sequence Base Object Reference)
   Output Animation Pose(Result=Blend Poses by bool(
     BlendPose_0=Sequence Player(Sequence=@n1, PlayRate=@n2) {bLoopAnimation=false}
     BlendPose_1=Sequence Player(Sequence=@n3, PlayRate=1) {bLoopAnimation=false}
     BlendTime_0=0
     BlendTime_1=0
     bActiveValue=Get DEBUG_FootShift_UseProceduralSystem?()
    ) {blendType=Linear})
  ]
 }
 Transition {
  path=AnimGraph.AnimGraphNode_StateMachine_11.FootShift 2ndFoot Lift.AnimStateTransitionNode_0.Transition
  nodes [Result(bCanEnterTransition=Get FootShift_2ndFoot_ProcLiftCurrentlyActive?())]
 }
 AnimationTransitionGraph_1 {
  path=AnimGraph.AnimGraphNode_StateMachine_11.FootShift 2ndFoot Lift.AnimStateTransitionNode_1.AnimationTransitionGraph_1
  nodes [Result(bCanEnterTransition=NOT Boolean(A=Get FootShift_2ndFoot_ProcLiftCurrentlyActive?()))]
 }
 Transition {
  path=AnimGraph.AnimGraphNode_StateMachine_11.FootShift 2ndFoot Lift.AnimStateTransitionNode_10.Transition
  nodes [Result(bCanEnterTransition=Get DEBUG_FootShift_ResetSystem?())]
 }
 Transition {
  path=AnimGraph.AnimGraphNode_StateMachine_11.FootShift 2ndFoot Lift.AnimStateTransitionNode_6.Transition
  nodes [Result(bCanEnterTransition=Get DEBUG_FootShift_ResetSystem?())]
 }
 Transition {
  path=AnimGraph.AnimGraphNode_StateMachine_11.FootShift 2ndFoot Lift.AnimStateTransitionNode_9.Transition
  nodes [Result(bCanEnterTransition=true)]
 }
 LowerBody {
  entry=LowerBody ON
  path=AnimGraph.AnimGraphNode_StateMachine_12.LowerBody
  ids=id
  nodes [AnimStateNode_2 = LowerBody ON() {
    is=title AnimStateNode
    stateEntered {guid=#4}
    stateFullyBlended {guid=#5}
    stateLeft {guid=#6}
   }]
 }
 LowerBody ON {
  path=AnimGraph.AnimGraphNode_StateMachine_12.LowerBody.AnimStateNode_2.LowerBody ON
  nodes [Output Animation Pose(Result=Sequence Player(Sequence=Get LowerBody_NextSequence() -> (LowerBody_NextSequence:Anim Sequence Base Object Reference), PlayRate=1))]
 }
 BaseBody StrumArm {
  entry=Conduit
  path=AnimGraph.AnimGraphNode_StateMachine_13.BaseBody StrumArm
  ids=id
  nodes [
   AnimStateConduitNode_0 = Conduit()
   AnimStateNode_0 = Bass() {
    is=title AnimStateNode
    stateEntered {guid=#10}
    stateFullyBlended {guid=#11}
    stateLeft {guid=#12}
   }
   AnimStateNode_1 = Keytar() {is=title AnimStateNode}
   AnimStateNode_4 = Guitar() {
    is=title AnimStateNode
    stateEntered {guid=#13}
    stateFullyBlended {guid=#14}
    stateLeft {guid=#15}
   }
   AnimStateTransitionNode_10 = AnimStateTransitionNode() {
    transitionEnd {guid=#16}
    transitionInterrupt {guid=#17}
    transitionStart {guid=#18}
   }
   AnimStateTransitionNode_12 = AnimStateTransitionNode()
   AnimStateTransitionNode_13 = AnimStateTransitionNode()
   AnimStateTransitionNode_4 = AnimStateTransitionNode()
   AnimStateTransitionNode_8 = AnimStateTransitionNode()
   AnimStateTransitionNode_9 = AnimStateTransitionNode()
  ]
  links [
   {
    dst=Conduit
    rule=AnimGraph.AnimGraphNode_StateMachine_13.BaseBody StrumArm.AnimStateTransitionNode_13.Transition
    src=Bass
    via=AnimStateTransitionNode_13
   }
   {
    dst=Bass
    rule=AnimGraph.AnimGraphNode_StateMachine_13.BaseBody StrumArm.AnimStateTransitionNode_10.Transition
    src=Conduit
    via=AnimStateTransitionNode_10
   }
   {
    dst=Guitar
    rule=AnimGraph.AnimGraphNode_StateMachine_13.BaseBody StrumArm.AnimStateTransitionNode_4.Transition
    src=Conduit
    via=AnimStateTransitionNode_4
   }
   {
    dst=Keytar
    rule=AnimGraph.AnimGraphNode_StateMachine_13.BaseBody StrumArm.AnimStateTransitionNode_8.Transition
    src=Conduit
    via=AnimStateTransitionNode_8
   }
   {
    dst=Conduit
    rule=AnimGraph.AnimGraphNode_StateMachine_13.BaseBody StrumArm.AnimStateTransitionNode_12.Transition
    src=Guitar
    via=AnimStateTransitionNode_12
   }
   {
    dst=Conduit
    rule=AnimGraph.AnimGraphNode_StateMachine_13.BaseBody StrumArm.AnimStateTransitionNode_9.Transition
    src=Keytar
    via=AnimStateTransitionNode_9
   }
  ]
 }
 Conduit {path=AnimGraph.AnimGraphNode_StateMachine_13.BaseBody StrumArm.AnimStateConduitNode_0.Conduit, nodes [Result(bCanEnterTransition=true)]}
 Bass {
  path=AnimGraph.AnimGraphNode_StateMachine_13.BaseBody StrumArm.AnimStateNode_0.Bass
  nodes [
   n1 = Sequence Player!(Sequence, PlayRate=0, bLoopAnimation:bool=False) -> (Pose) {bLoopAnimation=false, playRate=0}
   n2 = Use cached pose 'BaseBody Pose'()
   Output Animation Pose(Result=Apply Additive(
     Base=Apply Additive(Base=@n2, Additive=@n1) {alphaBoolBlend {blendOption=Linear}}
     Additive=Sequence Player(PlayRate=0, bLoopAnimation=false) {playRate=0}
    ) {alphaBoolBlend {blendOption=Linear}})
  ]
 }
 Keytar {
  path=AnimGraph.AnimGraphNode_StateMachine_13.BaseBody StrumArm.AnimStateNode_1.Keytar
  nodes [
   Sequence Player 'Keytar_FretHand_BasePose'() {sequence=/SparksCharacterCommon/Animation/MainPlayer/BandPlay/Keytar/FretHand/Keytar_FretHand_BasePose}
   Output Animation Pose(Result=Apply Additive(
     Base=Sequence Player 'Keytar_FretHand_BasePose'() {sequence=&Keytar_FretHand_BasePose}
     Additive=Blendspace Player 'Keytar_FretHand_BS'(X=0, Y=0) {blendSpace=/SparksCharacterCommon/Animation/MainPlayer/BandPlay/Keytar/FretHand/Keytar_FretHand_BS, startPosition=1}
    ) {alphaBoolBlend {blendOption=Linear}})
  ]
 }
 Guitar {
  path=AnimGraph.AnimGraphNode_StateMachine_13.BaseBody StrumArm.AnimStateNode_4.Guitar
  nodes [
   "Sequence Player (additive) 'Guitar_Strum_Test_A'"!(
    Sequence:Anim Sequence Base Object Reference=/SparksCharacterCommon/Animation/MainPlayer/BandPlay/Guitar/Strums/Guitar_Strum_Test_A
    PlayRate:float=0
    StartPosition:float=0
   ) -> (Pose) {playRate=0}
   Output Animation Pose(Result=Sequence Player 'Guitar_FretHand_BasePose'!(
     Sequence:Anim Sequence Base Object Reference=/SparksCharacterCommon/Animation/MainPlayer/BandPlay/Guitar/FretHand/Guitar_FretHand_BasePose
     PlayRate:float=1
     StartPosition:float=0
    ) -> (Pose))
   Use cached pose 'BaseBody Pose'()
  ]
 }
 Transition {
  path=AnimGraph.AnimGraphNode_StateMachine_13.BaseBody StrumArm.AnimStateTransitionNode_10.Transition
  nodes [
   Result(bCanEnterTransition="Equal (Enum)"(A:ESparksInstrumentType Enum=Get InstrumentType() -> (InstrumentType=Guitar), B:ESparksInstrumentType Enum=Bass))
   Get Instrument() -> (Instrument=Guitar)
  ]
 }
 Transition {
  path=AnimGraph.AnimGraphNode_StateMachine_13.BaseBody StrumArm.AnimStateTransitionNode_12.Transition
  nodes [
   Result(bCanEnterTransition="Not Equal (Enum)"!(A:ESparksInstrumentType Enum=Get InstrumentType() -> (InstrumentType=Guitar), B:ESparksInstrumentType Enum=Guitar) -> (ReturnValue:bool))
   Get Instrument() -> (Instrument=Guitar)
  ]
 }
 Transition {
  path=AnimGraph.AnimGraphNode_StateMachine_13.BaseBody StrumArm.AnimStateTransitionNode_13.Transition
  nodes [
   Result(bCanEnterTransition="Not Equal (Enum)"(A=Get InstrumentType() -> (InstrumentType=Guitar), B=Bass))
   Get Instrument() -> (Instrument=Guitar)
  ]
 }
 Transition {
  path=AnimGraph.AnimGraphNode_StateMachine_13.BaseBody StrumArm.AnimStateTransitionNode_4.Transition
  nodes [
   Result(bCanEnterTransition="Equal (Enum)"(A=Get InstrumentType() -> (InstrumentType=Guitar), B=Guitar))
   Get Instrument() -> (Instrument=Guitar)
  ]
 }
 Transition {
  path=AnimGraph.AnimGraphNode_StateMachine_13.BaseBody StrumArm.AnimStateTransitionNode_8.Transition
  nodes [
   Result(bCanEnterTransition="Equal (Enum)"(A=Get InstrumentType() -> (InstrumentType=Guitar), B=Keyboard))
   Get Instrument() -> (Instrument=Guitar)
  ]
 }
 Transition {
  path=AnimGraph.AnimGraphNode_StateMachine_13.BaseBody StrumArm.AnimStateTransitionNode_9.Transition
  nodes [
   Result(bCanEnterTransition="Not Equal (Enum)"(A=Get InstrumentType() -> (InstrumentType=Guitar), B=Keyboard))
   Get Instrument() -> (Instrument=Guitar)
  ]
 }
 BaseBody FretArm {
  entry=Conduit
  path=AnimGraph.AnimGraphNode_StateMachine_14.BaseBody FretArm
  ids=id
  nodes [
   AnimStateConduitNode_0 = Conduit()
   AnimStateNode_0 = Keytar()
   AnimStateNode_1 = Guitar()
   AnimStateNode_4 = Bass()
   AnimStateTransitionNode_10 = AnimStateTransitionNode()
   AnimStateTransitionNode_11 = AnimStateTransitionNode()
   AnimStateTransitionNode_2 = AnimStateTransitionNode()
   AnimStateTransitionNode_3 = AnimStateTransitionNode()
   AnimStateTransitionNode_8 = AnimStateTransitionNode()
   AnimStateTransitionNode_9 = AnimStateTransitionNode()
  ]
  links [
   {
    dst=Conduit
    rule=AnimGraph.AnimGraphNode_StateMachine_14.BaseBody FretArm.AnimStateTransitionNode_11.Transition
    src=Bass
    via=AnimStateTransitionNode_11
   }
   {
    dst=Bass
    rule=AnimGraph.AnimGraphNode_StateMachine_14.BaseBody FretArm.AnimStateTransitionNode_3.Transition
    src=Conduit
    via=AnimStateTransitionNode_3
   }
   {
    dst=Guitar
    rule=AnimGraph.AnimGraphNode_StateMachine_14.BaseBody FretArm.AnimStateTransitionNode_2.Transition
    src=Conduit
    via=AnimStateTransitionNode_2
   }
   {
    dst=Keytar
    rule=AnimGraph.AnimGraphNode_StateMachine_14.BaseBody FretArm.AnimStateTransitionNode_9.Transition
    src=Conduit
    via=AnimStateTransitionNode_9
   }
   {
    dst=Conduit
    rule=AnimGraph.AnimGraphNode_StateMachine_14.BaseBody FretArm.AnimStateTransitionNode_10.Transition
    src=Guitar
    via=AnimStateTransitionNode_10
   }
   {
    dst=Conduit
    rule=AnimGraph.AnimGraphNode_StateMachine_14.BaseBody FretArm.AnimStateTransitionNode_8.Transition
    src=Keytar
    via=AnimStateTransitionNode_8
   }
  ]
 }
 Conduit {path=AnimGraph.AnimGraphNode_StateMachine_14.BaseBody FretArm.AnimStateConduitNode_0.Conduit, nodes [Result(bCanEnterTransition=true)]}
 Keytar {
  path=AnimGraph.AnimGraphNode_StateMachine_14.BaseBody FretArm.AnimStateNode_0.Keytar
  nodes [Output Animation Pose(Result=Sequence Player 'Keytar_FretHand_BasePose'() {sequence=&Keytar_FretHand_BasePose})]
 }
 Guitar {
  path=AnimGraph.AnimGraphNode_StateMachine_14.BaseBody FretArm.AnimStateNode_1.Guitar
  nodes [
   AnimGraphNode_BlendSpaceGraph_0 = Blendspace 'Guitar_Fret_Chord_BS'!(X:float=0, Y:float=0) -> (Pose:Pose Link Structure) {is=title AnimGraphNode_BlendSpaceGraph}
   n1 = Blendspace Player 'Guitar_FretHand_BS'(X=0, Y=0) {blendSpace=/SparksCharacterCommon/Animation/MainPlayer/BandPlay/Guitar/FretHand/Guitar_FretHand_BS, startPosition=1}
   n2 = Additive Identity Pose()
   n3 = "Not Equal (Enum)"(
    A:ESparksAnimatableState Enum=Get AnimatableState() -> (AnimatableState:ESparksAnimatableState Enum=None)
    B:ESparksAnimatableState Enum=Animatable
   )
   Output Animation Pose(Result=Apply Additive(
     Base=Sequence Player 'Guitar_FretHand_BasePose'!() -> (Pose) {sequence=&Guitar_FretHand_BasePose}
     Additive=Blend Poses by bool(BlendPose_0=@n1, BlendPose_1=@n2, BlendTime_0=0.4, BlendTime_1=0.4, bActiveValue=@n3) {blendType=Linear}
    ) {alphaBoolBlend {blendOption=Linear}})
  ]
 }
 Guitar_Fret_Chord_BS {
  path=AnimGraph.AnimGraphNode_StateMachine_14.BaseBody FretArm.AnimStateNode_1.Guitar.AnimGraphNode_BlendSpaceGraph_0.Guitar_Fret_Chord_BS
  links []
  nodes []
 }
 Guitar_Fret_01_Chord_A_Major {
  path=AnimGraph.AnimGraphNode_StateMachine_14.BaseBody FretArm.AnimStateNode_1.Guitar.AnimGraphNode_BlendSpaceGraph_0.Guitar_Fret_Chord_BS.Guitar_Fret_01_Chord_A_Major
  nodes [Output Animation Pose!(Result:Pose Link Structure="Sequence Player (additive) 'Guitar_Fret01_A'"() {method=Graph, sequence=/SparksCharacterCommon/Animation/MainPlayer/BandPlay/Guitar/FretHand/Guitar_Fret01_A}) {is=title AnimGraphNode_BlendSpaceSampleResult}]
 }
 Guitar_Fret_01_Chord_C_Major {
  path=AnimGraph.AnimGraphNode_StateMachine_14.BaseBody FretArm.AnimStateNode_1.Guitar.AnimGraphNode_BlendSpaceGraph_0.Guitar_Fret_Chord_BS.Guitar_Fret_01_Chord_C_Major
  nodes [Output Animation Pose(Result="Sequence Player (additive) 'Guitar_Fret01_C'"() {method=Graph, sequence=/SparksCharacterCommon/Animation/MainPlayer/BandPlay/Guitar/FretHand/Guitar_Fret01_C})]
 }
 Guitar_Fret_01_Chord_D_Major {
  path=AnimGraph.AnimGraphNode_StateMachine_14.BaseBody FretArm.AnimStateNode_1.Guitar.AnimGraphNode_BlendSpaceGraph_0.Guitar_Fret_Chord_BS.Guitar_Fret_01_Chord_D_Major
  nodes [Output Animation Pose(Result="Sequence Player (additive) 'Guitar_Fret01_D'"() {method=Graph, sequence=/SparksCharacterCommon/Animation/MainPlayer/BandPlay/Guitar/FretHand/Guitar_Fret01_D})]
 }
 Guitar_Fret_01_Chord_E_Major {
  path=AnimGraph.AnimGraphNode_StateMachine_14.BaseBody FretArm.AnimStateNode_1.Guitar.AnimGraphNode_BlendSpaceGraph_0.Guitar_Fret_Chord_BS.Guitar_Fret_01_Chord_E_Major
  nodes [Output Animation Pose(Result="Sequence Player (additive) 'Guitar_Fret01_E'"() {method=Graph, sequence=/SparksCharacterCommon/Animation/MainPlayer/BandPlay/Guitar/FretHand/Guitar_Fret01_E})]
 }
 Guitar_Fret_01_Chord_G_Major {
  path=AnimGraph.AnimGraphNode_StateMachine_14.BaseBody FretArm.AnimStateNode_1.Guitar.AnimGraphNode_BlendSpaceGraph_0.Guitar_Fret_Chord_BS.Guitar_Fret_01_Chord_G_Major
  nodes [Output Animation Pose(Result="Sequence Player (additive) 'Guitar_Fret01_G'"() {method=Graph, sequence=/SparksCharacterCommon/Animation/MainPlayer/BandPlay/Guitar/FretHand/Guitar_Fret01_G})]
 }
 Guitar_Fret_20_Chord_A_Major {
  path=AnimGraph.AnimGraphNode_StateMachine_14.BaseBody FretArm.AnimStateNode_1.Guitar.AnimGraphNode_BlendSpaceGraph_0.Guitar_Fret_Chord_BS.Guitar_Fret_20_Chord_A_Major
  nodes [Output Animation Pose(Result="Sequence Player (additive) 'Guitar_Fret20_A'"() {method=Graph, sequence=/SparksCharacterCommon/Animation/MainPlayer/BandPlay/Guitar/FretHand/Guitar_Fret20_A})]
 }
 Guitar_Fret_20_Chord_C_Major {
  path=AnimGraph.AnimGraphNode_StateMachine_14.BaseBody FretArm.AnimStateNode_1.Guitar.AnimGraphNode_BlendSpaceGraph_0.Guitar_Fret_Chord_BS.Guitar_Fret_20_Chord_C_Major
  nodes [Output Animation Pose(Result="Sequence Player (additive) 'Guitar_Fret20_C'"() {method=Graph, sequence=/SparksCharacterCommon/Animation/MainPlayer/BandPlay/Guitar/FretHand/Guitar_Fret20_C})]
 }
 Guitar_Fret_20_Chord_D_Major {
  path=AnimGraph.AnimGraphNode_StateMachine_14.BaseBody FretArm.AnimStateNode_1.Guitar.AnimGraphNode_BlendSpaceGraph_0.Guitar_Fret_Chord_BS.Guitar_Fret_20_Chord_D_Major
  nodes [Output Animation Pose(Result="Sequence Player (additive) 'Guitar_Fret20_D'"() {method=Graph, sequence=/SparksCharacterCommon/Animation/MainPlayer/BandPlay/Guitar/FretHand/Guitar_Fret20_D})]
 }
 Guitar_Fret_20_Chord_E_Major {
  path=AnimGraph.AnimGraphNode_StateMachine_14.BaseBody FretArm.AnimStateNode_1.Guitar.AnimGraphNode_BlendSpaceGraph_0.Guitar_Fret_Chord_BS.Guitar_Fret_20_Chord_E_Major
  nodes [Output Animation Pose(Result="Sequence Player (additive) 'Guitar_Fret20_E'"() {method=Graph, sequence=/SparksCharacterCommon/Animation/MainPlayer/BandPlay/Guitar/FretHand/Guitar_Fret20_E})]
 }
 Guitar_Fret_20_Chord_G_Major {
  path=AnimGraph.AnimGraphNode_StateMachine_14.BaseBody FretArm.AnimStateNode_1.Guitar.AnimGraphNode_BlendSpaceGraph_0.Guitar_Fret_Chord_BS.Guitar_Fret_20_Chord_G_Major
  nodes [Output Animation Pose(Result="Sequence Player (additive) 'Guitar_Fret20_G'"() {method=Graph, sequence=/SparksCharacterCommon/Animation/MainPlayer/BandPlay/Guitar/FretHand/Guitar_Fret20_G})]
 }
 Bass {
  path=AnimGraph.AnimGraphNode_StateMachine_14.BaseBody FretArm.AnimStateNode_4.Bass
  nodes [
   n1 = Blendspace Player 'Bass_FretHand_BS'(X=0, Y=0) {blendSpace=/SparksCharacterCommon/Animation/MainPlayer/BandPlay/Bass/FretHand/Bass_FretHand_BS, startPosition=1}
   n2 = Additive Identity Pose()
   n3 = "Not Equal (Enum)"(A=Get AnimatableState() -> (AnimatableState=None), B=Animatable)
   Output Animation Pose(Result=Apply Additive(
     Base=Sequence Player 'Bass_FretHand_BasePose'() {sequence=/SparksCharacterCommon/Animation/MainPlayer/BandPlay/Bass/FretHand/Bass_FretHand_BasePose}
     Additive=Blend Poses by bool(BlendPose_0=@n1, BlendPose_1=@n2, bActiveValue=@n3) {blendType=Linear}
    ) {alphaBoolBlend {blendOption=Linear}}) {is=title AnimGraphNode_StateResult}
  ]
 }
 Transition {
  path=AnimGraph.AnimGraphNode_StateMachine_14.BaseBody FretArm.AnimStateTransitionNode_10.Transition
  nodes [
   Result(bCanEnterTransition="Not Equal (Enum)"(A:ESparksInstrumentType Enum=Get InstrumentType() -> (InstrumentType=Guitar), B:ESparksInstrumentType Enum=Guitar))
   Get Instrument() -> (Instrument=Guitar)
  ]
 }
 Transition {
  path=AnimGraph.AnimGraphNode_StateMachine_14.BaseBody FretArm.AnimStateTransitionNode_11.Transition
  nodes [
   Result(bCanEnterTransition="Not Equal (Enum)"(A=Get InstrumentType() -> (InstrumentType=Guitar), B=Bass))
   Get Instrument() -> (Instrument=Guitar)
  ]
 }
 Transition {
  path=AnimGraph.AnimGraphNode_StateMachine_14.BaseBody FretArm.AnimStateTransitionNode_2.Transition
  nodes [
   Result(bCanEnterTransition="Equal (Enum)"(A=Get InstrumentType() -> (InstrumentType=Guitar), B=Guitar))
   Get Instrument() -> (Instrument=Guitar)
  ]
 }
 Transition {
  path=AnimGraph.AnimGraphNode_StateMachine_14.BaseBody FretArm.AnimStateTransitionNode_3.Transition
  nodes [
   Result(bCanEnterTransition="Equal (Enum)"(A=Get InstrumentType() -> (InstrumentType=Guitar), B=Bass))
   Get Instrument() -> (Instrument=Guitar)
  ]
 }
 Transition {
  path=AnimGraph.AnimGraphNode_StateMachine_14.BaseBody FretArm.AnimStateTransitionNode_8.Transition
  nodes [
   Result(bCanEnterTransition="Not Equal (Enum)"(A=Get InstrumentType() -> (InstrumentType=Guitar), B=Keyboard))
   Get Instrument() -> (Instrument=Guitar)
  ]
 }
 Transition {
  path=AnimGraph.AnimGraphNode_StateMachine_14.BaseBody FretArm.AnimStateTransitionNode_9.Transition
  nodes [
   Result(bCanEnterTransition="Equal (Enum)"(A=Get InstrumentType() -> (InstrumentType=Guitar), B=Keyboard))
   Get InstrumentType() -> (InstrumentType=Guitar)
  ]
 }
 FootShift 1stFoot Translate {
  entry=FootShift 1stFoot Translate RESET
  path=AnimGraph.AnimGraphNode_StateMachine_15.FootShift 1stFoot Translate
  ids=id
  nodes [
   AnimStateNode_0 = FootShift 1stFoot Translate RESET() {
    is=title AnimStateNode
    stateEntered {guid=#1}
    stateFullyBlended {guid=#2}
    stateLeft {guid=#3}
   }
   AnimStateNode_1 = FootShift 1stFoot Translate ON() {
    is=title AnimStateNode
    stateEntered {guid=#4}
    stateFullyBlended {guid=#5}
    stateLeft {guid=#6}
   }
   AnimStateNode_3 = FootShift 1stFoot Translate OFF() {
    is=title AnimStateNode
    stateEntered {guid=#1}
    stateFullyBlended {guid=#2}
    stateLeft {guid=#3}
   }
   AnimStateTransitionNode_0 = AnimStateTransitionNode() {crossfadeDuration=0.22}
   AnimStateTransitionNode_10 = AnimStateTransitionNode() {bAllowInertializationForSelfTransitions=true, crossfadeDuration=0}
   AnimStateTransitionNode_11 = AnimStateTransitionNode() {bAllowInertializationForSelfTransitions=true, crossfadeDuration=0}
   AnimStateTransitionNode_16 = AnimStateTransitionNode() {
    bAllowInertializationForSelfTransitions=true
    crossfadeDuration=0.33
    priorityOrder=2
    transitionEnd {guid=#19}
    transitionInterrupt {guid=#20}
    transitionStart {guid=#21}
   }
   AnimStateTransitionNode_17 = AnimStateTransitionNode() {
    bAllowInertializationForSelfTransitions=true
    crossfadeDuration=0.44
    priorityOrder=3
    transitionEnd {guid=#22}
    transitionInterrupt {guid=#23}
    transitionStart {guid=#24}
   }
   AnimStateTransitionNode_18 = AnimStateTransitionNode() {bAllowInertializationForSelfTransitions=true, crossfadeDuration=0.33, priorityOrder=2}
   AnimStateTransitionNode_19 = AnimStateTransitionNode() {bAllowInertializationForSelfTransitions=true, crossfadeDuration=0.44, priorityOrder=3}
   AnimStateTransitionNode_6 = AnimStateTransitionNode() {
    bAllowInertializationForSelfTransitions=true
    crossfadeDuration=0
    transitionEnd {guid=#25}
    transitionInterrupt {guid=#26}
    transitionStart {guid=#27}
   }
   AnimStateTransitionNode_8 = AnimStateTransitionNode() {
    crossfadeDuration=0.22
    transitionEnd {guid=#7}
    transitionInterrupt {guid=#8}
    transitionStart {guid=#9}
   }
  ]
  links [
   {
    dst=FootShift 1stFoot Translate ON
    rule=AnimGraph.AnimGraphNode_StateMachine_15.FootShift 1stFoot Translate.AnimStateTransitionNode_0.Transition
    src=FootShift 1stFoot Translate OFF
    via=AnimStateTransitionNode_0
   }
   {
    dst=FootShift 1stFoot Translate ON
    rule=AnimGraph.AnimGraphNode_StateMachine_15.FootShift 1stFoot Translate.AnimStateTransitionNode_18.Transition
    src=FootShift 1stFoot Translate OFF
    via=AnimStateTransitionNode_18
   }
   {
    dst=FootShift 1stFoot Translate ON
    rule=AnimGraph.AnimGraphNode_StateMachine_15.FootShift 1stFoot Translate.AnimStateTransitionNode_19.Transition
    src=FootShift 1stFoot Translate OFF
    via=AnimStateTransitionNode_19
   }
   {
    dst=FootShift 1stFoot Translate RESET
    rule=AnimGraph.AnimGraphNode_StateMachine_15.FootShift 1stFoot Translate.AnimStateTransitionNode_10.Transition
    src=FootShift 1stFoot Translate OFF
    via=AnimStateTransitionNode_10
   }
   {
    dst=FootShift 1stFoot Translate OFF
    rule=AnimGraph.AnimGraphNode_StateMachine_15.FootShift 1stFoot Translate.AnimStateTransitionNode_16.Transition
    src=FootShift 1stFoot Translate ON
    via=AnimStateTransitionNode_16
   }
   {
    dst=FootShift 1stFoot Translate OFF
    rule=AnimGraph.AnimGraphNode_StateMachine_15.FootShift 1stFoot Translate.AnimStateTransitionNode_17.Transition
    src=FootShift 1stFoot Translate ON
    via=AnimStateTransitionNode_17
   }
   {
    dst=FootShift 1stFoot Translate OFF
    rule=AnimGraph.AnimGraphNode_StateMachine_15.FootShift 1stFoot Translate.AnimStateTransitionNode_8.AnimationTransitionGraph_0
    src=FootShift 1stFoot Translate ON
    via=AnimStateTransitionNode_8
   }
   {
    dst=FootShift 1stFoot Translate RESET
    rule=AnimGraph.AnimGraphNode_StateMachine_15.FootShift 1stFoot Translate.AnimStateTransitionNode_6.Transition
    src=FootShift 1stFoot Translate ON
    via=AnimStateTransitionNode_6
   }
   {
    dst=FootShift 1stFoot Translate OFF
    rule=AnimGraph.AnimGraphNode_StateMachine_15.FootShift 1stFoot Translate.AnimStateTransitionNode_11.Transition
    src=FootShift 1stFoot Translate RESET
    via=AnimStateTransitionNode_11
   }
  ]
 }
 FootShift 1stFoot Translate RESET {
  path=AnimGraph.AnimGraphNode_StateMachine_15.FootShift 1stFoot Translate.AnimStateNode_0.FootShift 1stFoot Translate RESET
  nodes [Output Animation Pose(Result=Use cached pose 'BaseBody Post-Additives Pose'())]
 }
 FootShift 1stFoot Translate ON {
  path=AnimGraph.AnimGraphNode_StateMachine_15.FootShift 1stFoot Translate.AnimStateNode_1.FootShift 1stFoot Translate ON
  nodes [
   n1 = "Transform (Modify) Bone - Bone: pelvis"(
    ComponentPose=Local To Component()
    Alpha=1
    Translation_X=Get FootShift_1stPelvis_ProcTranslateLR() -> (FootShift_1stPelvis_ProcTranslateLR:double)
    Translation_Y=Get FootShift_1stPelvis_ProcTranslateFB() -> (FootShift_1stPelvis_ProcTranslateFB:double)
    Translation_Z=Get FootShift_1stPelvis_ProcTranslateVertical() -> (FootShift_1stPelvis_ProcTranslateVertical:double)
    Rotation_Roll=0
    Rotation_Pitch=0
    Rotation_Yaw=Get FootShift_1stPelvis_ProcRotateTwist() -> (FootShift_1stPelvis_ProcRotateTwist:float)
   ) {
    rotationMode=BMM_Additive
    translationMode=BMM_Additive
    alphaBoolBlend {blendOption=Linear}
    boneToModify {boneName=pelvis}
    translation=<30 30>
   }
   Component To Local(ComponentPose=@n1)
   n2 = "Transform (Modify) Bone - Bone: ik_foot_l"!(
    ComponentPose=Local To Component(LocalPose=Use cached pose 'BaseBody Post-Additives Pose'())
    Alpha=1
    Translation_X:double=Get FootShift_1stFoot_ProcTranslateLR() -> (FootShift_1stFoot_ProcTranslateLR:double)
    Translation_Y:double=Get FootShift_1stFoot_ProcTranslateFB() -> (FootShift_1stFoot_ProcTranslateFB:double)
    Translation_Z:double=0
    Rotation_Roll:float=0
    Rotation_Pitch:float=0
    Rotation_Yaw:float=Get FootShift_1stFoot_ProcRotateTwist() -> (FootShift_1stFoot_ProcRotateTwist:double)
   ) -> (Pose) {
    rotationMode=BMM_Additive
    translationMode=BMM_Additive
    alphaBoolBlend {blendOption=Linear}
    boneToModify {boneName=ik_foot_l}
    translation=<30 30>
   }
   n3 = "Transform (Modify) Bone - Bone: ik_foot_r"!(
    ComponentPose=Local To Component(LocalPose=Use cached pose 'BaseBody Post-Additives Pose'())
    Alpha=1
    Translation_X:double=Get FootShift_1stFoot_ProcTranslateLR()
    Translation_Y:double=Get FootShift_1stFoot_ProcTranslateFB()
    Translation_Z:double=0
    Rotation_Roll:float=0
    Rotation_Pitch:float=0
    Rotation_Yaw:float=Get FootShift_1stFoot_ProcRotateTwist()
   ) -> (Pose) {
    rotationMode=BMM_Additive
    translationMode=BMM_Additive
    alphaBoolBlend {blendOption=Linear}
    boneToModify {boneName=ik_foot_r}
    translation=<30 30>
   }
   n4 = "Transform (Modify) Bone - Bone: ik_foot_l"(
    ComponentPose=Local To Component(LocalPose=Use cached pose 'BaseBody Post-Additives Pose'())
    Alpha=1
    Translation_X=Get FootShift_1stFoot_ProcTranslateLR()
    Translation_Y=Get FootShift_1stFoot_ProcTranslateFB()
    Translation_Z=0
    Rotation_Roll=0
    Rotation_Pitch=0
    Rotation_Yaw=Get FootShift_1stFoot_ProcRotateTwist()
   ) {
    rotationMode=BMM_Additive
    translationMode=BMM_Additive
    alphaBoolBlend {blendOption=Linear}
    boneToModify {boneName=ik_foot_l}
    translation=<30 30>
   }
   n5 = "Transform (Modify) Bone - Bone: ik_foot_r"(
    ComponentPose=Local To Component(LocalPose=Use cached pose 'BaseBody Post-Additives Pose'())
    Alpha=1
    Translation_X=Get FootShift_1stFoot_ProcTranslateLR()
    Translation_Y=Get FootShift_1stFoot_ProcTranslateFB()
    Translation_Z=0
    Rotation_Roll=0
    Rotation_Pitch=0
    Rotation_Yaw=Get FootShift_1stFoot_ProcRotateTwist()
   ) {
    rotationMode=BMM_Additive
    translationMode=BMM_Additive
    alphaBoolBlend {blendOption=Linear}
    boneToModify {boneName=ik_foot_r}
    translation=<30 30>
   }
   n6 = Blend Poses by int!(
    BlendPose_0:Pose Link Structure=Component To Local(ComponentPose=@n4)
    BlendPose_1:Pose Link Structure=Component To Local(ComponentPose=@n5)
    BlendPose_2:Pose Link Structure=Blend Poses by bool(
     BlendPose_0=Component To Local(ComponentPose=@n2)
     BlendPose_1=Component To Local(ComponentPose=@n3)
     BlendTime_0=0
     BlendTime_1=0
     bActiveValue=Get FootShift_Proc1stFootIsLeft?() -> (FootShift_Proc1stFootIsLeft?:bool)
    )
    BlendTime_0:float=0
    BlendTime_1:float=0
    BlendTime_2:float=0
    ActiveChildIndex:int=Get FootShift_ProcModeIndex() -> (FootShift_ProcModeIndex:int=0)
   ) -> (Pose:Pose Link Structure) {
    is=title AnimGraphNode_BlendListByInt
    blendPose [{linkId=-1, sourceLinkId=-1}, {linkId=-1, sourceLinkId=-1}, {linkId=-1, sourceLinkId=-1}]
    blendTime [0.1,0.1,0.1]
   }
   Output Animation Pose(Result=@n6)
  ]
 }
 FootShift 1stFoot Translate OFF {
  path=AnimGraph.AnimGraphNode_StateMachine_15.FootShift 1stFoot Translate.AnimStateNode_3.FootShift 1stFoot Translate OFF
  nodes [
   n1 = "Transform (Modify) Bone - Bone: pelvis"(
    ComponentPose=Local To Component()
    Alpha=1
    Translation_X=0
    Translation_Y=0
    Translation_Z=0
    Rotation_Roll=0
    Rotation_Pitch=0
    Rotation_Yaw=0
   ) {
    rotationMode=BMM_Additive
    translationMode=BMM_Additive
    alphaBoolBlend {blendOption=Linear}
    boneToModify {boneName=pelvis}
    translation=<30 30>
   }
   Component To Local(ComponentPose=@n1)
   n2 = "Transform (Modify) Bone - Bone: ik_foot_l"(
    ComponentPose=Local To Component(LocalPose=Use cached pose 'BaseBody Post-Additives Pose'())
    Alpha=1
    Translation_X=0
    Translation_Y=0
    Translation_Z=0
    Rotation_Roll=0
    Rotation_Pitch=0
    Rotation_Yaw=0
   ) {
    rotationMode=BMM_Additive
    translationMode=BMM_Additive
    alphaBoolBlend {blendOption=Linear}
    boneToModify {boneName=ik_foot_l}
    translation=<30 30>
   }
   n3 = "Transform (Modify) Bone - Bone: ik_foot_r"(
    ComponentPose=Local To Component(LocalPose=Use cached pose 'BaseBody Post-Additives Pose'())
    Alpha=1
    Translation_X=0
    Translation_Y=0
    Translation_Z=0
    Rotation_Roll=0
    Rotation_Pitch=0
    Rotation_Yaw=0
   ) {
    rotationMode=BMM_Additive
    translationMode=BMM_Additive
    alphaBoolBlend {blendOption=Linear}
    boneToModify {boneName=ik_foot_r}
    translation=<30 30>
   }
   n4 = "Transform (Modify) Bone - Bone: ik_foot_l"(
    ComponentPose=Local To Component(LocalPose=Use cached pose 'BaseBody Post-Additives Pose'())
    Alpha=1
    Translation_X=0
    Translation_Y=0
    Translation_Z=0
    Rotation_Roll=0
    Rotation_Pitch=0
    Rotation_Yaw=0
   ) {
    rotationMode=BMM_Additive
    translationMode=BMM_Additive
    alphaBoolBlend {blendOption=Linear}
    boneToModify {boneName=ik_foot_l}
    translation=<30 30>
   }
   n5 = "Transform (Modify) Bone - Bone: ik_foot_r"(
    ComponentPose=Local To Component(LocalPose=Use cached pose 'BaseBody Post-Additives Pose'())
    Alpha=1
    Translation_X=0
    Translation_Y=0
    Translation_Z=0
    Rotation_Roll=0
    Rotation_Pitch=0
    Rotation_Yaw=0
   ) {
    rotationMode=BMM_Additive
    translationMode=BMM_Additive
    alphaBoolBlend {blendOption=Linear}
    boneToModify {boneName=ik_foot_r}
    translation=<30 30>
   }
   n6 = Blend Poses by int(
    BlendPose_0=Component To Local(ComponentPose=@n4)
    BlendPose_1=Component To Local(ComponentPose=@n5)
    BlendPose_2=Blend Poses by bool(
     BlendPose_0=Component To Local(ComponentPose=@n2)
     BlendPose_1=Component To Local(ComponentPose=@n3)
     BlendTime_0=0
     BlendTime_1=0
     bActiveValue=Get FootShift_Proc1stFootIsLeft?()
    )
    BlendTime_0=0
    BlendTime_1=0
    BlendTime_2=0
    ActiveChildIndex=Get FootShift_ProcModeIndex() -> (FootShift_ProcModeIndex=0)
   ) {blendPose [{linkId=-1, sourceLinkId=-1}, {linkId=-1, sourceLinkId=-1}, {linkId=-1, sourceLinkId=-1}], blendTime [0.1,0.1,0.1]}
   Output Animation Pose(Result=@n6)
  ]
 }
 Transition {
  path=AnimGraph.AnimGraphNode_StateMachine_15.FootShift 1stFoot Translate.AnimStateTransitionNode_0.Transition
  nodes [
   n1 = Get FootShift_1stFoot_ProcTransitionSpeed() -> (FootShift_1stFoot_ProcTransitionSpeed:Enum_FootShift_ProcTransitionSpeed Enum=NewEnumerator0)
   Result(bCanEnterTransition=AND Boolean(
     A=Get FootShift_1stFoot_ProcTranslateCurrentlyActive?() -> (FootShift_1stFoot_ProcTranslateCurrentlyActive?:bool)
     B="Equal (Enum)"(A:Enum_FootShift_ProcTransitionSpeed Enum=@n1, B:Enum_FootShift_ProcTransitionSpeed Enum=NewEnumerator2)
    ))
  ]
 }
 Transition {
  path=AnimGraph.AnimGraphNode_StateMachine_15.FootShift 1stFoot Translate.AnimStateTransitionNode_10.Transition
  nodes [Result(bCanEnterTransition=Get DEBUG_FootShift_ResetSystem?())]
 }
 Transition {
  path=AnimGraph.AnimGraphNode_StateMachine_15.FootShift 1stFoot Translate.AnimStateTransitionNode_11.Transition
  nodes [Result(bCanEnterTransition=true)]
 }
 Transition {
  path=AnimGraph.AnimGraphNode_StateMachine_15.FootShift 1stFoot Translate.AnimStateTransitionNode_16.Transition
  nodes [
   n1 = Get FootShift_1stFoot_ProcTranslateCurrentlyActive?()
   n2 = Get FootShift_1stFoot_ProcTransitionSpeed() -> (FootShift_1stFoot_ProcTransitionSpeed=NewEnumerator0)
   Result(bCanEnterTransition=AND Boolean(A=NOT Boolean(A=@n1), B="Equal (Enum)"(A=@n2, B=NewEnumerator1)))
  ]
 }
 Transition {
  path=AnimGraph.AnimGraphNode_StateMachine_15.FootShift 1stFoot Translate.AnimStateTransitionNode_17.Transition
  nodes [
   n1 = Get FootShift_1stFoot_ProcTranslateCurrentlyActive?()
   n2 = Get FootShift_1stFoot_ProcTransitionSpeed() -> (FootShift_1stFoot_ProcTransitionSpeed=NewEnumerator0)
   Result(bCanEnterTransition=AND Boolean(A=NOT Boolean(A=@n1), B="Equal (Enum)"(A=@n2, B=NewEnumerator0)))
  ]
 }
 Transition {
  path=AnimGraph.AnimGraphNode_StateMachine_15.FootShift 1stFoot Translate.AnimStateTransitionNode_18.Transition
  nodes [
   n1 = Get FootShift_1stFoot_ProcTransitionSpeed() -> (FootShift_1stFoot_ProcTransitionSpeed=NewEnumerator0)
   Result(bCanEnterTransition=AND Boolean(A=Get FootShift_1stFoot_ProcTranslateCurrentlyActive?(), B="Equal (Enum)"(A=@n1, B=NewEnumerator1)))
  ]
 }
 Transition {
  path=AnimGraph.AnimGraphNode_StateMachine_15.FootShift 1stFoot Translate.AnimStateTransitionNode_19.Transition
  nodes [
   n1 = Get FootShift_1stFoot_ProcTransitionSpeed() -> (FootShift_1stFoot_ProcTransitionSpeed=NewEnumerator0)
   Result(bCanEnterTransition=AND Boolean(A=Get FootShift_1stFoot_ProcTranslateCurrentlyActive?(), B="Equal (Enum)"(A=@n1, B=NewEnumerator0)))
  ]
 }
 Transition {
  path=AnimGraph.AnimGraphNode_StateMachine_15.FootShift 1stFoot Translate.AnimStateTransitionNode_6.Transition
  nodes [Result(bCanEnterTransition=Get DEBUG_FootShift_ResetSystem?())]
 }
 AnimationTransitionGraph_0 {
  path=AnimGraph.AnimGraphNode_StateMachine_15.FootShift 1stFoot Translate.AnimStateTransitionNode_8.AnimationTransitionGraph_0
  nodes [
   n1 = Get FootShift_1stFoot_ProcTranslateCurrentlyActive?()
   n2 = Get FootShift_1stFoot_ProcTransitionSpeed() -> (FootShift_1stFoot_ProcTransitionSpeed=NewEnumerator0)
   Result(bCanEnterTransition=AND Boolean(A=NOT Boolean(A=@n1), B="Equal (Enum)"(A=@n2, B=NewEnumerator2)))
  ]
 }
 FootShift 2ndFoot Translate {
  entry=FootShift 2ndFoot Translate RESET
  path=AnimGraph.AnimGraphNode_StateMachine_16.FootShift 2ndFoot Translate
  ids=id
  nodes [
   AnimStateNode_0 = FootShift 2ndFoot Translate RESET() {
    is=title AnimStateNode
    stateEntered {guid=#1}
    stateFullyBlended {guid=#2}
    stateLeft {guid=#3}
   }
   AnimStateNode_1 = FootShift 2ndFoot Translate ON() {
    is=title AnimStateNode
    stateEntered {guid=#4}
    stateFullyBlended {guid=#5}
    stateLeft {guid=#6}
   }
   AnimStateNode_3 = FootShift 2ndFoot Translate OFF() {
    is=title AnimStateNode
    stateEntered {guid=#1}
    stateFullyBlended {guid=#2}
    stateLeft {guid=#3}
   }
   AnimStateTransitionNode_0 = AnimStateTransitionNode() {crossfadeDuration=0.22}
   AnimStateTransitionNode_1 = AnimStateTransitionNode() {bAllowInertializationForSelfTransitions=true, crossfadeDuration=0}
   AnimStateTransitionNode_10 = AnimStateTransitionNode() {bAllowInertializationForSelfTransitions=true, crossfadeDuration=0}
   AnimStateTransitionNode_16 = AnimStateTransitionNode() {bAllowInertializationForSelfTransitions=true, crossfadeDuration=0.33, priorityOrder=2}
   AnimStateTransitionNode_17 = AnimStateTransitionNode() {bAllowInertializationForSelfTransitions=true, crossfadeDuration=0.44, priorityOrder=3}
   AnimStateTransitionNode_18 = AnimStateTransitionNode() {bAllowInertializationForSelfTransitions=true, crossfadeDuration=0.33, priorityOrder=2}
   AnimStateTransitionNode_19 = AnimStateTransitionNode() {bAllowInertializationForSelfTransitions=true, crossfadeDuration=0.44, priorityOrder=3}
   AnimStateTransitionNode_6 = AnimStateTransitionNode() {bAllowInertializationForSelfTransitions=true, crossfadeDuration=0}
   AnimStateTransitionNode_8 = AnimStateTransitionNode() {
    crossfadeDuration=0.22
    transitionEnd {guid=#7}
    transitionInterrupt {guid=#8}
    transitionStart {guid=#9}
   }
  ]
  links [
   {
    dst=FootShift 2ndFoot Translate ON
    rule=AnimGraph.AnimGraphNode_StateMachine_16.FootShift 2ndFoot Translate.AnimStateTransitionNode_0.Transition
    src=FootShift 2ndFoot Translate OFF
    via=AnimStateTransitionNode_0
   }
   {
    dst=FootShift 2ndFoot Translate ON
    rule=AnimGraph.AnimGraphNode_StateMachine_16.FootShift 2ndFoot Translate.AnimStateTransitionNode_18.Transition
    src=FootShift 2ndFoot Translate OFF
    via=AnimStateTransitionNode_18
   }
   {
    dst=FootShift 2ndFoot Translate ON
    rule=AnimGraph.AnimGraphNode_StateMachine_16.FootShift 2ndFoot Translate.AnimStateTransitionNode_19.Transition
    src=FootShift 2ndFoot Translate OFF
    via=AnimStateTransitionNode_19
   }
   {
    dst=FootShift 2ndFoot Translate RESET
    rule=AnimGraph.AnimGraphNode_StateMachine_16.FootShift 2ndFoot Translate.AnimStateTransitionNode_10.Transition
    src=FootShift 2ndFoot Translate OFF
    via=AnimStateTransitionNode_10
   }
   {
    dst=FootShift 2ndFoot Translate OFF
    rule=AnimGraph.AnimGraphNode_StateMachine_16.FootShift 2ndFoot Translate.AnimStateTransitionNode_16.Transition
    src=FootShift 2ndFoot Translate ON
    via=AnimStateTransitionNode_16
   }
   {
    dst=FootShift 2ndFoot Translate OFF
    rule=AnimGraph.AnimGraphNode_StateMachine_16.FootShift 2ndFoot Translate.AnimStateTransitionNode_17.Transition
    src=FootShift 2ndFoot Translate ON
    via=AnimStateTransitionNode_17
   }
   {
    dst=FootShift 2ndFoot Translate OFF
    rule=AnimGraph.AnimGraphNode_StateMachine_16.FootShift 2ndFoot Translate.AnimStateTransitionNode_8.AnimationTransitionGraph_0
    src=FootShift 2ndFoot Translate ON
    via=AnimStateTransitionNode_8
   }
   {
    dst=FootShift 2ndFoot Translate RESET
    rule=AnimGraph.AnimGraphNode_StateMachine_16.FootShift 2ndFoot Translate.AnimStateTransitionNode_6.Transition
    src=FootShift 2ndFoot Translate ON
    via=AnimStateTransitionNode_6
   }
   {
    dst=FootShift 2ndFoot Translate OFF
    rule=AnimGraph.AnimGraphNode_StateMachine_16.FootShift 2ndFoot Translate.AnimStateTransitionNode_1.Transition
    src=FootShift 2ndFoot Translate RESET
    via=AnimStateTransitionNode_1
   }
  ]
 }
 FootShift 2ndFoot Translate RESET {
  path=AnimGraph.AnimGraphNode_StateMachine_16.FootShift 2ndFoot Translate.AnimStateNode_0.FootShift 2ndFoot Translate RESET
  nodes [Output Animation Pose(Result=Use cached pose 'BaseBody Post-Additives Pose'())]
 }
 FootShift 2ndFoot Translate ON {
  path=AnimGraph.AnimGraphNode_StateMachine_16.FootShift 2ndFoot Translate.AnimStateNode_1.FootShift 2ndFoot Translate ON
  nodes [
   n1 = "Transform (Modify) Bone - Bone: pelvis"(
    ComponentPose=Local To Component()
    Alpha=1
    Translation_X=Get FootShift_2ndPelvis_ProcTranslateLR()
    Translation_Y=Get FootShift_2ndPelvis_ProcTranslateFB()
    Translation_Z=Get FootShift_2ndPelvis_ProcTranslateVertical()
    Rotation_Roll=0
    Rotation_Pitch=0
    Rotation_Yaw=Get FootShift_2ndPelvis_ProcRotateTwist()
   ) {
    rotationMode=BMM_Additive
    translationMode=BMM_Additive
    alphaBoolBlend {blendOption=Linear}
    boneToModify {boneName=pelvis}
    translation=<30 30>
   }
   Component To Local(ComponentPose=@n1)
   n2 = "Transform (Modify) Bone - Bone: ik_foot_l"(
    ComponentPose=Local To Component(LocalPose=Use cached pose 'BaseBody Post-Additives Pose'())
    Alpha=1
    Translation_X=Get FootShift_2ndFoot_ProcTranslateLR() -> (FootShift_2ndFoot_ProcTranslateLR:double)
    Translation_Y=Get FootShift_2ndFoot_ProcTranslateFB() -> (FootShift_2ndFoot_ProcTranslateFB:double)
    Translation_Z=0
    Rotation_Roll=0
    Rotation_Pitch=0
    Rotation_Yaw=Get FootShift_2ndFoot_ProcRotateTwist() -> (FootShift_2ndFoot_ProcRotateTwist:double)
   ) {
    rotationMode=BMM_Additive
    translationMode=BMM_Additive
    alphaBoolBlend {blendOption=Linear}
    boneToModify {boneName=ik_foot_l}
    translation=<30 30>
   }
   n3 = "Transform (Modify) Bone - Bone: ik_foot_r"(
    ComponentPose=Local To Component(LocalPose=Use cached pose 'BaseBody Post-Additives Pose'())
    Alpha=1
    Translation_X=Get FootShift_2ndFoot_ProcTranslateLR()
    Translation_Y=Get FootShift_2ndFoot_ProcTranslateFB()
    Translation_Z=0
    Rotation_Roll=0
    Rotation_Pitch=0
    Rotation_Yaw=Get FootShift_2ndFoot_ProcRotateTwist()
   ) {
    rotationMode=BMM_Additive
    translationMode=BMM_Additive
    alphaBoolBlend {blendOption=Linear}
    boneToModify {boneName=ik_foot_r}
    translation=<30 30>
   }
   n4 = "Transform (Modify) Bone - Bone: ik_foot_l"(
    ComponentPose=Local To Component(LocalPose=Use cached pose 'BaseBody Post-Additives Pose'())
    Alpha=1
    Translation_X=Get FootShift_2ndFoot_ProcTranslateLR()
    Translation_Y=Get FootShift_2ndFoot_ProcTranslateFB()
    Translation_Z=0
    Rotation_Roll=0
    Rotation_Pitch=0
    Rotation_Yaw=Get FootShift_2ndFoot_ProcRotateTwist()
   ) {
    rotationMode=BMM_Additive
    translationMode=BMM_Additive
    alphaBoolBlend {blendOption=Linear}
    boneToModify {boneName=ik_foot_l}
    translation=<30 30>
   }
   n5 = "Transform (Modify) Bone - Bone: ik_foot_r"(
    ComponentPose=Local To Component(LocalPose=Use cached pose 'BaseBody Post-Additives Pose'())
    Alpha=1
    Translation_X=Get FootShift_2ndFoot_ProcTranslateLR()
    Translation_Y=Get FootShift_2ndFoot_ProcTranslateFB()
    Translation_Z=0
    Rotation_Roll=0
    Rotation_Pitch=0
    Rotation_Yaw=Get FootShift_2ndFoot_ProcRotateTwist()
   ) {
    rotationMode=BMM_Additive
    translationMode=BMM_Additive
    alphaBoolBlend {blendOption=Linear}
    boneToModify {boneName=ik_foot_r}
    translation=<30 30>
   }
   n6 = Blend Poses by int(
    BlendPose_0=Component To Local(ComponentPose=@n4)
    BlendPose_1=Component To Local(ComponentPose=@n5)
    BlendPose_2=Blend Poses by bool(
     BlendPose_0=Component To Local(ComponentPose=@n3)
     BlendPose_1=Component To Local(ComponentPose=@n2)
     BlendTime_0=0
     BlendTime_1=0
     bActiveValue=Get FootShift_Proc1stFootIsLeft?()
    )
    BlendTime_0=0
    BlendTime_1=0
    BlendTime_2=0
    ActiveChildIndex=Get FootShift_ProcModeIndex() -> (FootShift_ProcModeIndex=0)
   ) {blendPose [{linkId=-1, sourceLinkId=-1}, {linkId=-1, sourceLinkId=-1}, {linkId=-1, sourceLinkId=-1}], blendTime [0.1,0.1,0.1]}
   Output Animation Pose(Result=@n6)
  ]
 }
 FootShift 2ndFoot Translate OFF {
  path=AnimGraph.AnimGraphNode_StateMachine_16.FootShift 2ndFoot Translate.AnimStateNode_3.FootShift 2ndFoot Translate OFF
  nodes [
   n1 = "Transform (Modify) Bone - Bone: pelvis"(
    ComponentPose=Local To Component()
    Alpha=1
    Translation_X=0
    Translation_Y=0
    Translation_Z=0
    Rotation_Roll=0
    Rotation_Pitch=0
    Rotation_Yaw=0
   ) {
    rotationMode=BMM_Additive
    translationMode=BMM_Additive
    alphaBoolBlend {blendOption=Linear}
    boneToModify {boneName=pelvis}
    translation=<30 30>
   }
   Component To Local(ComponentPose=@n1)
   n2 = "Transform (Modify) Bone - Bone: ik_foot_l"(
    ComponentPose=Local To Component(LocalPose=Use cached pose 'BaseBody Post-Additives Pose'())
    Alpha=1
    Translation_X=0
    Translation_Y=0
    Translation_Z=0
    Rotation_Roll=0
    Rotation_Pitch=0
    Rotation_Yaw=0
   ) {
    rotationMode=BMM_Additive
    translationMode=BMM_Additive
    alphaBoolBlend {blendOption=Linear}
    boneToModify {boneName=ik_foot_l}
    translation=<30 30>
   }
   n3 = "Transform (Modify) Bone - Bone: ik_foot_r"(
    ComponentPose=Local To Component(LocalPose=Use cached pose 'BaseBody Post-Additives Pose'())
    Alpha=1
    Translation_X=0
    Translation_Y=0
    Translation_Z=0
    Rotation_Roll=0
    Rotation_Pitch=0
    Rotation_Yaw=0
   ) {
    rotationMode=BMM_Additive
    translationMode=BMM_Additive
    alphaBoolBlend {blendOption=Linear}
    boneToModify {boneName=ik_foot_r}
    translation=<30 30>
   }
   n4 = "Transform (Modify) Bone - Bone: ik_foot_l"(
    ComponentPose=Local To Component(LocalPose=Use cached pose 'BaseBody Post-Additives Pose'())
    Alpha=1
    Translation_X=0
    Translation_Y=0
    Translation_Z=0
    Rotation_Roll=0
    Rotation_Pitch=0
    Rotation_Yaw=0
   ) {
    rotationMode=BMM_Additive
    translationMode=BMM_Additive
    alphaBoolBlend {blendOption=Linear}
    boneToModify {boneName=ik_foot_l}
    translation=<30 30>
   }
   n5 = "Transform (Modify) Bone - Bone: ik_foot_r"(
    ComponentPose=Local To Component(LocalPose=Use cached pose 'BaseBody Post-Additives Pose'())
    Alpha=1
    Translation_X=0
    Translation_Y=0
    Translation_Z=0
    Rotation_Roll=0
    Rotation_Pitch=0
    Rotation_Yaw=0
   ) {
    rotationMode=BMM_Additive
    translationMode=BMM_Additive
    alphaBoolBlend {blendOption=Linear}
    boneToModify {boneName=ik_foot_r}
    translation=<30 30>
   }
   n6 = Blend Poses by int(
    BlendPose_0=Component To Local(ComponentPose=@n4)
    BlendPose_1=Component To Local(ComponentPose=@n5)
    BlendPose_2=Blend Poses by bool(
     BlendPose_0=Component To Local(ComponentPose=@n3)
     BlendPose_1=Component To Local(ComponentPose=@n2)
     BlendTime_0=0
     BlendTime_1=0
     bActiveValue=Get FootShift_Proc1stFootIsLeft?()
    )
    BlendTime_0=0
    BlendTime_1=0
    BlendTime_2=0
    ActiveChildIndex=Get FootShift_ProcModeIndex() -> (FootShift_ProcModeIndex=0)
   ) {blendPose [{linkId=-1, sourceLinkId=-1}, {linkId=-1, sourceLinkId=-1}, {linkId=-1, sourceLinkId=-1}], blendTime [0.1,0.1,0.1]}
   Output Animation Pose(Result=@n6)
  ]
 }
 Transition {
  path=AnimGraph.AnimGraphNode_StateMachine_16.FootShift 2ndFoot Translate.AnimStateTransitionNode_0.Transition
  nodes [
   n1 = Get FootShift_2ndFoot_ProcTransitionSpeed() -> (FootShift_2ndFoot_ProcTransitionSpeed=NewEnumerator0)
   Result(bCanEnterTransition=AND Boolean(
     A=Get FootShift_2ndFoot_ProcTranslateCurrentlyActive?() -> (FootShift_2ndFoot_ProcTranslateCurrentlyActive?:bool)
     B="Equal (Enum)"(A=@n1, B=NewEnumerator2)
    ))
  ]
 }
 Transition {
  path=AnimGraph.AnimGraphNode_StateMachine_16.FootShift 2ndFoot Translate.AnimStateTransitionNode_1.Transition
  nodes [Result(bCanEnterTransition=true)]
 }
 Transition {
  path=AnimGraph.AnimGraphNode_StateMachine_16.FootShift 2ndFoot Translate.AnimStateTransitionNode_10.Transition
  nodes [Result(bCanEnterTransition=Get DEBUG_FootShift_ResetSystem?())]
 }
 Transition {
  path=AnimGraph.AnimGraphNode_StateMachine_16.FootShift 2ndFoot Translate.AnimStateTransitionNode_16.Transition
  nodes [
   n1 = Get FootShift_2ndFoot_ProcTranslateCurrentlyActive?()
   n2 = Get FootShift_2ndFoot_ProcTransitionSpeed() -> (FootShift_2ndFoot_ProcTransitionSpeed=NewEnumerator0)
   Result(bCanEnterTransition=AND Boolean(A=NOT Boolean(A=@n1), B="Equal (Enum)"(A=@n2, B=NewEnumerator1)))
  ]
 }
 Transition {
  path=AnimGraph.AnimGraphNode_StateMachine_16.FootShift 2ndFoot Translate.AnimStateTransitionNode_17.Transition
  nodes [
   n1 = Get FootShift_2ndFoot_ProcTranslateCurrentlyActive?()
   n2 = Get FootShift_2ndFoot_ProcTransitionSpeed() -> (FootShift_2ndFoot_ProcTransitionSpeed=NewEnumerator0)
   Result(bCanEnterTransition=AND Boolean(A=NOT Boolean(A=@n1), B="Equal (Enum)"(A=@n2, B=NewEnumerator0)))
  ]
 }
 Transition {
  path=AnimGraph.AnimGraphNode_StateMachine_16.FootShift 2ndFoot Translate.AnimStateTransitionNode_18.Transition
  nodes [
   n1 = Get FootShift_2ndFoot_ProcTransitionSpeed() -> (FootShift_2ndFoot_ProcTransitionSpeed=NewEnumerator0)
   Result(bCanEnterTransition=AND Boolean(A=Get FootShift_2ndFoot_ProcTranslateCurrentlyActive?(), B="Equal (Enum)"(A=@n1, B=NewEnumerator1)))
  ]
 }
 Transition {
  path=AnimGraph.AnimGraphNode_StateMachine_16.FootShift 2ndFoot Translate.AnimStateTransitionNode_19.Transition
  nodes [
   n1 = Get FootShift_2ndFoot_ProcTransitionSpeed() -> (FootShift_2ndFoot_ProcTransitionSpeed=NewEnumerator0)
   Result(bCanEnterTransition=AND Boolean(A=Get FootShift_2ndFoot_ProcTranslateCurrentlyActive?(), B="Equal (Enum)"(A=@n1, B=NewEnumerator0)))
  ]
 }
 Transition {
  path=AnimGraph.AnimGraphNode_StateMachine_16.FootShift 2ndFoot Translate.AnimStateTransitionNode_6.Transition
  nodes [Result(bCanEnterTransition=Get DEBUG_FootShift_ResetSystem?())]
 }
 AnimationTransitionGraph_0 {
  path=AnimGraph.AnimGraphNode_StateMachine_16.FootShift 2ndFoot Translate.AnimStateTransitionNode_8.AnimationTransitionGraph_0
  nodes [
   n1 = Get FootShift_2ndFoot_ProcTranslateCurrentlyActive?()
   n2 = Get FootShift_2ndFoot_ProcTransitionSpeed() -> (FootShift_2ndFoot_ProcTransitionSpeed=NewEnumerator0)
   Result(bCanEnterTransition=AND Boolean(A=NOT Boolean(A=@n1), B="Equal (Enum)"(A=@n2, B=NewEnumerator2)))
  ]
 }
 FootShift 1stFoot Lift {
  entry=FootShift 1stFoot Lift RESET
  path=AnimGraph.AnimGraphNode_StateMachine_2.FootShift 1stFoot Lift
  ids=id
  nodes [
   AnimStateNode_0 = FootShift 1stFoot Lift RESET() {
    is=title AnimStateNode
    stateEntered {guid=#1}
    stateFullyBlended {guid=#2}
    stateLeft {guid=#3}
   }
   AnimStateNode_1 = FootShift 1stFoot Lift OFF() {
    is=title AnimStateNode
    stateEntered {guid=#1}
    stateFullyBlended {guid=#2}
    stateLeft {guid=#3}
   }
   AnimStateNode_2 = FootShift 1stFoot Lift ON() {
    is=title AnimStateNode
    stateEntered {guid=#4}
    stateFullyBlended {guid=#5}
    stateLeft {guid=#6}
   }
   AnimStateTransitionNode_0 = AnimStateTransitionNode() {crossfadeDuration=0}
   AnimStateTransitionNode_1 = AnimStateTransitionNode() {
    crossfadeDuration=0
    transitionEnd {guid=#7}
    transitionInterrupt {guid=#8}
    transitionStart {guid=#9}
   }
   AnimStateTransitionNode_10 = AnimStateTransitionNode() {bAllowInertializationForSelfTransitions=true, crossfadeDuration=0}
   AnimStateTransitionNode_6 = AnimStateTransitionNode() {bAllowInertializationForSelfTransitions=true, crossfadeDuration=0}
   AnimStateTransitionNode_9 = AnimStateTransitionNode() {bAllowInertializationForSelfTransitions=true, crossfadeDuration=0}
  ]
  links [
   {
    dst=FootShift 1stFoot Lift ON
    rule=AnimGraph.AnimGraphNode_StateMachine_2.FootShift 1stFoot Lift.AnimStateTransitionNode_0.Transition
    src=FootShift 1stFoot Lift OFF
    via=AnimStateTransitionNode_0
   }
   {
    dst=FootShift 1stFoot Lift RESET
    rule=AnimGraph.AnimGraphNode_StateMachine_2.FootShift 1stFoot Lift.AnimStateTransitionNode_10.Transition
    src=FootShift 1stFoot Lift OFF
    via=AnimStateTransitionNode_10
   }
   {
    dst=FootShift 1stFoot Lift OFF
    rule=AnimGraph.AnimGraphNode_StateMachine_2.FootShift 1stFoot Lift.AnimStateTransitionNode_1.AnimationTransitionGraph_1
    src=FootShift 1stFoot Lift ON
    via=AnimStateTransitionNode_1
   }
   {
    dst=FootShift 1stFoot Lift RESET
    rule=AnimGraph.AnimGraphNode_StateMachine_2.FootShift 1stFoot Lift.AnimStateTransitionNode_6.Transition
    src=FootShift 1stFoot Lift ON
    via=AnimStateTransitionNode_6
   }
   {
    dst=FootShift 1stFoot Lift OFF
    rule=AnimGraph.AnimGraphNode_StateMachine_2.FootShift 1stFoot Lift.AnimStateTransitionNode_9.Transition
    src=FootShift 1stFoot Lift RESET
    via=AnimStateTransitionNode_9
   }
  ]
 }
 FootShift 1stFoot Lift RESET {
  path=AnimGraph.AnimGraphNode_StateMachine_2.FootShift 1stFoot Lift.AnimStateNode_0.FootShift 1stFoot Lift RESET
  nodes [Output Animation Pose()]
 }
 FootShift 1stFoot Lift OFF {
  path=AnimGraph.AnimGraphNode_StateMachine_2.FootShift 1stFoot Lift.AnimStateNode_1.FootShift 1stFoot Lift OFF
  nodes [
   n1 = Get FootShift_1stFoot_ProcAnimPlayRate() -> (FootShift_1stFoot_ProcAnimPlayRate:float)
   n2 = Get FootShift_1stFoot_ProcCurrentSequence() -> (FootShift_1stFoot_ProcCurrentSequence:Anim Sequence Base Object Reference)
   n3 = Get FootShift_Reset_CurrentSequence()
   Output Animation Pose(Result=Blend Poses by bool(
     BlendPose_0=Sequence Player!(Sequence=@n2, PlayRate=@n1) -> (Pose) {bLoopAnimation=false}
     BlendPose_1=Sequence Player(Sequence=@n3, PlayRate=1) {bLoopAnimation=false}
     BlendTime_0=0
     BlendTime_1=0
     bActiveValue=Get DEBUG_FootShift_UseProceduralSystem?()
    ) {blendType=Linear})
  ]
 }
 FootShift 1stFoot Lift ON {
  path=AnimGraph.AnimGraphNode_StateMachine_2.FootShift 1stFoot Lift.AnimStateNode_2.FootShift 1stFoot Lift ON
  nodes [
   n1 = Get FootShift_1stFoot_ProcAnimPlayRate()
   n2 = Get FootShift_1stFoot_ProcCurrentSequence()
   n3 = Get FootShift_CurrentSequence()
   Output Animation Pose(Result=Blend Poses by bool(
     BlendPose_0=Sequence Player(Sequence=@n2, PlayRate=@n1) {bLoopAnimation=false}
     BlendPose_1=Sequence Player(Sequence=@n3, PlayRate=1) {bLoopAnimation=false}
     BlendTime_0=0
     BlendTime_1=0
     bActiveValue=Get DEBUG_FootShift_UseProceduralSystem?()
    ) {blendType=Linear})
  ]
 }
 Transition {
  path=AnimGraph.AnimGraphNode_StateMachine_2.FootShift 1stFoot Lift.AnimStateTransitionNode_0.Transition
  nodes [Result(bCanEnterTransition=Get FootShift_1stFoot_ProcLiftCurrentlyActive?() -> (FootShift_1stFoot_ProcLiftCurrentlyActive?:bool))]
 }
 AnimationTransitionGraph_1 {
  path=AnimGraph.AnimGraphNode_StateMachine_2.FootShift 1stFoot Lift.AnimStateTransitionNode_1.AnimationTransitionGraph_1
  nodes [Result(bCanEnterTransition=NOT Boolean(A=Get FootShift_1stFoot_ProcLiftCurrentlyActive?()))]
 }
 Transition {
  path=AnimGraph.AnimGraphNode_StateMachine_2.FootShift 1stFoot Lift.AnimStateTransitionNode_10.Transition
  nodes [Result(bCanEnterTransition=Get DEBUG_FootShift_ResetSystem?())]
 }
 Transition {
  path=AnimGraph.AnimGraphNode_StateMachine_2.FootShift 1stFoot Lift.AnimStateTransitionNode_6.Transition
  nodes [Result(bCanEnterTransition=Get DEBUG_FootShift_ResetSystem?())]
 }
 Transition {
  path=AnimGraph.AnimGraphNode_StateMachine_2.FootShift 1stFoot Lift.AnimStateTransitionNode_9.Transition
  nodes [Result(bCanEnterTransition=true)]
 }
 RightArm {
  entry=RightArm RESET
  path=AnimGraph.AnimGraphNode_StateMachine_3.RightArm
  ids=id
  nodes [
   AnimStateNode_0 = RightArm RESET() {is=title AnimStateNode}
   AnimStateNode_1 = RightArm OFF() {
    is=title AnimStateNode
    stateEntered {guid=#1}
    stateFullyBlended {guid=#2}
    stateLeft {guid=#3}
   }
   AnimStateNode_5 = RightArm ON() {
    is=title AnimStateNode
    stateEntered {guid=#4}
    stateFullyBlended {guid=#5}
    stateLeft {guid=#6}
   }
   AnimStateTransitionNode_10 = AnimStateTransitionNode() {crossfadeDuration=0.5}
   AnimStateTransitionNode_11 = AnimStateTransitionNode() {crossfadeDuration=0}
   AnimStateTransitionNode_12 = AnimStateTransitionNode() {crossfadeDuration=0}
   AnimStateTransitionNode_8 = AnimStateTransitionNode() {crossfadeDuration=0.5}
   AnimStateTransitionNode_9 = AnimStateTransitionNode() {crossfadeDuration=0.5}
  ]
  links [
   {
    dst=RightArm ON
    rule=AnimGraph.AnimGraphNode_StateMachine_3.RightArm.AnimStateTransitionNode_9.Transition
    src=RightArm OFF
    via=AnimStateTransitionNode_9
   }
   {
    dst=RightArm RESET
    rule=AnimGraph.AnimGraphNode_StateMachine_3.RightArm.AnimStateTransitionNode_11.Transition
    src=RightArm OFF
    via=AnimStateTransitionNode_11
   }
   {
    dst=RightArm OFF
    rule=AnimGraph.AnimGraphNode_StateMachine_3.RightArm.AnimStateTransitionNode_8.Transition
    src=RightArm ON
    via=AnimStateTransitionNode_8
   }
   {
    dst=RightArm RESET
    rule=AnimGraph.AnimGraphNode_StateMachine_3.RightArm.AnimStateTransitionNode_12.Transition
    src=RightArm ON
    via=AnimStateTransitionNode_12
   }
   {
    dst=RightArm OFF
    rule=AnimGraph.AnimGraphNode_StateMachine_3.RightArm.AnimStateTransitionNode_10.Transition
    src=RightArm RESET
    via=AnimStateTransitionNode_10
   }
  ]
 }
 RightArm RESET {
  path=AnimGraph.AnimGraphNode_StateMachine_3.RightArm.AnimStateNode_0.RightArm RESET
  nodes [Output Animation Pose(Result=Additive Identity Pose())]
 }
 RightArm OFF {
  path=AnimGraph.AnimGraphNode_StateMachine_3.RightArm.AnimStateNode_1.RightArm OFF
  nodes [Output Animation Pose(Result=Additive Identity Pose())]
 }
 RightArm ON {
  path=AnimGraph.AnimGraphNode_StateMachine_3.RightArm.AnimStateNode_5.RightArm ON
  nodes [Output Animation Pose(Result=Play Beat Synced Animation(
     ExplicitTime=0
     bShouldLoop=True
     InSequence=Get RightArm_CurrentSequence() -> (RightArm_CurrentSequence:Anim Sequence Base Object Reference)
     TempoMultiplier=1
     bEnableTempoVariants=False
     TempoVariantHysteresis=0.05
     bShowDebug=False
    ))]
 }
 Transition {
  path=AnimGraph.AnimGraphNode_StateMachine_3.RightArm.AnimStateTransitionNode_10.Transition
  nodes [Result(bCanEnterTransition=Get DEBUG_RightArm_ResetSystem?() -> (DEBUG_RightArm_ResetSystem?:bool))]
 }
 Transition {
  path=AnimGraph.AnimGraphNode_StateMachine_3.RightArm.AnimStateTransitionNode_11.Transition
  nodes [Result(bCanEnterTransition=NOT Boolean(A=Get DEBUG_RightArm_ResetSystem?()))]
 }
 Transition {
  path=AnimGraph.AnimGraphNode_StateMachine_3.RightArm.AnimStateTransitionNode_12.Transition
  nodes [Result(bCanEnterTransition=NOT Boolean(A=Get DEBUG_RightArm_ResetSystem?()))]
 }
 Transition {
  path=AnimGraph.AnimGraphNode_StateMachine_3.RightArm.AnimStateTransitionNode_8.Transition
  nodes [Result(bCanEnterTransition=NOT Boolean(A=Get RightArm_CurrentlyActive?()))]
 }
 Transition {
  path=AnimGraph.AnimGraphNode_StateMachine_3.RightArm.AnimStateTransitionNode_9.Transition
  nodes [Result(bCanEnterTransition=Get RightArm_CurrentlyActive?())]
 }
 SpineMotion {
  entry=SpineMotion RESET
  path=AnimGraph.AnimGraphNode_StateMachine_4.SpineMotion
  ids=id
  nodes [
   AnimStateNode_0 = SpineMotion RESET() {
    is=title AnimStateNode
    stateEntered {guid=#1}
    stateFullyBlended {guid=#2}
    stateLeft {guid=#3}
   }
   AnimStateNode_2 = SpineMotion OFF() {
    is=title AnimStateNode
    stateEntered {guid=#1}
    stateFullyBlended {guid=#2}
    stateLeft {guid=#3}
   }
   AnimStateNode_5 = SpineMotion ON() {
    is=title AnimStateNode
    stateEntered {guid=#4}
    stateFullyBlended {guid=#5}
    stateLeft {guid=#6}
   }
   AnimStateTransitionNode_0 = AnimStateTransitionNode() {bAllowInertializationForSelfTransitions=true, crossfadeDuration=1}
   AnimStateTransitionNode_1 = AnimStateTransitionNode() {bAllowInertializationForSelfTransitions=true, crossfadeDuration=0}
   AnimStateTransitionNode_7 = AnimStateTransitionNode() {bAllowInertializationForSelfTransitions=true, crossfadeDuration=0}
   AnimStateTransitionNode_8 = AnimStateTransitionNode() {crossfadeDuration=1}
   AnimStateTransitionNode_9 = AnimStateTransitionNode() {crossfadeDuration=1}
  ]
  links [
   {
    dst=SpineMotion ON
    rule=AnimGraph.AnimGraphNode_StateMachine_4.SpineMotion.AnimStateTransitionNode_9.AnimationTransitionGraph_0
    src=SpineMotion OFF
    via=AnimStateTransitionNode_9
   }
   {
    dst=SpineMotion RESET
    rule=AnimGraph.AnimGraphNode_StateMachine_4.SpineMotion.AnimStateTransitionNode_7.Transition
    src=SpineMotion OFF
    via=AnimStateTransitionNode_7
   }
   {
    dst=SpineMotion OFF
    rule=AnimGraph.AnimGraphNode_StateMachine_4.SpineMotion.AnimStateTransitionNode_8.AnimationTransitionGraph_0
    src=SpineMotion ON
    via=AnimStateTransitionNode_8
   }
   {
    dst=SpineMotion RESET
    rule=AnimGraph.AnimGraphNode_StateMachine_4.SpineMotion.AnimStateTransitionNode_1.Transition
    src=SpineMotion ON
    via=AnimStateTransitionNode_1
   }
   {
    dst=SpineMotion OFF
    rule=AnimGraph.AnimGraphNode_StateMachine_4.SpineMotion.AnimStateTransitionNode_0.Transition
    src=SpineMotion RESET
    via=AnimStateTransitionNode_0
   }
  ]
 }
 SpineMotion RESET {path=AnimGraph.AnimGraphNode_StateMachine_4.SpineMotion.AnimStateNode_0.SpineMotion RESET, nodes [Output Animation Pose()]}
 SpineMotion OFF {
  path=AnimGraph.AnimGraphNode_StateMachine_4.SpineMotion.AnimStateNode_2.SpineMotion OFF
  nodes [
   n1 = Get SpineMotion_PlayRate() -> (SpineMotion_PlayRate:float)
   n2 = Get SpineMotion_Reset_CurrentSequence() -> (SpineMotion_Reset_CurrentSequence:Anim Sequence Base Object Reference)
   Output Animation Pose(Result=Blend Poses by bool(
     BlendPose_0=Additive Identity Pose()
     BlendPose_1=Sequence Player(Sequence=@n2, PlayRate=@n1) {bLoopAnimation=false}
     BlendTime_0=0
     BlendTime_1=0
     bActiveValue=Get DEBUG_SpineMotion_UsePoseBlendSpace?() -> (DEBUG_SpineMotion_UsePoseBlendSpace?:bool)
    ) {blendType=Linear})
  ]
 }
 SpineMotion ON {
  path=AnimGraph.AnimGraphNode_StateMachine_4.SpineMotion.AnimStateNode_5.SpineMotion ON
  nodes [
   n1 = Get SpineMotion_PoseBS_Roll() -> (SpineMotion_PoseBS_Roll:float)
   n2 = Get SpineMotion_PoseBS_Pitch() -> (SpineMotion_PoseBS_Pitch:float)
   Blendspace Player!(X=@n1, Y=@n2, PlayRate:float=0, StartPosition:float=0.65) -> (Pose) {playRate=0, startPosition=0.65}
   Component To Local(ComponentPose="Transform (Modify) Bone - Bone: pelvis"!(
     ComponentPose=Local To Component()
     Alpha=1
     Translation:Vector=Get SpineMotion_CorrectPelvisTranslate() -> (SpineMotion_CorrectPelvisTranslate:Vector)
    ) -> (Pose) {translationMode=BMM_Additive, alphaBoolBlend {blendOption=Linear}, boneToModify {boneName=pelvis}})
   n3 = Get SpineMotion_PlayRate()
   n4 = Get SpineMotion_CurrentSequence() -> (SpineMotion_CurrentSequence:Anim Sequence Base Object Reference)
   AnimGraphNode_BlendSpaceGraph_0 = Blendspace 'Additive_SpineMotion_PoseBS'!(X:float=@n1, Y:float=@n2) -> (Pose:Pose Link Structure) {is=title AnimGraphNode_BlendSpaceGraph}
   Output Animation Pose(Result=Blend Poses by bool(
     BlendPose_0=@AnimGraphNode_BlendSpaceGraph_0
     BlendPose_1=Sequence Player(Sequence=@n4, PlayRate=@n3) {bLoopAnimation=false}
     BlendTime_0=0
     BlendTime_1=0
     bActiveValue=Get DEBUG_SpineMotion_UsePoseBlendSpace?()
    ) {blendType=Linear})
  ]
 }
 Additive_SpineMotion_PoseBS {
  path=AnimGraph.AnimGraphNode_StateMachine_4.SpineMotion.AnimStateNode_5.SpineMotion ON.AnimGraphNode_BlendSpaceGraph_0.Additive_SpineMotion_PoseBS
  links []
  nodes []
 }
 Additive_SpineMotion_PoseBack_A {
  path=AnimGraph.AnimGraphNode_StateMachine_4.SpineMotion.AnimStateNode_5.SpineMotion ON.AnimGraphNode_BlendSpaceGraph_0.Additive_SpineMotion_PoseBS.Additive_SpineMotion_PoseBack_A
  nodes [Output Animation Pose(Result="Sequence Player (additive) 'SpineMotion_PoseBack_A'"() {method=Graph, sequence=/SparksCharacterCommon/Animation/MainPlayer/Additives/SpineMotion_PoseBack_A}) {is=title AnimGraphNode_BlendSpaceSampleResult}]
 }
 Additive_SpineMotion_PoseForward_A {
  path=AnimGraph.AnimGraphNode_StateMachine_4.SpineMotion.AnimStateNode_5.SpineMotion ON.AnimGraphNode_BlendSpaceGraph_0.Additive_SpineMotion_PoseBS.Additive_SpineMotion_PoseForward_A
  nodes [Output Animation Pose(Result="Sequence Player (additive) 'SpineMotion_PoseForward_A'"() {method=Graph, sequence=/SparksCharacterCommon/Animation/MainPlayer/Additives/SpineMotion_PoseForward_A})]
 }
 Additive_SpineMotion_PoseLeft_A {
  path=AnimGraph.AnimGraphNode_StateMachine_4.SpineMotion.AnimStateNode_5.SpineMotion ON.AnimGraphNode_BlendSpaceGraph_0.Additive_SpineMotion_PoseBS.Additive_SpineMotion_PoseLeft_A
  nodes [Output Animation Pose(Result="Sequence Player (additive) 'SpineMotion_PoseLeft_A'"() {method=Graph, sequence=/SparksCharacterCommon/Animation/MainPlayer/Additives/SpineMotion_PoseLeft_A})]
 }
 Additive_SpineMotion_PoseRight_A {
  path=AnimGraph.AnimGraphNode_StateMachine_4.SpineMotion.AnimStateNode_5.SpineMotion ON.AnimGraphNode_BlendSpaceGraph_0.Additive_SpineMotion_PoseBS.Additive_SpineMotion_PoseRight_A
  nodes [Output Animation Pose(Result="Sequence Player (additive) 'SpineMotion_PoseRight_A'"() {method=Graph, sequence=/SparksCharacterCommon/Animation/MainPlayer/Additives/SpineMotion_PoseRight_A})]
 }
 Transition {
  path=AnimGraph.AnimGraphNode_StateMachine_4.SpineMotion.AnimStateTransitionNode_0.Transition
  nodes [Result(bCanEnterTransition=Get SpineMotion_CurrentlyActive?())]
 }
 Transition {
  path=AnimGraph.AnimGraphNode_StateMachine_4.SpineMotion.AnimStateTransitionNode_1.Transition
  nodes [Result(bCanEnterTransition=Get DEBUG_SpineMotion_ResetSystem?() -> (DEBUG_SpineMotion_ResetSystem?:bool))]
 }
 Transition {
  path=AnimGraph.AnimGraphNode_StateMachine_4.SpineMotion.AnimStateTransitionNode_7.Transition
  nodes [Result(bCanEnterTransition=Get DEBUG_SpineMotion_ResetSystem?())]
 }
 AnimationTransitionGraph_0 {
  path=AnimGraph.AnimGraphNode_StateMachine_4.SpineMotion.AnimStateTransitionNode_8.AnimationTransitionGraph_0
  nodes [Result(bCanEnterTransition=NOT Boolean(A=Get SpineMotion_CurrentlyActive?()))]
 }
 AnimationTransitionGraph_0 {
  path=AnimGraph.AnimGraphNode_StateMachine_4.SpineMotion.AnimStateTransitionNode_9.AnimationTransitionGraph_0
  nodes [Result(bCanEnterTransition=Get SpineMotion_CurrentlyActive?())]
 }
 HeadMotion {
  entry=HeadMotion RESET
  path=AnimGraph.AnimGraphNode_StateMachine_5.HeadMotion
  ids=id
  nodes [
   AnimStateNode_0 = HeadMotion RESET() {
    is=title AnimStateNode
    stateEntered {guid=#1}
    stateFullyBlended {guid=#2}
    stateLeft {guid=#3}
   }
   AnimStateNode_1 = HeadMotion OFF() {
    is=title AnimStateNode
    stateEntered {guid=#1}
    stateFullyBlended {guid=#2}
    stateLeft {guid=#3}
   }
   AnimStateNode_5 = HeadMotion ON() {
    is=title AnimStateNode
    stateEntered {guid=#4}
    stateFullyBlended {guid=#5}
    stateLeft {guid=#6}
   }
   AnimStateTransitionNode_0 = AnimStateTransitionNode() {bAllowInertializationForSelfTransitions=true, crossfadeDuration=0}
   AnimStateTransitionNode_10 = AnimStateTransitionNode() {bAllowInertializationForSelfTransitions=true, crossfadeDuration=0}
   AnimStateTransitionNode_11 = AnimStateTransitionNode() {bAllowInertializationForSelfTransitions=true, crossfadeDuration=0}
   AnimStateTransitionNode_8 = AnimStateTransitionNode() {
    crossfadeDuration=1
    transitionEnd {guid=#28}
    transitionInterrupt {guid=#29}
    transitionStart {guid=#30}
   }
   AnimStateTransitionNode_9 = AnimStateTransitionNode() {
    crossfadeDuration=1
    transitionEnd {guid=#31}
    transitionInterrupt {guid=#32}
    transitionStart {guid=#33}
   }
  ]
  links [
   {
    dst=HeadMotion ON
    rule=AnimGraph.AnimGraphNode_StateMachine_5.HeadMotion.AnimStateTransitionNode_9.Transition
    src=HeadMotion OFF
    via=AnimStateTransitionNode_9
   }
   {
    dst=HeadMotion RESET
    rule=AnimGraph.AnimGraphNode_StateMachine_5.HeadMotion.AnimStateTransitionNode_11.Transition
    src=HeadMotion OFF
    via=AnimStateTransitionNode_11
   }
   {
    dst=HeadMotion OFF
    rule=AnimGraph.AnimGraphNode_StateMachine_5.HeadMotion.AnimStateTransitionNode_8.Transition
    src=HeadMotion ON
    via=AnimStateTransitionNode_8
   }
   {
    dst=HeadMotion RESET
    rule=AnimGraph.AnimGraphNode_StateMachine_5.HeadMotion.AnimStateTransitionNode_10.Transition
    src=HeadMotion ON
    via=AnimStateTransitionNode_10
   }
   {
    dst=HeadMotion OFF
    rule=AnimGraph.AnimGraphNode_StateMachine_5.HeadMotion.AnimStateTransitionNode_0.Transition
    src=HeadMotion RESET
    via=AnimStateTransitionNode_0
   }
  ]
 }
 HeadMotion RESET {
  path=AnimGraph.AnimGraphNode_StateMachine_5.HeadMotion.AnimStateNode_0.HeadMotion RESET
  nodes [Output Animation Pose() {is=title AnimGraphNode_StateResult}]
 }
 HeadMotion OFF {path=AnimGraph.AnimGraphNode_StateMachine_5.HeadMotion.AnimStateNode_1.HeadMotion OFF, nodes [Output Animation Pose()]}
 HeadMotion ON {
  path=AnimGraph.AnimGraphNode_StateMachine_5.HeadMotion.AnimStateNode_5.HeadMotion ON
  nodes [
   n1 = Get HeadMotion_CurrentSequenceIsLooping?() -> (HeadMotion_CurrentSequenceIsLooping?:bool)
   n2 = Get HeadMotion_CurrentSequence() -> (HeadMotion_CurrentSequence:Anim Sequence Base Object Reference)
   Output Animation Pose(Result=Blend Poses by bool(
     BlendPose_0=Play Beat Synced Animation(
      ExplicitTime=0
      bShouldLoop=@n1
      InSequence=@n2
      TempoMultiplier=1
      bEnableTempoVariants=False
      TempoVariantHysteresis=0.05
      bShowDebug=False
     )
     BlendPose_1=Sequence Player!(Sequence=@n2, bLoopAnimation:bool=@n1) -> (Pose)
     BlendTime_0=0
     BlendTime_1=0
     bActiveValue=@n1
    ))
  ]
 }
 Transition {
  path=AnimGraph.AnimGraphNode_StateMachine_5.HeadMotion.AnimStateTransitionNode_0.Transition
  nodes [Result(bCanEnterTransition=Get HeadMotion_CurrentlyActive?())]
 }
 Transition {
  path=AnimGraph.AnimGraphNode_StateMachine_5.HeadMotion.AnimStateTransitionNode_10.Transition
  nodes [Result(bCanEnterTransition=Get DEBUG_HeadMotion_ResetSystem?() -> (DEBUG_HeadMotion_ResetSystem?:bool))]
 }
 Transition {
  path=AnimGraph.AnimGraphNode_StateMachine_5.HeadMotion.AnimStateTransitionNode_11.Transition
  nodes [Result(bCanEnterTransition=Get DEBUG_HeadMotion_ResetSystem?())]
 }
 Transition {
  path=AnimGraph.AnimGraphNode_StateMachine_5.HeadMotion.AnimStateTransitionNode_8.Transition
  nodes [Result(bCanEnterTransition=NOT Boolean(A=Get HeadMotion_CurrentlyActive?()))]
 }
 Transition {
  path=AnimGraph.AnimGraphNode_StateMachine_5.HeadMotion.AnimStateTransitionNode_9.Transition
  nodes [Result(bCanEnterTransition=Get HeadMotion_CurrentlyActive?())]
 }
 Instrument Mesh Chooser {
  entry=Instrument Mesh Chooser State
  path=AnimGraph.AnimGraphNode_StateMachine_6.Instrument Mesh Chooser
  ids=id
  nodes [AnimStateNode_0 = Instrument Mesh Chooser State() {is=title AnimStateNode}]
 }
 Instrument Mesh Chooser State {
  path=AnimGraph.AnimGraphNode_StateMachine_6.Instrument Mesh Chooser.AnimStateNode_0.Instrument Mesh Chooser State
  nodes [
   n1 = "Blend Poses (ESparksInstrumentType)"!(
    BlendPose_0:Pose Link Structure=Additive Identity Pose()
    BlendPose_1:Pose Link Structure="Sequence Player (additive) 'Guitar_SpawnInstrument_Pose'"() {sequence=/SparksCharacterCommon/Animation/MainPlayer/BandPlay/Guitar/Guitar_SpawnInstrument_Pose}
    BlendPose_2:Pose Link Structure="Sequence Player (additive) 'Bass_SpawnInstrument_Pose'"() {sequence=/SparksCharacterCommon/Animation/MainPlayer/BandPlay/Bass/Bass_SpawnInstrument_Pose}
    BlendPose_3:Pose Link Structure="Sequence Player (additive) 'Mic_SpawnInstrument_Pose'"() {sequence=/SparksCharacterCommon/Animation/MainPlayer/BandPlay/Mic/Mic_SpawnInstrument_Pose}
    BlendPose_4:Pose Link Structure="Sequence Player (additive) 'Drum_SpawnInstrument_Pose'"() {sequence=/SparksCharacterCommon/Animation/MainPlayer/BandPlay/Drum/Drum_SpawnInstrument_Pose}
    BlendPose_5:Pose Link Structure="Sequence Player (additive) 'Keytar_SpawnInstrument_Pose'"() {sequence=/SparksCharacterCommon/Animation/MainPlayer/BandPlay/Keytar/Keytar_SpawnInstrument_Pose}
    BlendPose_6:Pose Link Structure=Additive Identity Pose()
    BlendTime_0:float=0
    BlendTime_1:float=0
    BlendTime_2:float=0
    BlendTime_3:float=0
    BlendTime_4:float=0
    BlendTime_5:float=0
    BlendTime_6:float=0
    ActiveEnumValue:Byte=0
   ) -> (Pose:Pose Link Structure) {
    is=title AnimGraphNode_BlendListByEnum
    blendType=Linear
    blendPose [
     {linkId=-1, sourceLinkId=-1}
     {linkId=-1, sourceLinkId=-1}
     {linkId=-1, sourceLinkId=-1}
     {linkId=-1, sourceLinkId=-1}
     {linkId=-1, sourceLinkId=-1}
     {linkId=-1, sourceLinkId=-1}
     {linkId=-1, sourceLinkId=-1}
    ]
    blendTime [0.1,0.1,0.1,0.1,0.1,0.1,0.1]
   }
   n2 = "Blend Poses (ESparksInstrumentType)"(
    BlendPose_0=Additive Identity Pose()
    BlendPose_1="Sequence Player (additive) 'Guitar_SpawnInstrument_Pose'"() {sequence=&Guitar_SpawnInstrument_Pose}
    BlendPose_2="Sequence Player (additive) 'Bass_SpawnInstrument_Pose'"() {sequence=&Bass_SpawnInstrument_Pose}
    BlendPose_3="Sequence Player (additive) 'Mic_SpawnInstrument_Pose'"() {sequence=&Mic_SpawnInstrument_Pose}
    BlendPose_4="Sequence Player (additive) 'Drum_SpawnInstrument_Pose'"() {sequence=&Drum_SpawnInstrument_Pose}
    BlendPose_5="Sequence Player (additive) 'Keytar_SpawnInstrument_Pose'"() {sequence=&Keytar_SpawnInstrument_Pose}
    BlendPose_6=Additive Identity Pose()
    BlendTime_0=0
    BlendTime_1=0
    BlendTime_2=0
    BlendTime_3=0
    BlendTime_4=0
    BlendTime_5=0
    BlendTime_6=0
    ActiveEnumValue=0
   ) {
    blendType=Linear
    blendPose [
     {linkId=-1, sourceLinkId=-1}
     {linkId=-1, sourceLinkId=-1}
     {linkId=-1, sourceLinkId=-1}
     {linkId=-1, sourceLinkId=-1}
     {linkId=-1, sourceLinkId=-1}
     {linkId=-1, sourceLinkId=-1}
     {linkId=-1, sourceLinkId=-1}
    ]
    blendTime [0.1,0.1,0.1,0.1,0.1,0.1,0.1]
   }
   Output Animation Pose(Result=Blend Poses by bool(BlendPose_0=@n2, BlendPose_1=@n1, BlendTime_0=0, BlendTime_1=0) {blendType=Linear})
  ]
 }
 FootShift 1stPelvis Translate {
  entry=FootShift 1stPelvis Translate RESET
  path=AnimGraph.AnimGraphNode_StateMachine_7.FootShift 1stPelvis Translate
  ids=id
  nodes [
   AnimStateNode_0 = FootShift 1stPelvis Translate RESET() {
    is=title AnimStateNode
    stateEntered {guid=#1}
    stateFullyBlended {guid=#2}
    stateLeft {guid=#3}
   }
   AnimStateNode_1 = FootShift 1stPelvis Translate ON() {
    is=title AnimStateNode
    stateEntered {guid=#4}
    stateFullyBlended {guid=#5}
    stateLeft {guid=#6}
   }
   AnimStateNode_3 = FootShift 1stPelvis Translate OFF() {
    is=title AnimStateNode
    stateEntered {guid=#1}
    stateFullyBlended {guid=#2}
    stateLeft {guid=#3}
   }
   AnimStateTransitionNode_0 = AnimStateTransitionNode() {crossfadeDuration=0.53, customBlendCurve=&FootShift_Pelvis_ProceduralTranslate}
   AnimStateTransitionNode_10 = AnimStateTransitionNode() {bAllowInertializationForSelfTransitions=true, crossfadeDuration=0}
   AnimStateTransitionNode_11 = AnimStateTransitionNode() {bAllowInertializationForSelfTransitions=true, crossfadeDuration=0}
   AnimStateTransitionNode_16 = AnimStateTransitionNode() {
    bAllowInertializationForSelfTransitions=true
    crossfadeDuration=0.795
    customBlendCurve=&FootShift_Pelvis_ProceduralTranslate
    priorityOrder=2
    transitionEnd {guid=#19}
    transitionInterrupt {guid=#20}
    transitionStart {guid=#21}
   }
   AnimStateTransitionNode_17 = AnimStateTransitionNode() {
    bAllowInertializationForSelfTransitions=true
    crossfadeDuration=1.06
    customBlendCurve=&FootShift_Pelvis_ProceduralTranslate
    priorityOrder=3
    transitionEnd {guid=#22}
    transitionInterrupt {guid=#23}
    transitionStart {guid=#24}
   }
   AnimStateTransitionNode_18 = AnimStateTransitionNode() {
    bAllowInertializationForSelfTransitions=true
    crossfadeDuration=0.795
    customBlendCurve=&FootShift_Pelvis_ProceduralTranslate
    priorityOrder=2
   }
   AnimStateTransitionNode_19 = AnimStateTransitionNode() {
    bAllowInertializationForSelfTransitions=true
    crossfadeDuration=1.06
    customBlendCurve=&FootShift_Pelvis_ProceduralTranslate
    priorityOrder=3
    transitionEnd {guid=#34}
    transitionInterrupt {guid=#35}
    transitionStart {guid=#36}
   }
   AnimStateTransitionNode_6 = AnimStateTransitionNode() {
    bAllowInertializationForSelfTransitions=true
    crossfadeDuration=0
    transitionEnd {guid=#25}
    transitionInterrupt {guid=#26}
    transitionStart {guid=#27}
   }
   AnimStateTransitionNode_8 = AnimStateTransitionNode() {
    crossfadeDuration=0.53
    customBlendCurve=&FootShift_Pelvis_ProceduralTranslate
    transitionEnd {guid=#7}
    transitionInterrupt {guid=#8}
    transitionStart {guid=#9}
   }
  ]
  links [
   {
    dst=FootShift 1stPelvis Translate ON
    rule=AnimGraph.AnimGraphNode_StateMachine_7.FootShift 1stPelvis Translate.AnimStateTransitionNode_0.Transition
    src=FootShift 1stPelvis Translate OFF
    via=AnimStateTransitionNode_0
   }
   {
    dst=FootShift 1stPelvis Translate ON
    rule=AnimGraph.AnimGraphNode_StateMachine_7.FootShift 1stPelvis Translate.AnimStateTransitionNode_18.Transition
    src=FootShift 1stPelvis Translate OFF
    via=AnimStateTransitionNode_18
   }
   {
    dst=FootShift 1stPelvis Translate ON
    rule=AnimGraph.AnimGraphNode_StateMachine_7.FootShift 1stPelvis Translate.AnimStateTransitionNode_19.Transition
    src=FootShift 1stPelvis Translate OFF
    via=AnimStateTransitionNode_19
   }
   {
    dst=FootShift 1stPelvis Translate RESET
    rule=AnimGraph.AnimGraphNode_StateMachine_7.FootShift 1stPelvis Translate.AnimStateTransitionNode_10.Transition
    src=FootShift 1stPelvis Translate OFF
    via=AnimStateTransitionNode_10
   }
   {
    dst=FootShift 1stPelvis Translate OFF
    rule=AnimGraph.AnimGraphNode_StateMachine_7.FootShift 1stPelvis Translate.AnimStateTransitionNode_16.Transition
    src=FootShift 1stPelvis Translate ON
    via=AnimStateTransitionNode_16
   }
   {
    dst=FootShift 1stPelvis Translate OFF
    rule=AnimGraph.AnimGraphNode_StateMachine_7.FootShift 1stPelvis Translate.AnimStateTransitionNode_17.Transition
    src=FootShift 1stPelvis Translate ON
    via=AnimStateTransitionNode_17
   }
   {
    dst=FootShift 1stPelvis Translate OFF
    rule=AnimGraph.AnimGraphNode_StateMachine_7.FootShift 1stPelvis Translate.AnimStateTransitionNode_8.AnimationTransitionGraph_0
    src=FootShift 1stPelvis Translate ON
    via=AnimStateTransitionNode_8
   }
   {
    dst=FootShift 1stPelvis Translate RESET
    rule=AnimGraph.AnimGraphNode_StateMachine_7.FootShift 1stPelvis Translate.AnimStateTransitionNode_6.Transition
    src=FootShift 1stPelvis Translate ON
    via=AnimStateTransitionNode_6
   }
   {
    dst=FootShift 1stPelvis Translate OFF
    rule=AnimGraph.AnimGraphNode_StateMachine_7.FootShift 1stPelvis Translate.AnimStateTransitionNode_11.Transition
    src=FootShift 1stPelvis Translate RESET
    via=AnimStateTransitionNode_11
   }
  ]
 }
 FootShift 1stPelvis Translate RESET {
  path=AnimGraph.AnimGraphNode_StateMachine_7.FootShift 1stPelvis Translate.AnimStateNode_0.FootShift 1stPelvis Translate RESET
  nodes [Output Animation Pose(Result=Use cached pose 'BaseBody Post-Additives Pose'())]
 }
 FootShift 1stPelvis Translate ON {
  path=AnimGraph.AnimGraphNode_StateMachine_7.FootShift 1stPelvis Translate.AnimStateNode_1.FootShift 1stPelvis Translate ON
  nodes [
   n1 = "Transform (Modify) Bone - Bone: pelvis"!(
    ComponentPose=Local To Component(LocalPose=Use cached pose 'BaseBody Post-Additives Pose'())
    Alpha=1
    Translation_X:double=Get FootShift_1stPelvis_ProcTranslateLR()
    Translation_Y:double=Get FootShift_1stPelvis_ProcTranslateFB()
    Translation_Z:double=Get FootShift_1stPelvis_ProcTranslateVertical()
    Rotation_Roll:float=0
    Rotation_Pitch:float=0
    Rotation_Yaw:float=Get FootShift_1stPelvis_ProcRotateTwist()
   ) -> (Pose) {
    rotationMode=BMM_Additive
    translationMode=BMM_Additive
    alphaBoolBlend {blendOption=Linear}
    boneToModify {boneName=pelvis}
    translation=<30 30>
   }
   Output Animation Pose(Result=Component To Local(ComponentPose=@n1))
  ]
 }
 FootShift 1stPelvis Translate OFF {
  path=AnimGraph.AnimGraphNode_StateMachine_7.FootShift 1stPelvis Translate.AnimStateNode_3.FootShift 1stPelvis Translate OFF
  nodes [
   n1 = "Transform (Modify) Bone - Bone: pelvis"(
    ComponentPose=Local To Component(LocalPose=Use cached pose 'BaseBody Post-Additives Pose'())
    Alpha=1
    Translation_X=0
    Translation_Y=0
    Translation_Z=0
    Rotation_Roll=0
    Rotation_Pitch=0
    Rotation_Yaw=0
   ) {
    rotationMode=BMM_Additive
    translationMode=BMM_Additive
    alphaBoolBlend {blendOption=Linear}
    boneToModify {boneName=pelvis}
    translation=<30 30>
   }
   Output Animation Pose(Result=Component To Local(ComponentPose=@n1))
  ]
 }
 Transition {
  path=AnimGraph.AnimGraphNode_StateMachine_7.FootShift 1stPelvis Translate.AnimStateTransitionNode_0.Transition
  nodes [
   n1 = Get FootShift_1stFoot_ProcTransitionSpeed() -> (FootShift_1stFoot_ProcTransitionSpeed=NewEnumerator0)
   Result(bCanEnterTransition=AND Boolean(A=Get FootShift_1stFoot_ProcLiftCurrentlyActive?(), B="Equal (Enum)"(A=@n1, B=NewEnumerator2)))
  ]
 }
 Transition {
  path=AnimGraph.AnimGraphNode_StateMachine_7.FootShift 1stPelvis Translate.AnimStateTransitionNode_10.Transition
  nodes [Result(bCanEnterTransition=Get DEBUG_FootShift_ResetSystem?())]
 }
 Transition {
  path=AnimGraph.AnimGraphNode_StateMachine_7.FootShift 1stPelvis Translate.AnimStateTransitionNode_11.Transition
  nodes [Result(bCanEnterTransition=true)]
 }
 Transition {
  path=AnimGraph.AnimGraphNode_StateMachine_7.FootShift 1stPelvis Translate.AnimStateTransitionNode_16.Transition
  nodes [
   n1 = Get FootShift_1stFoot_ProcLiftCurrentlyActive?()
   n2 = Get FootShift_1stFoot_ProcTransitionSpeed() -> (FootShift_1stFoot_ProcTransitionSpeed=NewEnumerator0)
   Result(bCanEnterTransition=AND Boolean(A=NOT Boolean(A=@n1), B="Equal (Enum)"(A=@n2, B=NewEnumerator1)))
  ]
 }
 Transition {
  path=AnimGraph.AnimGraphNode_StateMachine_7.FootShift 1stPelvis Translate.AnimStateTransitionNode_17.Transition
  nodes [
   n1 = Get FootShift_1stFoot_ProcLiftCurrentlyActive?()
   n2 = Get FootShift_1stFoot_ProcTransitionSpeed() -> (FootShift_1stFoot_ProcTransitionSpeed=NewEnumerator0)
   Result(bCanEnterTransition=AND Boolean(A=NOT Boolean(A=@n1), B="Equal (Enum)"(A=@n2, B=NewEnumerator0)))
  ]
 }
 Transition {
  path=AnimGraph.AnimGraphNode_StateMachine_7.FootShift 1stPelvis Translate.AnimStateTransitionNode_18.Transition
  nodes [
   n1 = Get FootShift_1stFoot_ProcTransitionSpeed() -> (FootShift_1stFoot_ProcTransitionSpeed=NewEnumerator0)
   Result(bCanEnterTransition=AND Boolean(A=Get FootShift_1stFoot_ProcLiftCurrentlyActive?(), B="Equal (Enum)"(A=@n1, B=NewEnumerator1)))
  ]
 }
 Transition {
  path=AnimGraph.AnimGraphNode_StateMachine_7.FootShift 1stPelvis Translate.AnimStateTransitionNode_19.Transition
  nodes [
   n1 = Get FootShift_1stFoot_ProcTransitionSpeed() -> (FootShift_1stFoot_ProcTransitionSpeed=NewEnumerator0)
   Result(bCanEnterTransition=AND Boolean(A=Get FootShift_1stFoot_ProcLiftCurrentlyActive?(), B="Equal (Enum)"(A=@n1, B=NewEnumerator0)))
  ]
 }
 Transition {
  path=AnimGraph.AnimGraphNode_StateMachine_7.FootShift 1stPelvis Translate.AnimStateTransitionNode_6.Transition
  nodes [Result(bCanEnterTransition=Get DEBUG_FootShift_ResetSystem?())]
 }
 AnimationTransitionGraph_0 {
  path=AnimGraph.AnimGraphNode_StateMachine_7.FootShift 1stPelvis Translate.AnimStateTransitionNode_8.AnimationTransitionGraph_0
  nodes [
   n1 = Get FootShift_1stFoot_ProcLiftCurrentlyActive?()
   n2 = Get FootShift_1stFoot_ProcTransitionSpeed() -> (FootShift_1stFoot_ProcTransitionSpeed=NewEnumerator0)
   Result(bCanEnterTransition=AND Boolean(A=NOT Boolean(A=@n1), B="Equal (Enum)"(A=@n2, B=NewEnumerator2)))
  ]
 }
 BaseBody {
  entry=Play 0
  path=AnimGraph.AnimGraphNode_StateMachine_8.BaseBody
  ids=id
  nodes [
   AnimStateNode_31 = Play 0() {
    is=title AnimStateNode
    stateEntered {guid=#37}
    stateFullyBlended {guid=#38}
    stateLeft {guid=#39}
   }
   AnimStateNode_8 = Play 1() {
    is=title AnimStateNode
    stateEntered {guid=#37}
    stateFullyBlended {guid=#38}
    stateLeft {guid=#39}
   }
   AnimStateTransitionNode_12 = AnimStateTransitionNode() {crossfadeDuration=0.6}
   AnimStateTransitionNode_14 = AnimStateTransitionNode() {crossfadeDuration=0.6}
  ]
  links [
   {
    dst=Play 1
    rule=AnimGraph.AnimGraphNode_StateMachine_8.BaseBody.AnimStateTransitionNode_12.Transition
    src=Play 0
    via=AnimStateTransitionNode_12
   }
   {
    dst=Play 0
    rule=AnimGraph.AnimGraphNode_StateMachine_8.BaseBody.AnimStateTransitionNode_14.Transition
    src=Play 1
    via=AnimStateTransitionNode_14
   }
  ]
 }
 Play 0 {
  path=AnimGraph.AnimGraphNode_StateMachine_8.BaseBody.AnimStateNode_31.Play 0
  nodes [
   n1 = Get PlaySequences() -> (PlaySequences:Array of Anim Sequence Object References)
   n2 = Get TempoMultipliers() -> (TempoMultipliers:"Array of Float (double-precision)s")
   Output Animation Pose(Result=Play Beat Synced Animation(
     ExplicitTime=0
     bShouldLoop=True
     InSequence="Get (a copy)"(Array:Array of Anim Sequence Object References=@n1) -> (Output:Anim Sequence Object Reference)
     TempoMultiplier="Get (a copy)"(Array:"Array of Float (double-precision)s"=@n2) -> (Output:double)
     bEnableTempoVariants=False
     TempoVariantHysteresis=0.05
     bShowDebug=False
    )) {
    node {stateEntryFunction {functionName=OnPlay0Entry}}
    stateEntryFunction {bSelfContext=true, memberGuid=#40, memberName=OnPlay0Entry}
   }
  ]
 }
 Play 1 {
  path=AnimGraph.AnimGraphNode_StateMachine_8.BaseBody.AnimStateNode_8.Play 1
  nodes [
   n1 = Get PlaySequences()
   n2 = Get TempoMultipliers()
   Output Animation Pose(Result=Play Beat Synced Animation(
     ExplicitTime=0
     bShouldLoop=True
     InSequence="Get (a copy)"(Array:Array of Anim Sequence Object References=@n1, Dimension 1=1) -> (Output:Anim Sequence Object Reference)
     TempoMultiplier="Get (a copy)"(Array:"Array of Float (double-precision)s"=@n2, Dimension 1=1) -> (Output:double)
     bEnableTempoVariants=False
     TempoVariantHysteresis=0.05
     bShowDebug=False
    )) {
    node {stateEntryFunction {functionName=OnPlay1Entry}}
    stateEntryFunction {bSelfContext=true, memberGuid=#41, memberName=OnPlay1Entry}
   }
  ]
 }
 Transition {
  path=AnimGraph.AnimGraphNode_StateMachine_8.BaseBody.AnimStateTransitionNode_12.Transition
  nodes [Result(bCanEnterTransition=JamShouldTransition!(self:Sparks Instrument Anim Instance Object Reference) -> (ReturnValue:bool))]
 }
 Transition {
  path=AnimGraph.AnimGraphNode_StateMachine_8.BaseBody.AnimStateTransitionNode_14.Transition
  nodes [Result(bCanEnterTransition=JamShouldTransition())]
 }
 FaceExpression {
  entry=FaceExpression ON
  path=AnimGraph.AnimGraphNode_StateMachine_9.FaceExpression
  ids=id
  nodes [AnimStateNode_1 = FaceExpression ON() {
    is=title AnimStateNode
    stateEntered {guid=#42}
    stateFullyBlended {guid=#43}
    stateLeft {guid=#44}
   }]
 }
 FaceExpression ON {
  path=AnimGraph.AnimGraphNode_StateMachine_9.FaceExpression.AnimStateNode_1.FaceExpression ON
  nodes [
   n1 = Get DEBUG_FaceExpression_OverrideSequence() -> (DEBUG_FaceExpression_OverrideSequence:Anim Sequence Base Object Reference)
   n2 = Get Song_GenreBlendTime() -> (Song_GenreBlendTime:float)
   n3 = Get Song_IntensityBlendTime() -> (Song_IntensityBlendTime:float)
   n4 = Blend Poses by int(
    BlendPose_0=Sequence Player 'FaceExpression_HipHop_Light'!(PlayRate:float=1, StartPosition:float=0) -> (Pose) {bLoopAnimation=false, sequence=/SparksCharacterCommon/Animation/MainPlayer/Additives/FaceExpression_HipHop_Light}
    BlendPose_1=Sequence Player 'FaceExpression_HipHop_Medium'!(PlayRate:float=1, StartPosition:float=0) -> (Pose) {bLoopAnimation=false, sequence=/SparksCharacterCommon/Animation/MainPlayer/Additives/FaceExpression_HipHop_Medium}
    BlendPose_2=Sequence Player 'FaceExpression_HipHop_Heavy'!(PlayRate:float=1, StartPosition:float=0) -> (Pose) {bLoopAnimation=false, sequence=/SparksCharacterCommon/Animation/MainPlayer/Additives/FaceExpression_HipHop_Heavy}
    BlendTime_0=@n3
    BlendTime_1=@n3
    BlendTime_2=@n3
    ActiveChildIndex=Get Song_IntensityIndex() -> (Song_IntensityIndex:int=0)
   ) {
    blendType=Linear
    blendPose [{linkId=-1, sourceLinkId=-1}, {linkId=-1, sourceLinkId=-1}, {linkId=-1, sourceLinkId=-1}]
    blendTime [0.1,0.1,0.1]
   }
   n5 = Get Song_IntensityBlendTime()
   n6 = Blend Poses by int(
    BlendPose_0=Sequence Player 'FaceExpression_Rock_Light'!(PlayRate:float=1, StartPosition:float=0) -> (Pose) {bLoopAnimation=false, sequence=/SparksCharacterCommon/Animation/MainPlayer/Additives/FaceExpression_Rock_Light}
    BlendPose_1=Sequence Player 'FaceExpression_Rock_Medium'!(PlayRate:float=1, StartPosition:float=0) -> (Pose) {bLoopAnimation=false, sequence=/SparksCharacterCommon/Animation/MainPlayer/Additives/FaceExpression_Rock_Medium}
    BlendPose_2=Sequence Player 'FaceExpression_Rock_Heavy'!(PlayRate:float=1, StartPosition:float=0) -> (Pose) {bLoopAnimation=false, sequence=/SparksCharacterCommon/Animation/MainPlayer/Additives/FaceExpression_Rock_Heavy}
    BlendTime_0=@n5
    BlendTime_1=@n5
    BlendTime_2=@n5
    ActiveChildIndex=Get Song_IntensityIndex() -> (Song_IntensityIndex=0)
   ) {
    blendType=Linear
    blendPose [{linkId=-1, sourceLinkId=-1}, {linkId=-1, sourceLinkId=-1}, {linkId=-1, sourceLinkId=-1}]
    blendTime [0.1,0.1,0.1]
   }
   n7 = Get Song_IntensityBlendTime()
   n8 = Blend Poses by int(
    BlendPose_0=Sequence Player 'FaceExpression_Pop_Light'!(PlayRate:float=1, StartPosition:float=0) -> (Pose) {bLoopAnimation=false, sequence=/SparksCharacterCommon/Animation/MainPlayer/Additives/FaceExpression_Pop_Light}
    BlendPose_1=Sequence Player 'FaceExpression_Pop_Medium'!(PlayRate:float=1, StartPosition:float=0) -> (Pose) {bLoopAnimation=false, sequence=/SparksCharacterCommon/Animation/MainPlayer/Additives/FaceExpression_Pop_Medium}
    BlendPose_2=Sequence Player 'FaceExpression_Pop_Heavy'!(PlayRate:float=1, StartPosition:float=0) -> (Pose) {bLoopAnimation=false, sequence=/SparksCharacterCommon/Animation/MainPlayer/Additives/FaceExpression_Pop_Heavy}
    BlendTime_0=@n7
    BlendTime_1=@n7
    BlendTime_2=@n7
    ActiveChildIndex=Get Song_IntensityIndex() -> (Song_IntensityIndex=0)
   ) {
    blendType=Linear
    blendPose [{linkId=-1, sourceLinkId=-1}, {linkId=-1, sourceLinkId=-1}, {linkId=-1, sourceLinkId=-1}]
    blendTime [0.1,0.1,0.1]
   }
   n9 = Blend Poses by int!(
    BlendPose_0=Sequence Player 'FaceExpression_Default'!(PlayRate:float=1, StartPosition:float=0) -> (Pose) {sequence=/SparksCharacterCommon/Animation/MainPlayer/Additives/FaceExpression_Default}
    BlendPose_1=@n4
    BlendPose_2=@n8
    BlendPose_3:Pose Link Structure=@n6
    BlendTime_0=@n2
    BlendTime_1=@n2
    BlendTime_2=@n2
    BlendTime_3:float=@n2
    ActiveChildIndex=Get Song_GenreIndex() -> (Song_GenreIndex:int=0)
   ) -> (Pose) {
    blendType=Linear
    blendPose [{linkId=-1, sourceLinkId=-1}, {linkId=-1, sourceLinkId=-1}, {linkId=-1, sourceLinkId=-1}, {linkId=-1, sourceLinkId=-1}]
    blendTime [0.1,0.1,0.1,0.1]
   }
   Output Animation Pose(Result=Blend Poses by bool(
     BlendPose_0=Sequence Player!(Sequence=@n1) -> (Pose) {bLoopAnimation=false}
     BlendPose_1=@n9
     BlendTime_0=0
     BlendTime_1=0
     bActiveValue=Get DEBUG_FaceExpression_UseOverrideSequence?() -> (DEBUG_FaceExpression_UseOverrideSequence?:bool)
    ) {blendType=Linear})
  ]
 }
 AnimNotify_PrintStringAssembler {nodes [
   n1 = AnimNotify_PrintStringAssembler!() -> (then, AnimNotifyEventName:String) {is=title K2Node_FunctionEntry}
   > Branch(Condition=Get DEBUG_PrintAnimLogic() -> (DEBUG_PrintAnimLogic:bool))
   n2 = Get DEBUG_PrintAnimLogic()
   n3 = Enum to String!(Enumerator:ESparksInstrumentType Enum=Get InstrumentType() -> (InstrumentType=Guitar)) -> (ReturnValue:String)
   > PrintString!(
    execute
    InString=JoinStringArray(SourceArray=Make Array!("[0]":String="[AnimNotify]", "[1]":String=@n3, "[2]":String=@n1) -> (Array:Array of Strings), Separator=" ")
    bPrintToScreen=@n2
    bPrintToLog=@n2
    TextColor=Get AnimNotify_TextColor() -> (AnimNotify_TextColor:Linear Color Structure)
    Duration
    Key
   ) -> (then)
   > Return Node()
  ]}
 EventGraph {nodes [
   n1 = Get DEBUG_PrintAnimLogic()
   n2 = Get SpineMotion_CurrentSequence()
   n3 = Get SpineMotion_Reset_CurrentSequence()
   PrintString(
    InString=Append!(
     A="SpineMotion_CurrentSequence="
     B=GetDisplayName(Object=@n2)
     C:String="\r\nSpineMotion_Reset_CurrentSequence="
     D:String=GetDisplayName(Object=@n3)
    ) -> (ReturnValue)
    bPrintToScreen=@n1
    bPrintToLog=@n1
    Duration=Get DeltaTimeX() -> (DeltaTimeX:float)
   )
   n4 = Get DEBUG_PrintAnimLogic()
   n5 = Get RightArm_CurrentSequence()
   > PrintString(
    InString=Append!(A="RightArm_CurrentSequence=", B=GetDisplayName(Object=@n5)) -> (ReturnValue)
    bPrintToScreen=@n4
    bPrintToLog=@n4
    Duration=Get DeltaTimeX()
   )
   n6 = Get DEBUG_PrintAnimLogic()
   n7 = Get LeftArm_CurrentSequence()
   > PrintString(InString=Append(A="LefTArm_CurrentSequence=", B=GetDisplayName(Object=@n7)), bPrintToScreen=@n6, bPrintToLog=@n6, Duration=Get DeltaTimeX())
   n8 = Get DEBUG_PrintAnimLogic()
   n9 = Get LowerBody_CurrentSequence() -> (LowerBody_CurrentSequence:Anim Sequence Base Object Reference)
   > PrintString(
    InString=Append(A="LowerBody_CurrentSequence=", B=GetDisplayName(Object=@n9))
    bPrintToScreen=@n8
    bPrintToLog=@n8
    Duration=Get DeltaTimeX()
   )
   n10 = Get DEBUG_PrintAnimLogic()
   n11 = Get HeadMotion_CurrentSequence()
   > PrintString(
    InString=Append(A="HeadMotion_CurrentSequence=", B=GetDisplayName(Object=@n11))
    bPrintToScreen=@n10
    bPrintToLog=@n10
    Duration=Get DeltaTimeX()
   )
   n12 = Get DEBUG_PrintAnimLogic()
   n13 = Get FootShift_Reset_CurrentSequence()
   n14 = Get FootShift_CurrentSequence()
   > PrintString(
    InString=Append!(
     A="FootShift_CurrentSequence="
     B=GetDisplayName(Object=@n14)
     C:String="\r\nFootShift_Reset_CurrentSequence="
     D:String=GetDisplayName(Object=@n13)
    ) -> (ReturnValue)
    bPrintToScreen=@n12
    bPrintToLog=@n12
    Duration=Get DeltaTimeX()
   )
   n15 = Get DEBUG_PrintAnimLogic()
   n16 = Get FaceExpression_CurrentSequence() -> (FaceExpression_CurrentSequence:Anim Sequence Base Object Reference)
   > PrintString(
    InString=Append!(A="FaceExpression_CurrentSequence=", B=GetDisplayName(Object=@n16)) -> (ReturnValue)
    bPrintToScreen=@n15
    bPrintToLog=@n15
    Duration=Get DeltaTimeX()
   )
   LowerBody_IncrementCurrentIndex() {is=title K2Node_CustomEvent}
   n17 = Get LowerBody_SequenceArray() -> (LowerBody_SequenceArray:Array of Anim Sequence Base Object References)
   > n18 = Branch(Condition="Equal (Integer)"(
     A=Get LowerBody_CurrentIndex() -> (LowerBody_CurrentIndex:int=0)
     B=Last Index!(TargetArray:Array of Anim Sequence Base Object References=@n17) -> (ReturnValue=0)
    ))
   > Set LowerBody_CurrentIndex(LowerBody_CurrentIndex:int=0) -> (Output_Get:int=0)
   Increment Int(" "=@n18.else, Value=Get LowerBody_CurrentIndex() -> (LowerBody_CurrentIndex=0))
   Guitar Strum() {is=title K2Node_CustomEvent}
   n19 = "Equal (Enum)"(A:ESparksAnimatableState Enum=Get AnimatableState() -> (AnimatableState=None), B:ESparksAnimatableState Enum=Animatable)
   > n20 = Branch(Condition=@n19)
   n21 = Branch(execute=@n20.else, Condition=Get UseStateMachineStrumming?() -> (UseStateMachineStrumming?:bool))
   Gate!(Enter:Exec=@n21.else, Open:Exec=@n23.then, Close:Exec=@n22.then_0, Toggle:Exec, bStartClosed:bool=false) -> (Exit:Exec) {is=title K2Node_MacroInstance}
   > K2Node_Composite_7 = Choose Strum Montage!(execute) -> (OutputPin:Exec) {is=title K2Node_Composite}
   > n22 = Sequence()
   n23 = Delay(execute=@n22.then_1, Duration=0.15)
   n24 = HeadMotion_ResetSystem() {is=title K2Node_CustomEvent}
   n25 = Sequence(execute=[@n24.then, @n30.then])
   Do Once(execute=@n25.then_1, Reset=@n25.then_0)
   > Event Disable Currently Active Booleans!(execute, self, AdditiveType:Enum_AdditiveType Enum=NewEnumerator1) -> (then)
   > Event Update Damplify Values!(execute, self, AdditiveType:Enum_AdditiveType Enum=NewEnumerator1) -> (then)
   > n26 = Branch(Condition=Get DEBUG_HeadMotion_UseOverrideSequence?() -> (DEBUG_HeadMotion_UseOverrideSequence?:bool))
   > n27 = Set HeadMotion_CurrentSequence(HeadMotion_CurrentSequence:Anim Sequence Base Object Reference=Get DEBUG_HeadMotion_OverrideSequence() -> (DEBUG_HeadMotion_OverrideSequence:Anim Sequence Base Object Reference)) -> (Output_Get:Anim Sequence Base Object Reference)
   Set HeadMotion_CurrentSequenceIsLooping?(
    execute=[@n31.then, @n27.then]
    HeadMotion_CurrentSequenceIsLooping?:bool=IsAnimationAssetLooping!(Asset:Object Reference=Get HeadMotion_CurrentSequence()) -> (bIsAssetLooping:bool)
   ) -> (Output_Get:bool)
   > Set HeadMotion_CurrentlyActive?(HeadMotion_CurrentlyActive?:bool=true) -> (Output_Get:bool)
   > n28 = Sequence()
   n29 = Get HeadMotion_TimeActive() -> (HeadMotion_TimeActive:float)
   > RetriggerableDelay(Duration=@n29)
   > Set HeadMotion_CurrentlyActive?(HeadMotion_CurrentlyActive?=false)
   n30 = RetriggerableDelay(execute=@n28.then_1, Duration=float + float(A=@n29, B=Get HeadMotion_TimeInactive() -> (HeadMotion_TimeInactive:float)))
   n31 = Set HeadMotion_CurrentSequence(
    execute=@n26.else
    HeadMotion_CurrentSequence=Random Array Item!(TargetArray:Array of Anim Sequence Base Object References=Get HeadMotion_SequenceArray() -> (HeadMotion_SequenceArray:Array of Anim Sequence Base Object References)) -> (OutItem:Anim Sequence Base Object Reference, OutIndex:int=0) {is=title K2Node_CallArrayFunction}
   )
   n32 = SpineMotion_ResetSystem() {is=title K2Node_CustomEvent}
   n33 = Sequence(execute=[@n32.then, @n35.then, @n40.then])
   Do Once(execute=@n33.then_1, Reset=@n33.then_0)
   > Event Disable Currently Active Booleans(AdditiveType=NewEnumerator4)
   > Event Update Damplify Values(AdditiveType=NewEnumerator4)
   > n34 = Branch(Condition=Get DEBUG_SpineMotion_UsePoseBlendSpace?())
   > Set SpineMotion_PoseBS_Pitch(SpineMotion_PoseBS_Pitch:float=RandomFloatInRange(Min=-90, Max=90)) -> (Output_Get:float)
   > Set SpineMotion_PoseBS_Roll(SpineMotion_PoseBS_Roll:float=RandomFloatInRange(Min=-90, Max=90)) -> (Output_Get:float)
   > Set SpineMotion_CurrentlyActive?(SpineMotion_CurrentlyActive?:bool=true) -> (Output_Get:bool)
   > n35 = Delay(Duration=4)
   n36 = Branch(execute=@n34.else, Condition=Get DEBUG_SpineMotion_UseOverrideSequence?() -> (DEBUG_SpineMotion_UseOverrideSequence?:bool))
   > n37 = Set SpineMotion_CurrentSequence(SpineMotion_CurrentSequence:Anim Sequence Base Object Reference=Get DEBUG_SpineMotion_OverrideSequence() -> (DEBUG_SpineMotion_OverrideSequence:Anim Sequence Base Object Reference)) -> (Output_Get:Anim Sequence Base Object Reference)
   Set SpineMotion_Reset_CurrentSequence(
    execute=[@n41.then, @n37.then]
    SpineMotion_Reset_CurrentSequence:Anim Sequence Base Object Reference=Find(
     TargetMap:Map of Anim Sequence Base Object References to Anim Sequence Base Object References=Get SpineMotion_SequenceMap() -> (SpineMotion_SequenceMap:Map of Anim Sequence Base Object References to Anim Sequence Base Object References)
     Key:"Anim Sequence Base Object Reference (by ref)"=Get SpineMotion_CurrentSequence()
    ) -> (Value:Anim Sequence Base Object Reference)
   ) -> (Output_Get:Anim Sequence Base Object Reference)
   > Set SpineMotion_CurrentlyActive?(SpineMotion_CurrentlyActive?=true)
   > n38 = Sequence()
   n39 = Get SpineMotion_TimeActive() -> (SpineMotion_TimeActive:float)
   > RetriggerableDelay(Duration=@n39)
   > Set SpineMotion_CurrentlyActive?(SpineMotion_CurrentlyActive?=false)
   n40 = RetriggerableDelay(execute=@n38.then_1, Duration=float + float(A=@n39, B=Get SpineMotion_TimeInactive() -> (SpineMotion_TimeInactive:float)))
   n41 = Set SpineMotion_CurrentSequence(
    execute=@n36.else
    SpineMotion_CurrentSequence=Random Array Item(TargetArray=Get SpineMotion_SequenceArray() -> (SpineMotion_SequenceArray:Array of Anim Sequence Base Object References)) -> (OutIndex=0)
   )
   n42 = FootShift_ResetSystem() {is=title K2Node_CustomEvent}
   n43 = Sequence(execute=[@n42.then, @n50.then])
   Do Once(execute=@n43.then_1, Reset=@n43.then_0)
   > Event Disable Currently Active Booleans(AdditiveType=NewEnumerator0)
   > Event Update Damplify Values(AdditiveType=NewEnumerator0)
   > n44 = Branch(Condition=Get DEBUG_FootShift_UseProceduralSystem?())
   > K2Node_Composite_14 = FootShift Proc Set Mode Index!(execute) -> (then:Exec) {is=title K2Node_Composite}
   > K2Node_Composite_5 = FootShift Proc Set Transform Values!(In:Exec) -> (Out:Exec) {is=title K2Node_Composite}
   > K2Node_Composite_0 = FootShift Proc Set Transition Speed!(In:Exec) -> (Out:Exec) {is=title K2Node_Composite}
   > K2Node_Composite_8 = FootShift Proc Set Sequences!(In:Exec) -> (Out:Exec) {is=title K2Node_Composite}
   n45 = Sequence!(execute=[@K2Node_Composite_10.then, @K2Node_Composite_8.Out]) -> (then_0, then_1, then_2:Exec)
   n46 = Get FootShift_ProcModeIndex() -> (FootShift_ProcModeIndex=0)
   Branch(execute=@n45.then_1, Condition=AND Boolean(A=Get DEBUG_FootShift_UseProceduralSystem?(), B="Equal (Integer)"(A=@n46, B=2)))
   > RetriggerableDelay(Duration=Get FootShift_2ndFoot_ProcLiftDelay() -> (FootShift_2ndFoot_ProcLiftDelay:float))
   > Set FootShift_2ndFoot_ProcLiftCurrentlyActive?(FootShift_2ndFoot_ProcLiftCurrentlyActive?:bool=true) -> (Output_Get:bool)
   > RetriggerableDelay(Duration=Get FootShift_2ndFoot_ProcTranslateDelay() -> (FootShift_2ndFoot_ProcTranslateDelay:float))
   > Set FootShift_2ndFoot_ProcTranslateCurrentlyActive?(FootShift_2ndFoot_ProcTranslateCurrentlyActive?:bool=true) -> (Output_Get:bool)
   Set FootShift_1stFoot_ProcLiftCurrentlyActive?(execute=@n45.then_0, FootShift_1stFoot_ProcLiftCurrentlyActive?:bool=true) -> (Output_Get:bool)
   > RetriggerableDelay(Duration=Get FootShift_1stFoot_ProcTranslateDelay() -> (FootShift_1stFoot_ProcTranslateDelay:float))
   > Set FootShift_1stFoot_ProcTranslateCurrentlyActive?(FootShift_1stFoot_ProcTranslateCurrentlyActive?:bool=true) -> (Output_Get:bool)
   n47 = Sequence!(execute=@n45.then_2) -> (then_0, then_1)
   > RetriggerableDelay(Duration=Get FootShift_TimeActive() -> (FootShift_TimeActive:float))
   > n48 = Sequence()
   n49 = Get FootShift_ProcModeIndex() -> (FootShift_ProcModeIndex=0)
   Branch(execute=@n48.then_1, Condition=AND Boolean(A=Get DEBUG_FootShift_UseProceduralSystem?(), B="Equal (Integer)"(A=@n49, B=2)))
   > RetriggerableDelay(Duration=Get FootShift_2ndFoot_ProcLiftDelay())
   > Set FootShift_2ndFoot_ProcLiftCurrentlyActive?(FootShift_2ndFoot_ProcLiftCurrentlyActive?=false)
   > RetriggerableDelay(Duration=Get FootShift_2ndFoot_ProcTranslateDelay())
   > Set FootShift_2ndFoot_ProcTranslateCurrentlyActive?(FootShift_2ndFoot_ProcTranslateCurrentlyActive?=false)
   Set FootShift_1stFoot_ProcLiftCurrentlyActive?(execute=@n48.then_0, FootShift_1stFoot_ProcLiftCurrentlyActive?=false)
   > RetriggerableDelay(Duration=Get FootShift_1stFoot_ProcTranslateDelay())
   > Set FootShift_1stFoot_ProcTranslateCurrentlyActive?(FootShift_1stFoot_ProcTranslateCurrentlyActive?=false)
   n50 = RetriggerableDelay(execute=@n47.then_1, Duration=float + float(A=Get FootShift_TimeActive(), B=Get FootShift_TimeInactive() -> (FootShift_TimeInactive:float)))
   K2Node_Composite_10 = FootShift Choose Sequence!(execute=@n44.else) -> (then:Exec) {is=title K2Node_Composite}
   Change Chord Hand Pose() {is=title K2Node_CustomEvent}
   > n51 = Branch(Condition=@n19)
   Set ChordHandPoseIndex(execute=@n51.else, ChordHandPoseIndex:int=RandomIntegerInRange!(Min:int=0, Max:int=4) -> (ReturnValue:int=0)) -> (Output_Get:int=0)
   n52 = Event Update Damplify Values!() -> (OutputDelegate, then, AdditiveType:Enum_AdditiveType Enum=NewEnumerator0) {is=title K2Node_CustomEvent}
   > K2Node_Composite_9 = Update Damplify Values!(In:Exec, AdditiveType:Enum_AdditiveType Enum=@n52.AdditiveType) -> (Out:Exec) {is=title K2Node_Composite}
   n53 = LeftArm_ResetSystem() {is=title K2Node_CustomEvent}
   n54 = Sequence(execute=[@n53.then, @n59.then])
   Do Once(execute=@n54.then_1, Reset=@n54.then_0)
   > Event Disable Currently Active Booleans(AdditiveType=NewEnumerator2)
   > Event Update Damplify Values(AdditiveType=NewEnumerator2) {is=title K2Node_CallFunction}
   > n55 = Branch(Condition=Get DEBUG_LeftArm_UseOverrideSequence?() -> (DEBUG_LeftArm_UseOverrideSequence?:bool))
   > n56 = Set LeftArm_CurrentSequence(LeftArm_CurrentSequence:Anim Sequence Base Object Reference=Get DEBUG_LeftArm_OverrideSequence() -> (DEBUG_LeftArm_OverrideSequence:Anim Sequence Base Object Reference)) -> (Output_Get:Anim Sequence Base Object Reference)
   Set LeftArm_CurrentlyActive?(execute=[@n60.then, @n56.then], LeftArm_CurrentlyActive?:bool=true) -> (Output_Get:bool)
   > n57 = Sequence()
   n58 = Get LeftArm_TimeActive() -> (LeftArm_TimeActive:float)
   > RetriggerableDelay(Duration=@n58)
   > Set LeftArm_CurrentlyActive?(LeftArm_CurrentlyActive?=false)
   n59 = RetriggerableDelay(execute=@n57.then_1, Duration=float + float(A=@n58, B=Get LeftArm_TimeInactive() -> (LeftArm_TimeInactive:float)))
   n60 = Set LeftArm_CurrentSequence(
    execute=@n55.else
    LeftArm_CurrentSequence=Random Array Item(TargetArray=Get LeftArm_SequenceArray() -> (LeftArm_SequenceArray:Array of Anim Sequence Base Object References)) -> (OutIndex=0)
   )
   n61 = Event Disable Currently Active Booleans!() -> (OutputDelegate, then, AdditiveType:Enum_AdditiveType Enum=NewEnumerator0) {is=title K2Node_CustomEvent}
   > K2Node_Composite_11 = Disable Currently Active Booleans!(then:Exec, AdditiveType:Enum_AdditiveType Enum=@n61.AdditiveType) -> (OutputPin:Exec) {is=title K2Node_Composite}
   n62 = RightArm_ResetSystem() {is=title K2Node_CustomEvent}
   n63 = Sequence(execute=[@n62.then, @n68.then])
   Do Once(execute=@n63.then_1, Reset=@n63.then_0)
   > Event Disable Currently Active Booleans(AdditiveType=NewEnumerator3) {is=title K2Node_CallFunction}
   > Event Update Damplify Values(AdditiveType=NewEnumerator3)
   > n64 = Branch(Condition=Get DEBUG_RightArm_UseOverrideSequence?() -> (DEBUG_RightArm_UseOverrideSequence?:bool))
   > n65 = Set RightArm_CurrentSequence(RightArm_CurrentSequence:Anim Sequence Base Object Reference=Get DEBUG_RightArm_OverrideSequence() -> (DEBUG_RightArm_OverrideSequence:Anim Sequence Base Object Reference)) -> (Output_Get:Anim Sequence Base Object Reference)
   Set RightArm_CurrentlyActive?(execute=[@n69.then, @n65.then], RightArm_CurrentlyActive?:bool=true) -> (Output_Get:bool)
   > n66 = Sequence()
   n67 = Get RightArm_TimeActive() -> (RightArm_TimeActive:float)
   > RetriggerableDelay(Duration=@n67)
   > Set RightArm_CurrentlyActive?(RightArm_CurrentlyActive?=false)
   n68 = RetriggerableDelay(execute=@n66.then_1, Duration=float + float(A=@n67, B=Get RightArm_TimeInactive() -> (RightArm_TimeInactive:float)))
   n69 = Set RightArm_CurrentSequence(
    execute=@n64.else
    RightArm_CurrentSequence=Random Array Item(TargetArray=Get RightArm_SequenceArray() -> (RightArm_SequenceArray:Array of Anim Sequence Base Object References)) -> (OutIndex=0)
   )
   n70 = Event BlueprintUpdateAnimation()
   > Set DeltaTimeX(DeltaTimeX:float=@n70.DeltaTimeX) -> (Output_Get:float)
   > n71 = Sequence!(execute) -> (
    then_0
    then_1
    then_2:Exec
    then_3:Exec
    then_4:Exec
    then_5:Exec
    then_6:Exec
    then_7:Exec
    then_8:Exec
    then_9:Exec
    then_10:Exec
    then_11:Exec
    then_12:Exec
   )
   Reroute Node(InputPin=@n71.then_12)
   Reroute Node(InputPin=@n71.then_0)
   n72 = Sequence!(execute=@n71.then_3) -> (then_0, then_1, then_2, then_3, then_4, then_5, then_6)
   > Reroute Node()
   Reroute Node(InputPin=@n72.then_1)
   Reroute Node(InputPin=@n72.then_2)
   Reroute Node(InputPin=@n72.then_4)
   Reroute Node(InputPin=@n72.then_5)
   n73 = Cast To PlayerPawn_Athena(execute=@n72.then_6, Object=TryGetPawnOwner())
   > n74 = IsExceedingMaxSpeed!(
    execute
    Velocity:"Vector (by ref)"=GetLastUpdateVelocity!(self:Character Movement Component Object Reference=Get CharacterMovement!(self:Character Object Reference=@n73) -> (CharacterMovement:Character Movement Component Object Reference)) -> (ReturnValue:Vector)
    InMaxSpeed:float=0
   ) -> (then, ReturnValue:bool)
   > n75 = Set Traversal_IsCharacterMoving?(Traversal_IsCharacterMoving?:bool=@n74) -> (Output_Get:bool)
   > n76 = Branch(Condition=@n75)
   DoOnce MultiInput!(A In:Exec=@n76.else, Reset In:Exec=@n76.then) -> (A Out:Exec, Reset Out:Exec) {is=title K2Node_DoOnceMultiInput}
   > LowerBody_IncrementCurrentIndex() {is=title K2Node_CallFunction}
   n77 = Branch(execute=@n72.then_3, Condition=Get Traversal_IsCharacterMoving?())
   Do Once(execute=@n77.else, Reset=@n77.then, Start Closed=true)
   > n78 = Branch(Condition=Get DEBUG_LowerBody_UseOverrideSequence?() -> (DEBUG_LowerBody_UseOverrideSequence?:bool))
   > Set LowerBody_NextSequence(LowerBody_NextSequence:Anim Sequence Base Object Reference=Get DEBUG_LowerBody_OverrideSequence() -> (DEBUG_LowerBody_OverrideSequence:Anim Sequence Base Object Reference)) -> (Output_Get:Anim Sequence Base Object Reference)
   n79 = Branch(execute=@n78.else, Condition=Get DEBUG_LowerBody_ForceSequentialMode?() -> (DEBUG_LowerBody_ForceSequentialMode?:bool))
   > Set LowerBody_NextSequence(LowerBody_NextSequence="Get (a copy)"(
     Array:Array of Anim Sequence Base Object References=Get LowerBody_SequenceArray()
     Dimension 1=Get LowerBody_CurrentIndex() -> (LowerBody_CurrentIndex=0)
    ) -> (Output:Anim Sequence Base Object Reference))
   Set LowerBody_NextSequence(execute=[@n79.else, @n80.then], LowerBody_NextSequence=Random Array Item(TargetArray=Get LowerBody_SequenceArray()) -> (OutIndex=0))
   > n80 = Branch(Condition="Equal (Object)"(A=Get LowerBody_NextSequence(), B=Get LowerBody_CurrentSequence()))
   Set LowerBody_CurrentSequence(execute=@n80.else, LowerBody_CurrentSequence:Anim Sequence Base Object Reference=Get LowerBody_NextSequence()) -> (Output_Get:Anim Sequence Base Object Reference)
   n81 = Sequence!(execute=@n71.then_6) -> (then_0, then_1, then_2)
   > n82 = Branch(Condition=Get DEBUG_BaseBody_ForceSquat?() -> (DEBUG_BaseBody_ForceSquat?:bool))
   > Set BaseBody_Squat_BlendWeight(BaseBody_Squat_BlendWeight:double=float - float(A=1, B=GetCurveValue(CurveName=Disable_JamV2_BaseBodySquat))) -> (Output_Get:double)
   > n83 = JamV2_SineWaveMotionOcsillator!(
    execute
    self
    MaxValue:double=Get BaseBody_SquatDistanceMax() -> (BaseBody_SquatDistanceMax:double)
    "Frequency (Speed) Multiplier":double=1
    "Amplitude (Min && Max) Multiplier":double=1
   ) -> (then, ReturnValue:double)
   > Set BaseBody_SquatDistance(BaseBody_SquatDistance:double=@n83) -> (Output_Get:double)
   Set BaseBody_SquatDistance(execute=@n82.else, BaseBody_SquatDistance=0)
   n84 = Get DEBUG_FaceExpression_CloseEyes() -> (DEBUG_FaceExpression_CloseEyes:bool)
   Set FaceExpression_ClosedEyes(
    execute=@n81.then_1
    FaceExpression_ClosedEyes:float=FInterpTo(
     Current=Get FaceExpression_ClosedEyes()
     Target="To Float (Boolean)"(InBool=@n84)
     DeltaTime=GetWorldDeltaSeconds!() -> (ReturnValue:double)
     InterpSpeed=12
    )
   ) -> (Output_Get:float)
   n85 = Sequence!(execute=@n71.then_9) -> (then_0, then_1, then_2, then_3:Exec, then_4:Exec, then_5:Exec, then_6:Exec, then_7:Exec, then_8:Exec, then_9:Exec, then_10:Exec)
   n86 = Sequence!(execute=@n85.then_10) -> (then_0, then_1, then_2, then_3, then_4, then_5)
   > Branch(Condition=Get DEBUG_FootShift_ResetSystem?())
   > FootShift_ResetSystem() {is=title K2Node_CallFunction}
   > DelayUntilNextTick!(execute) -> (then)
   > Set DEBUG_FootShift_ResetSystem?(DEBUG_FootShift_ResetSystem?:bool=false) -> (Output_Get:bool)
   Branch(execute=@n86.then_2, Condition=Get DEBUG_LeftArm_ResetSystem?())
   > LeftArm_ResetSystem() {is=title K2Node_CallFunction}
   > DelayUntilNextTick()
   > Set DEBUG_LeftArm_ResetSystem?(DEBUG_LeftArm_ResetSystem?:bool=false) -> (Output_Get:bool)
   Branch(execute=@n86.then_1, Condition=Get DEBUG_HeadMotion_ResetSystem?())
   > HeadMotion_ResetSystem() {is=title K2Node_CallFunction}
   > DelayUntilNextTick()
   > Set DEBUG_HeadMotion_ResetSystem?(DEBUG_HeadMotion_ResetSystem?:bool=false) -> (Output_Get:bool)
   Branch(execute=@n86.then_3, Condition=Get DEBUG_RightArm_ResetSystem?())
   > RightArm_ResetSystem() {is=title K2Node_CallFunction}
   > DelayUntilNextTick()
   > Set DEBUG_RightArm_ResetSystem?(DEBUG_RightArm_ResetSystem?:bool=false) -> (Output_Get:bool)
   Branch(execute=@n86.then_4, Condition=Get DEBUG_SpineMotion_ResetSystem?())
   > SpineMotion_ResetSystem() {is=title K2Node_CallFunction}
   > DelayUntilNextTick()
   > Set DEBUG_SpineMotion_ResetSystem?(DEBUG_SpineMotion_ResetSystem?:bool=false) -> (Output_Get:bool)
   Reroute Node(InputPin=@n85.then_4)
   Reroute Node(InputPin=@n85.then_5)
   n87 = Branch(execute=@n85.then_2, Condition=Get DEBUG_FootShift_UseOverrideBlendWeight?() -> (DEBUG_FootShift_UseOverrideBlendWeight?:bool))
   > Set FootShift_BlendWeight(FootShift_BlendWeight:float=Get DEBUG_FootShift_OverrideBlendWeight() -> (DEBUG_FootShift_OverrideBlendWeight:float)) -> (Output_Get:float)
   Set FootShift_BlendWeight(execute=@n87.else, FootShift_BlendWeight=float - float(A=1, B=GetCurveValue(CurveName=Disable_JamV2_FootShiftAdditive)))
   n88 = Branch(execute=@n85.then_1, Condition=Get DEBUG_Song_UseOverrideIntensity?() -> (DEBUG_Song_UseOverrideIntensity?:bool))
   > Set Song_IntensityIndex(Song_IntensityIndex:int=Select!(
     NewEnumerator0:int=0
     NewEnumerator1:int=1
     NewEnumerator2:int=2
     Index:AnimIntensity Enum=Get DEBUG_Song_OverrideIntensity() -> (DEBUG_Song_OverrideIntensity:AnimIntensity Enum=NewEnumerator0)
    ) -> (ReturnValue:int)) -> (Output_Get:int=0)
   Set Song_IntensityIndex(
    execute=@n88.else
    Song_IntensityIndex=SongIntensityToIndex!(Intensity:Gameplay Tag Structure=Get SongIntensity() -> (SongIntensity:Gameplay Tag Structure)) -> (Index:int=0)
   ) -> (Output_Get=0)
   n89 = Branch(execute=@n85.then_0, Condition=Get DEBUG_Song_UseOverrideGenre?() -> (DEBUG_Song_UseOverrideGenre?:bool))
   > Set Song_GenreIndex(Song_GenreIndex:int=Select(
     NewEnumerator0=0
     NewEnumerator1=1
     NewEnumerator2=2
     Index:Enum_MusicGenre Enum=Get DEBUG_Song_OverrideGenre() -> (DEBUG_Song_OverrideGenre:Enum_MusicGenre Enum=NewEnumerator0)
    )) -> (Output_Get:int=0)
   Set Song_GenreIndex(
    execute=@n89.else
    Song_GenreIndex=SongGenreToIndex!(Genre:Gameplay Tag Structure=Get SongGenre() -> (SongGenre:Gameplay Tag Structure)) -> (Index:int=0)
   ) -> (Output_Get=0)
   Branch(execute=@n85.then_3, Condition=Get DEBUG_BaseBody_ForcePlayStateTransition?() -> (DEBUG_BaseBody_ForcePlayStateTransition?:bool))
   > DelayUntilNextTick()
   > Set DEBUG_BaseBody_ForcePlayStateTransition?(DEBUG_BaseBody_ForcePlayStateTransition?:bool=false) -> (Output_Get:bool)
   n90 = Branch(execute=@n85.then_7, Condition=Get DEBUG_LeftArm_UseOverrideBlendWeight?() -> (DEBUG_LeftArm_UseOverrideBlendWeight?:bool))
   > Set LeftArm_BlendWeight(LeftArm_BlendWeight:float=Get DEBUG_LeftArm_OverrideBlendWeight() -> (DEBUG_LeftArm_OverrideBlendWeight:float)) -> (Output_Get:float)
   Set LeftArm_BlendWeight(execute=@n90.else, LeftArm_BlendWeight=float - float(A=1, B=GetCurveValue(CurveName=Disable_JamV2_LeftArmReplace)))
   n91 = Branch(execute=@n85.then_8, Condition=Get DEBUG_RightArm_UseOverrideBlendWeight?() -> (DEBUG_RightArm_UseOverrideBlendWeight?:bool))
   > Set RightArm_BlendWeight(RightArm_BlendWeight:float=Get DEBUG_RightArm_OverrideBlendWeight() -> (DEBUG_RightArm_OverrideBlendWeight:float)) -> (Output_Get:float)
   Set RightArm_BlendWeight(execute=@n91.else, RightArm_BlendWeight=float - float(A=1, B=GetCurveValue(CurveName=Disable_JamV2_RightArmReplace)))
   n92 = Branch(execute=@n85.then_9, Condition=Get DEBUG_SpineMotion_UseOverrideBlendWeight?() -> (DEBUG_SpineMotion_UseOverrideBlendWeight?:bool))
   > Set SpineMotion_BlendWeight(SpineMotion_BlendWeight:float=Get DEBUG_SpineMotion_OverrideBlendWeight() -> (DEBUG_SpineMotion_OverrideBlendWeight:float)) -> (Output_Get:float)
   Set SpineMotion_BlendWeight(execute=@n92.else, SpineMotion_BlendWeight=float - float(A=1, B=GetCurveValue(CurveName=Disable_JamV2_SpineMotionAdditive)))
   n93 = Branch(execute=@n85.then_6, Condition=Get DEBUG_HeadMotion_UseOverrideBlendWeight?() -> (DEBUG_HeadMotion_UseOverrideBlendWeight?:bool))
   > Set HeadMotion_BlendWeight(HeadMotion_BlendWeight:float=Get DEBUG_HeadMotion_OverrideBlendWeight() -> (DEBUG_HeadMotion_OverrideBlendWeight:float)) -> (Output_Get:float)
   Set HeadMotion_BlendWeight(execute=@n93.else, HeadMotion_BlendWeight=float - float(A=1, B=GetCurveValue(CurveName=Disable_JamV2_HeadMotionReplace)))
   n94 = Event On LipSync Asset Changed!() -> (OutputDelegate, then, NewLipSyncAsset:Anim Sequence Object Reference) {is=title K2Node_Event}
   > n95 = Is Valid(InputObject=@n94.NewLipSyncAsset)
   Set LipSyncBlendWeight(execute=@n95.Is Not Valid, LipSyncBlendWeight:double=0) -> (Output_Get:double)
   Set LipSyncBlendWeight(execute=@n95.Is Valid, LipSyncBlendWeight=1)
   n96 = Event On MIDI Event Driver Changed!() -> (OutputDelegate, then, NewMidiEventDriver:"Midi Event Driver Interface Interface (by ref)") {is=title K2Node_Event}
   > n97 = Branch(Condition=IsValidInterface!(Interface:"Interface Interface (by ref)"=@n96.NewMidiEventDriver) -> (ReturnValue:bool))
   > n98 = Set MidiTrackName(MidiTrackName:Name=GetMidiTrackNameFromSparksInstrument!(
     self:Midi Event Driver Interface Interface=@n96.NewMidiEventDriver
     InstrumentType:ESparksInstrumentType Enum=Get Instrument() -> (Instrument=Guitar)
    ) -> (ReturnValue:Name=None)) -> (Output_Get:Name=None)
   K2Node_Composite_3 = Song Intensity!(In:Exec=[@n99.then, @n98.then]) -> (Out:Exec) {is=title K2Node_Composite}
   > K2Node_Composite_4 = Fret Position!(In:Exec) -> (Out:Exec) {is=title K2Node_Composite}
   > K2Node_Composite_2 = Strum!(In:Exec) -> (Out:Exec) {is=title K2Node_Composite}
   n99 = Set MidiTrackName(execute=@n97.else, MidiTrackName=None) -> (Output_Get=None)
   Event BlueprintInitializeAnimation()
   > K2Node_Composite_13 = Initialize Body Section Variables!(In:Exec) -> (Out:Exec) {is=title K2Node_Composite}
   > K2Node_Composite_12 = Reset System on all Body Sections!(In:Exec) -> (then:Exec) {is=title K2Node_Composite}
   Event On New Song() {is=title K2Node_Event}
   > SetSongDefaults!(
    execute
    AnimInstance:Sparks Instrument Anim Instance Object Reference=GetSelfAsSparksInstrumentAnimInstance!() -> (ReturnValue:Sparks Instrument Anim Instance Object Reference)
   ) -> (then)
   Event Sparks Debug Update() {is=title K2Node_Event}
   > Branch(Condition=Get DEBUG_StrumTestFire() -> (DEBUG_StrumTestFire:bool))
   > Guitar Strum() {is=title K2Node_CallFunction}
   > DelayUntilNextTick()
   > Set DEBUG_StrumTestFire(DEBUG_StrumTestFire:bool=false) -> (Output_Get:bool)
   n100 = Event On Cosmetic Instrument Type Changed!() -> (OutputDelegate, then, NewInstrument:ESparksInstrumentType Enum=Guitar) {is=title K2Node_Event}
   > Set LowerBody_SequenceArray(LowerBody_SequenceArray:Array of Anim Sequence Base Object References=Select!(
     Option 0:Array of Anim Sequence Base Object References=Make Array(
      "[0]":Anim Sequence Base Object Reference=/SparksCharacterCommon/Animation/MainPlayer/BandPlay/Drum/Drumsticks_Walkcycle_CMM
      "[1]":Anim Sequence Base Object Reference=/BRCosmetics/Animation/Game/MainPlayer/Emotes/SunnySteppin/CMM/Emote_SunnySteppin_Traversal_CMM
      "[2]":Anim Sequence Base Object Reference=/BRCosmetics/Animation/Game/MainPlayer/Emotes/PogoTraversal/Emote_PogoTraversal_CMM
     ) -> (Array:Array of Anim Sequence Base Object References)
     Option 1:Array of Anim Sequence Base Object References=Make Array(
      "[0]"=/BRCosmetics/Animation/Game/MainPlayer/Emotes/Relish/Emote_Relish_CMM
      "[1]"=/BRCosmetics/Animation/Game/MainPlayer/Emotes/BalletJumps/Emote_BalletJumps_CMM
      "[2]"=/BRCosmetics/Animation/Game/MainPlayer/Emotes/BunnyHop/Emote_BunnyHop_Traverse_CMF
     )
     Index:bool="Equal (Enum)"(A:ESparksInstrumentType Enum=@n100.NewInstrument, B:ESparksInstrumentType Enum=Bass)
    ) -> (ReturnValue:Array of Anim Sequence Base Object References)) -> (Output_Get:Array of Anim Sequence Base Object References)
   > Set InstrumentType(InstrumentType:ESparksInstrumentType Enum=@n100.NewInstrument) -> (Output_Get:ESparksInstrumentType Enum=Guitar)
   n101 = Find(
    TargetMap:Map of ESparksInstrumentType Enums to Sparks Anim BPMSet Structures=Get HomebaseBPMs_Defaults() -> (HomebaseBPMs_Defaults:Map of ESparksInstrumentType Enums to Sparks Anim BPMSet Structures=Guitar)
    Key:"ESparksInstrumentType Enum (by ref)"=@n100.NewInstrument
   ) -> (Value:Sparks Anim BPMSet Structure)
   > Branch(Condition=@n101.ReturnValue)
   > Set HomebaseBPMs(HomebaseBPMs:Sparks Anim BPMSet Structure=@n101) -> (Output_Get:Sparks Anim BPMSet Structure)
   n102 = For Each Loop(Array:Array of Names=Get DEBUG_CurveNameArray() -> (DEBUG_CurveNameArray:Array of Names)) -> (Array Element:Name)
   > JamV2_CurveValue_Printer!(execute, self, CurveName:Name=@n102) -> (then)
   RandomFloatInRange(Min=8, Max=12)
   RandomFloatInRange(Min=8, Max=12)
   RandomFloatInRange(Min=8, Max=12)
   RandomFloatInRange(Min=8, Max=12)
   n103 = Get FootShift_1stFoot_ProcLiftCurrentlyActive?()
   n104 = Get FootShift_2ndFoot_ProcLiftCurrentlyActive?()
   n105 = Get FootShift_1stFoot_ProcTranslateCurrentlyActive?()
   n106 = Get FootShift_2ndFoot_ProcTranslateCurrentlyActive?()
   PrintString(
    InString=Append!(
     A="1st Foot LIFT Active? == "
     B="To String (Boolean)"(InBool=@n103)
     C:String="\r\n1st Foot TRANSLATE Active? == "
     D:String="To String (Boolean)"(InBool=@n105)
     E:String="\r\n2nd Foot LIFT Active? == "
     F:String="To String (Boolean)"(InBool=@n104)
     G:String="\r\n2nd Foot TRANSLATE Active? == "
     H:String="To String (Boolean)"(InBool=@n106)
    ) -> (ReturnValue)
    Duration=Get DeltaTimeX()
   )
   RandomFloatInRange(Min=8, Max=12)
   RandomFloatInRange(Min=8, Max=12)
   RandomFloatInRange(Min=8, Max=12)
   RandomFloatInRange(Min=8, Max=12)
   K2Node_Composite_1 = Set Members of Structs!(ExecIn:Exec) -> (ExecOut:Exec) {is=title K2Node_Composite}
   Set IsPlayingKeytar(IsPlayingKeytar:bool="Equal (Enum)"(A=@n100.NewInstrument, B=Keyboard)) -> (Output_Get:bool)
  ]}
 FootShift Proc Set Transition Speed {
  path=EventGraph.K2Node_Composite_0.FootShift Proc Set Transition Speed
  nodes [
   Inputs!() -> (In:Exec)
   > n1 = Sequence!(execute) -> (then_0, then_1)
   > n2 = Branch(Condition=Get DEBUG_FootShift_ProcUseOverrideTransitionSpeed?() -> (DEBUG_FootShift_ProcUseOverrideTransitionSpeed?:bool))
   > Set FootShift_1stFoot_ProcTransitionSpeed(FootShift_1stFoot_ProcTransitionSpeed:Enum_FootShift_ProcTransitionSpeed Enum=Get DEBUG_FootShift_ProcOverrideTransitionSpeed() -> (DEBUG_FootShift_ProcOverrideTransitionSpeed:Enum_FootShift_ProcTransitionSpeed Enum=NewEnumerator0)) -> (Output_Get:Enum_FootShift_ProcTransitionSpeed Enum=NewEnumerator0)
   > Set FootShift_1stFoot_ProcAnimPlayRate(FootShift_1stFoot_ProcAnimPlayRate:float=1) -> (Output_Get:float)
   > Set FootShift_2ndFoot_ProcTransitionSpeed(FootShift_2ndFoot_ProcTransitionSpeed:Enum_FootShift_ProcTransitionSpeed Enum=Get DEBUG_FootShift_ProcOverrideTransitionSpeed() -> (DEBUG_FootShift_ProcOverrideTransitionSpeed=NewEnumerator0)) -> (Output_Get:Enum_FootShift_ProcTransitionSpeed Enum=NewEnumerator0)
   > Set FootShift_2ndFoot_ProcAnimPlayRate(FootShift_2ndFoot_ProcAnimPlayRate:float=1) -> (Output_Get:float)
   n3 = Sequence(execute=@n2.else)
   Branch(execute=@n3.then_1, Condition="Equal (Integer)"(A=Get FootShift_ProcModeIndex() -> (FootShift_ProcModeIndex=0), B=2))
   n4 = NormalizeToRange!(
    Value:double=Get FootShift_2ndFoot_ProcTranslateDistance() -> (FootShift_2ndFoot_ProcTranslateDistance:double)
    RangeMin:double=Get FootShift_2ndFoot_ProcTranslateDistanceMin() -> (FootShift_2ndFoot_ProcTranslateDistanceMin:double)
    RangeMax:double=Get FootShift_2ndFoot_ProcTranslateDistanceMax() -> (FootShift_2ndFoot_ProcTranslateDistanceMax:double)
   ) -> (ReturnValue:double)
   > n5 = Branch(Condition="float <= float"(A=@n4, B=0.333) -> (ReturnValue=~))
   > Set FootShift_2ndFoot_ProcTransitionSpeed(FootShift_2ndFoot_ProcTransitionSpeed=NewEnumerator2) -> (Output_Get=NewEnumerator0)
   > Set FootShift_2ndFoot_ProcAnimPlayRate(FootShift_2ndFoot_ProcAnimPlayRate=1)
   > Set FootShift_2ndFoot_ProcTranslateDelay(FootShift_2ndFoot_ProcTranslateDelay:float=0.12) -> (Output_Get:float)
   n6 = Branch(execute=@n5.else, Condition="float <= float"(A=@n4, B=0.666) -> (ReturnValue=~))
   > Set FootShift_2ndFoot_ProcTransitionSpeed(FootShift_2ndFoot_ProcTransitionSpeed=NewEnumerator1) -> (Output_Get=NewEnumerator0)
   > Set FootShift_2ndFoot_ProcAnimPlayRate(FootShift_2ndFoot_ProcAnimPlayRate=0.75)
   > Set FootShift_2ndFoot_ProcTranslateDelay(FootShift_2ndFoot_ProcTranslateDelay=0.18)
   Branch(execute=@n6.else, Condition="float <= float"(A=@n4, B=1) -> (ReturnValue=~))
   > Set FootShift_2ndFoot_ProcTransitionSpeed(FootShift_2ndFoot_ProcTransitionSpeed=NewEnumerator0) -> (Output_Get=NewEnumerator0)
   > Set FootShift_2ndFoot_ProcAnimPlayRate(FootShift_2ndFoot_ProcAnimPlayRate=0.5)
   > Set FootShift_2ndFoot_ProcTranslateDelay(FootShift_2ndFoot_ProcTranslateDelay=0.24)
   n7 = NormalizeToRange(
    Value=Get FootShift_1stFoot_ProcTranslateDistance() -> (FootShift_1stFoot_ProcTranslateDistance:double)
    RangeMin=Get FootShift_1stFoot_ProcTranslateDistanceMin() -> (FootShift_1stFoot_ProcTranslateDistanceMin:double)
    RangeMax=Get FootShift_1stFoot_ProcTranslateDistanceMax() -> (FootShift_1stFoot_ProcTranslateDistanceMax:double)
   )
   n8 = Branch(execute=@n3.then_0, Condition="float <= float"(A=@n7, B=0.333) -> (ReturnValue=~))
   > Set FootShift_1stFoot_ProcTransitionSpeed(FootShift_1stFoot_ProcTransitionSpeed=NewEnumerator2) -> (Output_Get=NewEnumerator0)
   > Set FootShift_1stFoot_ProcAnimPlayRate(FootShift_1stFoot_ProcAnimPlayRate=1)
   > Set FootShift_1stFoot_ProcTranslateDelay(FootShift_1stFoot_ProcTranslateDelay:float=0.12) -> (Output_Get:float)
   n9 = Branch(execute=@n8.else, Condition="float <= float"(A=@n7, B=0.666) -> (ReturnValue=~))
   > Set FootShift_1stFoot_ProcTransitionSpeed(FootShift_1stFoot_ProcTransitionSpeed=NewEnumerator1) -> (Output_Get=NewEnumerator0)
   > Set FootShift_1stFoot_ProcAnimPlayRate(FootShift_1stFoot_ProcAnimPlayRate=0.75)
   > Set FootShift_1stFoot_ProcTranslateDelay(FootShift_1stFoot_ProcTranslateDelay=0.18)
   Branch(execute=@n9.else, Condition="float <= float"(A=@n7, B=1) -> (ReturnValue=~))
   > Set FootShift_1stFoot_ProcTransitionSpeed(FootShift_1stFoot_ProcTransitionSpeed=NewEnumerator0) -> (Output_Get=NewEnumerator0)
   > Set FootShift_1stFoot_ProcAnimPlayRate(FootShift_1stFoot_ProcAnimPlayRate=0.5)
   > Set FootShift_1stFoot_ProcTranslateDelay(FootShift_1stFoot_ProcTranslateDelay=0.24)
   Outputs!(Out:Exec=@n1.then_1)
  ]
 }
 Set Members of Structs {
  path=EventGraph.K2Node_Composite_1.Set Members of Structs
  nodes [
   Inputs!() -> (ExecIn:Exec)
   > Set members in Sparks Jam V2 Additives Struct!(
    execute:Exec
    StructRef:"Sparks Jam V2 Additives Struct Structure (by ref)"=Get FootShift_Struct() -> (FootShift_Struct:Sparks Jam V2 Additives Struct Structure)
    SequenceArray_12_#45:Array of Anim Sequence Base Object References=Get FootShift_SequenceArray() -> (FootShift_SequenceArray:Array of Anim Sequence Base Object References)
    BlendWeight_14_#46:double=Get FootShift_BlendWeight()
    CurrentlyActive?_15_#47:bool=Get FootShift_1stFoot_ProcLiftCurrentlyActive?()
    TimeActive_16_#48:double=Get FootShift_TimeActive()
    TimeInactive_17_#49:double=Get FootShift_TimeInactive()
    UseOverrideSequence?_22_#50:bool=Get DEBUG_FootShift_UseOverrideSequence?() -> (DEBUG_FootShift_UseOverrideSequence?:bool)
    OverrideSequence_25_#51:Anim Sequence Base Object Reference=Get DEBUG_FootShift_OverrideSequence() -> (DEBUG_FootShift_OverrideSequence:Anim Sequence Base Object Reference)
    OverrideBlendWeight_28_#52:double=Get DEBUG_FootShift_OverrideBlendWeight()
    UseOverrideBlendWeight?_31_#53:bool=Get DEBUG_FootShift_UseOverrideBlendWeight?()
    DisableCurve_34_#54:Name=Disable_JamV2_FootShiftAdditive
   ) -> (then:Exec, StructOut:"Sparks Jam V2 Additives Struct Structure (by ref)") {is=title K2Node_SetFieldsInStruct}
   > Set members in Sparks Jam V2 Additives Struct(
    StructRef=Get HeadMotion_Struct() -> (HeadMotion_Struct:Sparks Jam V2 Additives Struct Structure)
    SequenceArray_12_#45=Get HeadMotion_SequenceArray()
    BlendWeight_14_#46=Get HeadMotion_BlendWeight()
    CurrentlyActive?_15_#47=Get HeadMotion_CurrentlyActive?()
    TimeActive_16_#48=Get HeadMotion_TimeActive()
    TimeInactive_17_#49=Get HeadMotion_TimeInactive()
    UseOverrideSequence?_22_#50=Get DEBUG_HeadMotion_UseOverrideSequence?()
    OverrideSequence_25_#51=Get DEBUG_HeadMotion_OverrideSequence()
    OverrideBlendWeight_28_#52=Get DEBUG_HeadMotion_OverrideBlendWeight()
    UseOverrideBlendWeight?_31_#53=Get DEBUG_HeadMotion_UseOverrideBlendWeight?()
    DisableCurve_34_#54=Disable_JamV2_HeadMotionReplace
   )
   > Set members in Sparks Jam V2 Additives Struct(
    StructRef=Get LeftArm_Struct() -> (LeftArm_Struct:Sparks Jam V2 Additives Struct Structure)
    SequenceArray_12_#45=Get LeftArm_SequenceArray()
    BlendWeight_14_#46=Get LeftArm_BlendWeight()
    CurrentlyActive?_15_#47=Get LeftArm_CurrentlyActive?()
    TimeActive_16_#48=Get LeftArm_TimeActive()
    TimeInactive_17_#49=Get LeftArm_TimeInactive()
    UseOverrideSequence?_22_#50=Get DEBUG_LeftArm_UseOverrideSequence?()
    OverrideSequence_25_#51=Get DEBUG_LeftArm_OverrideSequence()
    OverrideBlendWeight_28_#52=Get DEBUG_LeftArm_OverrideBlendWeight()
    UseOverrideBlendWeight?_31_#53=Get DEBUG_LeftArm_UseOverrideBlendWeight?()
    DisableCurve_34_#54=Disable_JamV2_LeftArmReplace
   )
   > Set members in Sparks Jam V2 Additives Struct(
    StructRef=Get RightArm_Struct() -> (RightArm_Struct:Sparks Jam V2 Additives Struct Structure)
    SequenceArray_12_#45=Get RightArm_SequenceArray()
    BlendWeight_14_#46=Get RightArm_BlendWeight()
    CurrentlyActive?_15_#47=Get RightArm_CurrentlyActive?()
    TimeActive_16_#48=Get RightArm_TimeActive()
    TimeInactive_17_#49=Get RightArm_TimeInactive()
    UseOverrideSequence?_22_#50=Get DEBUG_RightArm_UseOverrideSequence?()
    OverrideSequence_25_#51=Get DEBUG_RightArm_OverrideSequence()
    OverrideBlendWeight_28_#52=Get DEBUG_RightArm_OverrideBlendWeight()
    UseOverrideBlendWeight?_31_#53=Get DEBUG_RightArm_UseOverrideBlendWeight?()
    DisableCurve_34_#54=Disable_JamV2_RightArmReplace
   )
   > Set members in Sparks Jam V2 Additives Struct(
    StructRef=Get SpineMotion_Struct() -> (SpineMotion_Struct:Sparks Jam V2 Additives Struct Structure)
    SequenceArray_12_#45=Get SpineMotion_SequenceArray()
    BlendWeight_14_#46=Get SpineMotion_BlendWeight()
    CurrentlyActive?_15_#47=Get SpineMotion_CurrentlyActive?()
    TimeActive_16_#48=Get SpineMotion_TimeActive()
    TimeInactive_17_#49=Get SpineMotion_TimeInactive()
    UseOverrideSequence?_22_#50=Get DEBUG_SpineMotion_UseOverrideSequence?()
    OverrideSequence_25_#51=Get DEBUG_SpineMotion_OverrideSequence()
    OverrideBlendWeight_28_#52=Get DEBUG_SpineMotion_OverrideBlendWeight()
    UseOverrideBlendWeight?_31_#53=Get DEBUG_SpineMotion_UseOverrideBlendWeight?()
    DisableCurve_34_#54=Disable_JamV2_SpineMotionAdditive
   )
   > Outputs!(ExecOut:Exec)
  ]
 }
 FootShift Choose Sequence {
  path=EventGraph.K2Node_Composite_10.FootShift Choose Sequence
  nodes [
   Inputs!() -> (execute:Exec)
   > n1 = Branch(Condition=Get DEBUG_FootShift_UseOverrideSequence?())
   > n2 = Set FootShift_CurrentSequence(FootShift_CurrentSequence:Anim Sequence Base Object Reference=Get DEBUG_FootShift_OverrideSequence()) -> (Output_Get:Anim Sequence Base Object Reference)
   Set FootShift_Reset_CurrentSequence(
    execute=[@n3.then, @n2.then]
    FootShift_Reset_CurrentSequence:Anim Sequence Base Object Reference=Find(
     TargetMap:Map of Anim Sequence Base Object References to Anim Sequence Base Object References=Get FootShift_SequenceMap() -> (FootShift_SequenceMap:Map of Anim Sequence Base Object References to Anim Sequence Base Object References)
     Key:"Anim Sequence Base Object Reference (by ref)"=Get FootShift_CurrentSequence()
    ) -> (Value:Anim Sequence Base Object Reference)
   ) -> (Output_Get:Anim Sequence Base Object Reference)
   > Outputs!(then:Exec)
   n3 = Set FootShift_CurrentSequence(execute=@n1.else, FootShift_CurrentSequence=Random Array Item(TargetArray=Get FootShift_SequenceArray()) -> (OutIndex=0))
  ]
 }
 Disable Currently Active Booleans {
  path=EventGraph.K2Node_Composite_11.Disable Currently Active Booleans
  nodes [
   n1 = Inputs!() -> (then:Exec, AdditiveType:Enum_AdditiveType Enum=NewEnumerator0)
   > n2 = Switch on Enum_AdditiveType!(execute, Selection:Enum_AdditiveType Enum=@n1) -> (NewEnumerator0:Exec, NewEnumerator1:Exec, NewEnumerator2:Exec, NewEnumerator3:Exec, NewEnumerator4:Exec) {is=title K2Node_SwitchEnum}
   > Set FootShift_1stFoot_ProcLiftCurrentlyActive?(FootShift_1stFoot_ProcLiftCurrentlyActive?=false)
   > Set FootShift_2ndFoot_ProcLiftCurrentlyActive?(FootShift_2ndFoot_ProcLiftCurrentlyActive?=false)
   > Set FootShift_1stFoot_ProcTranslateCurrentlyActive?(FootShift_1stFoot_ProcTranslateCurrentlyActive?=false)
   > n3 = Set FootShift_2ndFoot_ProcTranslateCurrentlyActive?(FootShift_2ndFoot_ProcTranslateCurrentlyActive?=false)
   Outputs!(OutputPin:Exec=[@n4.then, @n3.then, @n6.then, @n5.then, @n7.then])
   n4 = Set HeadMotion_CurrentlyActive?(execute=@n2.NewEnumerator1, HeadMotion_CurrentlyActive?=false)
   n5 = Set LeftArm_CurrentlyActive?(execute=@n2.NewEnumerator2, LeftArm_CurrentlyActive?=false)
   n6 = Set RightArm_CurrentlyActive?(execute=@n2.NewEnumerator3, RightArm_CurrentlyActive?=false)
   n7 = Set SpineMotion_CurrentlyActive?(execute=@n2.NewEnumerator4, SpineMotion_CurrentlyActive?=false)
  ]
 }
 Reset System on all Body Sections {
  path=EventGraph.K2Node_Composite_12.Reset System on all Body Sections
  nodes [
   Inputs!() -> (In:Exec)
   > n1 = Sequence!(execute) -> (then_0, then_1, then_2:Exec, then_3:Exec, then_4:Exec)
   > Delay(Duration=Get FootShift_TimeInactive())
   > n2 = FootShift_ResetSystem()
   Outputs!(then:Exec=[@n3.then, @n2.then, @n6.then, @n5.then, @n4.then])
   Delay(execute=@n1.then_1, Duration=Get HeadMotion_TimeInactive())
   > n3 = HeadMotion_ResetSystem()
   Delay(execute=@n1.then_2, Duration=Get LeftArm_TimeInactive())
   > n4 = LeftArm_ResetSystem()
   Delay(execute=@n1.then_3, Duration=Get RightArm_TimeInactive())
   > n5 = RightArm_ResetSystem()
   Delay(execute=@n1.then_4, Duration=Get SpineMotion_TimeInactive())
   > n6 = SpineMotion_ResetSystem()
  ]
 }
 Initialize Body Section Variables {
  path=EventGraph.K2Node_Composite_13.Initialize Body Section Variables
  nodes [
   Inputs()
   > n1 = Keys!(execute, TargetMap:Map of Anim Sequence Base Object References to Anim Sequence Base Object References=Get FootShift_SequenceMap()) -> (then, Keys:Array of Anim Sequence Base Object References)
   > Set FootShift_SequenceArray(FootShift_SequenceArray:Array of Anim Sequence Base Object References=@n1) -> (Output_Get:Array of Anim Sequence Base Object References)
   > n2 = Keys(TargetMap=Get SpineMotion_SequenceMap())
   > Set SpineMotion_SequenceArray(SpineMotion_SequenceArray:Array of Anim Sequence Base Object References=@n2) -> (Output_Get:Array of Anim Sequence Base Object References)
   > Set LowerBody_NextSequence(LowerBody_NextSequence="Get (a copy)"(Array=Get LowerBody_SequenceArray()))
   > Set!(
    execute
    FootShift_Struct_SequenceArray_12_#45:Array of Anim Sequence Base Object References=Get FootShift_SequenceArray()
    FootShift_Struct_BlendWeight_14_#46:double=Get FootShift_BlendWeight()
    FootShift_Struct_CurrentlyActive?_15_#47:bool=Get FootShift_1stFoot_ProcLiftCurrentlyActive?()
    FootShift_Struct_TimeActive_16_#48:double=Get FootShift_TimeActive()
    FootShift_Struct_TimeInactive_17_#49:double=Get FootShift_TimeInactive()
    FootShift_Struct_UseOverrideSequence?_22_#50:bool=Get DEBUG_FootShift_UseOverrideSequence?()
    FootShift_Struct_OverrideSequence_25_#51:Anim Sequence Base Object Reference=Get DEBUG_FootShift_OverrideSequence()
    FootShift_Struct_OverrideBlendWeight_28_#52:double=Get DEBUG_FootShift_OverrideBlendWeight()
    FootShift_Struct_UseOverrideBlendWeight?_31_#53:bool=Get DEBUG_FootShift_UseOverrideBlendWeight?()
    FootShift_Struct_DisableCurve_34_#54:Name=Disable_JamV2_FootShiftAdditive
   ) -> (then, Output_Get:Sparks Jam V2 Additives Struct Structure)
   > Set!(
    execute
    HeadMotion_Struct_SequenceArray_12_#45:Array of Anim Sequence Base Object References=Get HeadMotion_SequenceArray()
    HeadMotion_Struct_BlendWeight_14_#46:double=Get HeadMotion_BlendWeight()
    HeadMotion_Struct_CurrentlyActive?_15_#47:bool=Get HeadMotion_CurrentlyActive?()
    HeadMotion_Struct_TimeActive_16_#48:double=Get HeadMotion_TimeActive()
    HeadMotion_Struct_TimeInactive_17_#49:double=Get HeadMotion_TimeInactive()
    HeadMotion_Struct_UseOverrideSequence?_22_#50:bool=Get DEBUG_HeadMotion_UseOverrideSequence?()
    HeadMotion_Struct_OverrideSequence_25_#51:Anim Sequence Base Object Reference=Get DEBUG_HeadMotion_OverrideSequence()
    HeadMotion_Struct_OverrideBlendWeight_28_#52:double=Get DEBUG_HeadMotion_OverrideBlendWeight()
    HeadMotion_Struct_UseOverrideBlendWeight?_31_#53:bool=Get DEBUG_HeadMotion_UseOverrideBlendWeight?()
    HeadMotion_Struct_DisableCurve_34_#54:Name=Disable_JamV2_HeadMotionReplace
   ) -> (then, Output_Get)
   > Set!(
    execute
    LeftArm_Struct_SequenceArray_12_#45:Array of Anim Sequence Base Object References=Get LeftArm_SequenceArray()
    LeftArm_Struct_BlendWeight_14_#46:double=Get LeftArm_BlendWeight()
    LeftArm_Struct_CurrentlyActive?_15_#47:bool=Get LeftArm_CurrentlyActive?()
    LeftArm_Struct_TimeActive_16_#48:double=Get LeftArm_TimeActive()
    LeftArm_Struct_TimeInactive_17_#49:double=Get LeftArm_TimeInactive()
    LeftArm_Struct_UseOverrideSequence?_22_#50:bool=Get DEBUG_LeftArm_UseOverrideSequence?()
    LeftArm_Struct_OverrideSequence_25_#51:Anim Sequence Base Object Reference=Get DEBUG_LeftArm_OverrideSequence()
    LeftArm_Struct_OverrideBlendWeight_28_#52:double=Get DEBUG_LeftArm_OverrideBlendWeight()
    LeftArm_Struct_UseOverrideBlendWeight?_31_#53:bool=Get DEBUG_LeftArm_UseOverrideBlendWeight?()
    LeftArm_Struct_DisableCurve_34_#54:Name=Disable_JamV2_LeftArmReplace
   ) -> (then, Output_Get)
   > Set!(
    execute
    RightArm_Struct_SequenceArray_12_#45:Array of Anim Sequence Base Object References=Get RightArm_SequenceArray()
    RightArm_Struct_BlendWeight_14_#46:double=Get RightArm_BlendWeight()
    RightArm_Struct_CurrentlyActive?_15_#47:bool=Get RightArm_CurrentlyActive?()
    RightArm_Struct_TimeActive_16_#48:double=Get RightArm_TimeActive()
    RightArm_Struct_TimeInactive_17_#49:double=Get RightArm_TimeInactive()
    RightArm_Struct_UseOverrideSequence?_22_#50:bool=Get DEBUG_RightArm_UseOverrideSequence?()
    RightArm_Struct_OverrideSequence_25_#51:Anim Sequence Base Object Reference=Get DEBUG_RightArm_OverrideSequence()
    RightArm_Struct_OverrideBlendWeight_28_#52:double=Get DEBUG_RightArm_OverrideBlendWeight()
    RightArm_Struct_UseOverrideBlendWeight?_31_#53:bool=Get DEBUG_RightArm_UseOverrideBlendWeight?()
    RightArm_Struct_DisableCurve_34_#54:Name=Disable_JamV2_RightArmReplace
   ) -> (then, Output_Get)
   > Set!(
    execute
    SpineMotion_Struct_SequenceArray_12_#45:Array of Anim Sequence Base Object References=Get SpineMotion_SequenceArray()
    SpineMotion_Struct_BlendWeight_14_#46:double=Get SpineMotion_BlendWeight()
    SpineMotion_Struct_CurrentlyActive?_15_#47:bool=Get SpineMotion_CurrentlyActive?()
    SpineMotion_Struct_TimeActive_16_#48:double=Get SpineMotion_TimeActive()
    SpineMotion_Struct_TimeInactive_17_#49:double=Get SpineMotion_TimeInactive()
    SpineMotion_Struct_UseOverrideSequence?_22_#50:bool=Get DEBUG_SpineMotion_UseOverrideSequence?()
    SpineMotion_Struct_OverrideSequence_25_#51:Anim Sequence Base Object Reference=Get DEBUG_SpineMotion_OverrideSequence()
    SpineMotion_Struct_OverrideBlendWeight_28_#52:double=Get DEBUG_SpineMotion_OverrideBlendWeight()
    SpineMotion_Struct_UseOverrideBlendWeight?_31_#53:bool=Get DEBUG_SpineMotion_UseOverrideBlendWeight?()
    SpineMotion_Struct_DisableCurve_34_#54:Name=Disable_JamV2_SpineMotionAdditive
   ) -> (then, Output_Get)
   > Outputs!(Out:Exec)
  ]
 }
 FootShift Proc Set Mode Index {
  path=EventGraph.K2Node_Composite_14.FootShift Proc Set Mode Index
  nodes [
   Inputs!() -> (execute:Exec)
   > n1 = Sequence!(execute) -> (then_0, then_1)
   > n2 = Branch(Condition=Get DEBUG_FootShift_ProcUseOverrideModeIndex?() -> (DEBUG_FootShift_ProcUseOverrideModeIndex?:bool))
   > n3 = Set FootShift_ProcModeIndex(FootShift_ProcModeIndex:int=Get DEBUG_FootShift_ProcOverrideModeIndex() -> (DEBUG_FootShift_ProcOverrideModeIndex:int=0)) -> (Output_Get:int=0)
   n4 = Switch on Int!(execute:Exec=[@n3.then, @n5.then], Selection:int=Get FootShift_ProcModeIndex() -> (FootShift_ProcModeIndex=0)) -> (Default:Exec, 0:Exec, 1:Exec, 2:Exec)
   Set FootShift_Proc1stFootIsLeft?(execute=@n4.0, FootShift_Proc1stFootIsLeft?:bool=true) -> (Output_Get:bool)
   Set FootShift_Proc1stFootIsLeft?(execute=@n4.1, FootShift_Proc1stFootIsLeft?=false)
   Set FootShift_Proc1stFootIsLeft?(execute=@n4.2, FootShift_Proc1stFootIsLeft?=RandomBool!() -> (ReturnValue:bool))
   n5 = Set FootShift_ProcModeIndex(execute=@n2.else, FootShift_ProcModeIndex=RandomIntegerInRange(Min=0, Max=2) -> (ReturnValue=0)) -> (Output_Get=0)
   Outputs!(then:Exec=@n1.then_1)
  ]
 }
 Strum {
  path=EventGraph.K2Node_Composite_2.Strum
  nodes [
   Inputs!() -> (In:Exec)
   > n1 = Sequence!(execute) -> (then_0, then_1, then_2:Exec)
   n2 = Get StrumMidiListener() -> (StrumMidiListener:Async Action Listen For Midi Note Event Object Reference)
   > Is Valid(InputObject=@n2)
   > Cancel(self:Cancellable Async Action Object Reference=@n2)
   > Set StrumMidiListener(StrumMidiListener:Async Action Listen For Midi Note Event Object Reference) -> (Output_Get:Async Action Listen For Midi Note Event Object Reference)
   Branch(execute=@n1.then_1, Condition=IsValidInterface(Interface=Get MidiEventDriver() -> (MidiEventDriver:Midi Event Driver Interface Interface)))
   > n3 = ListenForMidiNoteEvents!(
    execute:Exec
    MidiEventDriver:Midi Event Driver Interface Interface=Get MidiEventDriver()
    MidiTrackName:Name=Get MidiTrackName() -> (MidiTrackName:Name=None)
    EventNotes:Array of Bytes=Make Array!(
     "[0]":Byte=GetGuitarMidiNoteIntegerFromEnum!(MidiNoteEvent:EGuitarMidiNoteEvent Enum=StrumDown) -> (ReturnValue:Byte=0)
     "[1]":Byte=GetGuitarMidiNoteIntegerFromEnum(MidiNoteEvent=StrumUp) -> (ReturnValue=0)
    ) -> (Array:Array of Bytes)
    SecondsBeforeEventStart:float=0.2
    SecondsBeforeEventEnd:float=0
    InitBroadcast:EMidiEventListenerInitBroadcast Enum=MostRecentEvent
   ) -> (
    then:Exec
    AsyncTaskProxy:Async Action Listen For Midi Note Event Object Reference
    OnMidiEventStart:Exec
    OnMidiEventEnd:Exec
    OnFinished:Exec
    MidiEventNote:Byte
    MidiEventData:"Parsed Midi Note Event Structure (by ref)"
   ) {is=title K2Node_AsyncAction}
   > Set StrumMidiListener(StrumMidiListener=@n3)
   Guitar Strum(execute=@n3.OnMidiEventStart)
   Outputs!(Out:Exec=@n1.then_2)
  ]
 }
 Song Intensity {
  path=EventGraph.K2Node_Composite_3.Song Intensity
  nodes [
   Inputs()
   > n1 = Sequence()
   n2 = Get IntensityMidiListener() -> (IntensityMidiListener:Async Action Listen For Midi Text Event Object Reference)
   > Is Valid(InputObject=@n2)
   > Cancel(self=@n2)
   > Set IntensityMidiListener(IntensityMidiListener:Async Action Listen For Midi Text Event Object Reference) -> (Output_Get:Async Action Listen For Midi Text Event Object Reference)
   Branch(execute=@n1.then_1, Condition=IsValidInterface(Interface=Get MidiEventDriver()))
   > n3 = IntensityMidiTextEvents!(execute) -> (then, Events:Array of Strings)
   > n4 = ListenForMidiTextEvents!(
    execute:Exec
    MidiEventDriver:Midi Event Driver Interface Interface=Get MidiEventDriver()
    MidiTrackName:Name=Get MidiTrackName() -> (MidiTrackName=None)
    EventStrings:Array of Strings=@n3
    SecondsBeforeEventStart:float=Get SecBeforeIntensity() -> (SecBeforeIntensity:float)
    InitBroadcast:EMidiEventListenerInitBroadcast Enum=MostRecentEvent
   ) -> (
    then:Exec
    AsyncTaskProxy:Async Action Listen For Midi Text Event Object Reference
    OnMidiEventStart:Exec
    OnFinished:Exec
    MidiEventText:String
    MidiEventData:"Parsed Midi Text Event Structure (by ref)"
   ) {is=title K2Node_AsyncAction}
   > Set IntensityMidiListener(IntensityMidiListener=@n4)
   MidiEvent_Intensity_OnEvent!(execute=@n4.OnMidiEventStart, self, MidiEventText:String=@n4.MidiEventText) -> (then)
   Outputs(Out=@n1.then_2)
  ]
 }
 Fret Position {
  path=EventGraph.K2Node_Composite_4.Fret Position
  nodes [
   Inputs()
   > n1 = Sequence()
   n2 = Get FretMidiListener() -> (FretMidiListener:Async Action Listen For Midi Note Event Object Reference)
   > Is Valid(InputObject=@n2)
   > Cancel(self=@n2)
   > Set FretMidiListener(FretMidiListener:Async Action Listen For Midi Note Event Object Reference) -> (Output_Get:Async Action Listen For Midi Note Event Object Reference)
   Branch(execute=@n1.then_1, Condition=IsValidInterface(Interface=Get MidiEventDriver()))
   > n3 = ListenForMidiNoteEvents(
    MidiEventDriver=Get MidiEventDriver()
    MidiTrackName=Get MidiTrackName() -> (MidiTrackName=None)
    EventNotes=GetGuitarMidiNoteIntegersFromEnums!(MidiNoteEvents:Array of EGuitarMidiNoteEvent Enums=Get MidiFretEvents() -> (MidiFretEvents:Array of EGuitarMidiNoteEvent Enums=None)) -> (ReturnValue:Array of Bytes)
    SecondsBeforeEventStart=0
    SecondsBeforeEventEnd=0
    InitBroadcast=MostRecentEvent
   )
   > Set FretMidiListener(FretMidiListener=@n3)
   Set FretNumber(execute=@n3.OnMidiEventStart, FretNumber:int=GetFretNumberFromMidiNote!(MidiNote:Byte=@n3.MidiEventNote) -> (ReturnValue:int=0)) -> (Output_Get:int=0)
   > Change Chord Hand Pose() {is=title K2Node_CallFunction}
   Outputs(Out=@n1.then_2)
  ]
 }
 FootShift Proc Set Transform Values {
  path=EventGraph.K2Node_Composite_5.FootShift Proc Set Transform Values
  nodes [
   Branch(Condition=OR Boolean(
     A=Get FootShift_ProcUseInwardKnees?() -> (FootShift_ProcUseInwardKnees?:bool)
     B=Get DEBUG_FootShift_ProcForceInwardKnees() -> (DEBUG_FootShift_ProcForceInwardKnees:bool)
    ))
   > Set FootShift_2ndFoot_ProcRotateTwist(FootShift_2ndFoot_ProcRotateTwist:double=float * float(A=Get FootShift_2ndFoot_ProcRotateTwist(), B=-1) {is=title K2Node_PromotableOperator}) -> (Output_Get:double)
   Inputs()
   > n1 = Sequence!(execute) -> (then_0, then_1, then_2, then_3:Exec, then_4:Exec)
   > n2 = Set FootShift_1stFoot_ProcTranslateDistance(FootShift_1stFoot_ProcTranslateDistance:double=RandomFloatInRange(Min=Get FootShift_1stFoot_ProcTranslateDistanceMin(), Max=Get FootShift_1stFoot_ProcTranslateDistanceMax())) -> (Output_Get:double)
   > Set FootShift_1stFoot_ProcTranslateDistance(FootShift_1stFoot_ProcTranslateDistance=float * float(A=Get DEBUG_FootShift_ProcTranslateDistanceMultiplier() -> (DEBUG_FootShift_ProcTranslateDistanceMultiplier:double), B=@n2))
   n3 = Square!(A:double=@n2) -> (ReturnValue:double)
   n4 = RandomFloatInRange(Min=0, Max=@n3)
   > Set FootShift_1stFoot_ProcTranslateLR(FootShift_1stFoot_ProcTranslateLR:double=Sqrt!(A:double=@n4) -> (ReturnValue:double)) -> (Output_Get:double)
   > n5 = Branch(Condition=Get FootShift_Proc1stFootIsLeft?())
   Set FootShift_1stFoot_ProcTranslateFB(execute=[@n9.then, @n5.then], FootShift_1stFoot_ProcTranslateFB:double=Sqrt(A=float - float(A=@n3, B=@n4))) -> (Output_Get:double)
   > n6 = Branch(Condition=RandomBool())
   Set FootShift_1stFoot_ProcRotateTwist(
    execute=[@n6.then, @n8.then]
    FootShift_1stFoot_ProcRotateTwist:double=float * float(A=Get FootShift_1stFoot_ProcTranslateDistance(), B=-1.25)
   ) -> (Output_Get:double)
   > n7 = Branch(Condition=Get FootShift_Proc1stFootIsLeft?())
   > Reroute Node()
   Set FootShift_1stFoot_ProcRotateTwist(execute=@n7.else, FootShift_1stFoot_ProcRotateTwist=float * float(A=Get FootShift_1stFoot_ProcRotateTwist(), B=-1))
   n8 = Set FootShift_1stFoot_ProcTranslateFB(execute=@n6.else, FootShift_1stFoot_ProcTranslateFB=float * float(A=Get FootShift_1stFoot_ProcTranslateFB(), B=-1))
   n9 = Set FootShift_1stFoot_ProcTranslateLR(execute=@n5.else, FootShift_1stFoot_ProcTranslateLR=float * float(A=Get FootShift_1stFoot_ProcTranslateLR(), B=-1))
   Set FootShift_1stPelvis_ProcTranslateLR(execute=@n1.then_2, FootShift_1stPelvis_ProcTranslateLR:double=float / float(A=Get FootShift_1stFoot_ProcTranslateLR(), B=2)) -> (Output_Get:double)
   > Set FootShift_1stPelvis_ProcTranslateFB(FootShift_1stPelvis_ProcTranslateFB:double=float / float(A=Get FootShift_1stFoot_ProcTranslateFB(), B=2)) -> (Output_Get:double)
   > Set FootShift_1stPelvis_ProcTranslateVertical(FootShift_1stPelvis_ProcTranslateVertical:double=float * float(A=Get FootShift_1stFoot_ProcTranslateDistance(), B=-0.25)) -> (Output_Get:double)
   > Set FootShift_1stPelvis_ProcRotateTwist(FootShift_1stPelvis_ProcRotateTwist:float=float + float(A=Get FootShift_1stFoot_ProcRotateTwist(), B=Get FootShift_2ndFoot_ProcRotateTwist())) -> (Output_Get:float)
   Branch(execute=@n1.then_1, Condition="Equal (Integer)"(A=Get FootShift_ProcModeIndex() -> (FootShift_ProcModeIndex=0), B=2))
   > n10 = Set FootShift_2ndFoot_ProcTranslateDistance(FootShift_2ndFoot_ProcTranslateDistance:double=RandomFloatInRange(Min=Get FootShift_2ndFoot_ProcTranslateDistanceMin(), Max=Get FootShift_2ndFoot_ProcTranslateDistanceMax())) -> (Output_Get:double)
   > Set FootShift_2ndFoot_ProcTranslateDistance(FootShift_2ndFoot_ProcTranslateDistance=float * float(A=Get DEBUG_FootShift_ProcTranslateDistanceMultiplier(), B=@n10))
   n11 = Square(A=@n10)
   n12 = RandomFloatInRange(Min=0, Max=@n11)
   > Set FootShift_2ndFoot_ProcTranslateLR(FootShift_2ndFoot_ProcTranslateLR:double=Sqrt(A=@n12)) -> (Output_Get:double)
   > n13 = Branch(Condition=Get FootShift_Proc1stFootIsLeft?())
   > n14 = Set FootShift_2ndFoot_ProcTranslateLR(FootShift_2ndFoot_ProcTranslateLR=float * float(A=Get FootShift_2ndFoot_ProcTranslateLR(), B=-1))
   Set FootShift_2ndFoot_ProcTranslateFB(execute=[@n14.then, @n13.else], FootShift_2ndFoot_ProcTranslateFB:double=Sqrt(A=float - float(A=@n11, B=@n12))) -> (Output_Get:double)
   > n15 = Branch(Condition=RandomBool())
   Set FootShift_2ndFoot_ProcRotateTwist(execute=[@n15.then, @n18.then], FootShift_2ndFoot_ProcRotateTwist=float * float(A=Get FootShift_2ndFoot_ProcTranslateDistance(), B=-1.25))
   > n16 = Branch(Condition=Get FootShift_Proc1stFootIsLeft?())
   > n17 = Set FootShift_2ndFoot_ProcRotateTwist(FootShift_2ndFoot_ProcRotateTwist=float * float(A=Get FootShift_2ndFoot_ProcRotateTwist(), B=-1))
   Reroute Node(InputPin=[@n17.then, @n16.else])
   n18 = Set FootShift_2ndFoot_ProcTranslateFB(execute=@n15.else, FootShift_2ndFoot_ProcTranslateFB=float * float(A=Get FootShift_2ndFoot_ProcTranslateFB(), B=-1))
   Branch(execute=@n1.then_3, Condition="Equal (Integer)"(A=Get FootShift_ProcModeIndex() -> (FootShift_ProcModeIndex=0), B=2))
   > Set FootShift_2ndPelvis_ProcTranslateLR(FootShift_2ndPelvis_ProcTranslateLR:double=float / float(A=Get FootShift_2ndFoot_ProcTranslateLR(), B=2)) -> (Output_Get:double)
   > Set FootShift_2ndPelvis_ProcTranslateFB(FootShift_2ndPelvis_ProcTranslateFB:double=float / float(A=Get FootShift_2ndFoot_ProcTranslateFB(), B=2)) -> (Output_Get:double)
   > Set FootShift_2ndPelvis_ProcTranslateVertical(FootShift_2ndPelvis_ProcTranslateVertical:double=float * float(A=Get FootShift_2ndFoot_ProcTranslateDistance(), B=-0.25)) -> (Output_Get:double)
   Outputs(Out=@n1.then_4)
   n19 = Set FootShift_ProcUseInwardKnees?(FootShift_ProcUseInwardKnees?:bool=RandomBoolWithWeight!(Weight:float=0.25) -> (ReturnValue:bool)) -> (Output_Get:bool)
   > Branch(Condition=OR Boolean(A=@n19, B=Get DEBUG_FootShift_ProcForceInwardKnees()))
   > Set FootShift_1stFoot_ProcRotateTwist(FootShift_1stFoot_ProcRotateTwist=float * float(A=Get FootShift_1stFoot_ProcRotateTwist(), B=-1))
   Set FootShift_2ndPelvis_ProcRotateTwist(FootShift_2ndPelvis_ProcRotateTwist:float=float + float!(A=Get FootShift_1stFoot_ProcRotateTwist(), B=Get FootShift_2ndFoot_ProcRotateTwist(), C:double) -> (ReturnValue)) -> (Output_Get:float)
  ]
 }
 Choose Strum Montage {
  path=EventGraph.K2Node_Composite_7.Choose Strum Montage
  nodes [
   Inputs!() -> (execute:Exec)
   > n1 = Switch on ESparksInstrumentType!(execute, Selection:ESparksInstrumentType Enum=Get Instrument() -> (Instrument=Guitar)) -> (Guitar:Exec, Bass:Exec, Vocals:Exec, Drum:Exec, Keyboard:Exec, None:Exec) {is=title K2Node_SwitchEnum}
   n2 = Montage_Play!(
    execute=[@n1.Drum, @n1.Guitar, @n1.None, @n1.Vocals]
    self:Anim Instance Object Reference
    MontageToPlay:Anim Montage Object Reference=/SparksCharacterCommon/Animation/MainPlayer/BandPlay/Guitar/Strums/Guitar_Strum_Test_A_M
    InPlayRate:float=1
    ReturnValueType:EMontagePlayReturnType Enum=MontageLength
    InTimeToStartMontageAt:float=0
    bStopAllMontages:bool=true
   ) -> (then, ReturnValue:float)
   Outputs!(OutputPin:Exec=[@n4.then, @n6.then, @n5.then, @n2.then])
   n3 = Flip Flop!(None:Exec=@n1.Bass) -> (A:Exec, B:Exec, IsA:bool=~) {is=title K2Node_MacroInstance}
   > n4 = Montage_Play(
    MontageToPlay=/SparksCharacterCommon/Animation/MainPlayer/BandPlay/Bass/Strums/Bass_Strum_PluckMiddle_M
    InPlayRate=1
    ReturnValueType=MontageLength
    InTimeToStartMontageAt=0
    bStopAllMontages=true
   )
   n5 = Montage_Play(
    execute=@n3.B
    MontageToPlay=/SparksCharacterCommon/Animation/MainPlayer/BandPlay/Bass/Strums/Bass_Strum_PluckIndex_M
    InPlayRate=1
    ReturnValueType=MontageLength
    InTimeToStartMontageAt=0
    bStopAllMontages=true
   )
   n6 = Montage_Play(
    execute=@n1.Keyboard
    MontageToPlay=/SparksCharacterCommon/Animation/MainPlayer/BandPlay/Keytar/Strums/Keytar_Strum_Down_A_M
    InPlayRate=1
    ReturnValueType=MontageLength
    InTimeToStartMontageAt=0
    bStopAllMontages=true
   )
  ]
 }
 FootShift Proc Set Sequences {
  path=EventGraph.K2Node_Composite_8.FootShift Proc Set Sequences
  nodes [
   Inputs!() -> (In:Exec)
   > n1 = Sequence!(execute) -> (then_0, then_1)
   > n2 = Switch on Int!(execute, Selection=Get FootShift_ProcModeIndex() -> (FootShift_ProcModeIndex=0)) -> (0, 1, 2) {bHasDefaultPin=false}
   n3 = Branch(execute=@n2.2, Condition=Get FootShift_Proc1stFootIsLeft?())
   > Set FootShift_1stFoot_ProcCurrentSequence(FootShift_1stFoot_ProcCurrentSequence:Anim Sequence Base Object Reference=Random Array Item(TargetArray=Get FootShift_L_ProcSequenceArray() -> (FootShift_L_ProcSequenceArray:Array of Anim Sequence Base Object References)) -> (OutIndex=0)) -> (Output_Get:Anim Sequence Base Object Reference)
   > Set FootShift_2ndFoot_ProcCurrentSequence(FootShift_2ndFoot_ProcCurrentSequence:Anim Sequence Base Object Reference=Random Array Item(TargetArray=Get FootShift_R_ProcSequenceArray() -> (FootShift_R_ProcSequenceArray:Array of Anim Sequence Base Object References)) -> (OutIndex=0)) -> (Output_Get:Anim Sequence Base Object Reference)
   Set FootShift_1stFoot_ProcCurrentSequence(execute=@n3.else, FootShift_1stFoot_ProcCurrentSequence=Random Array Item(TargetArray=Get FootShift_R_ProcSequenceArray()) -> (OutIndex=0))
   > Set FootShift_2ndFoot_ProcCurrentSequence(FootShift_2ndFoot_ProcCurrentSequence=Random Array Item(TargetArray=Get FootShift_L_ProcSequenceArray()) -> (OutIndex=0))
   Set FootShift_1stFoot_ProcCurrentSequence(execute=@n2.1, FootShift_1stFoot_ProcCurrentSequence=Random Array Item(TargetArray=Get FootShift_R_ProcSequenceArray()) -> (OutIndex=0))
   Set FootShift_1stFoot_ProcCurrentSequence(execute=@n2.0, FootShift_1stFoot_ProcCurrentSequence=Random Array Item(TargetArray=Get FootShift_L_ProcSequenceArray()) -> (OutIndex=0))
   Outputs!(Out:Exec=@n1.then_1)
  ]
 }
 Update Damplify Values {
  path=EventGraph.K2Node_Composite_9.Update Damplify Values
  nodes [
   n1 = Inputs!() -> (In, AdditiveType:Enum_AdditiveType Enum=NewEnumerator0)
   > n2 = Switch on Enum_AdditiveType(Selection=@n1)
   > n3 = Branch(Condition=Get DEBUG_FootShift_UseOverrideDamplifyMultiplier?() -> (DEBUG_FootShift_UseOverrideDamplifyMultiplier?:bool))
   > n4 = Set FootShift_DamplifyMultiplier(FootShift_DamplifyMultiplier:float=Get DEBUG_FootShift_DamplifyMultiplier() -> (DEBUG_FootShift_DamplifyMultiplier:float)) -> (Output_Get:float)
   n5 = Branch(execute=[@n8.then, @n4.then], Condition="float >= float"(A=Get FootShift_DamplifyMultiplier() -> (FootShift_DamplifyMultiplier:float), B=1))
   > Set FootShift_DampenAlpha(FootShift_DampenAlpha:float=1) -> (Output_Get:float)
   > n6 = Set FootShift_AmplifyAlpha(FootShift_AmplifyAlpha:float=float - float(A=Get FootShift_DamplifyMultiplier(), B=1)) -> (Output_Get:float)
   Outputs(Out=[@n6.then, @n7.then, @n29.then, @n28.then, @n12.then, @n13.then, @n18.then, @n19.then, @n23.then, @n24.then])
   Set FootShift_DampenAlpha(execute=@n5.else, FootShift_DampenAlpha="% (Float)"!(A:double=Get FootShift_DamplifyMultiplier(), B:double=1) -> (ReturnValue:double))
   > n7 = Set FootShift_AmplifyAlpha(FootShift_AmplifyAlpha=0)
   n8 = Set FootShift_DamplifyMultiplier(execute=@n3.else, FootShift_DamplifyMultiplier=RandomFloatInRange(Min=0.9, Max=1.1))
   n9 = Branch(
    execute=@n2.NewEnumerator1
    Condition=Get DEBUG_HeadMotion_UseOverrideDamplifyMultiplier?() -> (DEBUG_HeadMotion_UseOverrideDamplifyMultiplier?:bool)
   )
   > n10 = Set HeadMotion_DamplifyMultiplier(HeadMotion_DamplifyMultiplier:float=Get DEBUG_HeadMotion_DamplifyMultiplier() -> (DEBUG_HeadMotion_DamplifyMultiplier:float)) -> (Output_Get:float)
   n11 = Branch(
    execute=[@n10.then, @n14.then]
    Condition="float >= float"(A=Get HeadMotion_DamplifyMultiplier() -> (HeadMotion_DamplifyMultiplier:float), B=1)
   )
   > Set HeadMotion_DampenAlpha(HeadMotion_DampenAlpha:float=1) -> (Output_Get:float)
   > n12 = Set HeadMotion_AmplifyAlpha(HeadMotion_AmplifyAlpha:float=float - float(A=Get HeadMotion_DamplifyMultiplier(), B=1)) -> (Output_Get:float)
   Set HeadMotion_DampenAlpha(execute=@n11.else, HeadMotion_DampenAlpha="% (Float)"(A=Get HeadMotion_DamplifyMultiplier(), B=1))
   > n13 = Set HeadMotion_AmplifyAlpha(HeadMotion_AmplifyAlpha=0)
   n14 = Set HeadMotion_DamplifyMultiplier(execute=@n9.else, HeadMotion_DamplifyMultiplier=RandomFloatInRange(Min=0.9, Max=1.1))
   n15 = Branch(
    execute=@n2.NewEnumerator2
    Condition=Get DEBUG_LeftArm_UseOverrideDamplifyMultiplier?() -> (DEBUG_LeftArm_UseOverrideDamplifyMultiplier?:bool)
   )
   > n16 = Set LeftArm_DamplifyMultiplier(LeftArm_DamplifyMultiplier:float=Get DEBUG_LeftArm_DamplifyMultiplier() -> (DEBUG_LeftArm_DamplifyMultiplier:float)) -> (Output_Get:float)
   n17 = Branch(execute=[@n20.then, @n16.then], Condition="float >= float"(A=Get LeftArm_DamplifyMultiplier() -> (LeftArm_DamplifyMultiplier:float), B=1))
   > Set LeftArm_DampenAlpha(LeftArm_DampenAlpha:float=1) -> (Output_Get:float)
   > n18 = Set LeftArm_AmplifyAlpha(LeftArm_AmplifyAlpha:float=float - float(A=Get LeftArm_DamplifyMultiplier(), B=1)) -> (Output_Get:float)
   Set LeftArm_DampenAlpha(execute=@n17.else, LeftArm_DampenAlpha="% (Float)"(A=Get LeftArm_DamplifyMultiplier(), B=1))
   > n19 = Set LeftArm_AmplifyAlpha(LeftArm_AmplifyAlpha=0)
   n20 = Set LeftArm_DamplifyMultiplier(execute=@n15.else, LeftArm_DamplifyMultiplier=RandomFloatInRange(Min=0.9, Max=1.1))
   n21 = Branch(
    execute=@n2.NewEnumerator3
    Condition=Get DEBUG_RightArm_UseOverrideDamplifyMultiplier?() -> (DEBUG_RightArm_UseOverrideDamplifyMultiplier?:bool)
   )
   > Set RightArm_DamplifyMultiplier(RightArm_DamplifyMultiplier:float=Get DEBUG_RightArm_DamplifyMultiplier() -> (DEBUG_RightArm_DamplifyMultiplier:float)) -> (Output_Get:float)
   Set RightArm_DamplifyMultiplier(execute=@n21.else, RightArm_DamplifyMultiplier=RandomFloatInRange(Min=0.9, Max=1.1))
   > n22 = Branch(Condition="float >= float"(A=Get RightArm_DamplifyMultiplier() -> (RightArm_DamplifyMultiplier:float), B=1))
   > Set RightArm_DampenAlpha(RightArm_DampenAlpha:float=1) -> (Output_Get:float)
   > n23 = Set RightArm_AmplifyAlpha(RightArm_AmplifyAlpha:float=float - float(A=Get RightArm_DamplifyMultiplier(), B=1)) -> (Output_Get:float)
   Set RightArm_DampenAlpha(execute=@n22.else, RightArm_DampenAlpha="% (Float)"(A=Get RightArm_DamplifyMultiplier(), B=1))
   > n24 = Set RightArm_AmplifyAlpha(RightArm_AmplifyAlpha=0)
   n25 = Branch(
    execute=@n2.NewEnumerator4
    Condition=Get DEBUG_SpineMotion_UseOverrideDamplifyMultiplier?() -> (DEBUG_SpineMotion_UseOverrideDamplifyMultiplier?:bool)
   )
   > n26 = Set SpineMotion_DamplifyMultiplier(SpineMotion_DamplifyMultiplier:float=Get DEBUG_SpineMotion_DamplifyMultiplier() -> (DEBUG_SpineMotion_DamplifyMultiplier:float)) -> (Output_Get:float)
   n27 = Branch(
    execute=[@n26.then, @n30.then]
    Condition="float >= float"(A=Get SpineMotion_DamplifyMultiplier() -> (SpineMotion_DamplifyMultiplier:float), B=1)
   )
   > Set SpineMotion_DampenAlpha(SpineMotion_DampenAlpha:float=1) -> (Output_Get:float)
   > n28 = Set SpineMotion_AmplifyAlpha(SpineMotion_AmplifyAlpha:float=float - float(A=Get SpineMotion_DamplifyMultiplier(), B=1)) -> (Output_Get:float)
   Set SpineMotion_DampenAlpha(execute=@n27.else, SpineMotion_DampenAlpha="% (Float)"(A=Get SpineMotion_DamplifyMultiplier(), B=1))
   > n29 = Set SpineMotion_AmplifyAlpha(SpineMotion_AmplifyAlpha=0)
   n30 = Set SpineMotion_DamplifyMultiplier(execute=@n25.else, SpineMotion_DamplifyMultiplier=RandomFloatInRange(Min=0.9, Max=1.1))
  ]
 }
 FootShift_PrintBoneTransforms {nodes [
   n1 = FootShift_PrintBoneTransforms!() -> (then, InBoneName:Name) {is=title K2Node_FunctionEntry}
   n2 = GetBoneTransform!(
    self:Skinned Mesh Component Object Reference=GetOwningComponent()
    InBoneName:Name=@n1
    TransformSpace:ERelativeTransformSpace Enum=RTS_Component
   ) -> (
    ReturnValue_Location_X:double
    ReturnValue_Location_Y:double
    ReturnValue_Location_Z:double
    ReturnValue_Rotation_Roll:float
    ReturnValue_Rotation_Pitch:float
    ReturnValue_Rotation_Yaw:float
    ReturnValue_Scale:Vector
   )
   > PrintString(
    InString=Append(
     A="\r\nbone: "
     B="To String (Name)"(InName=@n1)
     C="\r\nTranslateX: "
     D="To String (Float)"(InDouble=@n2)
     E="\r\nTranslateY: "
     F="To String (Float)"(InDouble=@n2.ReturnValue_Location_Y)
     G="\r\nRotateZ: "
     H="To String (Float)"(InDouble=@n2.ReturnValue_Rotation_Yaw)
    )
    TextColor=(R=0.0438,G=1,B=0,A=1)
    Duration=4
   )
  ]}
 JamV2_CurveValue_Printer {nodes [
   n1 = JamV2_CurveValue_Printer!() -> (then, CurveName:Name) {is=title K2Node_FunctionEntry}
   n2 = Get DEBUG_PrintAnimLogic()
   n3 = GetCurveValue(CurveName=@n1)
   > PrintString(
    InString=Append!(A="To String (Name)"(InName=@n1), B=".CurveValue=", C="To String (Float)"(InDouble=@n3)) -> (ReturnValue)
    bPrintToScreen=@n2
    bPrintToLog=@n2
    Duration=Get DeltaTimeX()
   )
   > Return Node()
  ]}
 JamV2_PickSequence {nodes [
   n1 = JamV2_PickSequence!() -> (
    then
    UseOverrideSequence?:bool
    SequenceArray:Array of Anim Sequence Base Object References
    OverrideSequence:Anim Sequence Base Object Reference
    PreviousSequence:Anim Sequence Base Object Reference
    PreventRepeatSequence?:bool=~
   ) {is=title K2Node_FunctionEntry}
   > n2 = Branch(Condition=@n1)
   > n3 = Set l_CurrentSequence(l_CurrentSequence:Anim Sequence Base Object Reference=@n1.OverrideSequence) -> (Output_Get:Anim Sequence Base Object Reference)
   Return Node!(
    execute=[@n4.else, @n5.else, @n3.then]
    CurrentSequence:Anim Sequence Base Object Reference=Get l_CurrentSequence() -> (l_CurrentSequence:Anim Sequence Base Object Reference)
   )
   Set l_CurrentSequence(execute=[@n5.then, @n2.else], l_CurrentSequence=Random Array Item(TargetArray=@n1.SequenceArray) -> (OutIndex=0))
   > n4 = Branch(Condition=@n1.PreventRepeatSequence?)
   > n5 = Branch(Condition="Equal (Object)"(A=Get l_CurrentSequence(), B=@n1.PreviousSequence))
  ]}
 JamV2_SineWaveMotionOcsillator {nodes [
   n1 = JamV2_SineWaveMotionOcsillator!() -> (then, MaxValue:double=~, "Frequency (Speed) Multiplier":double=1, "Amplitude (Min && Max) Multiplier":double=1) {is=title K2Node_FunctionEntry}
   Set SineWave(execute=[@n1.then, @n4.else], SineWave:double=float - float(A=GetGameTimeInSeconds(), B=Get GameTime_Stored() -> (GameTime_Stored:double))) -> (Output_Get:double)
   n2 = GetGameTimeInSeconds()
   n3 = NormalizeToRange(Value="Sin (Radians)"(A=float * float(A=@n1."Frequency (Speed) Multiplier", B=@n2)), RangeMin=-1, RangeMax=1)
   > Return Node!(execute, ReturnValue:double=float * float(A=@n1, B=float * float(A=@n3, B=@n1."Amplitude (Min && Max) Multiplier")))
   n4 = Branch(Condition=Get AllowSineWaveReset?() -> (AllowSineWaveReset?:bool))
   > Do Once()
   > Set GameTime_Stored(GameTime_Stored:double=GetGameTimeInSeconds()) -> (Output_Get:double)
   > Set AllowSineWaveReset?(AllowSineWaveReset?:bool=false) -> (Output_Get:bool)
  ]}
 MidiEvent_Intensity_OnEvent {nodes [
   n1 = MidiEvent_Intensity_OnEvent!() -> (then, MidiEventText:String) {is=title K2Node_FunctionEntry}
   > n2 = Sequence()
   > n3 = TranslateMidiEventTextToSongIntensity!(execute, MidiEventText:Name=String To Name(InString=@n1)) -> (then, IdleRealtime:bool, MidiSongIntensityIsIdle:bool, SongIntensity:Gameplay Tag Structure)
   > n4 = Branch(Condition=@n3)
   Set DesiredSongIntensity(execute=@n4.else, DesiredSongIntensity:Gameplay Tag Structure=@n3.SongIntensity) -> (Output_Get:Gameplay Tag Structure)
   MidiEvent_PrintStringAssembler!(
    execute=@n2.then_1
    self
    InstrumentType:ESparksInstrumentType Enum=Get InstrumentType() -> (InstrumentType=Guitar)
    MidiEventText:String=@n1
    SecondsBeforeEvent:double=Get SecBeforeIntensity()
   ) -> (then)
  ]}
 MidiEvent_PrintStringAssembler {nodes [
   n1 = MidiEvent_PrintStringAssembler!() -> (then, InstrumentType:ESparksInstrumentType Enum=Guitar, MidiEventText:String, SecondsBeforeEvent:double=~) {is=title K2Node_FunctionEntry}
   > Branch(Condition=Get DEBUG_PrintAnimLogic())
   > n2 = Branch(Condition="Equal (Float)"(A=@n1.SecondsBeforeEvent))
   > n3 = Set Duration(Duration:float=2) -> (Output_Get:float)
   n4 = Get DEBUG_PrintAnimLogic()
   n5 = "To String (Float)"(InDouble=@n1.SecondsBeforeEvent)
   n6 = Enum to String(Enumerator=@n1)
   PrintString(
    execute=[@n3.then, @n7.then]
    InString=JoinStringArray(
     SourceArray=Make Array!("[0]":String="[Anim Midi]", "[1]":String=@n6, "[2]":String=@n1.MidiEventText, "[3]":String=in, "[4]":String=@n5, "[5]":String=seconds...) -> (Array:Array of Strings)
     Separator=" "
    )
    bPrintToScreen=@n4
    bPrintToLog=@n4
    TextColor=Get AnimNotify_TextColor()
    Duration=Get Duration() -> (Duration:float)
   )
   > Return Node!(execute)
   n7 = Set Duration(execute=@n2.else, Duration=@n1.SecondsBeforeEvent)
  ]}
 OnPlay0Entry {nodes [
   OnPlay0Entry!() -> (then, Context:"Anim Update Context Structure (by ref)", Node:"Anim Node Reference Structure (by ref)") {is=title K2Node_FunctionEntry}
   > SetPlayAnimAndTempoMultiplier!(execute, self, Index:int=0) -> (then)
   > Return Node()
  ]}
 OnPlay1Entry {nodes [
   OnPlay1Entry!() -> (then, Context:"Anim Update Context Structure (by ref)", Node:"Anim Node Reference Structure (by ref)") {is=title K2Node_FunctionEntry}
   > SetPlayAnimAndTempoMultiplier(Index=1)
   > Return Node()
  ]}
 SetPlayAnimAndTempoMultiplier {nodes [
   n1 = SetPlayAnimAndTempoMultiplier!() -> (then, Index:int) {is=title K2Node_FunctionEntry}
   > Set SongIntensity(SongIntensity:Gameplay Tag Structure=Get DesiredSongIntensity() -> (DesiredSongIntensity:Gameplay Tag Structure)) -> (Output_Get:Gameplay Tag Structure)
   n2 = Select(
    Option 0:Sparks Anim BPMSet Structure=Get HomebaseBPMs() -> (HomebaseBPMs:Sparks Anim BPMSet Structure)
    Option 1:Sparks Anim BPMSet Structure=Get DEBUG_BaseBody_TraversalHomebaseBPMs() -> (DEBUG_BaseBody_TraversalHomebaseBPMs:Sparks Anim BPMSet Structure)
    Index=Get Traversal_IsCharacterMoving?()
   ) -> (ReturnValue:Sparks Anim BPMSet Structure)
   > n3 = PickIntensityGenreAnimLoopCycle!(
    execute
    self:Sparks Instrument Anim Instance Object Reference
    BPMSet:"Sparks Anim BPMSet Structure (by ref)"=@n2
    ForceVariant:bool=false
    VariantIndex:int=0
   ) -> (then, ReturnValue:"Sparks Anim Loop Cycle Structure (by ref)")
   > Set Array Elem!(
    execute
    TargetArray:Array of Anim Sequence Object References=Get PlaySequences()
    Index:int=@n1
    Item:"Anim Sequence Object Reference (by ref)"=Break Sparks Anim Loop Cycle!(SparksAnimLoopCycle:"Sparks Anim Loop Cycle Structure (by ref)"=@n3) -> (Loop:Anim Sequence Object Reference, Entry:Anim Sequence Object Reference, Exit:Anim Sequence Object Reference) {is=title K2Node_BreakStruct}
    bSizeToFit:bool=true
   ) -> (then)
   > Set Array Elem(
    TargetArray:"Array of Float (double-precision)s"=Get TempoMultipliers()
    Index=@n1
    Item:"Float (double-precision) (by ref)"=Break Sparks Anim BPMSet!(SparksAnimBPMSet:"Sparks Anim BPMSet Structure (by ref)"=@n2) -> (BPMBuckets:Array of Sparks Anim Intensity Set Object References, CurrentTempoMultiplier:float=~) {is=title K2Node_BreakStruct}.CurrentTempoMultiplier
    bSizeToFit=true
   )
  ]}
]
variables [
 AnimatableState {type=ESparksAnimatableState, default=None}
 AnimNotify_TextColor {type=LinearColor, default=<rgba 1 0 1 1>}
 BaseBody_FretArm_BlendWeight {type=number, default=1}
 BaseBody_Squat_BlendWeight {type=number}
 BaseBody_SquatDistance {type=number}
 BaseBody_SquatDistanceMax {type=number, default=-40}
 BaseBody_StrumArm_BlendWeight {type=number, default=1}
 BPM Preview {type=number, default=125}
 ChordHandPoseIndex {type=integer}
 DEBUG__BypassBodySectionAdditives? {type=boolean}
 DEBUG_BaseBody_ForcePlayStateTransition? {type=boolean}
 DEBUG_BaseBody_ForceSquat? {type=boolean}
 DEBUG_BaseBody_OverrideSequence {type=AnimSequenceBase, default=&TPose}
 DEBUG_BaseBody_TraversalHomebaseBPMs {
  type=SparksAnimBPMSet
  default {
   allowDoublingHalving=true
   hysteresis=0.07
   bPMBuckets [
    /SparksCharacterCommon/Animation/MainPlayer/BandPlay/Guitar/Loops/JamV2Guitar_IntensitySet_095
    /SparksCharacterCommon/Animation/MainPlayer/BandPlay/Guitar/Loops/JamV2Guitar_IntensitySet_125
    /SparksCharacterCommon/Animation/MainPlayer/BandPlay/Guitar/Loops/JamV2Guitar_IntensitySet_165
   ]
  }
 }
 DEBUG_BaseBody_UseOverrideSequence? {type=boolean}
 DEBUG_BaseBody_UsePlayLoopOnly? {type=boolean}
 DEBUG_CurveNameArray {type="array<string>", default []}
 DEBUG_EnableLeftHandIK? {type=boolean}
 DEBUG_EnableRightHandIK? {type=boolean, default=true}
 DEBUG_FaceExpression_BlendWeight {type=number, default=1}
 DEBUG_FaceExpression_CloseEyes {type=boolean}
 DEBUG_FaceExpression_DisableLipSync? {type=boolean}
 DEBUG_FaceExpression_OverrideSequence {type=AnimSequenceBase, default=&FaceExpression_Pop_Heavy}
 DEBUG_FaceExpression_UseOverrideSequence? {type=boolean}
 DEBUG_FootShift_DamplifyMultiplier {type=number, default=1}
 DEBUG_FootShift_OverrideBlendWeight {type=number}
 DEBUG_FootShift_OverrideSequence {type=AnimSequenceBase, default=/SparksCharacterCommon/Animation/MainPlayer/Additives/FootShift_LForward_A}
 DEBUG_FootShift_ProcForceInwardKnees {type=boolean}
 DEBUG_FootShift_ProcOverrideModeIndex {type=integer}
 DEBUG_FootShift_ProcOverrideTransitionSpeed {type=Enum_FootShift_ProcTransitionSpeed, default=Fast}
 DEBUG_FootShift_ProcTranslateDistanceMultiplier {type=number, default=1}
 DEBUG_FootShift_ProcUseOverrideModeIndex? {type=boolean}
 DEBUG_FootShift_ProcUseOverrideTransitionSpeed? {type=boolean}
 DEBUG_FootShift_ResetSystem? {type=boolean}
 DEBUG_FootShift_UseOverrideBlendWeight? {type=boolean}
 DEBUG_FootShift_UseOverrideDamplifyMultiplier? {type=boolean}
 DEBUG_FootShift_UseOverrideSequence? {type=boolean}
 DEBUG_FootShift_UseProceduralSystem? {type=boolean, default=true}
 DEBUG_HeadMotion_DamplifyMultiplier {type=number, default=1}
 DEBUG_HeadMotion_OverrideBlendWeight {type=number}
 DEBUG_HeadMotion_OverrideSequence {type=AnimSequenceBase, default=/SparksCharacterCommon/Animation/MainPlayer/Additives/HeadMotion_TurnL_A}
 DEBUG_HeadMotion_ResetSystem? {type=boolean}
 DEBUG_HeadMotion_UseOverrideBlendWeight? {type=boolean, default=true}
 DEBUG_HeadMotion_UseOverrideDamplifyMultiplier? {type=boolean}
 DEBUG_HeadMotion_UseOverrideSequence? {type=boolean}
 DEBUG_Instrument_OverrideMesh {type=ESparksInstrumentType, default=Guitar}
 DEBUG_Instrument_UseOverrideMesh? {type=boolean}
 DEBUG_LeftArm_DamplifyMultiplier {type=number, default=1}
 DEBUG_LeftArm_OverrideBlendWeight {type=number}
 DEBUG_LeftArm_OverrideSequence {type=AnimSequenceBase, default=/SparksCharacterCommon/Animation/MainPlayer/Additives/LeftArm_Wave_A}
 DEBUG_LeftArm_ResetSystem? {type=boolean}
 DEBUG_LeftArm_UseOverrideBlendWeight? {type=boolean}
 DEBUG_LeftArm_UseOverrideDamplifyMultiplier? {type=boolean}
 DEBUG_LeftArm_UseOverrideSequence? {type=boolean}
 DEBUG_LowerBody_ForceSequentialMode? {type=boolean, default=true}
 DEBUG_LowerBody_OverrideSequence {type=AnimSequenceBase, default=/BRCosmetics/Animation/Game/MainPlayer/Emotes/JumpingJoy/Emote_JumpingJoy_CMF}
 DEBUG_LowerBody_UseLocomotionSystem? {type=boolean, default=true}
 DEBUG_LowerBody_UseOverrideSequence? {type=boolean}
 DEBUG_PrintAnimLogic {type=boolean}
 DEBUG_RightArm_DamplifyMultiplier {type=number, default=1}
 DEBUG_RightArm_OverrideBlendWeight {type=number}
 DEBUG_RightArm_OverrideSequence {type=AnimSequenceBase, default=/SparksCharacterCommon/Animation/MainPlayer/Additives/RightArm_Fist_A}
 DEBUG_RightArm_ResetSystem? {type=boolean}
 DEBUG_RightArm_UseOverrideBlendWeight? {type=boolean}
 DEBUG_RightArm_UseOverrideDamplifyMultiplier? {type=boolean}
 DEBUG_RightArm_UseOverrideSequence? {type=boolean}
 DEBUG_Song_OverrideGenre {type=Enum_MusicGenre, default=Rap/HipHop}
 DEBUG_Song_OverrideIntensity {type=AnimIntensity, default=Light}
 DEBUG_Song_UseOverrideGenre? {type=boolean}
 DEBUG_Song_UseOverrideIntensity? {type=boolean}
 DEBUG_SpineMotion_DamplifyMultiplier {type=number, default=1}
 DEBUG_SpineMotion_OverrideBlendWeight {type=number}
 DEBUG_SpineMotion_OverrideSequence {type=AnimSequenceBase, default=/SparksCharacterCommon/Animation/MainPlayer/Additives/SpineMotion_TwistLeft_A}
 DEBUG_SpineMotion_ResetSystem? {type=boolean}
 DEBUG_SpineMotion_UseOverrideBlendWeight? {type=boolean}
 DEBUG_SpineMotion_UseOverrideDamplifyMultiplier? {type=boolean}
 DEBUG_SpineMotion_UseOverrideSequence? {type=boolean}
 DEBUG_SpineMotion_UsePoseBlendSpace? {type=boolean}
 DEBUG_StrumTestFire {type=boolean}
 DeltaTimeX {type=number, default=0.0363}
 FaceExpression_ClosedEyes {type=number}
 FaceExpression_CurrentSequence {type=AnimSequenceBase, default=null}
 FootShift_1stFoot_ProcAnimPlayRate {type=number, default=1}
 FootShift_1stFoot_ProcCurrentSequence {type=AnimSequenceBase, default=null}
 FootShift_1stFoot_ProcLiftCurrentlyActive? {type=boolean}
 FootShift_1stFoot_ProcRotateTwist {type=number}
 FootShift_1stFoot_ProcTransitionSpeed {type=Enum_FootShift_ProcTransitionSpeed, default=Fast}
 FootShift_1stFoot_ProcTranslateCurrentlyActive? {type=boolean}
 FootShift_1stFoot_ProcTranslateDelay {type=number, default=0.12}
 FootShift_1stFoot_ProcTranslateDistance {type=number}
 FootShift_1stFoot_ProcTranslateDistanceMax {type=number, default=45}
 FootShift_1stFoot_ProcTranslateDistanceMin {type=number, default=15}
 FootShift_1stFoot_ProcTranslateFB {type=number}
 FootShift_1stFoot_ProcTranslateLR {type=number}
 FootShift_1stPelvis_ProcRotateTwist {type=number}
 FootShift_1stPelvis_ProcTranslateFB {type=number}
 FootShift_1stPelvis_ProcTranslateLR {type=number}
 FootShift_1stPelvis_ProcTranslateVertical {type=number}
 FootShift_2ndFoot_ProcAnimPlayRate {type=number, default=1}
 FootShift_2ndFoot_ProcCurrentSequence {type=AnimSequenceBase, default=null}
 FootShift_2ndFoot_ProcLiftCurrentlyActive? {type=boolean}
 FootShift_2ndFoot_ProcLiftDelay {type=number, default=0.35}
 FootShift_2ndFoot_ProcRotateTwist {type=number}
 FootShift_2ndFoot_ProcTransitionSpeed {type=Enum_FootShift_ProcTransitionSpeed, default=Fast}
 FootShift_2ndFoot_ProcTranslateCurrentlyActive? {type=boolean}
 FootShift_2ndFoot_ProcTranslateDelay {type=number, default=0.12}
 FootShift_2ndFoot_ProcTranslateDistance {type=number}
 FootShift_2ndFoot_ProcTranslateDistanceMax {type=number, default=45}
 FootShift_2ndFoot_ProcTranslateDistanceMin {type=number, default=15}
 FootShift_2ndFoot_ProcTranslateFB {type=number}
 FootShift_2ndFoot_ProcTranslateLR {type=number}
 FootShift_2ndPelvis_ProcRotateTwist {type=number}
 FootShift_2ndPelvis_ProcTranslateFB {type=number}
 FootShift_2ndPelvis_ProcTranslateLR {type=number}
 FootShift_2ndPelvis_ProcTranslateVertical {type=number}
 FootShift_AmplifyAlpha {type=number}
 FootShift_BlendWeight {type=number, default=1}
 FootShift_CurrentSequence {type=AnimSequenceBase, default=&FootShift_LForward_A}
 FootShift_DampenAlpha {type=number, default=0.9555}
 FootShift_DamplifyMultiplier {type=number, default=1}
 FootShift_L_ProcSequenceArray {type="array<AnimSequenceBase>", default [/SparksCharacterCommon/Animation/MainPlayer/Additives/FootShift_LLift_A]}
 FootShift_Proc1stFootIsLeft? {type=boolean, default=true}
 FootShift_ProcModeIndex {type=integer, default=2}
 FootShift_ProcUseInwardKnees? {type=boolean}
 FootShift_R_ProcSequenceArray {type="array<AnimSequenceBase>", default [/SparksCharacterCommon/Animation/MainPlayer/Additives/FootShift_RLift_A]}
 FootShift_Reset_CurrentSequence {type=AnimSequenceBase, default=/SparksCharacterCommon/Animation/MainPlayer/Additives/FootShift_LForward_A_Reset}
 FootShift_SequenceArray {
  type="array<AnimSequenceBase>"
  default [
   &FootShift_LForward_A
   /SparksCharacterCommon/Animation/MainPlayer/Additives/FootShift_RForward_A
   /SparksCharacterCommon/Animation/MainPlayer/Additives/FootShift_LShift_A
   /SparksCharacterCommon/Animation/MainPlayer/Additives/FootShift_RShift_A
   /SparksCharacterCommon/Animation/MainPlayer/Additives/FootShift_ShuffleRight_A
  ]
 }
 FootShift_SequenceMap {default {
   /Script/Engine.AnimSequence'/SparksCharacterCommon/Animation/MainPlayer/Additives/FootShift_LForward_A.FootShift_LForward_A'=&FootShift_LForward_A_Reset
   &AnimSequence'/SparksCharacterCommon/Animation/MainPlayer/Additives/FootShift_LShift_A.FootShift_LShift_A'=/SparksCharacterCommon/Animation/MainPlayer/Additives/FootShift_LShift_A_Reset
   &AnimSequence'/SparksCharacterCommon/Animation/MainPlayer/Additives/FootShift_RForward_A.FootShift_RForward_A'=/SparksCharacterCommon/Animation/MainPlayer/Additives/FootShift_RForward_A_Reset
   &AnimSequence'/SparksCharacterCommon/Animation/MainPlayer/Additives/FootShift_RShift_A.FootShift_RShift_A'=/SparksCharacterCommon/Animation/MainPlayer/Additives/FootShift_RShift_A_Reset
   &AnimSequence'/SparksCharacterCommon/Animation/MainPlayer/Additives/FootShift_ShuffleRight_A.FootShift_ShuffleRight_A'=/SparksCharacterCommon/Animation/MainPlayer/Additives/FootShift_ShuffleRight_A_Reset
  }}
 FootShift_Struct {
  type=Sparks_JamV2_Additives_Struct
  default {
   blendWeight=1
   currentlyActive?=false
   disableCurve=Disable_JamV2_FootShiftAdditive
   overrideBlendWeight=0
   overrideSequence=/SparksCharacterCommon/Animation/MainPlayer/Additives/FootShift_LForward_A
   timeActive=4
   timeInactive=4
   useOverrideBlendWeight?=false
   useOverrideSequence?=false
   sequenceArray [
    &FootShift_LForward_A
    /SparksCharacterCommon/Animation/MainPlayer/Additives/FootShift_RForward_A
    /SparksCharacterCommon/Animation/MainPlayer/Additives/FootShift_LShift_A
    /SparksCharacterCommon/Animation/MainPlayer/Additives/FootShift_RShift_A
    /SparksCharacterCommon/Animation/MainPlayer/Additives/FootShift_ShuffleRight_A
   ]
  }
 }
 FootShift_TimeActive {type=number, default=4}
 FootShift_TimeInactive {type=number, default=4}
 FretMidiListener {type=AsyncAction_ListenForMidiNoteEvent, default=null}
 HeadMotion_AmplifyAlpha {type=number}
 HeadMotion_BlendWeight {type=number}
 HeadMotion_CurrentlyActive? {type=boolean, default=true}
 HeadMotion_CurrentSequence {type=AnimSequenceBase, default=/SparksCharacterCommon/Animation/MainPlayer/Additives/HeadMotion_GroovyPetals}
 HeadMotion_CurrentSequenceIsLooping? {type=boolean, default=true}
 HeadMotion_DampenAlpha {type=number, default=0.9583}
 HeadMotion_DamplifyMultiplier {type=number, default=1}
 HeadMotion_SequenceArray {
  type="array<AnimSequenceBase>"
  default [
   /SparksCharacterCommon/Animation/MainPlayer/Additives/HeadMotion_CircularMotion
   &HeadMotion_GroovyPetals
   /SparksCharacterCommon/Animation/MainPlayer/Additives/HeadMotion_WiggleKnees
  ]
 }
 HeadMotion_Struct {
  type=Sparks_JamV2_Additives_Struct
  default {
   blendWeight=0
   currentlyActive?=true
   disableCurve=Disable_JamV2_HeadMotionReplace
   overrideBlendWeight=0
   overrideSequence=&HeadMotion_TurnL_A
   timeActive=6
   timeInactive=4
   useOverrideBlendWeight?=true
   useOverrideSequence?=false
   sequenceArray [
    &HeadMotion_CircularMotion
    &HeadMotion_GroovyPetals
    &HeadMotion_WiggleKnees
   ]
  }
 }
 HeadMotion_TimeActive {type=number, default=8}
 HeadMotion_TimeInactive {type=number, default=16}
 HomebaseBPMs_Defaults {default {
   bass {
    allowDoublingHalving=false
    hysteresis=0
    bPMBuckets [
     /SparksCharacterCommon/Animation/MainPlayer/BandPlay/Bass/Loops/JamV2Bass_IntensitySet_095
     /SparksCharacterCommon/Animation/MainPlayer/BandPlay/Bass/Loops/JamV2Bass_IntensitySet_125
     /SparksCharacterCommon/Animation/MainPlayer/BandPlay/Bass/Loops/JamV2Bass_IntensitySet_165
    ]
   }
   guitar {
    allowDoublingHalving=false
    hysteresis=0
    bPMBuckets [
     &JamV2Guitar_IntensitySet_095
     &JamV2Guitar_IntensitySet_125
     &JamV2Guitar_IntensitySet_165
    ]
   }
   keyboard {
    allowDoublingHalving=false
    hysteresis=0
    bPMBuckets [
     &JamV2Guitar_IntensitySet_095
     &JamV2Guitar_IntensitySet_125
     &JamV2Guitar_IntensitySet_165
    ]
   }
  }}
 InstrumentType {type=ESparksInstrumentType, default=Guitar}
 IntensityMidiListener {type=AsyncAction_ListenForMidiTextEvent, default=null}
 IsPlayingKeytar {type=boolean}
 LeftArm_AmplifyAlpha {type=number}
 LeftArm_BlendWeight {type=number}
 LeftArm_CurrentlyActive? {type=boolean}
 LeftArm_CurrentSequence {type=AnimSequenceBase, default=/SparksCharacterCommon/Animation/MainPlayer/Additives/LeftArm_PointUp_A}
 LeftArm_DampenAlpha {type=number, default=0.9103}
 LeftArm_DamplifyMultiplier {type=number, default=1}
 LeftArm_SequenceArray {
  type="array<AnimSequenceBase>"
  default [
   /SparksCharacterCommon/Animation/MainPlayer/Additives/LeftArm_Fist_A
   &LeftArm_PointUp_A
   &LeftArm_Wave_A
   /SparksCharacterCommon/Animation/MainPlayer/Additives/LeftArm_WaveOverHead_A
  ]
 }
 LeftArm_Struct {
  type=Sparks_JamV2_Additives_Struct
  default {
   blendWeight=0
   currentlyActive?=false
   disableCurve=Disable_JamV2_LeftArmReplace
   overrideBlendWeight=0
   overrideSequence=&LeftArm_Wave_A
   timeActive=6
   timeInactive=2
   useOverrideBlendWeight?=true
   useOverrideSequence?=false
   sequenceArray [
    &LeftArm_Fist_A
    &LeftArm_PointUp_A
    &LeftArm_Wave_A
    &LeftArm_WaveOverHead_A
   ]
  }
 }
 LeftArm_TimeActive {type=number, default=8}
 LeftArm_TimeInactive {type=number, default=16}
 LipSyncBlendWeight {type=number}
 LowerBody_CurrentIndex {type=integer}
 LowerBody_CurrentSequence {type=AnimSequenceBase, default=null}
 LowerBody_NextSequence {type=AnimSequenceBase, default=&Drumsticks_Walkcycle_CMM}
 LowerBody_SequenceArray {
  type="array<AnimSequenceBase>"
  default [
   &Drumsticks_Walkcycle_CMM
   &Emote_SunnySteppin_Traversal_CMM
   &Emote_PogoTraversal_CMM
  ]
 }
 MidiFretEvents {
  type="array<EGuitarMidiNoteEvent>"
  default [
   FretPosition1
   FretPosition2
   FretPosition3
   FretPosition4
   FretPosition5
   FretPosition6
   FretPosition7
   FretPosition8
   FretPosition9
   FretPosition10
   FretPosition11
   FretPosition12
   FretPosition13
   FretPosition14
   FretPosition15
   FretPosition16
   FretPosition17
   FretPosition18
   FretPosition19
   FretPosition20
  ]
 }
 MidiTrackName {type=string, default=None}
 PlaySequences {
  type="array<AnimSequence>"
  default [
   &JamV2Guitar_PlayMedium_Pop_095_A
   &JamV2Guitar_PlayMedium_Pop_125_A
  ]
 }
 RightArm_AmplifyAlpha {type=number, default=0.0944}
 RightArm_BlendWeight {type=number}
 RightArm_CurrentlyActive? {type=boolean}
 RightArm_CurrentSequence {type=AnimSequenceBase, default=/SparksCharacterCommon/Animation/MainPlayer/Additives/RightArm_PointUp_A}
 RightArm_DampenAlpha {type=number, default=1}
 RightArm_DamplifyMultiplier {type=number, default=1}
 RightArm_SequenceArray {
  type="array<AnimSequenceBase>"
  default [
   &RightArm_Fist_A
   &RightArm_PointUp_A
   /SparksCharacterCommon/Animation/MainPlayer/Additives/RightArm_Wave_A
  ]
 }
 RightArm_Struct {
  type=Sparks_JamV2_Additives_Struct
  default {
   blendWeight=0
   currentlyActive?=false
   disableCurve=Disable_JamV2_RightArmReplace
   overrideBlendWeight=0
   overrideSequence=&RightArm_Fist_A
   timeActive=6
   timeInactive=2
   useOverrideBlendWeight?=true
   useOverrideSequence?=false
   sequenceArray [
    &RightArm_Fist_A
    &RightArm_PointUp_A
    &RightArm_Wave_A
   ]
  }
 }
 RightArm_TimeActive {type=number, default=8}
 RightArm_TimeInactive {type=number, default=16}
 SecBeforeIntensity {type=number}
 Song_GenreBlendTime {type=number, default=0.5}
 Song_GenreIndex {type=integer}
 Song_IntensityBlendTime {type=number, default=0.5}
 Song_IntensityIndex {type=integer, default=1}
 SpineMotion_AmplifyAlpha {type=number}
 SpineMotion_BlendWeight {type=number}
 SpineMotion_CorrectPelvisTranslate {type=Vector}
 SpineMotion_CurrentlyActive? {type=boolean, default=true}
 SpineMotion_CurrentSequence {type=AnimSequenceBase, default=&SpineMotion_TwistLeft_A}
 SpineMotion_DampenAlpha {type=number, default=0.9335}
 SpineMotion_DamplifyMultiplier {type=number, default=1}
 SpineMotion_PlayRate {type=number, default=0.333}
 SpineMotion_PoseBS_Pitch {type=number}
 SpineMotion_PoseBS_Roll {type=number}
 SpineMotion_Reset_CurrentSequence {type=AnimSequenceBase, default=/SparksCharacterCommon/Animation/MainPlayer/Additives/SpineMotion_TwistLeft_A_Reset}
 SpineMotion_SequenceArray {
  type="array<AnimSequenceBase>"
  default [
   &SpineMotion_TwistLeft_A
   /SparksCharacterCommon/Animation/MainPlayer/Additives/SpineMotion_TwistRight_A
  ]
 }
 SpineMotion_SequenceMap {default {
   &AnimSequence'/SparksCharacterCommon/Animation/MainPlayer/Additives/SpineMotion_TwistLeft_A.SpineMotion_TwistLeft_A'=&SpineMotion_TwistLeft_A_Reset
   &AnimSequence'/SparksCharacterCommon/Animation/MainPlayer/Additives/SpineMotion_TwistRight_A.SpineMotion_TwistRight_A'=/SparksCharacterCommon/Animation/MainPlayer/Additives/SpineMotion_TwistRight_A_Reset
  }}
 SpineMotion_Struct {
  type=Sparks_JamV2_Additives_Struct
  default {
   blendWeight=0
   currentlyActive?=true
   disableCurve=Disable_JamV2_SpineMotionAdditive
   overrideBlendWeight=0
   overrideSequence=/SparksCharacterCommon/Animation/MainPlayer/Additives/SpineMotion_TwistLeft_A
   timeActive=8
   timeInactive=4
   useOverrideBlendWeight?=true
   useOverrideSequence?=false
   sequenceArray [
    &SpineMotion_TwistLeft_A
    /SparksCharacterCommon/Animation/MainPlayer/Additives/SpineMotion_TwistRight_A
   ]
  }
 }
 SpineMotion_TimeActive {type=number, default=8}
 SpineMotion_TimeInactive {type=number, default=8}
 StrumCorrectivePose {type=AnimSequenceBase, default=null}
 StrumMidiListener {type=AsyncAction_ListenForMidiNoteEvent, default=null}
 StrumStartPose {type=AnimSequenceBase, default=null}
 TempoMultipliers {type="array<number>", default [1,1]}
 Traversal_IsCharacterMoving? {type=boolean}
 UseStateMachineStrumming? {type=boolean}
]
