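# Jam guitar and keytar animation layer for the Fortnite player that keeps instrument playing working during traversal.
# When still, Body alternates beat-synced intensity loops; when moving, the input pose gets a beat-synced upper-body additive.
# Left and right arm state machines, switched by instrument, are layered per bone over that.
# MIDI listeners feed song intensity, fret changes and strum montages; initialization registers item animation and lower-body mask overrides.
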AnimBlueprint /SparksCharacterCommon/Animation/MainPlayer/BandPlay/Sparks_Guitar_Jam_AnimBP_Traversal
target_skeleton=&Fortnite_M_Avg_Player_Skeleton
graphs [
 AnimGraph {nodes [
   n1 = Additive Identity Pose()
   n2 = Get TempoMultipliers() -> (TempoMultipliers:"Array of Float (double-precision)s")
   n3 = Layered blend per bone(
    BasePose=Play Beat Synced Animation!(
     ExplicitTime:float=0
     bShouldLoop:bool=True
     InSequence:Anim Sequence Base Object Reference=Get TraversalGeneralAdditiveSequence() -> (TraversalGeneralAdditiveSequence:Anim Sequence Base Object Reference)
     MusicClock:Music Clock Object Reference
     TempoMultiplier:float="Get (a copy)"(Array:"Array of Float (double-precision)s"=@n2) -> (Output:double)
     bEnableTempoVariants:bool=False
     TempoVariantHysteresis:float=0.05
     bShowDebug:bool=False
    ) -> (Pose:Pose Link Structure) {is=title AnimGraphNode_PlayBeatSyncedAnim}
    BlendPoses_0=Additive Identity Pose()
   ) {layerSetup [{branchFilters [{boneName=upperarm_l}, {boneName=upperarm_r}]}]}
   n4 = Use cached pose 'CachedInputPose'()
   n5 = Get Head Additive Alpha() -> (Head Additive Alpha:float)
   AnimGraphNode_StateMachine_1 = Body() {is=title AnimGraphNode_StateMachine, bAllowConduitEntryStates=true}
   AnimGraphNode_StateMachine_2 = Left Arm() {is=title AnimGraphNode_StateMachine, bAllowConduitEntryStates=true}
   AnimGraphNode_StateMachine_3 = Right Arm() {is=title AnimGraphNode_StateMachine, bAllowConduitEntryStates=true}
   n6 = Layered blend per bone!(
    BasePose=Blend Poses by bool(
     BlendPose_0=@AnimGraphNode_StateMachine_1
     BlendPose_1=Apply Additive(Base=@n4, Additive=@n3, Alpha=@n5)
     BlendTime_0=0.5
     BlendTime_1=0.5
     bActiveValue=Get ShouldUseJamAnim() -> (ShouldUseJamAnim:bool)
    )
    BlendPoses_0=@AnimGraphNode_StateMachine_2
    BlendPoses_1:Pose Link Structure=@AnimGraphNode_StateMachine_3
    BlendWeights_0
    BlendWeights_1:float=1
   ) -> (Pose) {
    blendPoses [{linkId=-1, sourceLinkId=-1}, {linkId=-1, sourceLinkId=-1}]
    blendWeights [1,1]
    layerSetup [
     {branchFilters [{boneName=clavicle_l}, {boneName=VB VB spine_05_hand_attach_r_ik_hand_l}, {boneName=ik_hand_l}]}
     {branchFilters [{boneName=clavicle_r}, {boneName=VB VB spine_05_hand_attach_r_ik_hand_gun}, {boneName=ik_hand_gun}]}
    ]
   }
   n7 = Blend Poses by bool(
    BlendPose_0=Apply Additive(Base=@n6, Additive=Slot 'AdditiveLayer'(Source=@n1) {slotName=AdditiveLayer}) {alphaBoolBlend {blendOption=Linear}}
    BlendPose_1=Sequence Player 'JamGuitar_PlayMedium_125_A'() {sequence=/SparksCharacterCommon/Animation/MainPlayer/BandPlay/Guitar/Loops/JamGuitar_PlayMedium_125_A}
    BlendTime_0=0.5
    BlendTime_1=0.5
    bActiveValue=Get bCanPlayMidiAnimations() -> (bCanPlayMidiAnimations:bool)
   ) {blendType=Linear}
   n8 = Blend Poses by bool(
    BlendPose_0="Sequence Player (additive) 'Keytar_SpawnInstrument_Pose'"() {sequence=/SparksCharacterCommon/Animation/MainPlayer/BandPlay/Keytar/Keytar_SpawnInstrument_Pose}
    BlendPose_1="Sequence Player (additive) 'Guitar_SpawnInstrument_Pose'"() {sequence=/SparksCharacterCommon/Animation/MainPlayer/BandPlay/Guitar/Guitar_SpawnInstrument_Pose}
    bActiveValue=Get IsPlayingKeytar() -> (IsPlayingKeytar:bool)
   ) {blendType=Linear}
   n9 = "Transform (Modify) Bone - Bone: VB ik_foot_root_hand_attach_r"(
    ComponentPose=Local To Component(LocalPose=Apply Additive(Base=@n7, Additive=@n8) {alphaBoolBlend {blendOption=Linear}})
    Alpha="To Float (Boolean)"(InBool=Get IsPlayingKeytar())
    Translation="200,0,0"
    Rotation="0,0,0"
    Scale="1,1,1"
   ) {translationMode=BMM_Additive, translationSpace=BCS_WorldSpace, boneToModify {boneName=VB ik_foot_root_hand_attach_r}, translation {x=200}}
   Output Pose(Result=Modify Curve(SourcePose=Component To Local(ComponentPose=@n9), CurveValues_0=1, Alpha=1) {curveNames [DisableLargeBodyEmoteIKAdjustements], curveValues [0]}) {layerGroup=None}
   n10 = Local To Component(LocalPose=Input Pose() {name=PassThroughFullBodyPose})
   Save cached pose 'CachedInputPose'(Pose=Component To Local(ComponentPose=Copy Bone - Source Bone: hand_attach_r - Target Bone: VB spine_05_hand_attach_r!(
      ComponentPose:Component Space Pose Link Structure=@n10
      Alpha:float=1
      bCopyTranslation:bool=true
      bCopyRotation:bool=true
      bCopyScale:bool=true
     ) -> (Pose:Component Space Pose Link Structure) {is=title AnimGraphNode_CopyBone, sourceBone {boneName=hand_attach_r}, targetBone {boneName=VB spine_05_hand_attach_r}})) {cacheName=CachedInputPose}
   "Sequence Player (additive) 'Traversal_UpperBodyBeatAdditive'"() {sequence=/SparksCharacterCommon/Animation/MainPlayer/Additives/Traversal_UpperBodyBeatAdditive}
  ]}
 Body {
  entry=Play 0
  path=AnimGraph.AnimGraphNode_StateMachine_1.Body
  ids=id
  nodes [
   AnimStateNode_10 = Play 0() {
    is=title AnimStateNode
    stateEntered {guid=#1}
    stateFullyBlended {guid=#2}
    stateLeft {guid=#3}
   }
   AnimStateNode_32 = Play 1() {
    is=title AnimStateNode
    stateEntered {guid=#1}
    stateFullyBlended {guid=#2}
    stateLeft {guid=#3}
   }
   AnimStateTransitionNode_10 = AnimStateTransitionNode()
   AnimStateTransitionNode_6 = AnimStateTransitionNode()
  ]
  links [
   {dst=Play 1, rule=AnimGraph.AnimGraphNode_StateMachine_1.Body.AnimStateTransitionNode_6.Transition, src=Play 0, via=AnimStateTransitionNode_6}
   {
    dst=Play 0
    rule=AnimGraph.AnimGraphNode_StateMachine_1.Body.AnimStateTransitionNode_10.Transition
    src=Play 1
    via=AnimStateTransitionNode_10
   }
  ]
 }
 Play 0 {
  path=AnimGraph.AnimGraphNode_StateMachine_1.Body.AnimStateNode_10.Play 0
  nodes [
   n1 = Get PlaySequences() -> (PlaySequences:Array of Anim Sequence Object References)
   n2 = Get TempoMultipliers()
   Output Animation Pose(Result=Play Beat Synced Animation(
     ExplicitTime=0
     bShouldLoop=True
     InSequence="Get (a copy)"(Array:Array of Anim Sequence Object References=@n1) -> (Output:Anim Sequence Object Reference)
     TempoMultiplier="Get (a copy)"(Array:"Array of Float (double-precision)s"=@n2) -> (Output:double)
     bEnableTempoVariants=False
     TempoVariantHysteresis=0.05
     bShowDebug=False
    )) {
    node {stateEntryFunction {functionName=OnPlay0Entry}}
    stateEntryFunction {bSelfContext=true, memberGuid=#4, memberName=OnPlay0Entry}
   }
  ]
 }
 Play 1 {
  path=AnimGraph.AnimGraphNode_StateMachine_1.Body.AnimStateNode_32.Play 1
  nodes [
   n1 = Get PlaySequences()
   n2 = Get TempoMultipliers()
   Output Animation Pose(Result=Play Beat Synced Animation(
     ExplicitTime=0
     bShouldLoop=True
     InSequence="Get (a copy)"(Array:Array of Anim Sequence Object References=@n1, Dimension 1=1) -> (Output:Anim Sequence Object Reference)
     TempoMultiplier="Get (a copy)"(Array:"Array of Float (double-precision)s"=@n2, Dimension 1=1) -> (Output:double)
     bEnableTempoVariants=False
     TempoVariantHysteresis=0.05
     bShowDebug=False
    )) {
    node {stateEntryFunction {functionName=OnPlay1Entry}}
    stateEntryFunction {bSelfContext=true, memberGuid=#5, memberName=OnPlay1Entry}
   }
  ]
 }
 Transition {
  path=AnimGraph.AnimGraphNode_StateMachine_1.Body.AnimStateTransitionNode_10.Transition
  nodes [Result(bCanEnterTransition=JamShouldTransition!(self:Sparks Instrument Anim Instance Object Reference) -> (ReturnValue:bool))]
 }
 Transition {
  path=AnimGraph.AnimGraphNode_StateMachine_1.Body.AnimStateTransitionNode_6.Transition
  nodes [Result(bCanEnterTransition=JamShouldTransition())]
 }
 Left Arm {
  entry=Guitar
  path=AnimGraph.AnimGraphNode_StateMachine_2.Left Arm
  ids=id
  nodes [
   AnimStateNode_0 = Keytar() {is=title AnimStateNode}
   AnimStateNode_4 = Guitar() {is=title AnimStateNode}
   AnimStateTransitionNode_0 = AnimStateTransitionNode() {crossfadeDuration=0.5}
   AnimStateTransitionNode_1 = AnimStateTransitionNode() {crossfadeDuration=0.5}
  ]
  links [
   {
    dst=Keytar
    rule=AnimGraph.AnimGraphNode_StateMachine_2.Left Arm.AnimStateTransitionNode_0.Transition
    src=Guitar
    via=AnimStateTransitionNode_0
   }
   {
    dst=Guitar
    rule=AnimGraph.AnimGraphNode_StateMachine_2.Left Arm.AnimStateTransitionNode_1.Transition
    src=Keytar
    via=AnimStateTransitionNode_1
   }
  ]
 }
 Keytar {
  path=AnimGraph.AnimGraphNode_StateMachine_2.Left Arm.AnimStateNode_0.Keytar
  nodes [Output Animation Pose(Result=Sequence Player 'Keytar_FretHand_BasePose'() {sequence=/SparksCharacterCommon/Animation/MainPlayer/BandPlay/Keytar/FretHand/Keytar_FretHand_BasePose})]
 }
 Guitar {
  path=AnimGraph.AnimGraphNode_StateMachine_2.Left Arm.AnimStateNode_4.Guitar
  nodes [
   n1 = Blendspace Player 'Guitar_FretHand_BS'(X=0, Y=0) {blendSpace=/SparksCharacterCommon/Animation/MainPlayer/BandPlay/Guitar/FretHand/Guitar_FretHand_BS, startPosition=1}
   n2 = Additive Identity Pose()
   n3 = "Not Equal (Enum)"!(
    A:ESparksAnimatableState Enum=Get AnimatableState() -> (AnimatableState:ESparksAnimatableState Enum=None)
    B:ESparksAnimatableState Enum=Animatable
   ) -> (ReturnValue:bool)
   Output Animation Pose(Result=Apply Additive(
     Base=Sequence Player 'Guitar_FretHand_BasePose'() {sequence=/SparksCharacterCommon/Animation/MainPlayer/BandPlay/Guitar/FretHand/Guitar_FretHand_BasePose}
     Additive=Blend Poses by bool(BlendPose_0=@n1, BlendPose_1=@n2, bActiveValue=@n3) {blendType=Linear}
    ) {alphaBoolBlend {blendOption=Linear}})
  ]
 }
 Transition {
  path=AnimGraph.AnimGraphNode_StateMachine_2.Left Arm.AnimStateTransitionNode_0.Transition
  nodes [Result(bCanEnterTransition=Get IsPlayingKeytar())]
 }
 Transition {
  path=AnimGraph.AnimGraphNode_StateMachine_2.Left Arm.AnimStateTransitionNode_1.Transition
  nodes [Result(bCanEnterTransition=NOT Boolean(A=Get IsPlayingKeytar()))]
 }
 Right Arm {
  entry=Guitar
  path=AnimGraph.AnimGraphNode_StateMachine_3.Right Arm
  ids=id
  nodes [
   AnimStateNode_0 = Keytar()
   AnimStateNode_4 = Guitar() {
    stateEntered {guid=#6}
    stateFullyBlended {guid=#7}
    stateLeft {guid=#8}
   }
   AnimStateTransitionNode_6 = AnimStateTransitionNode() {crossfadeDuration=0.5}
   AnimStateTransitionNode_7 = AnimStateTransitionNode() {crossfadeDuration=0.5}
  ]
  links [
   {
    dst=Keytar
    rule=AnimGraph.AnimGraphNode_StateMachine_3.Right Arm.AnimStateTransitionNode_6.Transition
    src=Guitar
    via=AnimStateTransitionNode_6
   }
   {
    dst=Guitar
    rule=AnimGraph.AnimGraphNode_StateMachine_3.Right Arm.AnimStateTransitionNode_7.Transition
    src=Keytar
    via=AnimStateTransitionNode_7
   }
  ]
 }
 Keytar {
  path=AnimGraph.AnimGraphNode_StateMachine_3.Right Arm.AnimStateNode_0.Keytar
  nodes [
   Sequence Player 'Keytar_FretHand_BasePose'() {sequence=&Keytar_FretHand_BasePose}
   Output Animation Pose(Result=Apply Additive(
     Base=Sequence Player 'Keytar_FretHand_BasePose'() {sequence=&Keytar_FretHand_BasePose}
     Additive=Blendspace Player 'Keytar_FretHand_BS'(X=0, Y=0) {blendSpace=/SparksCharacterCommon/Animation/MainPlayer/BandPlay/Keytar/FretHand/Keytar_FretHand_BS, startPosition=1}
    ) {alphaBoolBlend {blendOption=Linear}})
  ]
 }
 Guitar {
  path=AnimGraph.AnimGraphNode_StateMachine_3.Right Arm.AnimStateNode_4.Guitar
  nodes [Output Animation Pose(Result=Sequence Player 'Guitar_FretHand_BasePose'() {sequence=&Guitar_FretHand_BasePose})]
 }
 Transition {
  path=AnimGraph.AnimGraphNode_StateMachine_3.Right Arm.AnimStateTransitionNode_6.Transition
  nodes [Result(bCanEnterTransition=Get IsPlayingKeytar())]
 }
 Transition {
  path=AnimGraph.AnimGraphNode_StateMachine_3.Right Arm.AnimStateTransitionNode_7.Transition
  nodes [Result(bCanEnterTransition=NOT Boolean(A=Get IsPlayingKeytar()))]
 }
 AnimNotify_PrintStringAssembler {nodes [
   n1 = AnimNotify_PrintStringAssembler!() -> (then, AnimNotifyEventName:String) {is=title K2Node_FunctionEntry}
   > Branch(Condition=Get DEBUG_PrintAnimLogic() -> (DEBUG_PrintAnimLogic:bool))
   n2 = Get DEBUG_PrintAnimLogic()
   n3 = Enum to String!(Enumerator:ESparksInstrumentType Enum=Get InstrumentType() -> (InstrumentType:ESparksInstrumentType Enum=Guitar)) -> (ReturnValue:String)
   > PrintString!(
    execute
    InString=JoinStringArray(SourceArray=Make Array!("[0]":String="[AnimNotify]", "[1]":String=@n3, "[2]":String=@n1) -> (Array:Array of Strings), Separator=" ")
    bPrintToScreen=@n2
    bPrintToLog=@n2
    TextColor=Get AnimNotify_TextColor() -> (AnimNotify_TextColor:Linear Color Structure)
    Duration
    Key
   ) -> (then)
   > Return Node()
  ]}
 EventGraph {nodes [
   n1 = ResetAnimSets() {is=title K2Node_CustomEvent}
   n2 = Cast To PlayerPawn_Athena(execute=[@n1.then, @n23.then], Object=TryGetPawnOwner())
   > n3 = Set As Player Pawn Athena(As Player Pawn Athena:Player Pawn Athena Object Reference=@n2) -> (Output_Get:Player Pawn Athena Object Reference)
   > n4 = GetAnimationInterface!(execute, self:Fort Player Pawn Object Reference=@n3) -> (then, ReturnValue:Fort Player Animation Interface Object Reference)
   > SetItemAnimLayerChooserContextOverride!(execute, self:Fort Player Animation Interface Object Reference=@n4, ItemAnimLayer:Anim Instance Object Reference=Self-Reference()) -> (then)
   > SetItemAnimationDataOverride!(
    execute
    self:Fort Player Animation Interface Object Reference=@n4
    AnimDataOverride:"Fort Item Animation Data Override Structure (by ref)"=Get Anim Data Override() -> (Anim Data Override:Fort Item Animation Data Override Structure)
    OwningLayer:Anim Instance Object Reference=Self-Reference()
   ) -> (then)
   > Branch(Condition=Get bUseUpperBodyLayer() -> (bUseUpperBodyLayer:bool))
   > OverrideLowerBodyBlendMask!(
    execute
    self:Fort Player Animation Interface Object Reference=@n4
    BlendMaskPreset:"Fort Player Lower Body Blend Mask Preset Structure (by ref)"=Get Blend Mask Preset() -> (Blend Mask Preset:Fort Player Lower Body Blend Mask Preset Structure)
   ) -> (then)
   Guitar Strum() {is=title K2Node_CustomEvent}
   n5 = "Equal (Enum)"(A:ESparksAnimatableState Enum=Get AnimatableState() -> (AnimatableState=None), B:ESparksAnimatableState Enum=Animatable)
   > n6 = Branch(Condition=@n5)
   Gate!(Enter:Exec=@n6.else, Open:Exec=@n10.then, Close:Exec=@n9.then_0, Toggle:Exec, bStartClosed:bool=false) -> (Exit:Exec) {is=title K2Node_MacroInstance}
   > n7 = Branch(Condition=Get IsPlayingKeytar())
   > n8 = Montage_Play!(
    execute
    self:Anim Instance Object Reference
    MontageToPlay:Anim Montage Object Reference=/SparksCharacterCommon/Animation/MainPlayer/BandPlay/Keytar/Strums/Keytar_Strum_Down_A_M
    InPlayRate:float=1
    ReturnValueType:EMontagePlayReturnType Enum=MontageLength
    InTimeToStartMontageAt:float=0
    bStopAllMontages:bool=true
   ) -> (then, ReturnValue:float)
   n9 = Sequence(execute=[@n8.then, @n11.then])
   n10 = Delay(execute=@n9.then_1, Duration=0.15)
   n11 = Montage_Play(
    execute=@n7.else
    MontageToPlay=/SparksCharacterCommon/Animation/MainPlayer/BandPlay/Guitar/Strums/Guitar_Strum_Test_A_M
    InPlayRate=1
    ReturnValueType=MontageLength
    InTimeToStartMontageAt=0
    bStopAllMontages=true
   )
   Change Chord Hand Pose() {is=title K2Node_CustomEvent}
   > n12 = Branch(Condition=@n5)
   Set ChordHandPoseIndex(execute=@n12.else, ChordHandPoseIndex:int=RandomIntegerInRange!(Min:int=0, Max:int=4) -> (ReturnValue:int=0)) -> (Output_Get:int=0)
   Event BlueprintUpdateAnimation()
   > n13 = Cast To PlayerPawn_Athena(Object=TryGetPawnOwner())
   > n14 = Set As Player Pawn Athena(As Player Pawn Athena=@n13)
   n15 = GetAnimationInterface!(self=@n14) -> (ReturnValue)
   > DisableProceduralLayeringPostProcess(self:Fort Player Animation Interface Object Reference=@n15)
   > EnableStopsEarlyExit(self:Fort Player Animation Interface Object Reference=@n15)
   > n16 = IsCharacterMoving!(execute, self) -> (then, Moving:bool)
   > Set ShouldUseJamAnim(ShouldUseJamAnim:bool=NOT Boolean(A=@n16)) -> (Output_Get:bool)
   n17 = Event On LipSync Asset Changed!() -> (OutputDelegate, then, NewLipSyncAsset:Anim Sequence Object Reference) {is=title K2Node_Event}
   > n18 = Is Valid(InputObject=@n17.NewLipSyncAsset)
   Set LipSyncAnimSequence(execute=@n18.Is Not Valid, LipSyncAnimSequence:Anim Sequence Object Reference) -> (Output_Get:Anim Sequence Object Reference)
   Set LipSyncAnimSequence(execute=@n18.Is Valid, LipSyncAnimSequence=@n17.NewLipSyncAsset)
   n19 = Event On MIDI Event Driver Changed!() -> (OutputDelegate, then, NewMidiEventDriver:"Midi Event Driver Interface Interface (by ref)") {is=title K2Node_Event}
   > n20 = Branch(Condition=IsValidInterface!(Interface:"Interface Interface (by ref)"=@n19.NewMidiEventDriver) -> (ReturnValue:bool))
   > n21 = Set MidiTrackName(MidiTrackName:Name=GetMidiTrackNameFromSparksInstrument!(self:Midi Event Driver Interface Interface=@n19.NewMidiEventDriver, InstrumentType:ESparksInstrumentType Enum=Guitar) -> (ReturnValue:Name=None)) -> (Output_Get:Name=None)
   K2Node_Composite_3 = Song Intensity!(In:Exec=[@n21.then, @n22.then]) -> (Out:Exec) {is=title K2Node_Composite}
   > K2Node_Composite_1 = Fret Position!(In:Exec) -> (Out:Exec) {is=title K2Node_Composite}
   > K2Node_Composite_2 = Strum!(In:Exec) -> (Out:Exec) {is=title K2Node_Composite}
   n22 = Set MidiTrackName(execute=@n20.else, MidiTrackName=None) -> (Output_Get=None)
   Event On New Song() {is=title K2Node_Event}
   > SetSongDefaults!(
    execute
    AnimInstance:Sparks Instrument Anim Instance Object Reference=GetSelfAsSparksInstrumentAnimInstance!() -> (ReturnValue:Sparks Instrument Anim Instance Object Reference)
   ) -> (then)
   n23 = Event BlueprintInitializeAnimation()
   Event Sparks Debug Update() {is=title K2Node_Event}
   > Branch(Condition=Get DEBUG_StrumTestFire() -> (DEBUG_StrumTestFire:bool))
   > Guitar Strum() {is=title K2Node_CallFunction}
   > Set DEBUG_StrumTestFire(DEBUG_StrumTestFire:bool=false) -> (Output_Get:bool)
   n24 = Event On Cosmetic Instrument Type Changed!() -> (OutputDelegate, then, NewInstrument:ESparksInstrumentType Enum=Guitar) {is=title K2Node_Event}
   > Set IsPlayingKeytar(IsPlayingKeytar:bool="Equal (Enum)"(A:ESparksInstrumentType Enum=@n24.NewInstrument, B:ESparksInstrumentType Enum=Keyboard)) -> (Output_Get:bool)
   Get Anim Data Override()
   Get bUseUpperBodyLayer()
  ]}
 Fret Position {
  path=EventGraph.K2Node_Composite_1.Fret Position
  nodes [
   Inputs!() -> (In:Exec)
   > n1 = Sequence!(execute) -> (then_0, then_1, then_2:Exec)
   n2 = Get FretMidiListener() -> (FretMidiListener:Async Action Listen For Midi Note Event Object Reference)
   > Is Valid(InputObject=@n2)
   > Cancel(self:Cancellable Async Action Object Reference=@n2)
   > Set FretMidiListener(FretMidiListener:Async Action Listen For Midi Note Event Object Reference) -> (Output_Get:Async Action Listen For Midi Note Event Object Reference)
   Branch(execute=@n1.then_1, Condition=IsValidInterface(Interface=Get MidiEventDriver() -> (MidiEventDriver:Midi Event Driver Interface Interface)))
   > n3 = ListenForMidiNoteEvents!(
    execute:Exec
    MidiEventDriver:Midi Event Driver Interface Interface=Get MidiEventDriver()
    MidiTrackName:Name=Get MidiTrackName() -> (MidiTrackName:Name=None)
    EventNotes:Array of Bytes=GetGuitarMidiNoteIntegersFromEnums!(MidiNoteEvents:Array of EGuitarMidiNoteEvent Enums=Get MidiFretEvents() -> (MidiFretEvents:Array of EGuitarMidiNoteEvent Enums=None)) -> (ReturnValue:Array of Bytes)
    SecondsBeforeEventStart:float=0
    SecondsBeforeEventEnd:float=0
    InitBroadcast:EMidiEventListenerInitBroadcast Enum=MostRecentEvent
   ) -> (
    then:Exec
    AsyncTaskProxy:Async Action Listen For Midi Note Event Object Reference
    OnMidiEventStart:Exec
    OnMidiEventEnd:Exec
    OnFinished:Exec
    MidiEventNote:Byte
    MidiEventData:"Parsed Midi Note Event Structure (by ref)"
   ) {is=title K2Node_AsyncAction}
   > Set FretMidiListener(FretMidiListener=@n3)
   Set FretNumber(execute=@n3.OnMidiEventStart, FretNumber:int=GetFretNumberFromMidiNote!(MidiNote:Byte=@n3.MidiEventNote) -> (ReturnValue:int=0)) -> (Output_Get:int=0)
   > Change Chord Hand Pose() {is=title K2Node_CallFunction}
   Outputs!(Out:Exec=@n1.then_2)
  ]
 }
 Strum {
  path=EventGraph.K2Node_Composite_2.Strum
  nodes [
   Inputs()
   > n1 = Sequence()
   n2 = Get StrumMidiListener() -> (StrumMidiListener:Async Action Listen For Midi Note Event Object Reference)
   > Is Valid(InputObject=@n2)
   > Cancel(self=@n2)
   > Set StrumMidiListener(StrumMidiListener:Async Action Listen For Midi Note Event Object Reference) -> (Output_Get:Async Action Listen For Midi Note Event Object Reference)
   Branch(execute=@n1.then_1, Condition=IsValidInterface(Interface=Get MidiEventDriver()))
   > n3 = ListenForMidiNoteEvents(
    MidiEventDriver=Get MidiEventDriver()
    MidiTrackName=Get MidiTrackName() -> (MidiTrackName=None)
    EventNotes=Make Array!(
     "[0]":Byte=GetGuitarMidiNoteIntegerFromEnum!(MidiNoteEvent:EGuitarMidiNoteEvent Enum=StrumDown) -> (ReturnValue:Byte=0)
     "[1]":Byte=GetGuitarMidiNoteIntegerFromEnum(MidiNoteEvent=StrumUp) -> (ReturnValue=0)
    ) -> (Array:Array of Bytes)
    SecondsBeforeEventStart=0.2
    SecondsBeforeEventEnd=0
    InitBroadcast=MostRecentEvent
   )
   > Set StrumMidiListener(StrumMidiListener=@n3)
   Guitar Strum(execute=@n3.OnMidiEventStart)
   Outputs(Out=@n1.then_2)
  ]
 }
 Song Intensity {
  path=EventGraph.K2Node_Composite_3.Song Intensity
  nodes [
   Inputs()
   > n1 = Sequence()
   n2 = Get IntensityMidiListener() -> (IntensityMidiListener:Async Action Listen For Midi Text Event Object Reference)
   > Is Valid(InputObject=@n2)
   > Cancel(self=@n2)
   > Set IntensityMidiListener(IntensityMidiListener:Async Action Listen For Midi Text Event Object Reference) -> (Output_Get:Async Action Listen For Midi Text Event Object Reference)
   Branch(execute=@n1.then_1, Condition=IsValidInterface(Interface=Get MidiEventDriver()))
   > n3 = ListenForMidiTextEvents!(
    execute:Exec
    MidiEventDriver:Midi Event Driver Interface Interface=Get MidiEventDriver()
    MidiTrackName:Name=Get MidiTrackName() -> (MidiTrackName=None)
    EventStrings:Array of Strings=Make Array!(
     "[0]":String=GetGuitarMidiTextStringFromEnum!(MidiTextEvent:EGuitarMidiTextEvent Enum=PlayingMellow) -> (ReturnValue:String)
     "[1]":String=GetGuitarMidiTextStringFromEnum(MidiTextEvent=PlayingStandard)
     "[2]":String=GetGuitarMidiTextStringFromEnum(MidiTextEvent=PlayingIntense)
     "[3]":String=GetGuitarMidiTextStringFromEnum(MidiTextEvent=IdleRhythmMellow)
     "[4]":String=GetGuitarMidiTextStringFromEnum(MidiTextEvent=IdleRhythmStandard)
     "[5]":String=GetGuitarMidiTextStringFromEnum(MidiTextEvent=IdleRhythmIntense)
     "[6]":String=GetGuitarMidiTextStringFromEnum(MidiTextEvent=IdleRealtime)
    ) -> (Array:Array of Strings)
    SecondsBeforeEventStart:float=Get SecBeforeIntensity() -> (SecBeforeIntensity:float)
    InitBroadcast:EMidiEventListenerInitBroadcast Enum=MostRecentEvent
   ) -> (
    then:Exec
    AsyncTaskProxy:Async Action Listen For Midi Text Event Object Reference
    OnMidiEventStart:Exec
    OnFinished:Exec
    MidiEventText:String
    MidiEventData:"Parsed Midi Text Event Structure (by ref)"
   ) {is=title K2Node_AsyncAction}
   > Set IntensityMidiListener(IntensityMidiListener=@n3)
   MidiEvent_Intensity_OnEvent!(execute=@n3.OnMidiEventStart, self, MidiEventText:String=@n3.MidiEventText) -> (then)
   Outputs(Out=@n1.then_2)
  ]
 }
 IsCharacterMoving {nodes [
   IsCharacterMoving() {is=title K2Node_FunctionEntry}
   > n1 = Cast To PlayerPawn_Athena(Object=TryGetPawnOwner())
   > n2 = Cast To FortPlayerAnimInstance(Object=GetAnimInstance(self=Get Mesh(self=@n1)))
   n3 = Get TurnInPlaceInput!(self:Fort Player Anim Instance Object Reference=@n2) -> (TurnInPlaceInput:Fort Anim Input Turn In Place Structure)
   n4 = Get bIsIdleOrSettling!(self:Fort Player Anim Instance Object Reference=@n2) -> (bIsIdleOrSettling:bool)
   n5 = GetCurveValue(self=@n2, CurveName=TurnRotationAmount)
   > Return Node!(
    execute
    Moving:bool=OR Boolean!(
     A=Get bIsInAir!(self:Fort Player Anim Instance Object Reference=@n2) -> (bIsInAir:bool)
     B=NOT Boolean(A=@n4)
     C:bool="Not Equal (Float)"(A=@n5)
     D:bool=Get bIsCrouching!(self:Fort Player Anim Instance Object Reference=@n2) -> (bIsCrouching)
     E:bool=Break Fort Anim Input Turn in Place!(FortAnimInput_TurnInPlace:"Fort Anim Input Turn In Place Structure (by ref)"=@n3) -> (
      RootYawOffset:float=~
      TurnAnimPlayRate:float=~
      bWantsToTurn:bool=~
      bExitTurn:bool=~
      bUseTurnA:bool=~
      TurnACurveAlpha:float=~
      TurnBCurveAlpha:float=~
      bWantsToTurnInVehicle:bool=~
      bWantsToTurnAgain:bool=~
      bTurningLeft:bool=~
      ShufflePlayRate:float=~
      CrouchShufflePlayRate:float=~
      CrouchShuffleStartPosition:float=~
      CrouchShuffleSpeedScaling:float=~
      bTransition_LowerBody_Shuffle:bool=~
      bTransition_LowerBody_ShuffleADS:bool=~
      bPlayOutro:bool=~
     ) {is=title K2Node_BreakStruct}.bWantsToTurn
    ) -> (ReturnValue)
   )
   Return Node(execute=@n2.CastFailed, Moving=false)
  ]}
 MidiEvent_Intensity_OnEvent {nodes [
   n1 = MidiEvent_Intensity_OnEvent!() -> (then, MidiEventText:String) {is=title K2Node_FunctionEntry}
   > n2 = Sequence!(execute) -> (then_0, then_1)
   > n3 = TranslateMidiEventTextToSongIntensity!(execute, MidiEventText:Name=String To Name(InString=@n1)) -> (then, IdleRealtime:bool, MidiSongIntensityIsIdle:bool, SongIntensity:Gameplay Tag Structure)
   > n4 = Branch(Condition=@n3)
   Set DesiredSongIntensity(execute=@n4.else, DesiredSongIntensity:Gameplay Tag Structure=@n3.SongIntensity) -> (Output_Get:Gameplay Tag Structure)
   MidiEvent_PrintStringAssembler!(
    execute=@n2.then_1
    self
    InstrumentType:ESparksInstrumentType Enum=Get InstrumentType() -> (InstrumentType=Guitar)
    MidiEventText:String=@n1
    SecondsBeforeEvent:double=Get SecBeforeIntensity()
   ) -> (then)
  ]}
 MidiEvent_PrintStringAssembler {nodes [
   n1 = MidiEvent_PrintStringAssembler!() -> (then, InstrumentType:ESparksInstrumentType Enum=Guitar, MidiEventText:String, SecondsBeforeEvent:double=~) {is=title K2Node_FunctionEntry}
   > Branch(Condition=Get DEBUG_PrintAnimLogic())
   > n2 = Branch(Condition="Equal (Float)"(A=@n1.SecondsBeforeEvent))
   > n3 = Set Duration(Duration:float=2) -> (Output_Get:float)
   n4 = Get DEBUG_PrintAnimLogic()
   n5 = "To String (Float)"(InDouble=@n1.SecondsBeforeEvent)
   n6 = Enum to String(Enumerator=@n1)
   PrintString(
    execute=[@n3.then, @n7.then]
    InString=JoinStringArray(
     SourceArray=Make Array!("[0]"="[Anim Midi]", "[1]"=@n6, "[2]"=@n1.MidiEventText, "[3]"=in, "[4]"=@n5, "[5]"=seconds...) -> (Array)
     Separator=" "
    )
    bPrintToScreen=@n4
    bPrintToLog=@n4
    TextColor=Get AnimNotify_TextColor()
    Duration=Get Duration() -> (Duration:float)
   )
   > Return Node!(execute)
   n7 = Set Duration(execute=@n2.else, Duration=@n1.SecondsBeforeEvent)
  ]}
 OnPlay0Entry {nodes [
   OnPlay0Entry!() -> (then, Context:"Anim Update Context Structure (by ref)", Node:"Anim Node Reference Structure (by ref)") {is=title K2Node_FunctionEntry}
   > SetPlayAnimAndTempoMultiplier!(execute, self, Index:int=0) -> (then)
   > Return Node()
  ]}
 OnPlay1Entry {nodes [
   OnPlay1Entry!() -> (then, Context:"Anim Update Context Structure (by ref)", Node:"Anim Node Reference Structure (by ref)") {is=title K2Node_FunctionEntry}
   > SetPlayAnimAndTempoMultiplier(Index=1)
   > Return Node()
  ]}
 SetPlayAnimAndTempoMultiplier {nodes [
   n1 = SetPlayAnimAndTempoMultiplier!() -> (then, Index:int) {is=title K2Node_FunctionEntry}
   > Set SongIntensity(SongIntensity:Gameplay Tag Structure=Get DesiredSongIntensity() -> (DesiredSongIntensity:Gameplay Tag Structure)) -> (Output_Get:Gameplay Tag Structure)
   > n2 = PickIntensityAnimLoopCycle!(
    execute
    self:Sparks Instrument Anim Instance Object Reference
    BPMSet:"Sparks Anim BPMSet Structure (by ref)"=Get HomebaseBPMs() -> (HomebaseBPMs:Sparks Anim BPMSet Structure)
    TransitionTag:Gameplay Tag Structure
    ForceVariant:bool=false
    VariantIndex:int=0
   ) -> (then, ReturnValue:"Sparks Anim Loop Cycle Structure (by ref)")
   > Set Array Elem!(
    execute
    TargetArray:Array of Anim Sequence Object References=Get PlaySequences()
    Index:int=@n1
    Item:"Anim Sequence Object Reference (by ref)"=Break Sparks Anim Loop Cycle!(SparksAnimLoopCycle:"Sparks Anim Loop Cycle Structure (by ref)"=@n2) -> (Loop:Anim Sequence Object Reference, Entry:Anim Sequence Object Reference, Exit:Anim Sequence Object Reference) {is=title K2Node_BreakStruct}
    bSizeToFit:bool=true
   ) -> (then)
   > Set Array Elem(
    TargetArray:"Array of Float (double-precision)s"=Get TempoMultipliers()
    Index=@n1
    Item:"Float (double-precision) (by ref)"=Break Sparks Anim BPMSet!(SparksAnimBPMSet:"Sparks Anim BPMSet Structure (by ref)"=Get HomebaseBPMs()) -> (BPMBuckets:Array of Sparks Anim Intensity Set Object References, CurrentTempoMultiplier:float=~) {is=title K2Node_BreakStruct}.CurrentTempoMultiplier
    bSizeToFit=true
   )
  ]}
]
variables [
 Anim Data Override {
  type=FortItemAnimationDataOverride
  default {
   weaponAdditiveAnimSet=null
   weaponAnimSet=/Game/Animation/Game/MainPlayer/AnimAssets/FortPlayerDefault_AnimAsset
   weaponCoreAnim=Melee
   playerAnimAssets {firstPersonWeaponAnimAsset=null, tacticalStateAnimAsset=null}
   proceduralLayeringData {
    bUseMotionMatchingWithoutProxyTable=false
    bUseProceduralLayeringForTacticalSprint=false
    bUseWeaponProxyTableForDBNOCarry=false
    itemChooserDataClass=null
    proxyTable=/SparksCharacterCommon/Animation/MainPlayer/ProxyTables/JamGuitar_PT
    proxyTableOverride=null
   }
  }
 }
 AnimatableState {type=ESparksAnimatableState, default=None}
 AnimNotify_TextColor {type=LinearColor, default=<rgba 1 0 1 1>}
 As Player Pawn Athena {type=PlayerPawn_Athena_C}
 Blend Mask Preset {
  type=FortPlayerLowerBodyBlendMaskPreset
  default {
   bApplyInMeshSpace=true
   blendProfile=&/JamTraversalLowerBody_BlendMask
   defaultBlendInTime=0.1
   defaultBlendOutTime=0.1
  }
 }
 BPM Preview {type=number, default=125}
 ChordHandPoseIndex {type=integer}
 DEBUG_PrintAnimLogic {type=boolean}
 DEBUG_StrumTestFire {type=boolean}
 DEBUG_UseInputLayering {type=boolean, default=true}
 FretMidiListener {type=AsyncAction_ListenForMidiNoteEvent, default=null}
 Head Additive Alpha {type=number, default=0.2}
 Index {type=integer, default=3}
 InstrumentType {type=ESparksInstrumentType, default=Guitar}
 IntensityMidiListener {type=AsyncAction_ListenForMidiTextEvent, default=null}
 IsPlayingKeytar {type=boolean}
 LipSyncAnimSequence {type=AnimSequence, default=null}
 mainAnimBP {type=Fortnite_M_Avg_Player_AnimBlueprint_C, default=null}
 MidiFretEvents {
  type="array<EGuitarMidiNoteEvent>"
  default [
   FretPosition1
   FretPosition2
   FretPosition3
   FretPosition4
   FretPosition5
   FretPosition6
   FretPosition7
   FretPosition8
   FretPosition9
   FretPosition10
   FretPosition11
   FretPosition12
   FretPosition13
   FretPosition14
   FretPosition15
   FretPosition16
   FretPosition17
   FretPosition18
   FretPosition19
   FretPosition20
  ]
 }
 MidiTrackName {type=string, default=None}
 NewVar {type=boolean}
 PlaySequences {type="array<AnimSequence>", default [null,null]}
 SecBeforeIntensity {type=number}
 ShouldUseJamAnim {type=boolean, default=true}
 StrumMidiListener {type=AsyncAction_ListenForMidiNoteEvent, default=null}
 TempoMultipliers {type="array<number>", default [0,0]}
 Traversal_IsCharacterMoving? {type=boolean}
 TraversalGeneralAdditiveSequence {type=AnimSequenceBase, default=&Traversal_UpperBodyBeatAdditive}
]
